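# Level Sequence for authoring the Boss Yacht weapon reload animation on a Fortnite player mannequin holding a legendary heavy AR.
# The rifle is a spawnable skeletal mesh actor with its own animation section and a transform carrying a parent constraint.
# The mannequin, with head, body and face components, is driven by the Fortnite_Biped_CtrlRig track, whose densely keyed channels set legs to IK and arms to FK.
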LevelSequence /Clyde/Weapons/Boss_Yacht/Animation/Boss_Yacht_Reload_LS
bindings [
 Face {
  id=#1
  display_name=Face Accessory
  parent_id=#2
  possessed_class=SkeletalMeshComponentBudgeted
  tracks []
 }
 SK_LegendaryHeavyAR {
  id=#3
  possessed_class=FortSkeletalMeshActor
  tracks [
   {type=MovieSceneSpawnTrack, sections [{type=MovieSceneSpawnSection, channels [None {type=MovieSceneScriptingBoolChannel, default=true}]}]}
   {
    type=MovieScene3DTransformTrack
    sections [{
      type=MovieScene3DTransformSection
      channels [
       Location.X {type=MovieSceneScriptingDoubleChannel, default=0, keys [[-62.0013,0],[-62,0],[-0.0013,-15.1676],[0,0]]}
       Location.Y {type=MovieSceneScriptingDoubleChannel, default=0, keys [[-62.0013,0],[-62,0],[-0.0013,9.2619],[0,0]]}
       Location.Z {type=MovieSceneScriptingDoubleChannel, default=0, keys [[-62.0013,0],[-62,0],[-0.0013,124.1712],[0,0]]}
       Rotation.X {type=MovieSceneScriptingDoubleChannel, default=0, thinned_from=4, keys [[-62.0013,0],[0,0]]}
       Rotation.Y {type=MovieSceneScriptingDoubleChannel, default=0, thinned_from=4, keys [[-62.0013,0],[0,0]]}
       Rotation.Z {type=MovieSceneScriptingDoubleChannel, default=0, thinned_from=4, keys [[-62.0013,0],[-0.0013,0],[0,-90.0004]]}
       Scale.X {type=MovieSceneScriptingDoubleChannel, default=1, const [1,4]}
       Scale.Y {type=MovieSceneScriptingDoubleChannel, default=1, const [1,4]}
       Scale.Z {type=MovieSceneScriptingDoubleChannel, default=1, const [1,4]}
       ParentConstraint_3 {type=MovieSceneScriptingConstraintChannel, const [true,1]}
      ]
     }]
   }
   {type=MovieSceneSkeletalAnimationTrack, sections [{type=MovieSceneSkeletalAnimationSection, end_frame=80, start_frame=0}]}
  ]
 }
 Head {id=#4, parent_id=#2, possessed_class=SkeletalMeshComponentBudgeted, tracks []}
 Body {id=#5, parent_id=#2, possessed_class=SkeletalMeshComponentBudgeted, tracks []}
 mannequin {
  id=#2
  possessed_class=FortPlayerMannequin
  tracks [
   {
    type=MovieSceneSpawnTrack
    sections [
     {type=MovieSceneSpawnSection, channels [None {type=MovieSceneScriptingBoolChannel, default=true}]}
     {type=MovieSceneSpawnSection, channels [None {type=MovieSceneScriptingBoolChannel, default=true}]}
    ]
   }
   {
    type=MovieScene3DTransformTrack
    sections [{
      type=MovieScene3DTransformSection
      channels [
       Location.X {type=MovieSceneScriptingDoubleChannel, default=0}
       Location.Y {type=MovieSceneScriptingDoubleChannel, default=0}
       Location.Z {type=MovieSceneScriptingDoubleChannel, default=0}
       Rotation.X {type=MovieSceneScriptingDoubleChannel, default=0}
       Rotation.Y {type=MovieSceneScriptingDoubleChannel, default=0}
       Rotation.Z {type=MovieSceneScriptingDoubleChannel, default=0}
       Scale.X {type=MovieSceneScriptingDoubleChannel, default=1}
       Scale.Y {type=MovieSceneScriptingDoubleChannel, default=1}
       Scale.Z {type=MovieSceneScriptingDoubleChannel, default=1}
      ]
     }]
   }
   {
    type=MovieSceneControlRigParameterTrack
    display_name=Fortnite_Biped_CtrlRig
    sections [
     {
      type=MovieSceneControlRigParameterSection
      channels [
       leg_l_stretch_switch {type=MovieSceneScriptingBoolChannel, default=false, const [false,80]}
       leg_r_stretch_switch {type=MovieSceneScriptingBoolChannel, default=false, const [false,80]}
       arm_l_fk_local {type=MovieSceneScriptingBoolChannel, default=true, const [true,80]}
       Auto Interpolate_3 {type=MovieSceneScriptingBoolChannel, default=false, const [false,80]}
       Auto Interpolate {type=MovieSceneScriptingBoolChannel, default=false, const [false,80]}
       arm_l_stretch_switch {type=MovieSceneScriptingBoolChannel, default=false, const [false,80]}
       Keep Aligned {type=MovieSceneScriptingBoolChannel, default=false, const [false,80]}
       arm_r_fk_local {type=MovieSceneScriptingBoolChannel, default=true, const [true,80]}
       Auto Interpolate_4 {type=MovieSceneScriptingBoolChannel, default=false, const [false,80]}
       Auto Interpolate_2 {type=MovieSceneScriptingBoolChannel, default=false, const [false,80]}
       arm_r_stretch_switch {type=MovieSceneScriptingBoolChannel, default=false, const [false,80]}
       Keep Aligned_2 {type=MovieSceneScriptingBoolChannel, default=false, const [false,80]}
       head_local {type=MovieSceneScriptingBoolChannel, default=true, const [true,80]}
       Stretch {type=MovieSceneScriptingBoolChannel, default=false, const [false,80]}
       head_stretch {type=MovieSceneScriptingBoolChannel, default=false, const [false,80]}
       spine_stretch_switch {type=MovieSceneScriptingBoolChannel, default=false, const [false,80]}
       arm_l_fk_ik_switch {type=MovieSceneScriptingBoolChannel, default=false, const [false,80]}
       leg_l_fk_ik_switch {type=MovieSceneScriptingBoolChannel, default=true, const [true,80]}
       arm_r_fk_ik_switch {type=MovieSceneScriptingBoolChannel, default=false, const [false,80]}
       leg_r_fk_ik_switch {type=MovieSceneScriptingBoolChannel, default=true, const [true,80]}
       ShowBodyControls {type=MovieSceneScriptingBoolChannel, default=true, const [true,80]}
       spine_fk_ik_switch {type=MovieSceneScriptingBoolChannel, default=false, const [false,80]}
       neck_fk_ik_switch {type=MovieSceneScriptingBoolChannel, default=false, const [false,80]}
       space {type=MovieSceneScriptingByteChannel, default=1, const [1,80]}
       global_ctrl.Location.X {type=MovieSceneScriptingFloatChannel, default=0, const [0,80]}
       global_ctrl.Location.Y {type=MovieSceneScriptingFloatChannel, default=0, const [0,80]}
       global_ctrl.Location.Z {type=MovieSceneScriptingFloatChannel, default=0, const [0,80]}
       global_ctrl.Rotation.X {type=MovieSceneScriptingFloatChannel, default=0, const [0,80]}
       global_ctrl.Rotation.Y {type=MovieSceneScriptingFloatChannel, default=0, const [0,80]}
       global_ctrl.Rotation.Z {type=MovieSceneScriptingFloatChannel, default=0, const [0,80]}
       root_ctrl.Location.X {type=MovieSceneScriptingFloatChannel, default=0, const [0,80]}
       root_ctrl.Location.Y {type=MovieSceneScriptingFloatChannel, default=0, const [0,80]}
       root_ctrl.Location.Z {type=MovieSceneScriptingFloatChannel, default=0, const [0,80]}
       root_ctrl.Rotation.X {type=MovieSceneScriptingFloatChannel, default=0, const [0,80]}
       root_ctrl.Rotation.Y {type=MovieSceneScriptingFloatChannel, default=0, const [0,80]}
       root_ctrl.Rotation.Z {type=MovieSceneScriptingFloatChannel, default=0, const [0,80]}
       root_ctrl.Scale.X {type=MovieSceneScriptingFloatChannel, default=1, const [1,80]}
       root_ctrl.Scale.Y {type=MovieSceneScriptingFloatChannel, default=1, const [1,80]}
       root_ctrl.Scale.Z {type=MovieSceneScriptingFloatChannel, default=1, const [1,80]}
       body_offset_ctrl.Location.X {type=MovieSceneScriptingFloatChannel, default=0, const [0,80]}
       body_offset_ctrl.Location.Y {type=MovieSceneScriptingFloatChannel, default=0, const [0,80]}
       body_offset_ctrl.Location.Z {type=MovieSceneScriptingFloatChannel, default=0, const [0,80]}
       body_offset_ctrl.Rotation.X {type=MovieSceneScriptingFloatChannel, default=0, const [0,80]}
       body_offset_ctrl.Rotation.Y {type=MovieSceneScriptingFloatChannel, default=0, const [0,80]}
       body_offset_ctrl.Rotation.Z {type=MovieSceneScriptingFloatChannel, default=0, const [0,80]}
       body_offset_ctrl.Scale.X {type=MovieSceneScriptingFloatChannel, default=1, const [1,80]}
       body_offset_ctrl.Scale.Y {type=MovieSceneScriptingFloatChannel, default=1, const [1,80]}
       body_offset_ctrl.Scale.Z {type=MovieSceneScriptingFloatChannel, default=1, const [1,80]}
       thigh_l_fk_ctrl.Location.X {type=MovieSceneScriptingFloatChannel, default=0, thinned_from=80, keys [[0,0],[79.9875,0]]}
       thigh_l_fk_ctrl.Location.Y {type=MovieSceneScriptingFloatChannel, default=0, thinned_from=80, keys [[0,0],[79.9875,0]]}
       thigh_l_fk_ctrl.Location.Z {type=MovieSceneScriptingFloatChannel, default=0, thinned_from=80, keys [[0,0],[79.9875,0]]}
       thigh_l_fk_ctrl.Rotation.X {
        type=MovieSceneScriptingFloatChannel
        default=8.9448
        thinned_from=80
        keys [[0,8.9448],[1.0125,8.988],[2.025,9.1024],[3.0375,9.2663],[4.05,9.4623],[5.0625,9.7481],[7.0875,10.4686],[8.1,10.7683],[9.1125,10.8873],[10.125,10.9679],[11.1375,11.1961],[12.15,11.6702],[13.1625,12.3405],[14.175,13.2781],[15.1875,14.324],[16.2,15.1204],[17.2125,15.5955],[18.225,15.8947],[19.2375,16.0611],[20.25,16.1296],[22.275,16.1006],[23.2875,15.9766],[24.3,15.7233],[25.3125,15.3622],[26.325,14.9119],[27.3375,14.3899],[29.3625,13.2643],[30.375,12.7305],[31.3875,12.2566],[32.4,11.8736],[33.4125,11.617],[34.425,11.4949],[35.4375,11.4351],[36.45,11.3401],[37.4625,11.1935],[38.475,11.0064],[39.4875,10.7311],[40.5,9.9476],[41.5125,9.243],[42.525,9.2456],[44.55,9.7263],[45.5625,9.9364],[46.575,10.0686],[47.5875,10.0665],[48.6,9.9714],[49.6125,9.834],[50.625,9.6641],[51.6375,9.4674],[52.65,9.2511],[53.6625,9.0907],[54.675,9.0963],[55.6875,9.6149],[56.7,10.6887],[57.7125,11.9916],[58.725,13.2503],[59.7375,14.2551],[60.75,14.8941],[61.7625,15.0979],[62.775,14.963],[63.7875,14.5826],[64.8,14.0513],[65.8125,13.4559],[66.825,12.7964],[68.85,11.3191],[69.8625,10.6955],[70.875,10.2112],[71.8875,9.7966],[72.9,9.4154],[73.9125,9.0769],[74.925,8.8284],[75.9375,8.6828],[76.95,8.6077],[77.9625,8.6683],[78.975,8.8427],[79.9875,8.9448]]
       }
       thigh_l_fk_ctrl.Rotation.Y {
        type=MovieSceneScriptingFloatChannel
        default=9.3892
        thinned_from=80
        keys [[0,9.3892],[1.0125,9.409],[2.025,9.461],[4.05,9.6238],[5.0625,9.7685],[6.075,9.959],[7.0875,10.1711],[8.1,10.3443],[10.125,10.4634],[11.1375,10.5916],[12.15,10.8954],[13.1625,11.6169],[14.175,12.7862],[15.1875,13.9271],[16.2,14.5764],[17.2125,14.5605],[18.225,14.2904],[19.2375,14.1278],[20.25,14.0724],[21.2625,14.066],[22.275,14.0073],[23.2875,13.849],[24.3,13.5956],[25.3125,13.2763],[28.35,12.2296],[29.3625,11.8979],[30.375,11.5912],[31.3875,11.3253],[32.4,11.1135],[33.4125,10.9729],[34.425,10.9079],[35.4375,10.9079],[37.4625,10.9778],[38.475,10.9617],[39.4875,10.7259],[40.5,9.4143],[41.5125,7.6892],[42.525,7.0673],[43.5375,7.1507],[45.5625,7.5536],[46.575,7.742],[47.5875,7.8968],[49.6125,8.1066],[50.625,8.2582],[51.6375,8.5206],[52.65,8.9091],[53.6625,9.4736],[54.675,10.3928],[55.6875,12.29],[56.7,14.5585],[57.7125,15.6787],[58.725,15.8309],[59.7375,15.6886],[60.75,15.356],[62.775,14.3826],[65.8125,13.0117],[66.825,12.4634],[68.85,11.1251],[69.8625,10.5905],[70.875,10.2467],[71.8875,10.0184],[72.9,9.8391],[73.9125,9.6955],[75.9375,9.503],[76.95,9.4221],[77.9625,9.3828],[79.9875,9.3892]]
       }
       thigh_l_fk_ctrl.Rotation.Z {
        type=MovieSceneScriptingFloatChannel
        default=-28.5326
        thinned_from=80
        keys [[0,-28.5326],[2.025,-28.4288],[4.05,-28.2386],[7.0875,-27.7614],[8.1,-27.6633],[10.125,-27.7999],[11.1375,-27.7906],[12.15,-27.6182],[13.1625,-27.3532],[14.175,-26.9713],[15.1875,-26.4216],[18.225,-24.4825],[19.2375,-23.9502],[20.25,-23.7634],[21.2625,-23.8625],[22.275,-24.0647],[25.3125,-24.9088],[26.325,-25.1201],[34.425,-26.1016],[35.4375,-26.333],[36.45,-26.7519],[37.4625,-27.24],[38.475,-27.5833],[39.4875,-27.6044],[40.5,-26.8899],[41.5125,-25.7996],[42.525,-25.4587],[44.55,-26.2072],[45.5625,-26.3748],[47.5875,-26.418],[50.625,-26.1746],[51.6375,-26.219],[52.65,-26.4182],[55.6875,-27.3692],[56.7,-27.509],[57.7125,-26.9901],[58.725,-26.2585],[59.7375,-25.735],[60.75,-25.4333],[61.7625,-25.325],[62.775,-25.3513],[63.7875,-25.5394],[64.8,-25.9275],[65.8125,-26.4477],[66.825,-27.2121],[67.8375,-28.0992],[68.85,-28.7755],[69.8625,-29.1299],[72.9,-29.694],[73.9125,-29.7899],[74.925,-29.807],[75.9375,-29.6881],[76.95,-29.4549],[78.975,-28.7237],[79.9875,-28.5326]]
       }
       thigh_l_fk_ctrl.Scale.X {type=MovieSceneScriptingFloatChannel, default=1, thinned_from=80, keys [[0,1],[79.9875,1]]}
       thigh_l_fk_ctrl.Scale.Y {type=MovieSceneScriptingFloatChannel, default=1, const [1,80]}
       thigh_l_fk_ctrl.Scale.Z {type=MovieSceneScriptingFloatChannel, default=1, const [1,80]}
       calf_l_fk_ctrl.Location.X {
        type=MovieSceneScriptingFloatChannel
        default=0.0896
        thinned_from=80
        keys [[0,0.0896],[1.0125,0.0894],[3.0375,0.0882],[4.05,0.0874],[5.0625,0.0869],[10.125,0.0861],[11.1375,0.0863],[12.15,0.0871],[13.1625,0.0901],[14.175,0.0951],[15.1875,0.0993],[16.2,0.1006],[17.2125,0.0981],[18.225,0.0934],[19.2375,0.0892],[20.25,0.0872],[21.2625,0.0871],[24.3,0.088],[25.3125,0.088],[27.3375,0.0871],[28.35,0.0869],[32.4,0.0885],[34.425,0.0889],[35.4375,0.09],[36.45,0.0923],[37.4625,0.0949],[38.475,0.0964],[39.4875,0.0944],[40.5,0.0807],[41.5125,0.0643],[42.525,0.0607],[43.5375,0.0649],[44.55,0.0699],[45.5625,0.0733],[46.575,0.0753],[47.5875,0.0765],[48.6,0.0773],[51.6375,0.078],[52.65,0.0789],[53.6625,0.0805],[54.675,0.0832],[55.6875,0.0894],[56.7,0.0974],[57.7125,0.1027],[58.725,0.1054],[59.7375,0.1066],[60.75,0.1068],[61.7625,0.1064],[62.775,0.1056],[65.8125,0.1023],[66.825,0.1007],[68.85,0.0968],[69.8625,0.0954],[70.875,0.0952],[73.9125,0.0977],[74.925,0.0982],[75.9375,0.0973],[76.95,0.0956],[78.975,0.0908],[79.9875,0.0896]]
       }
       calf_l_fk_ctrl.Location.Y {
        type=MovieSceneScriptingFloatChannel
        default=0.0023
        thinned_from=80
        keys [[0,0.0023],[10.125,0.0015],[12.15,0.0017],[13.1625,0.0024],[15.1875,0.0048],[16.2,0.0052],[17.2125,0.0044],[18.225,0.0032],[19.2375,0.0022],[20.25,0.0017],[26.325,0.0018],[29.3625,0.0017],[34.425,0.0021],[35.4375,0.0024],[37.4625,0.0036],[38.475,0.004],[39.4875,0.0035],[41.5125,-0.0021],[42.525,-0.0025],[45.5625,-0.0009],[48.6,-0.0003],[52.65,0],[53.6625,0.0003],[54.675,0.0008],[55.6875,0.0022],[56.7,0.0043],[57.7125,0.0058],[58.725,0.0066],[59.7375,0.007],[60.75,0.0071],[65.8125,0.0056],[68.85,0.0041],[69.8625,0.0037],[70.875,0.0036],[74.925,0.0045],[75.9375,0.0042],[78.975,0.0025],[79.9875,0.0023]]
       }
       calf_l_fk_ctrl.Location.Z {
        type=MovieSceneScriptingFloatChannel
        default=0.0031
        thinned_from=80
        keys [[0,0.0031],[13.1625,0.0031],[16.2,0.0035],[39.4875,0.0033],[41.5125,0.0023],[42.525,0.0021],[54.675,0.0029],[58.725,0.0036],[79.9875,0.0031]]
       }
       calf_l_fk_ctrl.Rotation.X {
        type=MovieSceneScriptingFloatChannel
        default=0.7763
        thinned_from=80
        keys [[0,0.7763],[1.0125,0.7746],[2.025,0.7704],[4.05,0.7584],[5.0625,0.7543],[10.125,0.7485],[11.1375,0.7497],[12.15,0.7561],[13.1625,0.7801],[14.175,0.8185],[15.1875,0.8512],[16.2,0.8617],[17.2125,0.842],[18.225,0.8055],[19.2375,0.7727],[20.25,0.7572],[21.2625,0.7561],[24.3,0.7633],[25.3125,0.763],[27.3375,0.7564],[28.35,0.7547],[29.3625,0.7569],[32.4,0.7673],[34.425,0.7704],[35.4375,0.7794],[36.45,0.7973],[37.4625,0.8175],[38.475,0.8294],[39.4875,0.8132],[40.5,0.7039],[41.5125,0.5714],[42.525,0.5414],[43.5375,0.5768],[44.55,0.6182],[45.5625,0.6455],[46.575,0.6615],[47.5875,0.6717],[48.6,0.6778],[49.6125,0.6805],[51.6375,0.684],[52.65,0.6907],[53.6625,0.7037],[54.675,0.7249],[55.6875,0.774],[56.7,0.8364],[57.7125,0.8776],[58.725,0.8983],[59.7375,0.9072],[60.75,0.909],[61.7625,0.906],[62.775,0.8998],[65.8125,0.8743],[66.825,0.8623],[68.85,0.8323],[69.8625,0.8216],[70.875,0.8197],[73.9125,0.8392],[74.925,0.843],[75.9375,0.8361],[76.95,0.8227],[78.975,0.7856],[79.9875,0.7763]]
       }
       calf_l_fk_ctrl.Rotation.Y {
        type=MovieSceneScriptingFloatChannel
        default=0.1954
        thinned_from=80
        keys [[0,0.1954],[2.025,0.1946],[5.0625,0.1924],[10.125,0.1915],[11.1375,0.1917],[12.15,0.1926],[13.1625,0.1959],[14.175,0.2009],[15.1875,0.2048],[16.2,0.2059],[17.2125,0.2037],[19.2375,0.1949],[20.25,0.1928],[21.2625,0.1926],[24.3,0.1936],[26.325,0.1932],[28.35,0.1924],[32.4,0.1942],[34.425,0.1946],[35.4375,0.1958],[37.4625,0.2008],[38.475,0.2022],[39.4875,0.2002],[40.5,0.1829],[41.5125,0.1605],[42.525,0.1547],[43.5375,0.1618],[44.55,0.1698],[45.5625,0.1749],[46.575,0.1777],[47.5875,0.1794],[48.6,0.1804],[51.6375,0.1815],[52.65,0.1826],[53.6625,0.1847],[54.675,0.188],[55.6875,0.1949],[56.7,0.2029],[57.7125,0.2076],[58.725,0.2097],[59.7375,0.2105],[60.75,0.2107],[62.775,0.2098],[65.8125,0.2073],[69.8625,0.2013],[70.875,0.2011],[73.9125,0.2034],[74.925,0.2038],[75.9375,0.203],[76.95,0.2014],[78.975,0.1967],[79.9875,0.1954]]
       }
       calf_l_fk_ctrl.Rotation.Z {
        type=MovieSceneScriptingFloatChannel
        default=21.3725
        thinned_from=80
        keys [[0,21.3725],[1.0125,21.3223],[2.025,21.1934],[4.05,20.8247],[5.0625,20.7],[10.125,20.5224],[11.1375,20.5582],[12.15,20.754],[13.1625,21.4914],[14.175,22.6873],[15.1875,23.7196],[16.2,24.0536],[17.2125,23.4284],[18.225,22.2798],[19.2375,21.2638],[20.25,20.7874],[21.2625,20.7552],[24.3,20.9756],[25.3125,20.9644],[27.3375,20.7631],[28.35,20.71],[29.3625,20.7796],[32.4,21.0965],[34.425,21.1936],[35.4375,21.4683],[36.45,22.0258],[37.4625,22.6534],[38.475,23.0269],[39.4875,22.5234],[40.5,19.2213],[41.5125,15.2982],[42.525,14.4317],[43.5375,15.4458],[44.55,16.6414],[45.5625,17.4402],[46.575,17.9105],[47.5875,18.2137],[48.6,18.3952],[49.6125,18.4746],[50.625,18.5072],[51.6375,18.579],[52.65,18.7778],[53.6625,19.1666],[54.675,19.8068],[55.6875,21.3066],[56.7,23.2545],[57.7125,24.5621],[58.725,25.231],[59.7375,25.5222],[60.75,25.5799],[61.7625,25.4817],[62.775,25.2801],[65.8125,24.4556],[66.825,24.0706],[68.85,23.1202],[69.8625,22.7823],[70.875,22.7246],[73.9125,23.3363],[74.925,23.4563],[75.9375,23.239],[76.95,22.817],[78.975,21.6595],[79.9875,21.3725]]
       }
       calf_l_fk_ctrl.Scale.X {type=MovieSceneScriptingFloatChannel, default=1, const [1,80]}
       calf_l_fk_ctrl.Scale.Y {type=MovieSceneScriptingFloatChannel, default=1, const [1,80]}
       calf_l_fk_ctrl.Scale.Z {type=MovieSceneScriptingFloatChannel, default=1, const [1,80]}
       foot_l_fk_ctrl.Location.X {type=MovieSceneScriptingFloatChannel, default=0, thinned_from=80, keys [[0,0],[79.9875,0]]}
       foot_l_fk_ctrl.Location.Y {type=MovieSceneScriptingFloatChannel, default=0, thinned_from=80, keys [[0,0],[79.9875,0]]}
       foot_l_fk_ctrl.Location.Z {type=MovieSceneScriptingFloatChannel, default=0, thinned_from=80, keys [[0,0],[79.9875,0]]}
       foot_l_fk_ctrl.Rotation.X {
        type=MovieSceneScriptingFloatChannel
        default=-4.4173
        thinned_from=80
        keys [[0,-4.4173],[4.05,-4.4299],[5.0625,-4.4469],[8.1,-4.5291],[10.125,-4.538],[11.1375,-4.5602],[12.15,-4.6167],[13.1625,-4.7271],[15.1875,-5.0572],[16.2,-5.161],[17.2125,-5.1787],[19.2375,-5.1205],[20.25,-5.1066],[22.275,-5.1052],[23.2875,-5.0956],[24.3,-5.0731],[25.3125,-5.0386],[30.375,-4.7857],[32.4,-4.7057],[33.4125,-4.6803],[34.425,-4.6692],[35.4375,-4.6733],[37.4625,-4.7002],[38.475,-4.7006],[39.4875,-4.6547],[40.5,-4.4282],[41.5125,-4.1728],[42.525,-4.1242],[43.5375,-4.1854],[44.55,-4.2584],[45.5625,-4.3124],[46.575,-4.3458],[47.5875,-4.3619],[49.6125,-4.3634],[52.65,-4.3464],[53.6625,-4.3609],[54.675,-4.4166],[55.6875,-4.5871],[56.7,-4.837],[57.7125,-5.0472],[58.725,-5.1912],[59.7375,-5.2811],[60.75,-5.3255],[61.7625,-5.3271],[62.775,-5.2956],[63.7875,-5.2346],[64.8,-5.1497],[65.8125,-5.0512],[66.825,-4.9329],[67.8375,-4.801],[68.85,-4.6827],[69.8625,-4.5997],[70.875,-4.556],[72.9,-4.5129],[74.925,-4.4908],[77.9625,-4.4322],[79.9875,-4.4173]]
       }
       foot_l_fk_ctrl.Rotation.Y {
        type=MovieSceneScriptingFloatChannel
        default=-4.36
        thinned_from=80
        keys [[0,-4.36],[3.0375,-4.3364],[8.1,-4.2502],[12.15,-4.2222],[14.175,-4.2448],[15.1875,-4.2193],[16.2,-4.1567],[17.2125,-4.047],[18.225,-3.9739],[19.2375,-3.9841],[20.25,-4.032],[21.2625,-4.0912],[22.275,-4.1359],[23.2875,-4.1639],[25.3125,-4.1935],[27.3375,-4.1963],[32.4,-4.1502],[33.4125,-4.1529],[34.425,-4.1655],[35.4375,-4.191],[36.45,-4.2317],[38.475,-4.3405],[39.4875,-4.4494],[40.5,-4.6818],[41.5125,-4.777],[42.525,-4.7027],[45.5625,-4.5436],[49.6125,-4.3127],[50.625,-4.2709],[51.6375,-4.2587],[52.65,-4.2883],[53.6625,-4.3762],[54.675,-4.5504],[56.7,-5.1898],[57.7125,-5.2271],[58.725,-5.0666],[59.7375,-4.8853],[60.75,-4.7294],[61.7625,-4.6101],[62.775,-4.5262],[63.7875,-4.4623],[64.8,-4.4101],[66.825,-4.3309],[67.8375,-4.3018],[68.85,-4.2839],[69.8625,-4.2811],[70.875,-4.2929],[74.925,-4.3772],[76.95,-4.3856],[79.9875,-4.36]]
       }
       foot_l_fk_ctrl.Rotation.Z {
        type=MovieSceneScriptingFloatChannel
        default=5.7962
        thinned_from=80
        keys [[0,5.7962],[1.0125,5.8162],[2.025,5.8683],[4.05,6.0223],[5.0625,6.0676],[6.075,6.0814],[8.1,6.0837],[10.125,6.1814],[11.1375,6.1999],[12.15,6.1324],[13.1625,5.8051],[14.175,5.2633],[15.1875,4.8116],[16.2,4.7135],[17.2125,5.0957],[18.225,5.6983],[19.2375,6.1941],[20.25,6.4285],[21.2625,6.4522],[23.2875,6.3983],[24.3,6.3945],[25.3125,6.4177],[27.3375,6.5018],[28.35,6.4708],[29.3625,6.3254],[30.375,6.1417],[31.3875,5.9707],[32.4,5.8307],[33.4125,5.7602],[34.425,5.7043],[35.4375,5.5603],[36.45,5.2949],[37.4625,5.0035],[38.475,4.8329],[39.4875,5.055],[40.5,6.5381],[41.5125,8.3234],[42.525,8.7221],[43.5375,8.2448],[44.55,7.682],[45.5625,7.305],[46.575,7.0896],[47.5875,6.9646],[48.6,6.901],[49.6125,6.8878],[51.6375,6.9177],[52.65,6.8782],[53.6625,6.7539],[54.675,6.5374],[55.6875,6.0005],[56.7,5.2751],[57.7125,4.7311],[58.725,4.3989],[59.7375,4.2348],[60.75,4.1983],[61.7625,4.2561],[62.775,4.3546],[63.7875,4.4409],[64.8,4.4768],[65.8125,4.4759],[66.825,4.5338],[67.8375,4.676],[68.85,4.8552],[69.8625,5.0244],[70.875,5.0924],[71.8875,5.0488],[73.9125,4.9121],[74.925,4.8751],[75.9375,4.9832],[76.95,5.1778],[78.975,5.676],[79.9875,5.7962]]
       }
       foot_l_fk_ctrl.Scale.X {type=MovieSceneScriptingFloatChannel, default=1, const [1,80]}
       foot_l_fk_ctrl.Scale.Y {type=MovieSceneScriptingFloatChannel, default=1, const [1,80]}
       foot_l_fk_ctrl.Scale.Z {type=MovieSceneScriptingFloatChannel, default=1, const [1,80]}
       ball_l_fk_ctrl.Location.X {type=MovieSceneScriptingFloatChannel, default=0, const [0,80]}
       ball_l_fk_ctrl.Location.Y {type=MovieSceneScriptingFloatChannel, default=0, const [0,80]}
       ball_l_fk_ctrl.Location.Z {type=MovieSceneScriptingFloatChannel, default=0, const [0,80]}
       ball_l_fk_ctrl.Rotation.X {type=MovieSceneScriptingFloatChannel, default=0, const [0,80]}
       ball_l_fk_ctrl.Rotation.Y {type=MovieSceneScriptingFloatChannel, default=0, const [0,80]}
       ball_l_fk_ctrl.Rotation.Z {type=MovieSceneScriptingFloatChannel, default=0, const [0,80]}
       ball_l_fk_ctrl.Scale.X {type=MovieSceneScriptingFloatChannel, default=1, const [1,80]}
       ball_l_fk_ctrl.Scale.Y {type=MovieSceneScriptingFloatChannel, default=1, const [1,80]}
       ball_l_fk_ctrl.Scale.Z {type=MovieSceneScriptingFloatChannel, default=1, const [1,80]}
       foot_l_ik_ctrl.Location.X {
        type=MovieSceneScriptingFloatChannel
        default=6.4161
        thinned_from=80
        keys [[0,6.4161],[39.4875,6.4171],[40.5,6.4392],[41.5125,6.4202],[54.675,6.417],[56.7,6.4246],[79.9875,6.4161]]
       }
       foot_l_ik_ctrl.Location.Y {type=MovieSceneScriptingFloatChannel, default=21.9207, thinned_from=80, keys [[0,21.9207],[79.9875,21.9207]]}
       foot_l_ik_ctrl.Location.Z {
        type=MovieSceneScriptingFloatChannel
        default=-0.0001
        thinned_from=80
        keys [[0,-0.0001],[11.1375,-0.0002],[12.15,-0.0006],[13.1625,-0.0017],[14.175,-0.0021],[16.2,-0.0006],[17.2125,-0.0008],[18.225,-0.0008],[19.2375,-0.0003],[22.275,-0.0001],[34.425,-0.0001],[36.45,-0.0005],[38.475,-0.0002],[39.4875,-0.0013],[40.5,-0.0222],[41.5125,-0.0046],[42.525,-0.0002],[43.5375,-0.001],[44.55,-0.0008],[45.5625,-0.0003],[52.65,-0.0001],[53.6625,-0.0002],[54.675,-0.0007],[55.6875,-0.0041],[56.7,-0.0052],[57.7125,-0.0025],[58.725,-0.001],[59.7375,-0.0003],[60.75,-0.0001],[65.8125,-0.0007],[67.8375,-0.0012],[69.8625,-0.0004],[79.9875,-0.0001]]
       }
       foot_l_ik_ctrl.Rotation.X {
        type=MovieSceneScriptingFloatChannel
        default=0.027
        thinned_from=80
        keys [[0,0.027],[11.1375,0.0272],[12.15,0.0277],[13.1625,0.0292],[14.175,0.0298],[16.2,0.0272],[17.2125,0.027],[22.275,0.027],[28.35,0.0275],[34.425,0.027],[38.475,0.027],[39.4875,0.0273],[40.5,0.0373],[41.5125,0.0297],[42.525,0.0274],[43.5375,0.0278],[47.5875,0.0272],[52.65,0.027],[53.6625,0.0272],[54.675,0.0282],[55.6875,0.0349],[56.7,0.0367],[57.7125,0.0288],[58.725,0.0264],[59.7375,0.0264],[60.75,0.0265],[62.775,0.0274],[64.8,0.0276],[65.8125,0.0276],[67.8375,0.0281],[70.875,0.0272],[79.9875,0.027]]
       }
       foot_l_ik_ctrl.Rotation.Y {
        type=MovieSceneScriptingFloatChannel
        default=0.1388
        thinned_from=80
        keys [[0,0.1388],[13.1625,0.1386],[15.1875,0.1378],[20.25,0.1388],[25.3125,0.1386],[35.4375,0.1387],[38.475,0.1386],[39.4875,0.1392],[40.5,0.1499],[41.5125,0.1411],[42.525,0.1389],[43.5375,0.1394],[47.5875,0.1388],[52.65,0.1388],[54.675,0.1394],[55.6875,0.1429],[56.7,0.1424],[57.7125,0.1368],[58.725,0.137],[59.7375,0.138],[61.7625,0.1388],[63.7875,0.1387],[65.8125,0.1382],[66.825,0.1376],[67.8375,0.1374],[69.8625,0.1386],[79.9875,0.1388]]
       }
       foot_l_ik_ctrl.Rotation.Z {type=MovieSceneScriptingFloatChannel, default=10.5353, thinned_from=80, keys [[0,10.5353],[79.9875,10.5353]]}
       foot_l_ik_ctrl.Scale.X {type=MovieSceneScriptingFloatChannel, default=1, const [1,80]}
       foot_l_ik_ctrl.Scale.Y {type=MovieSceneScriptingFloatChannel, default=1, const [1,80]}
       foot_l_ik_ctrl.Scale.Z {type=MovieSceneScriptingFloatChannel, default=1, const [1,80]}
       ball_l_ik_ctrl.Location.X {type=MovieSceneScriptingFloatChannel, default=0, const [0,80]}
       ball_l_ik_ctrl.Location.Y {type=MovieSceneScriptingFloatChannel, default=0, const [0,80]}
       ball_l_ik_ctrl.Location.Z {type=MovieSceneScriptingFloatChannel, default=0, const [0,80]}
       ball_l_ik_ctrl.Rotation.X {type=MovieSceneScriptingFloatChannel, default=0, const [0,80]}
       ball_l_ik_ctrl.Rotation.Y {type=MovieSceneScriptingFloatChannel, default=0, const [0,80]}
       ball_l_ik_ctrl.Rotation.Z {type=MovieSceneScriptingFloatChannel, default=0, const [0,80]}
       ball_l_ik_ctrl.Scale.X {type=MovieSceneScriptingFloatChannel, default=1, const [1,80]}
       ball_l_ik_ctrl.Scale.Y {type=MovieSceneScriptingFloatChannel, default=1, const [1,80]}
       ball_l_ik_ctrl.Scale.Z {type=MovieSceneScriptingFloatChannel, default=1, const [1,80]}
       foot_bk1_l_ctrl.Location.X {type=MovieSceneScriptingFloatChannel, default=0, const [0,80]}
       foot_bk1_l_ctrl.Location.Y {type=MovieSceneScriptingFloatChannel, default=0, const [0,80]}
       foot_bk1_l_ctrl.Location.Z {type=MovieSceneScriptingFloatChannel, default=0, const [0,80]}
       foot_bk1_l_ctrl.Rotation.X {type=MovieSceneScriptingFloatChannel, default=0, const [0,80]}
       foot_bk1_l_ctrl.Rotation.Y {type=MovieSceneScriptingFloatChannel, default=0, const [0,80]}
       foot_bk1_l_ctrl.Rotation.Z {type=MovieSceneScriptingFloatChannel, default=0, const [0,80]}
       foot_bk1_l_ctrl.Scale.X {type=MovieSceneScriptingFloatChannel, default=1, const [1,80]}
       foot_bk1_l_ctrl.Scale.Y {type=MovieSceneScriptingFloatChannel, default=1, const [1,80]}
       foot_bk1_l_ctrl.Scale.Z {type=MovieSceneScriptingFloatChannel, default=1, const [1,80]}
       foot_roll_l_ctrl.Location.X {type=MovieSceneScriptingFloatChannel, default=0, const [0,80]}
       foot_roll_l_ctrl.Location.Y {type=MovieSceneScriptingFloatChannel, default=0, const [0,80]}
       foot_roll_l_ctrl.Location.Z {type=MovieSceneScriptingFloatChannel, default=0, const [0,80]}
       foot_roll_l_ctrl.Rotation.X {type=MovieSceneScriptingFloatChannel, default=0, const [0,80]}
       foot_roll_l_ctrl.Rotation.Y {type=MovieSceneScriptingFloatChannel, default=0, const [0,80]}
       foot_roll_l_ctrl.Rotation.Z {type=MovieSceneScriptingFloatChannel, default=0, const [0,80]}
       foot_roll_l_ctrl.Scale.X {type=MovieSceneScriptingFloatChannel, default=1, const [1,80]}
       foot_roll_l_ctrl.Scale.Y {type=MovieSceneScriptingFloatChannel, default=1, const [1,80]}
       foot_roll_l_ctrl.Scale.Z {type=MovieSceneScriptingFloatChannel, default=1, const [1,80]}
       roll_blend_l_ctrl {type=MovieSceneScriptingFloatChannel, default=1, const [1,80]}
       tip_l_ctrl.Location.X {type=MovieSceneScriptingFloatChannel, default=0, const [0,80]}
       tip_l_ctrl.Location.Y {type=MovieSceneScriptingFloatChannel, default=0, const [0,80]}
       tip_l_ctrl.Location.Z {type=MovieSceneScriptingFloatChannel, default=0, const [0,80]}
       tip_l_ctrl.Rotation.X {type=MovieSceneScriptingFloatChannel, default=0, const [0,80]}
       tip_l_ctrl.Rotation.Y {type=MovieSceneScriptingFloatChannel, default=0, const [0,80]}
       tip_l_ctrl.Rotation.Z {type=MovieSceneScriptingFloatChannel, default=0, const [0,80]}
       tip_l_ctrl.Scale.X {type=MovieSceneScriptingFloatChannel, default=1, const [1,80]}
       tip_l_ctrl.Scale.Y {type=MovieSceneScriptingFloatChannel, default=1, const [1,80]}
       tip_l_ctrl.Scale.Z {type=MovieSceneScriptingFloatChannel, default=1, const [1,80]}
       heel_l_ctrl.Location.X {type=MovieSceneScriptingFloatChannel, default=0, const [0,80]}
       heel_l_ctrl.Location.Y {type=MovieSceneScriptingFloatChannel, default=0, const [0,80]}
       heel_l_ctrl.Location.Z {type=MovieSceneScriptingFloatChannel, default=0, const [0,80]}
       heel_l_ctrl.Rotation.X {type=MovieSceneScriptingFloatChannel, default=0, const [0,80]}
       heel_l_ctrl.Rotation.Y {type=MovieSceneScriptingFloatChannel, default=0, const [0,80]}
       heel_l_ctrl.Rotation.Z {type=MovieSceneScriptingFloatChannel, default=0, const [0,80]}
       heel_l_ctrl.Scale.X {type=MovieSceneScriptingFloatChannel, default=1, const [1,80]}
       heel_l_ctrl.Scale.Y {type=MovieSceneScriptingFloatChannel, default=1, const [1,80]}
       heel_l_ctrl.Scale.Z {type=MovieSceneScriptingFloatChannel, default=1, const [1,80]}
       leg_l_softness {type=MovieSceneScriptingFloatChannel, default=0, const [0,80]}
       leg_l_pv_ik_ctrl.Location.X {
        type=MovieSceneScriptingFloatChannel
        default=-0.2696
        thinned_from=80
        keys [[0,-0.2696],[1.0125,-0.2643],[2.025,-0.2506],[3.0375,-0.2317],[4.05,-0.2102],[5.0625,-0.181],[7.0875,-0.1136],[8.1,-0.0888],[9.1125,-0.0831],[10.125,-0.0825],[11.1375,-0.0704],[12.15,-0.0387],[13.1625,0.0014],[14.175,0.0589],[15.1875,0.1364],[16.2,0.2104],[17.2125,0.2765],[18.225,0.3301],[19.2375,0.3589],[20.25,0.3613],[21.2625,0.346],[22.275,0.3277],[23.2875,0.3044],[24.3,0.2742],[26.325,0.2072],[27.3375,0.1742],[28.35,0.1435],[29.3625,0.1173],[30.375,0.0939],[31.3875,0.0728],[32.4,0.0548],[33.4125,0.0397],[34.425,0.0278],[35.4375,0.0103],[36.45,-0.022],[37.4625,-0.0634],[38.475,-0.1013],[39.4875,-0.1345],[40.5,-0.1633],[41.5125,-0.1364],[42.525,-0.1003],[43.5375,-0.1034],[44.55,-0.1082],[45.5625,-0.1009],[46.575,-0.0892],[47.5875,-0.0839],[49.6125,-0.0776],[50.625,-0.08],[51.6375,-0.0947],[52.65,-0.1251],[53.6625,-0.1693],[54.675,-0.2223],[55.6875,-0.2885],[56.7,-0.3192],[57.7125,-0.2465],[58.725,-0.1191],[59.7375,-0.0068],[60.75,0.0711],[61.7625,0.1084],[62.775,0.1158],[63.7875,0.1035],[64.8,0.0785],[65.8125,0.0469],[66.825,0.0008],[67.8375,-0.059],[68.85,-0.118],[69.8625,-0.1674],[70.875,-0.21],[71.8875,-0.2533],[72.9,-0.2943],[73.9125,-0.3289],[74.925,-0.3524],[75.9375,-0.3574],[76.95,-0.3493],[77.9625,-0.3245],[78.975,-0.2884],[79.9875,-0.2696]]
       }
       leg_l_pv_ik_ctrl.Location.Y {
        type=MovieSceneScriptingFloatChannel
        default=7.4021
        thinned_from=80
        keys [[0,7.4021],[1.0125,7.3866],[2.025,7.3467],[4.05,7.2322],[5.0625,7.1929],[10.125,7.135],[11.1375,7.1456],[12.15,7.205],[13.1625,7.4298],[14.175,7.7899],[15.1875,8.095],[16.2,8.1908],[17.2125,8.0029],[18.225,7.6565],[19.2375,7.3465],[20.25,7.2002],[21.2625,7.1909],[24.3,7.2616],[25.3125,7.2595],[27.3375,7.2002],[28.35,7.185],[29.3625,7.2074],[32.4,7.3071],[34.425,7.3378],[35.4375,7.4224],[36.45,7.5932],[37.4625,7.7841],[38.475,7.8974],[39.4875,7.7473],[40.5,6.7318],[41.5125,5.4685],[42.525,5.1807],[43.5375,5.5162],[44.55,5.9066],[45.5625,6.1642],[46.575,6.3144],[47.5875,6.4109],[48.6,6.4685],[49.6125,6.4936],[51.6375,6.5272],[52.65,6.591],[53.6625,6.7149],[54.675,6.917],[55.6875,7.3825],[56.7,7.9723],[57.7125,8.3568],[58.725,8.5477],[59.7375,8.628],[60.75,8.6417],[61.7625,8.6119],[62.775,8.5532],[65.8125,8.3154],[66.825,8.2039],[68.85,7.9258],[69.8625,7.8262],[70.875,7.8103],[73.9125,7.9969],[74.925,8.0334],[75.9375,7.9688],[76.95,7.8422],[78.975,7.4903],[79.9875,7.4021]]
       }
       leg_l_pv_ik_ctrl.Location.Z {
        type=MovieSceneScriptingFloatChannel
        default=8.6367
        thinned_from=80
        keys [[0,8.6367],[1.0125,8.6152],[2.025,8.56],[4.05,8.4025],[5.0625,8.3508],[10.125,8.2825],[11.1375,8.2993],[12.15,8.3877],[13.1625,8.7144],[14.175,9.2467],[15.1875,9.7122],[16.2,9.8681],[17.2125,9.5958],[18.225,9.0923],[19.2375,8.6483],[20.25,8.44],[21.2625,8.4245],[23.2875,8.4933],[24.3,8.514],[25.3125,8.506],[27.3375,8.4117],[28.35,8.3857],[29.3625,8.4136],[32.4,8.5463],[34.425,8.5862],[35.4375,8.7051],[36.45,8.9471],[37.4625,9.2202],[38.475,9.382],[39.4875,9.156],[40.5,7.7099],[41.5125,6.0398],[42.525,5.6801],[43.5375,6.105],[44.55,6.6099],[45.5625,6.9507],[46.575,7.1529],[47.5875,7.2834],[48.6,7.3617],[49.6125,7.3961],[50.625,7.4099],[51.6375,7.4394],[52.65,7.5222],[53.6625,7.6858],[54.675,7.9581],[55.6875,8.606],[56.7,9.4625],[57.7125,10.0523],[58.725,10.3643],[59.7375,10.5059],[60.75,10.5392],[61.7625,10.4985],[62.775,10.4085],[65.8125,10.0322],[66.825,9.8559],[68.85,9.4218],[69.8625,9.2674],[70.875,9.2379],[73.9125,9.4979],[74.925,9.5491],[75.9375,9.452],[76.95,9.2657],[78.975,8.7609],[79.9875,8.6367]]
       }
       leg_l_pv_ik_ctrl.Rotation.X {
        type=MovieSceneScriptingFloatChannel
        default=-11.1718
        thinned_from=80
        keys [[0,-11.1718],[1.0125,-11.1461],[2.025,-11.08],[4.05,-10.891],[5.0625,-10.8287],[10.125,-10.7462],[11.1375,-10.7663],[12.15,-10.8723],[13.1625,-11.2633],[14.175,-11.8965],[15.1875,-12.4461],[16.2,-12.6291],[17.2125,-12.3083],[18.225,-11.7119],[19.2375,-11.1825],[20.25,-10.9331],[21.2625,-10.9146],[24.3,-11.0223],[25.3125,-11.0128],[27.3375,-10.9001],[28.35,-10.869],[29.3625,-10.9026],[32.4,-11.062],[34.425,-11.11],[35.4375,-11.2521],[36.45,-11.541],[37.4625,-11.8657],[38.475,-12.0575],[39.4875,-11.7899],[40.5,-10.055],[41.5125,-8.0016],[42.525,-7.5516],[43.5375,-8.0827],[44.55,-8.7088],[45.5625,-9.1283],[46.575,-9.3762],[47.5875,-9.5357],[48.6,-9.6312],[49.6125,-9.6731],[51.6375,-9.726],[52.65,-9.8269],[53.6625,-10.026],[54.675,-10.3559],[55.6875,-11.1352],[56.7,-12.1537],[57.7125,-12.8475],[58.725,-13.2118],[59.7375,-13.3764],[60.75,-13.4148],[61.7625,-13.3672],[62.775,-13.2622],[65.8125,-12.8226],[66.825,-12.6157],[68.85,-12.1047],[69.8625,-11.9222],[70.875,-11.8874],[73.9125,-12.1956],[74.925,-12.2561],[75.9375,-12.1416],[76.95,-11.921],[78.975,-11.3203],[79.9875,-11.1718]]
       }
       leg_l_pv_ik_ctrl.Rotation.Y {
        type=MovieSceneScriptingFloatChannel
        default=0.0414
        thinned_from=80
        keys [[0,0.0414],[1.0125,0.0412],[4.05,0.0394],[5.0625,0.0398],[6.075,0.041],[7.0875,0.0425],[8.1,0.0437],[10.125,0.0433],[11.1375,0.0443],[12.15,0.0477],[13.1625,0.0566],[15.1875,0.0862],[16.2,0.0936],[17.2125,0.0902],[18.225,0.0811],[19.2375,0.0728],[20.25,0.0686],[21.2625,0.0676],[23.2875,0.0669],[24.3,0.0659],[25.3125,0.064],[26.325,0.0615],[27.3375,0.0588],[28.35,0.0567],[29.3625,0.056],[32.4,0.0557],[33.4125,0.0552],[34.425,0.0552],[35.4375,0.0569],[36.45,0.0606],[37.4625,0.0647],[38.475,0.0667],[39.4875,0.0598],[40.5,0.0291],[41.5125,0.0026],[42.525,-0.0003],[43.5375,0.0051],[44.55,0.0121],[45.5625,0.0178],[46.575,0.0218],[47.5875,0.0243],[48.6,0.0259],[49.6125,0.0266],[50.625,0.0267],[52.65,0.0265],[53.6625,0.0272],[54.675,0.0298],[55.6875,0.0399],[57.7125,0.0761],[58.725,0.0903],[59.7375,0.0995],[60.75,0.1041],[61.7625,0.1048],[62.775,0.1028],[63.7875,0.0991],[64.8,0.0947],[65.8125,0.0898],[66.825,0.0831],[67.8375,0.0748],[68.85,0.0668],[69.8625,0.0607],[70.875,0.0579],[71.8875,0.0575],[72.9,0.0577],[73.9125,0.0582],[74.925,0.0582],[75.9375,0.0557],[76.95,0.0516],[77.9625,0.0472],[78.975,0.0432],[79.9875,0.0414]]
       }
       leg_l_pv_ik_ctrl.Rotation.Z {
        type=MovieSceneScriptingFloatChannel
        default=0.3518
        thinned_from=80
        keys [[0,0.3518],[1.0125,0.3454],[2.025,0.329],[3.0375,0.3062],[4.05,0.2805],[5.0625,0.2459],[7.0875,0.1668],[8.1,0.1378],[9.1125,0.1309],[10.125,0.1301],[11.1375,0.1161],[12.15,0.0802],[13.1625,0.0377],[14.175,-0.0215],[16.2,-0.186],[17.2125,-0.2666],[18.225,-0.3375],[19.2375,-0.3793],[20.25,-0.386],[21.2625,-0.3684],[22.275,-0.3465],[23.2875,-0.3185],[24.3,-0.283],[26.325,-0.2056],[27.3375,-0.168],[28.35,-0.1325],[29.3625,-0.1016],[30.375,-0.0734],[31.3875,-0.0482],[32.4,-0.0267],[33.4125,-0.009],[34.425,0.0053],[35.4375,0.0272],[36.45,0.068],[37.4625,0.1194],[38.475,0.165],[39.4875,0.2006],[40.5,0.2195],[41.5125,0.1723],[42.525,0.1257],[43.5375,0.1329],[44.55,0.143],[45.5625,0.1375],[46.575,0.1256],[47.5875,0.1206],[48.6,0.1176],[49.6125,0.1143],[50.625,0.1173],[51.6375,0.135],[52.65,0.1718],[53.6625,0.2256],[54.675,0.2904],[55.6875,0.3735],[56.7,0.4167],[57.7125,0.3393],[58.725,0.1977],[59.7375,0.0718],[60.75,-0.0162],[61.7625,-0.0592],[62.775,-0.0691],[63.7875,-0.057],[64.8,-0.0303],[65.8125,0.0038],[66.825,0.054],[67.8375,0.1197],[68.85,0.1848],[69.8625,0.2399],[70.875,0.2887],[71.8875,0.3394],[72.9,0.3875],[73.9125,0.4282],[74.925,0.4556],[75.9375,0.4606],[76.95,0.4496],[77.9625,0.4189],[78.975,0.3747],[79.9875,0.3518]]
       }
       leg_l_pv_ik_ctrl.Scale.X {type=MovieSceneScriptingFloatChannel, default=1, const [1,80]}
       leg_l_pv_ik_ctrl.Scale.Y {type=MovieSceneScriptingFloatChannel, default=1, const [1,80]}
       leg_l_pv_ik_ctrl.Scale.Z {type=MovieSceneScriptingFloatChannel, default=1, const [1,80]}
       thigh_r_fk_ctrl.Location.X {type=MovieSceneScriptingFloatChannel, default=0, thinned_from=80, keys [[0,0],[79.9875,0]]}
       thigh_r_fk_ctrl.Location.Y {type=MovieSceneScriptingFloatChannel, default=0, thinned_from=80, keys [[0,0],[79.9875,0]]}
       thigh_r_fk_ctrl.Location.Z {type=MovieSceneScriptingFloatChannel, default=0, thinned_from=80, keys [[0,0],[79.9875,0]]}
       thigh_r_fk_ctrl.Rotation.X {
        type=MovieSceneScriptingFloatChannel
        default=17.9915
        thinned_from=80
        keys [[0,17.9915],[1.0125,17.952],[2.025,17.8473],[4.05,17.5145],[5.0625,17.2135],[6.075,16.822],[7.0875,16.3959],[8.1,16.052],[9.1125,15.927],[10.125,15.8382],[11.1375,15.5592],[12.15,14.9434],[13.1625,13.8998],[14.175,12.3743],[15.1875,10.7804],[16.2,9.693],[17.2125,9.2798],[18.225,9.2019],[21.2625,9.2126],[22.275,9.2564],[23.2875,9.4174],[24.3,9.7323],[25.3125,10.1778],[26.325,10.73],[27.3375,11.3601],[29.3625,12.7172],[30.375,13.3684],[31.3875,13.9502],[32.4,14.4247],[33.4125,14.745],[34.425,14.8955],[36.45,14.9576],[37.4625,15.0196],[38.475,15.1602],[39.4875,15.5709],[40.5,17.1958],[41.5125,18.9319],[42.525,19.2342],[43.5375,18.8839],[44.55,18.4509],[45.5625,18.0918],[46.575,17.8488],[47.5875,17.7621],[48.6,17.7895],[51.6375,18.0542],[52.65,18.0993],[53.6625,17.9955],[54.675,17.5231],[55.6875,16.0282],[56.7,13.816],[57.7125,11.9224],[58.725,10.5408],[59.7375,9.5842],[60.75,9.0707],[61.7625,9.0301],[62.775,9.3423],[63.7875,9.9307],[64.8,10.7216],[65.8125,11.6214],[66.825,12.6684],[67.8375,13.8522],[68.85,14.9813],[69.8625,15.8557],[70.875,16.4349],[71.8875,16.8684],[72.9,17.2461],[73.9125,17.5624],[74.925,17.793],[75.9375,17.9795],[76.95,18.1257],[77.9625,18.1479],[79.9875,17.9915]]
       }
       thigh_r_fk_ctrl.Rotation.Y {
        type=MovieSceneScriptingFloatChannel
        default=10.9852
        thinned_from=80
        keys [[0,10.9852],[1.0125,10.9642],[3.0375,10.8378],[7.0875,10.4726],[8.1,10.3977],[11.1375,10.3507],[12.15,10.2822],[13.1625,9.9797],[15.1875,8.8475],[16.2,8.5056],[17.2125,8.4743],[18.225,8.5064],[19.2375,8.4355],[20.25,8.3975],[21.2625,8.4271],[22.275,8.5322],[23.2875,8.7209],[25.3125,9.2126],[26.325,9.437],[27.3375,9.6097],[29.3625,9.8604],[30.375,9.9534],[31.3875,10.0175],[34.425,10.1159],[35.4375,10.183],[37.4625,10.4273],[38.475,10.5171],[39.4875,10.568],[40.5,10.8759],[41.5125,11.4977],[42.525,11.9499],[43.5375,12.1918],[44.55,12.3555],[45.5625,12.392],[46.575,12.3199],[48.6,12.0368],[49.6125,11.8441],[50.625,11.5783],[51.6375,11.2125],[52.65,10.7625],[53.6625,10.2393],[54.675,9.5435],[55.6875,8.38],[56.7,7.1669],[57.7125,6.8863],[58.725,7.2991],[59.7375,7.8036],[61.7625,8.8535],[62.775,9.2629],[63.7875,9.5311],[64.8,9.7074],[65.8125,9.8356],[66.825,10.0091],[68.85,10.5335],[69.8625,10.7648],[70.875,10.9476],[71.8875,11.1053],[72.9,11.2298],[73.9125,11.3246],[74.925,11.3763],[75.9375,11.3495],[76.95,11.2737],[78.975,11.0429],[79.9875,10.9852]]
       }
       thigh_r_fk_ctrl.Rotation.Z {
        type=MovieSceneScriptingFloatChannel
        default=-6.2638
        thinned_from=80
        keys [[0,-6.2638],[2.025,-6.2739],[3.0375,-6.3002],[4.05,-6.3521],[5.0625,-6.4193],[6.075,-6.5055],[7.0875,-6.6097],[8.1,-6.7504],[9.1125,-6.9344],[10.125,-7.1439],[11.1375,-7.3946],[13.1625,-7.9614],[14.175,-8.2869],[15.1875,-8.553],[16.2,-8.6],[17.2125,-8.436],[19.2375,-7.8794],[20.25,-7.882],[21.2625,-8.1113],[23.2875,-8.684],[24.3,-8.8812],[25.3125,-8.939],[26.325,-8.8107],[27.3375,-8.4941],[28.35,-8.0204],[29.3625,-7.4198],[30.375,-6.8001],[31.3875,-6.2382],[32.4,-5.7929],[33.4125,-5.5521],[34.425,-5.5221],[35.4375,-5.6978],[36.45,-6.0543],[37.4625,-6.4538],[38.475,-6.6952],[39.4875,-6.6681],[40.5,-6.0339],[41.5125,-5.1947],[42.525,-5.041],[44.55,-5.7368],[45.5625,-5.8691],[46.575,-5.8433],[47.5875,-5.691],[48.6,-5.4108],[49.6125,-5.0576],[50.625,-4.7228],[51.6375,-4.5124],[52.65,-4.4677],[53.6625,-4.6598],[54.675,-5.0807],[55.6875,-6.2045],[56.7,-7.653],[57.7125,-8.4064],[58.725,-8.7251],[59.7375,-9.0062],[60.75,-9.2368],[61.7625,-9.3262],[62.775,-9.2263],[63.7875,-8.9627],[65.8125,-8.293],[66.825,-8.1636],[67.8375,-8.1794],[68.85,-8.1058],[70.875,-7.7912],[72.9,-7.6364],[74.925,-7.3352],[75.9375,-7.1262],[77.9625,-6.5952],[78.975,-6.3646],[79.9875,-6.2638]]
       }
       thigh_r_fk_ctrl.Scale.X {type=MovieSceneScriptingFloatChannel, default=1, thinned_from=80, keys [[0,1],[79.9875,1]]}
       thigh_r_fk_ctrl.Scale.Y {type=MovieSceneScriptingFloatChannel, default=1, const [1,80]}
       thigh_r_fk_ctrl.Scale.Z {type=MovieSceneScriptingFloatChannel, default=1, const [1,80]}
       calf_r_fk_ctrl.Location.X {
        type=MovieSceneScriptingFloatChannel
        default=0.0977
        thinned_from=80
        keys [[0,0.0977],[1.0125,0.0975],[2.025,0.0969],[5.0625,0.0944],[7.0875,0.0936],[8.1,0.0934],[9.1125,0.0935],[10.125,0.094],[11.1375,0.0947],[12.15,0.096],[13.1625,0.0981],[14.175,0.101],[15.1875,0.1033],[16.2,0.1042],[17.2125,0.1028],[18.225,0.1],[19.2375,0.0974],[20.25,0.0967],[21.2625,0.0977],[24.3,0.1019],[25.3125,0.1028],[26.325,0.1029],[27.3375,0.1024],[28.35,0.1011],[31.3875,0.0953],[32.4,0.0937],[33.4125,0.0928],[34.425,0.0927],[35.4375,0.0938],[36.45,0.0962],[37.4625,0.099],[38.475,0.1009],[39.4875,0.1005],[40.5,0.0944],[41.5125,0.0873],[42.525,0.0863],[44.55,0.0923],[45.5625,0.0939],[46.575,0.0941],[47.5875,0.0937],[48.6,0.0929],[49.6125,0.0916],[51.6375,0.0887],[52.65,0.0879],[53.6625,0.0882],[54.675,0.0902],[55.6875,0.0963],[56.7,0.1046],[57.7125,0.1106],[58.725,0.1141],[59.7375,0.1161],[60.75,0.1174],[61.7625,0.1183],[62.775,0.1182],[63.7875,0.1167],[64.8,0.1139],[66.825,0.1068],[67.8375,0.1038],[68.85,0.1015],[69.8625,0.1007],[70.875,0.1012],[72.9,0.1037],[73.9125,0.1048],[74.925,0.1054],[75.9375,0.1047],[76.95,0.1032],[77.9625,0.1011],[78.975,0.0988],[79.9875,0.0977]]
       }
       calf_r_fk_ctrl.Location.Y {
        type=MovieSceneScriptingFloatChannel
        default=0.0042
        thinned_from=80
        keys [[0,0.0042],[7.0875,0.0031],[10.125,0.0032],[12.15,0.0037],[13.1625,0.0043],[15.1875,0.0058],[16.2,0.006],[17.2125,0.0056],[19.2375,0.0041],[20.25,0.0039],[25.3125,0.0056],[26.325,0.0057],[28.35,0.0051],[32.4,0.0031],[33.4125,0.0029],[34.425,0.0029],[35.4375,0.0031],[37.4625,0.0045],[38.475,0.0051],[39.4875,0.0049],[41.5125,0.0016],[42.525,0.0014],[44.55,0.0028],[45.5625,0.0031],[47.5875,0.0031],[52.65,0.0017],[53.6625,0.0018],[54.675,0.0022],[55.6875,0.0038],[56.7,0.0062],[57.7125,0.0081],[58.725,0.0093],[59.7375,0.0101],[61.7625,0.0109],[62.775,0.0109],[63.7875,0.0103],[64.8,0.0093],[66.825,0.0069],[67.8375,0.0059],[68.85,0.0052],[69.8625,0.005],[70.875,0.0051],[73.9125,0.0062],[74.925,0.0064],[75.9375,0.0062],[78.975,0.0045],[79.9875,0.0042]]
       }
       calf_r_fk_ctrl.Location.Z {
        type=MovieSceneScriptingFloatChannel
        default=0.0033
        thinned_from=80
        keys [[0,0.0033],[39.4875,0.0034],[41.5125,0.003],[54.675,0.0031],[58.725,0.0038],[79.9875,0.0033]]
       }
       calf_r_fk_ctrl.Rotation.X {
        type=MovieSceneScriptingFloatChannel
        default=0.8291
        thinned_from=80
        keys [[0,0.8291],[1.0125,0.8273],[5.0625,0.8041],[7.0875,0.7973],[8.1,0.7963],[9.1125,0.7972],[10.125,0.8003],[11.1375,0.8062],[12.15,0.8158],[13.1625,0.8324],[14.175,0.8539],[15.1875,0.8719],[16.2,0.8782],[17.2125,0.8676],[19.2375,0.8264],[20.25,0.8216],[21.2625,0.829],[24.3,0.861],[25.3125,0.8676],[26.325,0.8688],[27.3375,0.8643],[28.35,0.8547],[31.3875,0.8109],[32.4,0.7987],[33.4125,0.7916],[34.425,0.791],[35.4375,0.7995],[36.45,0.8178],[37.4625,0.8389],[38.475,0.8533],[39.4875,0.8504],[40.5,0.8037],[41.5125,0.7487],[42.525,0.7414],[44.55,0.7877],[45.5625,0.7995],[46.575,0.8015],[47.5875,0.7987],[48.6,0.7922],[49.6125,0.7822],[51.6375,0.7601],[52.65,0.7538],[53.6625,0.756],[54.675,0.7714],[55.6875,0.8184],[56.7,0.8811],[57.7125,0.9258],[58.725,0.9517],[59.7375,0.9665],[60.75,0.9764],[61.7625,0.9826],[62.775,0.9818],[63.7875,0.9708],[64.8,0.9507],[66.825,0.8981],[67.8375,0.8749],[68.85,0.8582],[69.8625,0.8516],[70.875,0.8556],[73.9125,0.8828],[74.925,0.8876],[75.9375,0.8822],[76.95,0.8706],[78.975,0.8375],[79.9875,0.8291]]
       }
       calf_r_fk_ctrl.Rotation.Y {
        type=MovieSceneScriptingFloatChannel
        default=0.2032
        thinned_from=80
        keys [[0,0.2032],[2.025,0.2024],[6.075,0.1995],[8.1,0.199],[10.125,0.1995],[11.1375,0.2003],[12.15,0.2015],[13.1625,0.2035],[14.175,0.206],[15.1875,0.208],[16.2,0.2087],[17.2125,0.2075],[19.2375,0.2028],[20.25,0.2022],[24.3,0.2068],[25.3125,0.2076],[26.325,0.2077],[27.3375,0.2072],[28.35,0.2061],[32.4,0.1993],[33.4125,0.1984],[34.425,0.1983],[35.4375,0.1994],[37.4625,0.2043],[38.475,0.206],[39.4875,0.2056],[40.5,0.1996],[41.5125,0.1923],[42.525,0.1913],[44.55,0.1979],[45.5625,0.1995],[46.575,0.1997],[47.5875,0.1993],[48.6,0.1985],[51.6375,0.1941],[52.65,0.1932],[53.6625,0.1935],[54.675,0.1956],[55.6875,0.2017],[56.7,0.2088],[57.7125,0.2132],[58.725,0.2154],[59.7375,0.2166],[61.7625,0.2177],[62.775,0.2177],[63.7875,0.2169],[64.8,0.2153],[67.8375,0.2083],[68.85,0.2065],[69.8625,0.2058],[70.875,0.2062],[73.9125,0.2092],[74.925,0.2097],[75.9375,0.2091],[76.95,0.2079],[78.975,0.2042],[79.9875,0.2032]]
       }
       calf_r_fk_ctrl.Rotation.Z {
        type=MovieSceneScriptingFloatChannel
        default=23.3351
        thinned_from=80
        keys [[0,23.3351],[1.0125,23.276],[2.025,23.1266],[5.0625,22.5364],[6.075,22.4049],[7.0875,22.3208],[8.1,22.2908],[9.1125,22.3193],[10.125,22.4177],[11.1375,22.604],[12.15,22.9084],[13.1625,23.4391],[14.175,24.1322],[15.1875,24.7179],[16.2,24.9242],[17.2125,24.5782],[18.225,23.8844],[19.2375,23.2485],[20.25,23.0943],[21.2625,23.3322],[24.3,24.3623],[25.3125,24.5789],[26.325,24.6165],[27.3375,24.4699],[28.35,24.1605],[31.3875,22.7534],[32.4,22.3644],[33.4125,22.1426],[34.425,22.1238],[35.4375,22.3907],[36.45,22.9724],[37.4625,23.6505],[38.475,24.1152],[39.4875,24.02],[40.5,22.5354],[41.5125,20.8007],[42.525,20.5748],[44.55,22.0205],[45.5625,22.3924],[46.575,22.4542],[47.5875,22.3661],[48.6,22.1605],[49.6125,21.845],[51.6375,21.1537],[52.65,20.959],[53.6625,21.0279],[54.675,21.5064],[55.6875,22.9975],[56.7,25.0281],[57.7125,26.5047],[58.725,27.379],[59.7375,27.8819],[60.75,28.2217],[61.7625,28.4358],[62.775,28.4078],[63.7875,28.0294],[64.8,27.343],[66.825,25.5818],[67.8375,24.8176],[68.85,24.2738],[69.8625,24.0569],[70.875,24.1874],[72.9,24.8069],[73.9125,25.077],[74.925,25.233],[75.9375,25.0576],[76.95,24.6772],[78.975,23.603],[79.9875,23.3351]]
       }
       calf_r_fk_ctrl.Scale.X {type=MovieSceneScriptingFloatChannel, default=1, const [1,80]}
       calf_r_fk_ctrl.Scale.Y {type=MovieSceneScriptingFloatChannel, default=1, const [1,80]}
       calf_r_fk_ctrl.Scale.Z {type=MovieSceneScriptingFloatChannel, default=1, const [1,80]}
       foot_r_fk_ctrl.Location.X {type=MovieSceneScriptingFloatChannel, default=0, thinned_from=80, keys [[0,0],[79.9875,0]]}
       foot_r_fk_ctrl.Location.Y {type=MovieSceneScriptingFloatChannel, default=0, thinned_from=80, keys [[0,0],[79.9875,0]]}
       foot_r_fk_ctrl.Location.Z {type=MovieSceneScriptingFloatChannel, default=0, thinned_from=80, keys [[0,0],[79.9875,0]]}
       foot_r_fk_ctrl.Rotation.X {
        type=MovieSceneScriptingFloatChannel
        default=-2.2655
        thinned_from=80
        keys [[0,-2.2655],[1.0125,-2.2603],[2.025,-2.2468],[4.05,-2.2058],[6.075,-2.1482],[7.0875,-2.1162],[8.1,-2.091],[9.1125,-2.0821],[10.125,-2.0788],[11.1375,-2.0669],[12.15,-2.0378],[13.1625,-1.9867],[15.1875,-1.8328],[16.2,-1.7904],[18.225,-1.7806],[19.2375,-1.7575],[20.25,-1.7398],[21.2625,-1.732],[22.275,-1.7337],[23.2875,-1.75],[24.3,-1.7793],[25.3125,-1.8169],[28.35,-1.94],[29.3625,-1.9746],[30.375,-2.0047],[31.3875,-2.0295],[32.4,-2.0479],[33.4125,-2.0596],[35.4375,-2.0678],[38.475,-2.103],[40.5,-2.0936],[42.525,-2.1595],[44.55,-2.1901],[46.575,-2.1984],[47.5875,-2.2059],[49.6125,-2.2314],[50.625,-2.2381],[51.6375,-2.2346],[52.65,-2.2195],[53.6625,-2.1868],[54.675,-2.1244],[55.6875,-1.9896],[56.7,-1.825],[57.7125,-1.7455],[58.725,-1.7358],[59.7375,-1.7438],[60.75,-1.7692],[61.7625,-1.8139],[62.775,-1.8623],[63.7875,-1.9046],[64.8,-1.9372],[66.825,-1.9857],[67.8375,-2.0215],[68.85,-2.0688],[70.875,-2.1698],[71.8875,-2.2144],[73.9125,-2.2902],[74.925,-2.3152],[75.9375,-2.3235],[76.95,-2.3216],[77.9625,-2.3056],[78.975,-2.2795],[79.9875,-2.2655]]
       }
       foot_r_fk_ctrl.Rotation.Y {
        type=MovieSceneScriptingFloatChannel
        default=-9.5261
        thinned_from=80
        keys [[0,-9.5261],[11.1375,-9.3605],[15.1875,-9.2072],[17.2125,-9.1624],[18.225,-9.1068],[20.25,-8.9523],[21.2625,-8.9097],[22.275,-8.8898],[24.3,-8.8996],[26.325,-8.9466],[27.3375,-8.9842],[28.35,-9.0446],[31.3875,-9.3301],[32.4,-9.4031],[33.4125,-9.4426],[37.4625,-9.5107],[38.475,-9.5113],[39.4875,-9.4366],[40.5,-9.2221],[41.5125,-9.1067],[42.525,-9.1647],[44.55,-9.3194],[46.575,-9.4237],[49.6125,-9.5182],[50.625,-9.5263],[51.6375,-9.5046],[52.65,-9.4527],[53.6625,-9.3579],[54.675,-9.1782],[56.7,-8.4957],[57.7125,-8.4248],[58.725,-8.5461],[59.7375,-8.6862],[60.75,-8.8076],[61.7625,-8.9042],[63.7875,-9.0568],[67.8375,-9.4504],[68.85,-9.5234],[69.8625,-9.5636],[74.925,-9.65],[76.95,-9.6139],[78.975,-9.5438],[79.9875,-9.5261]]
       }
       foot_r_fk_ctrl.Rotation.Z {
        type=MovieSceneScriptingFloatChannel
        default=-15.1146
        thinned_from=80
        keys [[0,-15.1146],[1.0125,-15.0804],[2.025,-14.9926],[6.075,-14.4873],[7.0875,-14.3889],[8.1,-14.3199],[9.1125,-14.2889],[11.1375,-14.3068],[12.15,-14.35],[14.175,-14.5873],[15.1875,-14.6129],[16.2,-14.476],[17.2125,-14.1071],[18.225,-13.652],[19.2375,-13.32],[20.25,-13.2564],[21.2625,-13.3977],[24.3,-13.9972],[25.3125,-14.1626],[26.325,-14.2616],[27.3375,-14.2949],[29.3625,-14.2522],[32.4,-14.1134],[33.4125,-14.0966],[34.425,-14.1377],[35.4375,-14.3033],[36.45,-14.6237],[37.4625,-14.9987],[38.475,-15.2786],[39.4875,-15.355],[40.5,-14.9968],[41.5125,-14.4449],[42.525,-14.3358],[44.55,-14.911],[45.5625,-15.0253],[46.575,-14.9897],[47.5875,-14.8919],[48.6,-14.7465],[51.6375,-14.1378],[52.65,-14.0171],[53.6625,-14.0348],[54.675,-14.2309],[55.6875,-14.7958],[56.7,-15.4972],[57.7125,-15.9282],[58.725,-16.1089],[61.7625,-16.2249],[62.775,-16.2147],[63.7875,-16.1102],[64.8,-15.9308],[66.825,-15.4923],[67.8375,-15.3234],[68.85,-15.228],[69.8625,-15.2281],[70.875,-15.3467],[73.9125,-15.9026],[74.925,-16.0093],[75.9375,-15.9456],[76.95,-15.776],[77.9625,-15.5293],[78.975,-15.2502],[79.9875,-15.1146]]
       }
       foot_r_fk_ctrl.Scale.X {type=MovieSceneScriptingFloatChannel, default=1, const [1,80]}
       foot_r_fk_ctrl.Scale.Y {type=MovieSceneScriptingFloatChannel, default=1, const [1,80]}
       foot_r_fk_ctrl.Scale.Z {type=MovieSceneScriptingFloatChannel, default=1, const [1,80]}
       ball_r_fk_ctrl.Location.X {type=MovieSceneScriptingFloatChannel, default=0, const [0,80]}
       ball_r_fk_ctrl.Location.Y {type=MovieSceneScriptingFloatChannel, default=0, const [0,80]}
       ball_r_fk_ctrl.Location.Z {type=MovieSceneScriptingFloatChannel, default=0, const [0,80]}
       ball_r_fk_ctrl.Rotation.X {type=MovieSceneScriptingFloatChannel, default=0, const [0,80]}
       ball_r_fk_ctrl.Rotation.Y {type=MovieSceneScriptingFloatChannel, default=0, const [0,80]}
       ball_r_fk_ctrl.Rotation.Z {type=MovieSceneScriptingFloatChannel, default=0, const [0,80]}
       ball_r_fk_ctrl.Scale.X {type=MovieSceneScriptingFloatChannel, default=1, const [1,80]}
       ball_r_fk_ctrl.Scale.Y {type=MovieSceneScriptingFloatChannel, default=1, const [1,80]}
       ball_r_fk_ctrl.Scale.Z {type=MovieSceneScriptingFloatChannel, default=1, const [1,80]}
       foot_r_ik_ctrl.Location.X {
        type=MovieSceneScriptingFloatChannel
        default=-0.7919
        thinned_from=80
        keys [[0,-0.7919],[14.175,-0.7931],[39.4875,-0.7919],[40.5,-0.8013],[41.5125,-0.795],[42.525,-0.7919],[53.6625,-0.7923],[54.675,-0.7939],[55.6875,-0.806],[56.7,-0.8077],[57.7125,-0.7927],[58.725,-0.7907],[65.8125,-0.7925],[79.9875,-0.7919]]
       }
       foot_r_ik_ctrl.Location.Y {type=MovieSceneScriptingFloatChannel, default=-22.7162, thinned_from=80, keys [[0,-22.7162],[79.9875,-22.7162]]}
       foot_r_ik_ctrl.Location.Z {
        type=MovieSceneScriptingFloatChannel
        default=0
        thinned_from=80
        keys [[0,0],[12.15,-0.0001],[15.1875,0.0002],[17.2125,-0.0002],[18.225,-0.0003],[19.2375,0],[25.3125,0.0001],[29.3625,-0.0007],[31.3875,-0.0005],[33.4125,0],[36.45,-0.0003],[38.475,0],[39.4875,-0.0003],[40.5,-0.0057],[41.5125,-0.0007],[42.525,-0.0003],[43.5375,-0.0008],[45.5625,0],[53.6625,0],[54.675,-0.0003],[55.6875,-0.0033],[56.7,-0.0041],[57.7125,-0.0013],[61.7625,0],[62.775,0],[65.8125,-0.0005],[67.8375,0.0001],[79.9875,0]]
       }
       foot_r_ik_ctrl.Rotation.X {
        type=MovieSceneScriptingFloatChannel
        default=0.027
        thinned_from=80
        keys [[0,0.027],[11.1375,0.027],[12.15,0.0274],[13.1625,0.0284],[14.175,0.0287],[16.2,0.027],[22.275,0.027],[28.35,0.0274],[34.425,0.027],[39.4875,0.027],[40.5,0.0328],[41.5125,0.0292],[42.525,0.0269],[52.65,0.027],[53.6625,0.0273],[54.675,0.0285],[55.6875,0.035],[56.7,0.0353],[57.7125,0.0264],[58.725,0.025],[59.7375,0.0259],[61.7625,0.0271],[62.775,0.0274],[79.9875,0.027]]
       }
       foot_r_ik_ctrl.Rotation.Y {
        type=MovieSceneScriptingFloatChannel
        default=0.1388
        thinned_from=80
        keys [[0,0.1388],[14.175,0.1385],[15.1875,0.139],[16.2,0.1391],[22.275,0.1389],[24.3,0.1391],[28.35,0.1376],[29.3625,0.1375],[31.3875,0.1379],[33.4125,0.1388],[38.475,0.1389],[39.4875,0.1385],[40.5,0.1305],[41.5125,0.1376],[42.525,0.1386],[43.5375,0.138],[45.5625,0.1388],[53.6625,0.1387],[54.675,0.1383],[55.6875,0.1325],[56.7,0.1301],[57.7125,0.1366],[58.725,0.1377],[61.7625,0.1388],[65.8125,0.1382],[67.8375,0.1391],[79.9875,0.1388]]
       }
       foot_r_ik_ctrl.Rotation.Z {type=MovieSceneScriptingFloatChannel, default=-36.3115, thinned_from=80, keys [[0,-36.3115],[79.9875,-36.3115]]}
       foot_r_ik_ctrl.Scale.X {type=MovieSceneScriptingFloatChannel, default=1, const [1,80]}
       foot_r_ik_ctrl.Scale.Y {type=MovieSceneScriptingFloatChannel, default=1, const [1,80]}
       foot_r_ik_ctrl.Scale.Z {type=MovieSceneScriptingFloatChannel, default=1, const [1,80]}
       ball_r_ik_ctrl.Location.X {type=MovieSceneScriptingFloatChannel, default=0, const [0,80]}
       ball_r_ik_ctrl.Location.Y {type=MovieSceneScriptingFloatChannel, default=0, const [0,80]}
       ball_r_ik_ctrl.Location.Z {type=MovieSceneScriptingFloatChannel, default=0, const [0,80]}
       ball_r_ik_ctrl.Rotation.X {type=MovieSceneScriptingFloatChannel, default=0, const [0,80]}
       ball_r_ik_ctrl.Rotation.Y {type=MovieSceneScriptingFloatChannel, default=0, const [0,80]}
       ball_r_ik_ctrl.Rotation.Z {type=MovieSceneScriptingFloatChannel, default=0, const [0,80]}
       ball_r_ik_ctrl.Scale.X {type=MovieSceneScriptingFloatChannel, default=1, const [1,80]}
       ball_r_ik_ctrl.Scale.Y {type=MovieSceneScriptingFloatChannel, default=1, const [1,80]}
       ball_r_ik_ctrl.Scale.Z {type=MovieSceneScriptingFloatChannel, default=1, const [1,80]}
       foot_bk1_r_ctrl.Location.X {type=MovieSceneScriptingFloatChannel, default=0, const [0,80]}
       foot_bk1_r_ctrl.Location.Y {type=MovieSceneScriptingFloatChannel, default=0, const [0,80]}
       foot_bk1_r_ctrl.Location.Z {type=MovieSceneScriptingFloatChannel, default=0, const [0,80]}
       foot_bk1_r_ctrl.Rotation.X {type=MovieSceneScriptingFloatChannel, default=0, const [0,80]}
       foot_bk1_r_ctrl.Rotation.Y {type=MovieSceneScriptingFloatChannel, default=0, const [0,80]}
       foot_bk1_r_ctrl.Rotation.Z {type=MovieSceneScriptingFloatChannel, default=0, const [0,80]}
       foot_bk1_r_ctrl.Scale.X {type=MovieSceneScriptingFloatChannel, default=1, const [1,80]}
       foot_bk1_r_ctrl.Scale.Y {type=MovieSceneScriptingFloatChannel, default=1, const [1,80]}
       foot_bk1_r_ctrl.Scale.Z {type=MovieSceneScriptingFloatChannel, default=1, const [1,80]}
       foot_roll_r_ctrl.Location.X {type=MovieSceneScriptingFloatChannel, default=0, const [0,80]}
       foot_roll_r_ctrl.Location.Y {type=MovieSceneScriptingFloatChannel, default=0, const [0,80]}
       foot_roll_r_ctrl.Location.Z {type=MovieSceneScriptingFloatChannel, default=0, const [0,80]}
       foot_roll_r_ctrl.Rotation.X {type=MovieSceneScriptingFloatChannel, default=0, const [0,80]}
       foot_roll_r_ctrl.Rotation.Y {type=MovieSceneScriptingFloatChannel, default=0, const [0,80]}
       foot_roll_r_ctrl.Rotation.Z {type=MovieSceneScriptingFloatChannel, default=0, const [0,80]}
       foot_roll_r_ctrl.Scale.X {type=MovieSceneScriptingFloatChannel, default=1, const [1,80]}
       foot_roll_r_ctrl.Scale.Y {type=MovieSceneScriptingFloatChannel, default=1, const [1,80]}
       foot_roll_r_ctrl.Scale.Z {type=MovieSceneScriptingFloatChannel, default=1, const [1,80]}
       roll_blend_r_ctrl {type=MovieSceneScriptingFloatChannel, default=1, const [1,80]}
       tip_r_ctrl.Location.X {type=MovieSceneScriptingFloatChannel, default=0, const [0,80]}
       tip_r_ctrl.Location.Y {type=MovieSceneScriptingFloatChannel, default=0, const [0,80]}
       tip_r_ctrl.Location.Z {type=MovieSceneScriptingFloatChannel, default=0, const [0,80]}
       tip_r_ctrl.Rotation.X {type=MovieSceneScriptingFloatChannel, default=0, const [0,80]}
       tip_r_ctrl.Rotation.Y {type=MovieSceneScriptingFloatChannel, default=0, const [0,80]}
       tip_r_ctrl.Rotation.Z {type=MovieSceneScriptingFloatChannel, default=0, const [0,80]}
       tip_r_ctrl.Scale.X {type=MovieSceneScriptingFloatChannel, default=1, const [1,80]}
       tip_r_ctrl.Scale.Y {type=MovieSceneScriptingFloatChannel, default=1, const [1,80]}
       tip_r_ctrl.Scale.Z {type=MovieSceneScriptingFloatChannel, default=1, const [1,80]}
       heel_r_ctrl.Location.X {type=MovieSceneScriptingFloatChannel, default=0, const [0,80]}
       heel_r_ctrl.Location.Y {type=MovieSceneScriptingFloatChannel, default=0, const [0,80]}
       heel_r_ctrl.Location.Z {type=MovieSceneScriptingFloatChannel, default=0, const [0,80]}
       heel_r_ctrl.Rotation.X {type=MovieSceneScriptingFloatChannel, default=0, const [0,80]}
       heel_r_ctrl.Rotation.Y {type=MovieSceneScriptingFloatChannel, default=0, const [0,80]}
       heel_r_ctrl.Rotation.Z {type=MovieSceneScriptingFloatChannel, default=0, const [0,80]}
       heel_r_ctrl.Scale.X {type=MovieSceneScriptingFloatChannel, default=1, const [1,80]}
       heel_r_ctrl.Scale.Y {type=MovieSceneScriptingFloatChannel, default=1, const [1,80]}
       heel_r_ctrl.Scale.Z {type=MovieSceneScriptingFloatChannel, default=1, const [1,80]}
       leg_r_softness {type=MovieSceneScriptingFloatChannel, default=0, const [0,80]}
       leg_r_pv_ik_ctrl.Location.X {
        type=MovieSceneScriptingFloatChannel
        default=1.0309
        thinned_from=80
        keys [[0,1.0309],[1.0125,1.0279],[2.025,1.0196],[3.0375,1.0071],[4.05,0.991],[5.0625,0.9655],[6.075,0.9322],[7.0875,0.8946],[8.1,0.8611],[10.125,0.8215],[11.1375,0.7844],[12.15,0.7181],[13.1625,0.6106],[14.175,0.4548],[15.1875,0.2963],[16.2,0.1931],[17.2125,0.1634],[18.225,0.1696],[19.2375,0.1785],[20.25,0.1732],[21.2625,0.1582],[22.275,0.1467],[23.2875,0.1471],[24.3,0.1644],[25.3125,0.1995],[26.325,0.2528],[27.3375,0.3221],[28.35,0.4067],[30.375,0.604],[31.3875,0.6927],[32.4,0.7644],[33.4125,0.8081],[34.425,0.825],[35.4375,0.8227],[37.4625,0.7979],[38.475,0.7962],[39.4875,0.8227],[40.5,0.9678],[41.5125,1.1426],[42.525,1.1873],[43.5375,1.1574],[44.55,1.1199],[45.5625,1.094],[46.575,1.0803],[47.5875,1.0807],[48.6,1.0939],[49.6125,1.1128],[50.625,1.1289],[51.6375,1.1329],[52.65,1.1217],[53.6625,1.0822],[54.675,0.9876],[55.6875,0.743],[56.7,0.4151],[57.7125,0.1998],[58.725,0.0935],[59.7375,0.0336],[60.75,0.0103],[61.7625,0.0259],[62.775,0.0738],[63.7875,0.1523],[64.8,0.256],[65.8125,0.3738],[67.8375,0.6166],[68.85,0.7287],[69.8625,0.8117],[70.875,0.8625],[71.8875,0.8971],[72.9,0.9274],[73.9125,0.9553],[76.95,1.0182],[77.9625,1.0295],[78.975,1.0316],[79.9875,1.0309]]
       }
       leg_r_pv_ik_ctrl.Location.Y {
        type=MovieSceneScriptingFloatChannel
        default=7.9528
        thinned_from=80
        keys [[0,7.9528],[1.0125,7.9354],[5.0625,7.7172],[7.0875,7.656],[8.1,7.6487],[9.1125,7.6584],[10.125,7.6891],[11.1375,7.7472],[12.15,7.842],[13.1625,8.0059],[14.175,8.2186],[15.1875,8.3978],[16.2,8.4624],[17.2125,8.3623],[19.2375,7.968],[20.25,7.9221],[21.2625,7.9938],[24.3,8.2988],[25.3125,8.3611],[26.325,8.3699],[27.3375,8.3239],[28.35,8.2292],[31.3875,7.7967],[32.4,7.676],[33.4125,7.6067],[34.425,7.6001],[35.4375,7.6809],[36.45,7.8565],[37.4625,8.0593],[38.475,8.1969],[39.4875,8.1674],[40.5,7.7168],[41.5125,7.1786],[42.525,7.1063],[44.55,7.5526],[45.5625,7.6666],[46.575,7.686],[47.5875,7.6594],[48.6,7.5965],[49.6125,7.4997],[51.6375,7.2876],[52.65,7.2285],[53.6625,7.2519],[54.675,7.4037],[55.6875,7.8675],[56.7,8.4832],[57.7125,8.9186],[58.725,9.1711],[59.7375,9.3148],[60.75,9.4107],[61.7625,9.4696],[62.775,9.4599],[63.7875,9.3513],[64.8,9.1541],[66.825,8.6402],[67.8375,8.4122],[68.85,8.2471],[69.8625,8.1789],[70.875,8.2146],[73.9125,8.47],[74.925,8.5141],[75.9375,8.4619],[76.95,8.3498],[78.975,8.0325],[79.9875,7.9528]]
       }
       leg_r_pv_ik_ctrl.Location.Z {
        type=MovieSceneScriptingFloatChannel
        default=9.6262
        thinned_from=80
        keys [[0,9.6262],[1.0125,9.5996],[2.025,9.5325],[5.0625,9.266],[6.075,9.2047],[7.0875,9.164],[8.1,9.1476],[9.1125,9.1583],[10.125,9.1999],[11.1375,9.2787],[12.15,9.4071],[13.1625,9.6325],[14.175,9.9265],[15.1875,10.1736],[16.2,10.2563],[17.2125,10.0985],[18.225,9.789],[19.2375,9.507],[20.25,9.438],[21.2625,9.5422],[24.3,10.0019],[25.3125,10.1022],[26.325,10.124],[27.3375,10.0649],[28.35,9.9346],[31.3875,9.3361],[32.4,9.1709],[33.4125,9.0772],[34.425,9.0705],[35.4375,9.188],[36.45,9.4443],[37.4625,9.7443],[38.475,9.9513],[39.4875,9.9113],[40.5,9.2657],[41.5125,8.52],[42.525,8.4256],[44.55,9.053],[45.5625,9.2146],[46.575,9.2406],[47.5875,9.2017],[48.6,9.1122],[49.6125,8.9751],[51.6375,8.6735],[52.65,8.5872],[53.6625,8.6136],[54.675,8.8143],[55.6875,9.449],[56.7,10.3239],[57.7125,10.9697],[58.725,11.3568],[59.7375,11.5806],[60.75,11.7338],[61.7625,11.8333],[62.775,11.825],[63.7875,11.6592],[64.8,11.3558],[66.825,10.5806],[67.8375,10.2486],[68.85,10.0158],[69.8625,9.9268],[70.875,9.9898],[72.9,10.2733],[73.9125,10.3971],[74.925,10.4693],[75.9375,10.3925],[76.95,10.2238],[77.9625,9.9946],[78.975,9.7454],[79.9875,9.6262]]
       }
       leg_r_pv_ik_ctrl.Rotation.X {
        type=MovieSceneScriptingFloatChannel
        default=-12.3399
        thinned_from=80
        keys [[0,-12.3399],[1.0125,-12.3086],[2.025,-12.2294],[5.0625,-11.9145],[6.075,-11.842],[7.0875,-11.7939],[8.1,-11.7746],[9.1125,-11.7874],[10.125,-11.8368],[11.1375,-11.9304],[12.15,-12.0827],[13.1625,-12.3493],[14.175,-12.696],[15.1875,-12.9862],[16.2,-13.0835],[17.2125,-12.899],[18.225,-12.5358],[19.2375,-12.2033],[20.25,-12.1219],[21.2625,-12.245],[24.3,-12.7858],[25.3125,-12.9031],[26.325,-12.9284],[27.3375,-12.8589],[28.35,-12.7057],[31.3875,-11.9988],[32.4,-11.8027],[33.4125,-11.6913],[34.425,-11.6832],[35.4375,-11.8228],[36.45,-12.1262],[37.4625,-12.4801],[38.475,-12.7234],[39.4875,-12.6764],[40.5,-11.9141],[41.5125,-11.0248],[42.525,-10.9114],[44.55,-11.661],[45.5625,-11.853],[46.575,-11.8839],[47.5875,-11.8377],[48.6,-11.7315],[49.6125,-11.5683],[51.6375,-11.2086],[52.65,-11.1054],[53.6625,-11.1372],[54.675,-11.3774],[55.6875,-12.1321],[56.7,-13.1614],[57.7125,-13.9132],[58.725,-14.3608],[59.7375,-14.6184],[60.75,-14.7943],[61.7625,-14.9083],[62.775,-14.8986],[63.7875,-14.7082],[64.8,-14.3588],[66.825,-13.4602],[67.8375,-13.0725],[68.85,-12.7994],[69.8625,-12.6946],[70.875,-12.7683],[72.9,-13.0999],[73.9125,-13.2444],[74.925,-13.3284],[75.9375,-13.2388],[76.95,-13.0416],[78.975,-12.4803],[79.9875,-12.3399]]
       }
       leg_r_pv_ik_ctrl.Rotation.Y {
        type=MovieSceneScriptingFloatChannel
        default=0.1267
        thinned_from=80
        keys [[0,0.1267],[1.0125,0.126],[2.025,0.124],[3.0375,0.1214],[6.075,0.1125],[7.0875,0.1098],[8.1,0.1079],[9.1125,0.1071],[11.1375,0.1071],[12.15,0.1069],[13.1625,0.1071],[14.175,0.1068],[15.1875,0.1054],[16.2,0.1025],[17.2125,0.0971],[18.225,0.0897],[19.2375,0.0833],[20.25,0.0814],[21.2625,0.0831],[22.275,0.0861],[23.2875,0.0901],[24.3,0.0947],[25.3125,0.0989],[26.325,0.102],[27.3375,0.1039],[28.35,0.1046],[29.3625,0.1047],[32.4,0.1038],[33.4125,0.1038],[34.425,0.1044],[35.4375,0.1069],[36.45,0.1122],[37.4625,0.1186],[38.475,0.1235],[39.4875,0.1238],[40.5,0.1158],[41.5125,0.1081],[42.525,0.1083],[43.5375,0.113],[44.55,0.1181],[45.5625,0.1204],[46.575,0.1203],[47.5875,0.1194],[48.6,0.1181],[49.6125,0.116],[51.6375,0.1107],[52.65,0.1084],[53.6625,0.107],[54.675,0.1067],[55.6875,0.1092],[56.7,0.115],[57.7125,0.122],[58.725,0.1278],[61.7625,0.1386],[62.775,0.1406],[63.7875,0.1393],[64.8,0.1354],[65.8125,0.1301],[66.825,0.1253],[67.8375,0.1225],[68.85,0.1219],[69.8625,0.1236],[70.875,0.1276],[72.9,0.1376],[73.9125,0.1421],[74.925,0.1449],[75.9375,0.1439],[76.95,0.1406],[77.9625,0.1355],[78.975,0.1296],[79.9875,0.1267]]
       }
       leg_r_pv_ik_ctrl.Rotation.Z {
        type=MovieSceneScriptingFloatChannel
        default=-1.1341
        thinned_from=80
        keys [[0,-1.1341],[1.0125,-1.1314],[2.025,-1.1238],[3.0375,-1.1119],[4.05,-1.096],[5.0625,-1.0689],[6.075,-1.0319],[7.0875,-0.9895],[8.1,-0.9512],[10.125,-0.9039],[11.1375,-0.859],[12.15,-0.7791],[13.1625,-0.6499],[15.1875,-0.2787],[16.2,-0.1595],[17.2125,-0.1282],[18.225,-0.1404],[19.2375,-0.1552],[20.25,-0.1502],[21.2625,-0.1312],[22.275,-0.1157],[23.2875,-0.1135],[24.3,-0.1309],[25.3125,-0.1694],[26.325,-0.2298],[27.3375,-0.3102],[28.35,-0.4094],[30.375,-0.6446],[31.3875,-0.7516],[32.4,-0.8384],[33.4125,-0.8915],[34.425,-0.9113],[35.4375,-0.9056],[37.4625,-0.8627],[38.475,-0.8556],[39.4875,-0.887],[40.5,-1.0713],[41.5125,-1.2968],[42.525,-1.3521],[43.5375,-1.3075],[44.55,-1.254],[45.5625,-1.2191],[46.575,-1.2025],[47.5875,-1.2041],[48.6,-1.2221],[49.6125,-1.2482],[50.625,-1.2717],[51.6375,-1.2808],[52.65,-1.2703],[53.6625,-1.2233],[54.675,-1.1071],[55.6875,-0.8068],[56.7,-0.4113],[57.7125,-0.155],[58.725,-0.0288],[59.7375,0.0419],[60.75,0.0702],[61.7625,0.0543],[62.775,0.0006],[63.7875,-0.0901],[64.8,-0.2118],[65.8125,-0.3518],[67.8375,-0.6429],[68.85,-0.7766],[69.8625,-0.874],[70.875,-0.9306],[71.8875,-0.9665],[72.9,-0.9974],[73.9125,-1.0259],[74.925,-1.049],[76.95,-1.1026],[77.9625,-1.1221],[78.975,-1.1315],[79.9875,-1.1341]]
       }
       leg_r_pv_ik_ctrl.Scale.X {type=MovieSceneScriptingFloatChannel, default=1, const [1,80]}
       leg_r_pv_ik_ctrl.Scale.Y {type=MovieSceneScriptingFloatChannel, default=1, const [1,80]}
       leg_r_pv_ik_ctrl.Scale.Z {type=MovieSceneScriptingFloatChannel, default=1, const [1,80]}
       upperarm_l_fk_ctrl.Location.X {type=MovieSceneScriptingFloatChannel, default=0, thinned_from=80, keys [[0,0],[79.9875,0]]}
       upperarm_l_fk_ctrl.Location.Y {type=MovieSceneScriptingFloatChannel, default=0, thinned_from=80, keys [[0,0],[79.9875,0]]}
       upperarm_l_fk_ctrl.Location.Z {type=MovieSceneScriptingFloatChannel, default=0, thinned_from=80, keys [[0,0],[79.9875,0]]}
       upperarm_l_fk_ctrl.Rotation.X {
        type=MovieSceneScriptingFloatChannel
        default=-5.7583
        thinned_from=80
        keys [[0,-5.7583],[2.025,-5.7294],[3.0375,-5.6339],[4.05,-5.4979],[5.0625,-5.6266],[6.075,-5.8858],[7.0875,-6.1042],[8.1,-5.8784],[9.1125,-5.2475],[10.125,-4.4213],[11.1375,-3.581],[12.15,-3.0541],[13.1625,-2.5184],[14.175,-1.7791],[15.1875,0.1676],[16.2,1.5286],[17.2125,2.523],[18.225,3.6822],[19.2375,4.8097],[20.25,5.5971],[21.2625,6.1899],[24.3,7.6796],[25.3125,8.1104],[26.325,8.3702],[27.3375,8.4104],[28.35,8.4089],[29.3625,8.6595],[30.375,9.21],[31.3875,9.8878],[32.4,10.5278],[33.4125,10.991],[34.425,11.2216],[35.4375,11.2336],[37.4625,10.9688],[38.475,10.9372],[39.4875,11.2011],[40.5,12.5797],[41.5125,13.2876],[42.525,12.0518],[43.5375,10.4898],[44.55,9.3428],[45.5625,8.7232],[46.575,8.4208],[47.5875,8.3069],[49.6125,8.3175],[50.625,8.1995],[52.65,7.6604],[53.6625,7.5354],[54.675,7.5901],[55.6875,8.0305],[56.7,10.4587],[57.7125,12.2688],[58.725,12.3889],[59.7375,11.8787],[60.75,11.2137],[61.7625,9.9111],[62.775,7.7733],[63.7875,5.4995],[64.8,3.3388],[65.8125,1.2455],[66.825,-0.8756],[67.8375,-2.9803],[68.85,-5.2281],[69.8625,-7.2201],[70.875,-8.605],[71.8875,-9.131],[72.9,-8.3364],[73.9125,-7.1832],[74.925,-5.999],[75.9375,-4.6881],[76.95,-4.0761],[77.9625,-4.4042],[78.975,-5.2269],[79.9875,-5.7583]]
       }
       upperarm_l_fk_ctrl.Rotation.Y {
        type=MovieSceneScriptingFloatChannel
        default=-10.6637
        thinned_from=80
        keys [[0,-10.6637],[1.0125,-10.6035],[2.025,-10.7511],[3.0375,-11.5304],[4.05,-13.3389],[5.0625,-16.0028],[7.0875,-22.0892],[8.1,-24.9828],[9.1125,-27.5535],[10.125,-29.3847],[11.1375,-30.1259],[12.15,-29.7369],[13.1625,-28.5692],[14.175,-26.6066],[15.1875,-24.2891],[16.2,-22.2556],[17.2125,-20.848],[18.225,-19.8113],[19.2375,-19.0307],[20.25,-18.4549],[21.2625,-18.172],[22.275,-18.2736],[23.2875,-18.7296],[24.3,-19.4767],[25.3125,-20.5361],[26.325,-21.9262],[27.3375,-23.5887],[28.35,-25.3198],[29.3625,-26.9889],[30.375,-28.4948],[31.3875,-29.6924],[32.4,-30.5369],[33.4125,-31.0458],[34.425,-31.29],[35.4375,-31.358],[36.45,-31.2908],[37.4625,-31.1345],[38.475,-30.8977],[39.4875,-30.5729],[40.5,-30.5734],[41.5125,-30.9298],[42.525,-31.0417],[43.5375,-30.7198],[44.55,-30.1256],[48.6,-27.2061],[49.6125,-26.6382],[50.625,-26.31],[52.65,-25.9043],[53.6625,-25.4694],[54.675,-24.3931],[55.6875,-22.0336],[56.7,-19.5755],[57.7125,-17.8168],[58.725,-16.872],[59.7375,-16.3672],[61.7625,-15.8029],[62.775,-15.4012],[63.7875,-15.049],[64.8,-14.7595],[65.8125,-14.5405],[66.825,-14.2514],[67.8375,-13.601],[68.85,-12.3452],[69.8625,-10.9086],[70.875,-9.7596],[71.8875,-9.4109],[72.9,-9.9884],[73.9125,-10.4394],[74.925,-10.7615],[75.9375,-11.2855],[76.95,-11.4556],[77.9625,-11.3173],[78.975,-10.9402],[79.9875,-10.6637]]
       }
       upperarm_l_fk_ctrl.Rotation.Z {
        type=MovieSceneScriptingFloatChannel
        default=64.7063
        thinned_from=80
        keys [[0,64.7063],[1.0125,66.2365],[3.0375,73.1076],[4.05,75.398],[5.0625,77.1815],[6.075,78.3755],[7.0875,78.5281],[8.1,77.5181],[11.1375,73.462],[13.1625,72.0112],[14.175,70.6327],[15.1875,66.7722],[16.2,64.5086],[17.2125,63.4536],[18.225,62.6559],[19.2375,62.1311],[21.2625,62.2199],[22.275,62.1143],[24.3,61.7037],[25.3125,61.7547],[26.325,62.2088],[28.35,63.8982],[29.3625,64.203],[30.375,63.9941],[32.4,62.8358],[33.4125,62.4698],[34.425,62.3967],[35.4375,62.5931],[37.4625,63.3717],[38.475,63.5892],[39.4875,63.4588],[40.5,62.389],[41.5125,62.2042],[42.525,64.0482],[43.5375,66.2825],[44.55,67.9915],[45.5625,69.0052],[46.575,69.6013],[47.5875,69.8319],[49.6125,69.5361],[50.625,69.6554],[52.65,70.7366],[53.6625,71.0213],[54.675,70.8591],[55.6875,70.2032],[56.7,66.6491],[57.7125,62.4598],[58.725,59.8422],[59.7375,58.5866],[60.75,58.2891],[61.7625,59.2819],[62.775,61.5292],[64.8,67.0348],[65.8125,69.5369],[66.825,71.6706],[67.8375,73.0368],[68.85,73.2641],[69.8625,72.0627],[70.875,68.5468],[71.8875,65.2528],[72.9,63.2658],[73.9125,62.5157],[74.925,62.0632],[75.9375,60.7427],[76.95,60.4876],[77.9625,61.8781],[78.975,63.708],[79.9875,64.7063]]
       }
       upperarm_l_fk_ctrl.Scale.X {type=MovieSceneScriptingFloatChannel, default=1, const [1,80]}
       upperarm_l_fk_ctrl.Scale.Y {type=MovieSceneScriptingFloatChannel, default=1, const [1,80]}
       upperarm_l_fk_ctrl.Scale.Z {type=MovieSceneScriptingFloatChannel, default=1, const [1,80]}
       lowerarm_l_fk_ctrl.Location.X {type=MovieSceneScriptingFloatChannel, default=0, thinned_from=80, keys [[0,0],[79.9875,0]]}
       lowerarm_l_fk_ctrl.Location.Y {type=MovieSceneScriptingFloatChannel, default=0, thinned_from=80, keys [[0,0],[79.9875,0]]}
       lowerarm_l_fk_ctrl.Location.Z {type=MovieSceneScriptingFloatChannel, default=0, thinned_from=80, keys [[0,0],[79.9875,0]]}
       lowerarm_l_fk_ctrl.Rotation.X {
        type=MovieSceneScriptingFloatChannel
        default=-3.2367
        thinned_from=80
        keys [[0,-3.2367],[1.0125,-3.2594],[2.025,-3.3268],[3.0375,-3.4103],[4.05,-3.4662],[5.0625,-3.5031],[6.075,-3.5205],[7.0875,-3.5145],[9.1125,-3.4539],[10.125,-3.4324],[11.1375,-3.4186],[12.15,-3.4159],[17.2125,-3.4192],[24.3,-3.3626],[26.325,-3.3622],[28.35,-3.3752],[34.425,-3.3581],[37.4625,-3.3707],[39.4875,-3.3936],[40.5,-3.4129],[43.5375,-3.4397],[44.55,-3.4413],[46.575,-3.4238],[49.6125,-3.3606],[50.625,-3.3517],[55.6875,-3.3502],[57.7125,-3.276],[58.725,-3.2623],[60.75,-3.2565],[61.7625,-3.2634],[64.8,-3.3181],[65.8125,-3.3294],[71.8875,-3.3455],[72.9,-3.3384],[76.95,-3.2602],[77.9625,-3.2449],[79.9875,-3.2367]]
       }
       lowerarm_l_fk_ctrl.Rotation.Y {
        type=MovieSceneScriptingFloatChannel
        default=1.3991
        thinned_from=80
        keys [[0,1.3991],[1.0125,1.4748],[2.025,1.6327],[3.0375,1.7685],[4.05,1.8402],[5.0625,1.8816],[6.075,1.8997],[7.0875,1.8935],[8.1,1.8617],[9.1125,1.8254],[10.125,1.7984],[11.1375,1.78],[12.15,1.7763],[14.175,1.7854],[17.2125,1.7809],[18.225,1.775],[20.25,1.7486],[22.275,1.7184],[24.3,1.6964],[25.3125,1.6914],[26.325,1.6956],[27.3375,1.7079],[28.35,1.7166],[29.3625,1.7148],[32.4,1.6922],[33.4125,1.6886],[35.4375,1.6919],[36.45,1.6985],[37.4625,1.7094],[38.475,1.7246],[39.4875,1.7446],[40.5,1.7722],[41.5125,1.7881],[43.5375,1.8077],[44.55,1.8098],[45.5625,1.8023],[46.575,1.787],[47.5875,1.7606],[48.6,1.7245],[49.6125,1.693],[50.625,1.6779],[51.6375,1.6751],[55.6875,1.6754],[56.7,1.6025],[57.7125,1.5205],[58.725,1.4832],[59.7375,1.4687],[60.75,1.466],[61.7625,1.4862],[62.775,1.5305],[63.7875,1.5786],[64.8,1.6156],[65.8125,1.6377],[66.825,1.6429],[67.8375,1.6441],[69.8625,1.6601],[71.8875,1.6672],[72.9,1.6544],[73.9125,1.6247],[74.925,1.5817],[75.9375,1.532],[76.95,1.477],[77.9625,1.4288],[78.975,1.4058],[79.9875,1.3991]]
       }
       lowerarm_l_fk_ctrl.Rotation.Z {
        type=MovieSceneScriptingFloatChannel
        default=61.7204
        thinned_from=80
        keys [[0,61.7204],[1.0125,57.6828],[2.025,48.9056],[3.0375,40.754],[4.05,36.1123],[5.0625,33.2763],[6.075,31.996],[7.0875,32.4366],[8.1,34.6518],[9.1125,37.0932],[10.125,38.8577],[11.1375,40.0311],[12.15,40.2673],[13.1625,39.8887],[14.175,39.6849],[15.1875,39.7432],[16.2,39.9562],[17.2125,39.9769],[18.225,40.3455],[20.25,41.9992],[21.2625,42.9947],[22.275,43.8483],[23.2875,44.5647],[24.3,45.1743],[25.3125,45.4676],[26.325,45.2208],[27.3375,44.4808],[28.35,43.9549],[29.3625,44.0664],[31.3875,45.0718],[32.4,45.4244],[33.4125,45.635],[34.425,45.6223],[35.4375,45.4379],[36.45,45.0473],[37.4625,44.3907],[38.475,43.4734],[39.4875,42.2447],[40.5,40.5227],[41.5125,39.5129],[42.525,38.9794],[43.5375,38.2505],[44.55,38.1178],[45.5625,38.6049],[46.575,39.5869],[47.5875,41.2518],[48.6,43.4783],[49.6125,45.3722],[50.625,46.2713],[51.6375,46.4367],[55.6875,46.42],[56.7,50.6301],[57.7125,55.2014],[58.725,57.2317],[59.7375,58.0112],[60.75,58.1549],[61.7625,57.0684],[62.775,54.6477],[63.7875,51.9812],[64.8,49.8861],[65.8125,48.6154],[66.825,48.3194],[67.8375,48.2445],[69.8625,47.319],[70.875,47.1777],[71.8875,46.903],[72.9,47.6472],[73.9125,49.3639],[74.925,51.8075],[75.9375,54.5701],[76.95,57.5658],[77.9625,60.1488],[78.975,61.3681],[79.9875,61.7204]]
       }
       lowerarm_l_fk_ctrl.Scale.X {type=MovieSceneScriptingFloatChannel, default=1, const [1,80]}
       lowerarm_l_fk_ctrl.Scale.Y {type=MovieSceneScriptingFloatChannel, default=1, const [1,80]}
       lowerarm_l_fk_ctrl.Scale.Z {type=MovieSceneScriptingFloatChannel, default=1, const [1,80]}
       hand_l_fk_ctrl.Location.X {type=MovieSceneScriptingFloatChannel, default=0, thinned_from=80, keys [[0,0],[79.9875,0]]}
       hand_l_fk_ctrl.Location.Y {type=MovieSceneScriptingFloatChannel, default=0, thinned_from=80, keys [[0,0],[79.9875,0]]}
       hand_l_fk_ctrl.Location.Z {type=MovieSceneScriptingFloatChannel, default=0, thinned_from=80, keys [[0,0],[79.9875,0]]}
       hand_l_fk_ctrl.Rotation.X {
        type=MovieSceneScriptingFloatChannel
        default=-101.0322
        thinned_from=80
        keys [[0,-101.0322],[1.0125,-99.3989],[2.025,-96.222],[3.0375,-93.4984],[4.05,-92.4451],[6.075,-92.5237],[7.0875,-91.3777],[9.1125,-87.3913],[10.125,-85.9347],[11.1375,-85.2992],[12.15,-85.3841],[13.1625,-84.7799],[14.175,-81.3296],[15.1875,-79.9037],[16.2,-82.9535],[17.2125,-85.3765],[18.225,-84.7017],[19.2375,-83.7464],[21.2625,-83.2538],[22.275,-83.2346],[23.2875,-83.5098],[27.3375,-86.2638],[28.35,-87.0766],[29.3625,-87.6966],[30.375,-88.0119],[31.3875,-88.0229],[36.45,-87.0139],[37.4625,-86.9733],[38.475,-87.258],[39.4875,-87.9633],[40.5,-90.2979],[41.5125,-94.2879],[42.525,-96.2821],[43.5375,-96.0453],[44.55,-95.5927],[46.575,-94.006],[47.5875,-93.3975],[48.6,-93.1329],[53.6625,-93.7473],[55.6875,-94.7348],[56.7,-93.3057],[57.7125,-91.3191],[58.725,-90.6488],[60.75,-90.8169],[61.7625,-91.1952],[63.7875,-92.76],[65.8125,-93.8593],[66.825,-94.7262],[67.8375,-96.0666],[68.85,-99.0738],[69.8625,-102.7507],[70.875,-105.0929],[71.8875,-106.3097],[72.9,-105.3888],[74.925,-102.4062],[75.9375,-100.6188],[76.95,-99.9933],[78.975,-100.8218],[79.9875,-101.0322]]
       }
       hand_l_fk_ctrl.Rotation.Y {
        type=MovieSceneScriptingFloatChannel
        default=-51.6014
        keys [[0,-51.6014],[1.0125,-47.9201],[2.025,-39.5689],[3.0375,-30.5416],[4.05,-22.9077],[5.0625,-16.4563],[6.075,-11.9988],[7.0875,-10.2843],[8.1,-10.9675],[9.1125,-11.9511],[10.125,-12.3516],[11.1375,-12.4454],[12.15,-11.9946],[13.1625,-10.7792],[14.175,-10.1365],[15.1875,-8.6702],[16.2,-7.5773],[17.2125,-6.6898],[18.225,-5.9698],[19.2375,-5.7583],[20.25,-5.9148],[21.2625,-6.5124],[22.275,-6.9607],[23.2875,-7.1305],[24.3,-7.1753],[25.3125,-7.0802],[26.325,-6.8067],[27.3375,-6.4204],[28.35,-6.3129],[29.3625,-6.6539],[30.375,-7.2633],[31.3875,-7.7838],[32.4,-8.1554],[33.4125,-8.4707],[34.425,-8.6729],[35.4375,-8.8256],[36.45,-8.8434],[37.4625,-8.4986],[38.475,-7.8499],[39.4875,-7.0018],[40.5,-6.721],[41.5125,-8.8536],[42.525,-10.5814],[43.5375,-10.3944],[44.55,-9.9041],[45.5625,-8.9731],[46.575,-8.0205],[47.5875,-7.7136],[48.6,-8.1189],[49.6125,-8.7729],[50.625,-9.3209],[51.6375,-9.7488],[52.65,-10.0529],[53.6625,-10.1249],[54.675,-9.599],[55.6875,-8.6096],[56.7,-10.4528],[57.7125,-12.4035],[58.725,-12.4446],[59.7375,-11.6666],[60.75,-10.8407],[61.7625,-9.7724],[62.775,-8.4894],[63.7875,-7.4994],[64.8,-7.0716],[65.8125,-7.5645],[66.825,-9.3403],[67.8375,-11.9093],[68.85,-16.1857],[69.8625,-21.1988],[70.875,-25.0493],[71.8875,-29.1032],[72.9,-34.6638],[73.9125,-40.9613],[74.925,-45.2453],[75.9375,-46.701],[76.95,-47.6652],[77.9625,-49.4544],[78.975,-50.95],[79.9875,-51.6014]]
       }
       hand_l_fk_ctrl.Rotation.Z {
        type=MovieSceneScriptingFloatChannel
        default=18.2638
        thinned_from=80
        keys [[0,18.2638],[1.0125,17.4122],[2.025,16.0489],[3.0375,15.0384],[4.05,14.0749],[5.0625,12.7344],[6.075,11.1258],[7.0875,9.7382],[8.1,8.5502],[9.1125,7.4166],[10.125,6.4104],[11.1375,5.8214],[12.15,5.928],[13.1625,6.2009],[14.175,6.3815],[15.1875,6.7984],[16.2,7.4959],[17.2125,7.9018],[18.225,7.6071],[19.2375,7.1696],[20.25,6.8736],[21.2625,6.6727],[22.275,6.3753],[23.2875,5.8292],[24.3,5.0514],[25.3125,4.1169],[26.325,3.1075],[27.3375,2.0703],[28.35,0.9975],[29.3625,-0.0581],[30.375,-1.0028],[31.3875,-1.6803],[32.4,-2.0742],[33.4125,-2.2905],[34.425,-2.3702],[35.4375,-2.343],[36.45,-2.2178],[37.4625,-2.0226],[38.475,-1.723],[39.4875,-1.2827],[40.5,-0.8353],[41.5125,-0.358],[42.525,-0.3103],[43.5375,-0.6805],[44.55,-1.0371],[45.5625,-1.5809],[46.575,-2.2643],[47.5875,-2.9555],[48.6,-3.6726],[49.6125,-4.3242],[50.625,-4.7477],[51.6375,-4.9196],[52.65,-4.9035],[53.6625,-4.7518],[54.675,-4.048],[55.6875,-3.1493],[56.7,-4.0083],[57.7125,-5.381],[58.725,-5.8429],[59.7375,-5.8455],[60.75,-5.7006],[61.7625,-4.9994],[62.775,-3.6824],[63.7875,-2.2984],[64.8,-1.0408],[65.8125,0.1175],[66.825,1.4422],[67.8375,3.2084],[68.85,6.1233],[69.8625,9.3656],[71.8875,14.5232],[72.9,16.7874],[73.9125,18.2169],[74.925,18.107],[75.9375,16.7918],[76.95,16.2759],[77.9625,16.6764],[78.975,17.6122],[79.9875,18.2638]]
       }
       hand_l_fk_ctrl.Scale.X {type=MovieSceneScriptingFloatChannel, default=1, const [1,80]}
       hand_l_fk_ctrl.Scale.Y {type=MovieSceneScriptingFloatChannel, default=1, const [1,80]}
       hand_l_fk_ctrl.Scale.Z {type=MovieSceneScriptingFloatChannel, default=1, const [1,80]}
       thumb_01_l_ctrl.Location.X {type=MovieSceneScriptingFloatChannel, default=0, thinned_from=80, keys [[0,0],[79.9875,0]]}
       thumb_01_l_ctrl.Location.Y {type=MovieSceneScriptingFloatChannel, default=0, thinned_from=80, keys [[0,0],[79.9875,0]]}
       thumb_01_l_ctrl.Location.Z {type=MovieSceneScriptingFloatChannel, default=0, thinned_from=80, keys [[0,0],[79.9875,0]]}
       thumb_01_l_ctrl.Rotation.X {type=MovieSceneScriptingFloatChannel, default=11.6194, const [11.6194,80]}
       thumb_01_l_ctrl.Rotation.Y {type=MovieSceneScriptingFloatChannel, default=9.9551, const [9.9551,80]}
       thumb_01_l_ctrl.Rotation.Z {type=MovieSceneScriptingFloatChannel, default=11.9413, const [11.9413,80]}
       thumb_01_l_ctrl.Scale.X {type=MovieSceneScriptingFloatChannel, default=1, const [1,80]}
       thumb_01_l_ctrl.Scale.Y {type=MovieSceneScriptingFloatChannel, default=1, const [1,80]}
       thumb_01_l_ctrl.Scale.Z {type=MovieSceneScriptingFloatChannel, default=1, const [1,80]}
       thumb_02_l_ctrl.Location.X {type=MovieSceneScriptingFloatChannel, default=0, thinned_from=80, keys [[0,0],[79.9875,0]]}
       thumb_02_l_ctrl.Location.Y {type=MovieSceneScriptingFloatChannel, default=0, thinned_from=80, keys [[0,0],[79.9875,0]]}
       thumb_02_l_ctrl.Location.Z {type=MovieSceneScriptingFloatChannel, default=0, const [0,80]}
       thumb_02_l_ctrl.Rotation.X {type=MovieSceneScriptingFloatChannel, default=5.5974, const [5.5974,80]}
       thumb_02_l_ctrl.Rotation.Y {type=MovieSceneScriptingFloatChannel, default=0, thinned_from=80, keys [[0,0],[79.9875,0]]}
       thumb_02_l_ctrl.Rotation.Z {type=MovieSceneScriptingFloatChannel, default=32.5258, const [32.5258,80]}
       thumb_02_l_ctrl.Scale.X {type=MovieSceneScriptingFloatChannel, default=1, const [1,80]}
       thumb_02_l_ctrl.Scale.Y {type=MovieSceneScriptingFloatChannel, default=1, const [1,80]}
       thumb_02_l_ctrl.Scale.Z {type=MovieSceneScriptingFloatChannel, default=1, const [1,80]}
       thumb_03_l_ctrl.Location.X {type=MovieSceneScriptingFloatChannel, default=0, const [0,80]}
       thumb_03_l_ctrl.Location.Y {type=MovieSceneScriptingFloatChannel, default=0, thinned_from=80, keys [[0,0],[79.9875,0]]}
       thumb_03_l_ctrl.Location.Z {type=MovieSceneScriptingFloatChannel, default=0, thinned_from=80, keys [[0,0],[79.9875,0]]}
       thumb_03_l_ctrl.Rotation.X {type=MovieSceneScriptingFloatChannel, default=5.5974, const [5.5974,80]}
       thumb_03_l_ctrl.Rotation.Y {type=MovieSceneScriptingFloatChannel, default=0, thinned_from=80, keys [[0,0],[79.9875,0]]}
       thumb_03_l_ctrl.Rotation.Z {type=MovieSceneScriptingFloatChannel, default=-10.8238, const [-10.8238,80]}
       thumb_03_l_ctrl.Scale.X {type=MovieSceneScriptingFloatChannel, default=1, const [1,80]}
       thumb_03_l_ctrl.Scale.Y {type=MovieSceneScriptingFloatChannel, default=1, const [1,80]}
       thumb_03_l_ctrl.Scale.Z {type=MovieSceneScriptingFloatChannel, default=1, const [1,80]}
       index_metacarpal_l_ctrl.Location.X {type=MovieSceneScriptingFloatChannel, default=0, const [0,80]}
       index_metacarpal_l_ctrl.Location.Y {type=MovieSceneScriptingFloatChannel, default=0, const [0,80]}
       index_metacarpal_l_ctrl.Location.Z {type=MovieSceneScriptingFloatChannel, default=0, const [0,80]}
       index_metacarpal_l_ctrl.Rotation.X {type=MovieSceneScriptingFloatChannel, default=0, const [0,80]}
       index_metacarpal_l_ctrl.Rotation.Y {type=MovieSceneScriptingFloatChannel, default=0, const [0,80]}
       index_metacarpal_l_ctrl.Rotation.Z {type=MovieSceneScriptingFloatChannel, default=0, const [0,80]}
       index_metacarpal_l_ctrl.Scale.X {type=MovieSceneScriptingFloatChannel, default=1, const [1,80]}
       index_metacarpal_l_ctrl.Scale.Y {type=MovieSceneScriptingFloatChannel, default=1, const [1,80]}
       index_metacarpal_l_ctrl.Scale.Z {type=MovieSceneScriptingFloatChannel, default=1, const [1,80]}
       index_01_l_ctrl.Location.X {type=MovieSceneScriptingFloatChannel, default=0, thinned_from=80, keys [[0,0],[79.9875,0]]}
       index_01_l_ctrl.Location.Y {type=MovieSceneScriptingFloatChannel, default=0, thinned_from=80, keys [[0,0],[79.9875,0]]}
       index_01_l_ctrl.Location.Z {type=MovieSceneScriptingFloatChannel, default=0, thinned_from=80, keys [[0,0],[79.9875,0]]}
       index_01_l_ctrl.Rotation.X {type=MovieSceneScriptingFloatChannel, default=-3.803, const [-3.803,80]}
       index_01_l_ctrl.Rotation.Y {type=MovieSceneScriptingFloatChannel, default=-7.0516, const [-7.0516,80]}
       index_01_l_ctrl.Rotation.Z {type=MovieSceneScriptingFloatChannel, default=4.081, const [4.081,80]}
       index_01_l_ctrl.Scale.X {type=MovieSceneScriptingFloatChannel, default=1, const [1,80]}
       index_01_l_ctrl.Scale.Y {type=MovieSceneScriptingFloatChannel, default=1, const [1,80]}
       index_01_l_ctrl.Scale.Z {type=MovieSceneScriptingFloatChannel, default=1, const [1,80]}
       index_02_l_ctrl.Location.X {type=MovieSceneScriptingFloatChannel, default=0, thinned_from=80, keys [[0,0],[79.9875,0]]}
       index_02_l_ctrl.Location.Y {type=MovieSceneScriptingFloatChannel, default=0, thinned_from=80, keys [[0,0],[79.9875,0]]}
       index_02_l_ctrl.Location.Z {type=MovieSceneScriptingFloatChannel, default=0, thinned_from=80, keys [[0,0],[79.9875,0]]}
       index_02_l_ctrl.Rotation.X {type=MovieSceneScriptingFloatChannel, default=-6.6486, const [-6.6486,80]}
       index_02_l_ctrl.Rotation.Y {type=MovieSceneScriptingFloatChannel, default=-1.4827, const [-1.4827,80]}
       index_02_l_ctrl.Rotation.Z {type=MovieSceneScriptingFloatChannel, default=-1.3705, const [-1.3705,80]}
       index_02_l_ctrl.Scale.X {type=MovieSceneScriptingFloatChannel, default=1, const [1,80]}
       index_02_l_ctrl.Scale.Y {type=MovieSceneScriptingFloatChannel, default=1, const [1,80]}
       index_02_l_ctrl.Scale.Z {type=MovieSceneScriptingFloatChannel, default=1, const [1,80]}
       index_03_l_ctrl.Location.X {type=MovieSceneScriptingFloatChannel, default=0, const [0,80]}
       index_03_l_ctrl.Location.Y {type=MovieSceneScriptingFloatChannel, default=0, thinned_from=80, keys [[0,0],[79.9875,0]]}
       index_03_l_ctrl.Location.Z {type=MovieSceneScriptingFloatChannel, default=0, thinned_from=80, keys [[0,0],[79.9875,0]]}
       index_03_l_ctrl.Rotation.X {type=MovieSceneScriptingFloatChannel, default=-6.5752, const [-6.5752,80]}
       index_03_l_ctrl.Rotation.Y {type=MovieSceneScriptingFloatChannel, default=0.9717, const [0.9717,80]}
       index_03_l_ctrl.Rotation.Z {type=MovieSceneScriptingFloatChannel, default=-6.3073, const [-6.3073,80]}
       index_03_l_ctrl.Scale.X {type=MovieSceneScriptingFloatChannel, default=1, const [1,80]}
       index_03_l_ctrl.Scale.Y {type=MovieSceneScriptingFloatChannel, default=1, const [1,80]}
       index_03_l_ctrl.Scale.Z {type=MovieSceneScriptingFloatChannel, default=1, const [1,80]}
       middle_metacarpal_l_ctrl.Location.X {type=MovieSceneScriptingFloatChannel, default=0, thinned_from=80, keys [[0,0],[79.9875,0]]}
       middle_metacarpal_l_ctrl.Location.Y {type=MovieSceneScriptingFloatChannel, default=0, thinned_from=80, keys [[0,0],[79.9875,0]]}
       middle_metacarpal_l_ctrl.Location.Z {type=MovieSceneScriptingFloatChannel, default=0, thinned_from=80, keys [[0,0],[79.9875,0]]}
       middle_metacarpal_l_ctrl.Rotation.X {type=MovieSceneScriptingFloatChannel, default=0, const [0,80]}
       middle_metacarpal_l_ctrl.Rotation.Y {type=MovieSceneScriptingFloatChannel, default=0.7323, const [0.7323,80]}
       middle_metacarpal_l_ctrl.Rotation.Z {type=MovieSceneScriptingFloatChannel, default=0, thinned_from=80, keys [[0,0],[79.9875,0]]}
       middle_metacarpal_l_ctrl.Scale.X {type=MovieSceneScriptingFloatChannel, default=1, const [1,80]}
       middle_metacarpal_l_ctrl.Scale.Y {type=MovieSceneScriptingFloatChannel, default=1, const [1,80]}
       middle_metacarpal_l_ctrl.Scale.Z {type=MovieSceneScriptingFloatChannel, default=1, const [1,80]}
       middle_01_l_ctrl.Location.X {type=MovieSceneScriptingFloatChannel, default=0, thinned_from=80, keys [[0,0],[79.9875,0]]}
       middle_01_l_ctrl.Location.Y {type=MovieSceneScriptingFloatChannel, default=0, thinned_from=80, keys [[0,0],[79.9875,0]]}
       middle_01_l_ctrl.Location.Z {type=MovieSceneScriptingFloatChannel, default=0, thinned_from=80, keys [[0,0],[79.9875,0]]}
       middle_01_l_ctrl.Rotation.X {type=MovieSceneScriptingFloatChannel, default=-13.9275, const [-13.9275,80]}
       middle_01_l_ctrl.Rotation.Y {type=MovieSceneScriptingFloatChannel, default=-3.1693, const [-3.1693,80]}
       middle_01_l_ctrl.Rotation.Z {type=MovieSceneScriptingFloatChannel, default=-24.8913, const [-24.8913,80]}
       middle_01_l_ctrl.Scale.X {type=MovieSceneScriptingFloatChannel, default=1, const [1,80]}
       middle_01_l_ctrl.Scale.Y {type=MovieSceneScriptingFloatChannel, default=1, const [1,80]}
       middle_01_l_ctrl.Scale.Z {type=MovieSceneScriptingFloatChannel, default=1, const [1,80]}
       middle_02_l_ctrl.Location.X {type=MovieSceneScriptingFloatChannel, default=0, thinned_from=80, keys [[0,0],[79.9875,0]]}
       middle_02_l_ctrl.Location.Y {type=MovieSceneScriptingFloatChannel, default=0, thinned_from=80, keys [[0,0],[79.9875,0]]}
       middle_02_l_ctrl.Location.Z {type=MovieSceneScriptingFloatChannel, default=0, thinned_from=80, keys [[0,0],[79.9875,0]]}
       middle_02_l_ctrl.Rotation.X {type=MovieSceneScriptingFloatChannel, default=-3.4377, const [-3.4377,80]}
       middle_02_l_ctrl.Rotation.Y {type=MovieSceneScriptingFloatChannel, default=0, thinned_from=80, keys [[0,0],[79.9875,0]]}
       middle_02_l_ctrl.Rotation.Z {type=MovieSceneScriptingFloatChannel, default=2.9748, const [2.9748,80]}
       middle_02_l_ctrl.Scale.X {type=MovieSceneScriptingFloatChannel, default=1, const [1,80]}
       middle_02_l_ctrl.Scale.Y {type=MovieSceneScriptingFloatChannel, default=1, const [1,80]}
       middle_02_l_ctrl.Scale.Z {type=MovieSceneScriptingFloatChannel, default=1, const [1,80]}
       middle_03_l_ctrl.Location.X {type=MovieSceneScriptingFloatChannel, default=0, thinned_from=80, keys [[0,0],[79.9875,0]]}
       middle_03_l_ctrl.Location.Y {type=MovieSceneScriptingFloatChannel, default=0, thinned_from=80, keys [[0,0],[79.9875,0]]}
       middle_03_l_ctrl.Location.Z {type=MovieSceneScriptingFloatChannel, default=0, thinned_from=80, keys [[0,0],[79.9875,0]]}
       middle_03_l_ctrl.Rotation.X {type=MovieSceneScriptingFloatChannel, default=-5.1617, const [-5.1617,80]}
       middle_03_l_ctrl.Rotation.Y {type=MovieSceneScriptingFloatChannel, default=-2.7884, const [-2.7884,80]}
       middle_03_l_ctrl.Rotation.Z {type=MovieSceneScriptingFloatChannel, default=-18.7575, const [-18.7575,80]}
       middle_03_l_ctrl.Scale.X {type=MovieSceneScriptingFloatChannel, default=1, const [1,80]}
       middle_03_l_ctrl.Scale.Y {type=MovieSceneScriptingFloatChannel, default=1, const [1,80]}
       middle_03_l_ctrl.Scale.Z {type=MovieSceneScriptingFloatChannel, default=1, const [1,80]}
       ring_metacarpal_l_ctrl.Location.X {type=MovieSceneScriptingFloatChannel, default=0, thinned_from=80, keys [[0,0],[79.9875,0]]}
       ring_metacarpal_l_ctrl.Location.Y {type=MovieSceneScriptingFloatChannel, default=0, thinned_from=80, keys [[0,0],[79.9875,0]]}
       ring_metacarpal_l_ctrl.Location.Z {type=MovieSceneScriptingFloatChannel, default=0, thinned_from=80, keys [[0,0],[79.9875,0]]}
       ring_metacarpal_l_ctrl.Rotation.X {type=MovieSceneScriptingFloatChannel, default=0, thinned_from=80, keys [[0,0],[79.9875,0]]}
       ring_metacarpal_l_ctrl.Rotation.Y {type=MovieSceneScriptingFloatChannel, default=-1.7423, const [-1.7423,80]}
       ring_metacarpal_l_ctrl.Rotation.Z {type=MovieSceneScriptingFloatChannel, default=-1.4338, const [-1.4338,80]}
       ring_metacarpal_l_ctrl.Scale.X {type=MovieSceneScriptingFloatChannel, default=1, const [1,80]}
       ring_metacarpal_l_ctrl.Scale.Y {type=MovieSceneScriptingFloatChannel, default=1, const [1,80]}
       ring_metacarpal_l_ctrl.Scale.Z {type=MovieSceneScriptingFloatChannel, default=1, const [1,80]}
       ring_01_l_ctrl.Location.X {type=MovieSceneScriptingFloatChannel, default=0, thinned_from=80, keys [[0,0],[79.9875,0]]}
       ring_01_l_ctrl.Location.Y {type=MovieSceneScriptingFloatChannel, default=0, const [0,80]}
       ring_01_l_ctrl.Location.Z {type=MovieSceneScriptingFloatChannel, default=0, thinned_from=80, keys [[0,0],[79.9875,0]]}
       ring_01_l_ctrl.Rotation.X {type=MovieSceneScriptingFloatChannel, default=-9.7285, const [-9.7285,80]}
       ring_01_l_ctrl.Rotation.Y {type=MovieSceneScriptingFloatChannel, default=-2.2294, const [-2.2294,80]}
       ring_01_l_ctrl.Rotation.Z {type=MovieSceneScriptingFloatChannel, default=-32.9959, const [-32.9959,80]}
       ring_01_l_ctrl.Scale.X {type=MovieSceneScriptingFloatChannel, default=1, const [1,80]}
       ring_01_l_ctrl.Scale.Y {type=MovieSceneScriptingFloatChannel, default=1, const [1,80]}
       ring_01_l_ctrl.Scale.Z {type=MovieSceneScriptingFloatChannel, default=1, const [1,80]}
       ring_02_l_ctrl.Location.X {type=MovieSceneScriptingFloatChannel, default=0, thinned_from=80, keys [[0,0],[79.9875,0]]}
       ring_02_l_ctrl.Location.Y {type=MovieSceneScriptingFloatChannel, default=0, thinned_from=80, keys [[0,0],[79.9875,0]]}
       ring_02_l_ctrl.Location.Z {type=MovieSceneScriptingFloatChannel, default=0, const [0,80]}
       ring_02_l_ctrl.Rotation.X {type=MovieSceneScriptingFloatChannel, default=-4.0107, const [-4.0107,80]}
       ring_02_l_ctrl.Rotation.Y {type=MovieSceneScriptingFloatChannel, default=0, thinned_from=80, keys [[0,0],[79.9875,0]]}
       ring_02_l_ctrl.Rotation.Z {type=MovieSceneScriptingFloatChannel, default=9.639, const [9.639,80]}
       ring_02_l_ctrl.Scale.X {type=MovieSceneScriptingFloatChannel, default=1, const [1,80]}
       ring_02_l_ctrl.Scale.Y {type=MovieSceneScriptingFloatChannel, default=1, const [1,80]}
       ring_02_l_ctrl.Scale.Z {type=MovieSceneScriptingFloatChannel, default=1, const [1,80]}
       ring_03_l_ctrl.Location.X {type=MovieSceneScriptingFloatChannel, default=0, thinned_from=80, keys [[0,0],[79.9875,0]]}
       ring_03_l_ctrl.Location.Y {type=MovieSceneScriptingFloatChannel, default=0, thinned_from=80, keys [[0,0],[79.9875,0]]}
       ring_03_l_ctrl.Location.Z {type=MovieSceneScriptingFloatChannel, default=0, thinned_from=80, keys [[0,0],[79.9875,0]]}
       ring_03_l_ctrl.Rotation.X {type=MovieSceneScriptingFloatChannel, default=-2.9794, const [-2.9794,80]}
       ring_03_l_ctrl.Rotation.Y {type=MovieSceneScriptingFloatChannel, default=-1.3751, const [-1.3751,80]}
       ring_03_l_ctrl.Rotation.Z {type=MovieSceneScriptingFloatChannel, default=-23.9971, const [-23.9971,80]}
       ring_03_l_ctrl.Scale.X {type=MovieSceneScriptingFloatChannel, default=1, const [1,80]}
       ring_03_l_ctrl.Scale.Y {type=MovieSceneScriptingFloatChannel, default=1, const [1,80]}
       ring_03_l_ctrl.Scale.Z {type=MovieSceneScriptingFloatChannel, default=1, const [1,80]}
       pinky_metacarpal_l_ctrl.Location.X {type=MovieSceneScriptingFloatChannel, default=0, thinned_from=80, keys [[0,0],[79.9875,0]]}
       pinky_metacarpal_l_ctrl.Location.Y {type=MovieSceneScriptingFloatChannel, default=0, thinned_from=80, keys [[0,0],[79.9875,0]]}
       pinky_metacarpal_l_ctrl.Location.Z {type=MovieSceneScriptingFloatChannel, default=0, thinned_from=80, keys [[0,0],[79.9875,0]]}
       pinky_metacarpal_l_ctrl.Rotation.X {type=MovieSceneScriptingFloatChannel, default=0, thinned_from=80, keys [[0,0],[79.9875,0]]}
       pinky_metacarpal_l_ctrl.Rotation.Y {type=MovieSceneScriptingFloatChannel, default=-3.539, const [-3.539,80]}
       pinky_metacarpal_l_ctrl.Rotation.Z {type=MovieSceneScriptingFloatChannel, default=-4.8356, const [-4.8356,80]}
       pinky_metacarpal_l_ctrl.Scale.X {type=MovieSceneScriptingFloatChannel, default=1, const [1,80]}
       pinky_metacarpal_l_ctrl.Scale.Y {type=MovieSceneScriptingFloatChannel, default=1, const [1,80]}
       pinky_metacarpal_l_ctrl.Scale.Z {type=MovieSceneScriptingFloatChannel, default=1, const [1,80]}
       pinky_01_l_ctrl.Location.X {type=MovieSceneScriptingFloatChannel, default=0, thinned_from=80, keys [[0,0],[79.9875,0]]}
       pinky_01_l_ctrl.Location.Y {type=MovieSceneScriptingFloatChannel, default=0, thinned_from=80, keys [[0,0],[79.9875,0]]}
       pinky_01_l_ctrl.Location.Z {type=MovieSceneScriptingFloatChannel, default=0, thinned_from=80, keys [[0,0],[79.9875,0]]}
       pinky_01_l_ctrl.Rotation.X {type=MovieSceneScriptingFloatChannel, default=-6.4845, const [-6.4845,80]}
       pinky_01_l_ctrl.Rotation.Y {type=MovieSceneScriptingFloatChannel, default=-2.0846, const [-2.0846,80]}
       pinky_01_l_ctrl.Rotation.Z {type=MovieSceneScriptingFloatChannel, default=-39.8382, const [-39.8382,80]}
       pinky_01_l_ctrl.Scale.X {type=MovieSceneScriptingFloatChannel, default=1, const [1,80]}
       pinky_01_l_ctrl.Scale.Y {type=MovieSceneScriptingFloatChannel, default=1, const [1,80]}
       pinky_01_l_ctrl.Scale.Z {type=MovieSceneScriptingFloatChannel, default=1, const [1,80]}
       pinky_02_l_ctrl.Location.X {type=MovieSceneScriptingFloatChannel, default=0, thinned_from=80, keys [[0,0],[79.9875,0]]}
       pinky_02_l_ctrl.Location.Y {type=MovieSceneScriptingFloatChannel, default=0, thinned_from=80, keys [[0,0],[79.9875,0]]}
       pinky_02_l_ctrl.Location.Z {type=MovieSceneScriptingFloatChannel, default=0, thinned_from=80, keys [[0,0],[79.9875,0]]}
       pinky_02_l_ctrl.Rotation.X {type=MovieSceneScriptingFloatChannel, default=-1.832, const [-1.832,80]}
       pinky_02_l_ctrl.Rotation.Y {type=MovieSceneScriptingFloatChannel, default=0.0728, const [0.0728,80]}
       pinky_02_l_ctrl.Rotation.Z {type=MovieSceneScriptingFloatChannel, default=17.0339, const [17.0339,80]}
       pinky_02_l_ctrl.Scale.X {type=MovieSceneScriptingFloatChannel, default=1, const [1,80]}
       pinky_02_l_ctrl.Scale.Y {type=MovieSceneScriptingFloatChannel, default=1, const [1,80]}
       pinky_02_l_ctrl.Scale.Z {type=MovieSceneScriptingFloatChannel, default=1, const [1,80]}
       pinky_03_l_ctrl.Location.X {type=MovieSceneScriptingFloatChannel, default=0, thinned_from=80, keys [[0,0],[79.9875,0]]}
       pinky_03_l_ctrl.Location.Y {type=MovieSceneScriptingFloatChannel, default=0, thinned_from=80, keys [[0,0],[79.9875,0]]}
       pinky_03_l_ctrl.Location.Z {type=MovieSceneScriptingFloatChannel, default=0, thinned_from=80, keys [[0,0],[79.9875,0]]}
       pinky_03_l_ctrl.Rotation.X {type=MovieSceneScriptingFloatChannel, default=-1.8335, const [-1.8335,80]}
       pinky_03_l_ctrl.Rotation.Y {type=MovieSceneScriptingFloatChannel, default=0, thinned_from=80, keys [[0,0],[79.9875,0]]}
       pinky_03_l_ctrl.Rotation.Z {type=MovieSceneScriptingFloatChannel, default=-12.4843, const [-12.4843,80]}
       pinky_03_l_ctrl.Scale.X {type=MovieSceneScriptingFloatChannel, default=1, const [1,80]}
       pinky_03_l_ctrl.Scale.Y {type=MovieSceneScriptingFloatChannel, default=1, const [1,80]}
       pinky_03_l_ctrl.Scale.Z {type=MovieSceneScriptingFloatChannel, default=1, const [1,80]}
       hand_l_ik_ctrl.Location.X {
        type=MovieSceneScriptingFloatChannel
        default=-25.018
        thinned_from=80
        keys [[0,-25.018],[1.0125,-24.6716],[2.025,-24.0859],[3.0375,-23.9308],[4.05,-24.6866],[5.0625,-26.152],[6.075,-27.9381],[7.0875,-29.5984],[8.1,-30.9536],[9.1125,-31.9389],[10.125,-32.3948],[11.1375,-32.5559],[12.15,-32.5937],[13.1625,-32.344],[14.175,-31.6095],[15.1875,-29.5406],[16.2,-28.258],[17.2125,-27.465],[18.225,-27.0346],[19.2375,-26.9773],[20.25,-27.2217],[22.275,-27.9078],[26.325,-28.8104],[27.3375,-29.1039],[30.375,-30.4628],[31.3875,-30.7005],[34.425,-30.8836],[35.4375,-30.8847],[36.45,-30.7733],[37.4625,-30.4708],[38.475,-29.9951],[39.4875,-29.3501],[40.5,-28.3404],[41.5125,-27.9317],[43.5375,-28.6469],[44.55,-29.0692],[48.6,-31.1025],[49.6125,-31.5564],[50.625,-31.8554],[51.6375,-32.0454],[52.65,-32.1458],[53.6625,-32.1333],[55.6875,-31.7191],[56.7,-31.6907],[57.7125,-31.501],[58.725,-31.1747],[59.7375,-30.9275],[62.775,-30.6477],[65.8125,-30.5448],[66.825,-30.3674],[67.8375,-29.7305],[68.85,-28.3695],[69.8625,-26.336],[70.875,-23.2651],[71.8875,-20.5698],[72.9,-19.3532],[73.9125,-19.1303],[74.925,-19.4834],[75.9375,-20.0462],[76.95,-21.3007],[77.9625,-23.0179],[78.975,-24.3995],[79.9875,-25.018]]
       }
       hand_l_ik_ctrl.Location.Y {
        type=MovieSceneScriptingFloatChannel
        default=25.82
        thinned_from=80
        keys [[0,25.82],[1.0125,26.4826],[2.025,27.9482],[3.0375,29.4865],[4.05,30.9111],[5.0625,32.4248],[6.075,33.8487],[7.0875,34.9733],[8.1,35.7682],[9.1125,36.4032],[10.125,36.9094],[11.1375,37.1837],[13.1625,37.47],[14.175,37.6872],[15.1875,38.3871],[16.2,38.9564],[17.2125,39.3422],[18.225,39.542],[19.2375,39.6049],[22.275,39.4137],[36.45,40.265],[39.4875,40.8131],[40.5,40.8357],[42.525,40.3658],[45.5625,39.9704],[46.575,39.7694],[47.5875,39.362],[49.6125,38.1867],[50.625,37.7667],[52.65,37.2137],[54.675,36.8981],[55.6875,36.8345],[56.7,36.8976],[59.7375,37.5528],[60.75,37.672],[61.7625,37.6537],[62.775,37.4749],[63.7875,37.1502],[64.8,36.6597],[65.8125,35.9484],[66.825,34.8412],[67.8375,33.3866],[68.85,31.6249],[69.8625,29.9553],[70.875,28.7561],[71.8875,28.2722],[72.9,28.4701],[73.9125,28.5819],[75.9375,28.4318],[76.95,27.886],[77.9625,27.0281],[78.975,26.2317],[79.9875,25.82]]
       }
       hand_l_ik_ctrl.Location.Z {
        type=MovieSceneScriptingFloatChannel
        default=58.4502
        thinned_from=80
        keys [[0,58.4502],[1.0125,59.2379],[2.025,60.7697],[3.0375,61.859],[4.05,62.1159],[5.0625,61.9064],[6.075,61.2721],[7.0875,60.2644],[9.1125,57.6484],[10.125,56.5445],[11.1375,55.7811],[14.175,54.9397],[15.1875,53.9259],[16.2,53.2797],[18.225,52.4213],[19.2375,52.0996],[22.275,51.8185],[27.3375,50.8274],[32.4,50.1742],[35.4375,50.1509],[38.475,50.4901],[41.5125,51.0051],[42.525,51.3764],[46.575,53.4842],[48.6,54.3152],[55.6875,55.9926],[56.7,54.6878],[57.7125,53.0118],[58.725,52.0269],[59.7375,51.5661],[60.75,51.4292],[61.7625,51.8726],[62.775,53.043],[65.8125,57.8646],[67.8375,60.8316],[68.85,62.1789],[69.8625,62.9218],[70.875,62.6749],[71.8875,61.8858],[73.9125,59.68],[75.9375,57.7658],[76.95,57.4524],[77.9625,57.6695],[78.975,58.1511],[79.9875,58.4502]]
       }
       hand_l_ik_ctrl.Rotation.X {
        type=MovieSceneScriptingFloatChannel
        default=-38.6378
        keys [[0,-38.6378],[1.0125,-37.2839],[2.025,-33.7602],[3.0375,-28.8463],[4.05,-23.4724],[5.0625,-18.5587],[6.075,-14.6558],[7.0875,-11.9618],[8.1,-10.4172],[9.1125,-9.6016],[10.125,-9.2603],[11.1375,-9.0829],[12.15,-8.6511],[13.1625,-7.1942],[14.175,-3.8012],[15.1875,-3.6354],[16.2,-6.6701],[17.2125,-8.7364],[18.225,-7.7158],[19.2375,-6.3408],[20.25,-5.3758],[21.2625,-4.7334],[22.275,-4.4358],[23.2875,-4.3465],[24.3,-4.3422],[25.3125,-4.4253],[26.325,-4.5991],[27.3375,-4.9127],[28.35,-5.4593],[29.3625,-6.1609],[30.375,-6.8486],[31.3875,-7.3636],[32.4,-7.6901],[33.4125,-7.9184],[34.425,-8.0716],[35.4375,-8.1471],[36.45,-8.1802],[37.4625,-8.3395],[38.475,-8.8431],[39.4875,-9.9629],[40.5,-14.2702],[41.5125,-21.0355],[42.525,-23.7376],[43.5375,-22.5721],[44.55,-20.5862],[45.5625,-17.6963],[46.575,-14.4019],[47.5875,-11.5128],[48.6,-9.2755],[49.6125,-7.833],[50.625,-7.2609],[51.6375,-7.2079],[52.65,-7.2726],[53.6625,-7.1267],[54.675,-6.4396],[55.6875,-5.1572],[56.7,-4.1394],[57.7125,-3.2492],[58.725,-2.3758],[59.7375,-1.5265],[60.75,-0.7343],[61.7625,-0.0992],[62.775,0.3642],[63.7875,0.85],[64.8,1.296],[65.8125,1.2674],[66.825,0.0938],[67.8375,-2.5835],[68.85,-10.2821],[69.8625,-21.5178],[70.875,-32.8194],[71.8875,-41.9403],[72.9,-45.5471],[73.9125,-46.223],[74.925,-45.9082],[75.9375,-44.9228],[76.95,-43.1495],[77.9625,-41.2488],[78.975,-39.554],[79.9875,-38.6378]]
       }
       hand_l_ik_ctrl.Rotation.Y {
        type=MovieSceneScriptingFloatChannel
        default=41.8512
        thinned_from=80
        keys [[0,41.8512],[1.0125,42.371],[2.025,43.6099],[3.0375,45.0718],[4.05,46.3796],[5.0625,47.2473],[6.075,47.5418],[7.0875,47.2285],[8.1,46.6062],[9.1125,46.1222],[11.1375,45.8147],[13.1625,45.2387],[14.175,45.0629],[15.1875,44.3991],[16.2,43.3775],[17.2125,42.5715],[21.2625,41.7401],[22.275,41.6773],[23.2875,41.8002],[25.3125,42.3407],[29.3625,43.8399],[30.375,44.0447],[31.3875,44.1152],[39.4875,43.5725],[40.5,43.3878],[41.5125,42.4841],[42.525,42.242],[43.5375,43.023],[44.55,43.9249],[45.5625,45.0869],[47.5875,47.5239],[48.6,48.6662],[49.6125,49.6512],[50.625,50.3267],[51.6375,50.7247],[53.6625,51.1181],[55.6875,50.7309],[59.7375,49.6126],[60.75,49.4429],[61.7625,49.4364],[62.775,49.6565],[63.7875,50.1032],[65.8125,51.3365],[66.825,52.1062],[67.8375,53.0884],[68.85,53.7385],[69.8625,53.5876],[70.875,52.6721],[71.8875,49.3781],[72.9,44.2487],[73.9125,39.9353],[74.925,38.0994],[75.9375,38.0157],[76.95,39.1379],[77.9625,40.4932],[78.975,41.4345],[79.9875,41.8512]]
       }
       hand_l_ik_ctrl.Rotation.Z {
        type=MovieSceneScriptingFloatChannel
        default=75.2292
        thinned_from=80
        keys [[0,75.2292],[1.0125,76.7668],[2.025,80.9271],[3.0375,87.1096],[4.05,95.325],[5.0625,104.4524],[6.075,112.2901],[7.0875,117.0492],[8.1,118.9843],[9.1125,119.9276],[13.1625,121.3773],[14.175,120.1507],[15.1875,117.2154],[16.2,115.865],[17.2125,115.1599],[18.225,114.757],[19.2375,114.6716],[22.275,115.8119],[24.3,117.3736],[28.35,121.3653],[29.3625,122.037],[30.375,122.3582],[32.4,122.1988],[37.4625,120.4879],[38.475,119.9718],[39.4875,119.0716],[40.5,115.9262],[41.5125,111.2574],[42.525,109.4402],[43.5375,110.2093],[44.55,111.8976],[45.5625,114.5897],[46.575,117.7253],[47.5875,120.6057],[48.6,123.0389],[49.6125,124.7459],[50.625,125.5116],[53.6625,125.8822],[54.675,126.7428],[55.6875,128.1054],[56.7,127.396],[57.7125,125.9105],[58.725,125.5923],[62.775,128.0753],[64.8,129.7896],[65.8125,130.1866],[66.825,129.5073],[67.8375,127.0831],[68.85,119.7344],[69.8625,108.7536],[70.875,96.7857],[71.8875,86.193],[72.9,78.7957],[73.9125,73.4783],[74.925,70.7117],[75.9375,70.0634],[76.95,71.4035],[77.9625,73.2517],[78.975,74.6063],[79.9875,75.2292]]
       }
       hand_l_ik_ctrl.Scale.X {type=MovieSceneScriptingFloatChannel, default=1, thinned_from=80, keys [[0,1],[79.9875,1]]}
       hand_l_ik_ctrl.Scale.Y {type=MovieSceneScriptingFloatChannel, default=1, const [1,80]}
       hand_l_ik_ctrl.Scale.Z {type=MovieSceneScriptingFloatChannel, default=1, const [1,80]}
       Softness {type=MovieSceneScriptingFloatChannel, default=0, const [0,80]}
       hand_l_ikorient_ctrl.Rotation.Roll {type=MovieSceneScriptingFloatChannel, default=0, const [0,80]}
       hand_l_ikorient_ctrl.Rotation.Pitch {type=MovieSceneScriptingFloatChannel, default=0, const [0,80]}
       hand_l_ikorient_ctrl.Rotation.Yaw {type=MovieSceneScriptingFloatChannel, default=0, const [0,80]}
       arm_l_pv_ik_ctrl.Location.X {
        type=MovieSceneScriptingFloatChannel
        default=-23.354
        thinned_from=80
        keys [[0,-23.354],[1.0125,-23.9495],[2.025,-25.4467],[4.05,-28.8007],[6.075,-32.6566],[7.0875,-34.172],[8.1,-35.067],[9.1125,-35.5639],[10.125,-35.64],[12.15,-35.128],[13.1625,-34.6553],[14.175,-33.735],[15.1875,-31.6146],[16.2,-30.2627],[18.225,-28.4003],[19.2375,-27.6941],[20.25,-27.3637],[21.2625,-27.2196],[24.3,-27.1932],[25.3125,-27.2677],[26.325,-27.4638],[27.3375,-27.8161],[28.35,-28.2816],[29.3625,-28.6876],[30.375,-28.9558],[31.3875,-29.0908],[36.45,-29.3547],[37.4625,-29.3703],[38.475,-29.2917],[39.4875,-29.094],[40.5,-28.3548],[41.5125,-27.7542],[42.525,-28.0148],[44.55,-29.0786],[45.5625,-29.5578],[46.575,-29.937],[47.5875,-30.0929],[50.625,-30.0171],[52.65,-30.1553],[53.6625,-30.1723],[54.675,-30.1201],[55.6875,-29.9852],[56.7,-28.7064],[57.7125,-27.0249],[58.725,-25.8889],[59.7375,-25.2929],[60.75,-25.0378],[61.7625,-25.1807],[62.775,-25.8881],[64.8,-28.1277],[65.8125,-29.116],[66.825,-29.7728],[67.8375,-29.9742],[68.85,-29.4995],[69.8625,-28.3272],[70.875,-26.3574],[71.8875,-24.7529],[72.9,-23.7785],[73.9125,-23.1727],[74.925,-22.6312],[75.9375,-21.8877],[76.95,-21.7123],[77.9625,-22.2414],[78.975,-22.9628],[79.9875,-23.354]]
       }
       arm_l_pv_ik_ctrl.Location.Y {
        type=MovieSceneScriptingFloatChannel
        default=55.8311
        thinned_from=80
        keys [[0,55.8311],[1.0125,56.5663],[2.025,58.1271],[3.0375,59.4463],[4.05,60.0285],[5.0625,60.1936],[6.075,59.7952],[7.0875,58.7614],[9.1125,55.5805],[10.125,54.2206],[11.1375,53.3911],[12.15,53.1598],[14.175,52.8966],[15.1875,51.8957],[16.2,51.4647],[19.2375,51.132],[21.2625,51.368],[23.2875,51.4295],[24.3,51.6138],[28.35,53.0175],[29.3625,53.2462],[39.4875,53.6735],[40.5,53.4817],[41.5125,53.4305],[42.525,53.9127],[45.5625,55.9045],[46.575,56.4692],[48.6,57.2501],[52.65,58.1518],[55.6875,58.5171],[56.7,57.5665],[57.7125,56.367],[58.725,55.6321],[59.7375,55.2643],[60.75,55.1071],[61.7625,55.2253],[62.775,55.8116],[63.7875,56.7967],[66.825,60.2008],[67.8375,61.1617],[68.85,61.779],[69.8625,61.6047],[70.875,60.0622],[71.8875,58.278],[72.9,56.8621],[74.925,55.1615],[75.9375,53.9655],[76.95,53.6141],[77.9625,54.2846],[78.975,55.2686],[79.9875,55.8311]]
       }
       arm_l_pv_ik_ctrl.Location.Z {
        type=MovieSceneScriptingFloatChannel
        default=-11.7694
        thinned_from=80
        keys [[0,-11.7694],[1.0125,-11.4039],[2.025,-10.4713],[3.0375,-9.387],[4.05,-8.5112],[5.0625,-7.7376],[6.075,-7.1538],[7.0875,-6.9341],[8.1,-7.145],[9.1125,-7.492],[10.125,-7.88],[11.1375,-8.1794],[13.1625,-8.2798],[14.175,-8.5619],[15.1875,-9.5188],[16.2,-10.0857],[17.2125,-10.3743],[18.225,-10.4514],[19.2375,-10.4042],[20.25,-10.3033],[21.2625,-10.24],[22.275,-10.2229],[23.2875,-10.2961],[24.3,-10.4539],[26.325,-10.8382],[27.3375,-10.9376],[29.3625,-10.9568],[30.375,-10.9961],[32.4,-11.1264],[33.4125,-11.1494],[35.4375,-11.1026],[36.45,-11.1055],[38.475,-11.2332],[39.4875,-11.2338],[40.5,-11.0726],[41.5125,-10.734],[42.525,-10.4191],[43.5375,-10.18],[45.5625,-9.7405],[46.575,-9.537],[47.5875,-9.3943],[49.6125,-9.1694],[51.6375,-8.8735],[52.65,-8.7557],[53.6625,-8.7109],[55.6875,-8.7203],[56.7,-9.3999],[57.7125,-10.1024],[58.725,-10.4186],[59.7375,-10.5355],[60.75,-10.5065],[61.7625,-10.241],[63.7875,-9.3439],[64.8,-8.935],[65.8125,-8.6234],[66.825,-8.4622],[67.8375,-8.4519],[68.85,-8.6227],[69.8625,-9.152],[70.875,-10.3518],[71.8875,-11.5208],[72.9,-12.3968],[73.9125,-13.1526],[74.925,-13.5897],[75.9375,-13.4789],[76.95,-13.0661],[77.9625,-12.5748],[78.975,-12.0486],[79.9875,-11.7694]]
       }
       arm_l_pv_ik_ctrl.Rotation.X {
        type=MovieSceneScriptingFloatChannel
        default=-48.298
        thinned_from=80
        keys [[0,-48.298],[1.0125,-49.7025],[3.0375,-56.2183],[4.05,-58.6046],[5.0625,-60.8276],[6.075,-62.6403],[7.0875,-63.5478],[8.1,-63.3321],[9.1125,-62.6303],[11.1375,-60.8194],[12.15,-60.2914],[13.1625,-59.5798],[14.175,-58.1206],[15.1875,-54.2967],[16.2,-51.6656],[17.2125,-49.9384],[18.225,-48.3885],[19.2375,-47.1788],[20.25,-46.6337],[21.2625,-46.3689],[23.2875,-46.1682],[24.3,-46.2118],[25.3125,-46.5001],[26.325,-47.1584],[28.35,-49.2506],[29.3625,-49.9662],[30.375,-50.3134],[31.3875,-50.3836],[34.425,-50.2125],[37.4625,-50.45],[38.475,-50.359],[39.4875,-50.122],[40.5,-49.3695],[41.5125,-48.9514],[42.525,-49.7262],[43.5375,-50.8964],[44.55,-51.8362],[45.5625,-52.4598],[46.575,-52.8479],[47.5875,-52.9617],[49.6125,-52.8162],[50.625,-52.9835],[52.65,-53.7638],[53.6625,-53.9824],[54.675,-53.9758],[55.6875,-53.7615],[56.7,-50.8115],[57.7125,-47.0553],[58.725,-44.6888],[59.7375,-43.4219],[60.75,-42.8597],[61.7625,-43.2547],[62.775,-44.9036],[64.8,-49.7556],[65.8125,-52.0173],[66.825,-53.9762],[67.8375,-55.5411],[68.85,-56.4455],[69.8625,-56.0608],[70.875,-53.4591],[71.8875,-50.7115],[72.9,-48.6496],[74.925,-46.5637],[75.9375,-44.6953],[76.95,-44.1913],[77.9625,-45.4601],[78.975,-47.2747],[79.9875,-48.298]]
       }
       arm_l_pv_ik_ctrl.Rotation.Y {
        type=MovieSceneScriptingFloatChannel
        default=-8.9379
        thinned_from=80
        keys [[0,-8.9379],[1.0125,-8.6071],[2.025,-7.9483],[3.0375,-7.3717],[4.05,-7.1188],[5.0625,-7.0665],[6.075,-7.3731],[7.0875,-8.1647],[8.1,-9.4034],[9.1125,-10.7235],[10.125,-11.7812],[11.1375,-12.2541],[12.15,-12.0924],[14.175,-11.3042],[15.1875,-11.3775],[16.2,-10.9798],[17.2125,-10.3873],[18.225,-9.8729],[20.25,-9.0453],[21.2625,-8.8146],[22.275,-8.7778],[23.2875,-8.7208],[24.3,-8.4991],[25.3125,-8.1589],[26.325,-7.7395],[27.3375,-7.3021],[28.35,-6.9879],[29.3625,-6.9017],[30.375,-6.9546],[31.3875,-7.0939],[32.4,-7.2725],[33.4125,-7.3943],[34.425,-7.4488],[36.45,-7.4603],[37.4625,-7.5008],[38.475,-7.5681],[39.4875,-7.6954],[40.5,-8.0464],[41.5125,-7.972],[42.525,-7.3054],[43.5375,-6.5566],[44.55,-5.9033],[45.5625,-5.3948],[46.575,-5.0162],[47.5875,-4.7501],[48.6,-4.5646],[49.6125,-4.4035],[50.625,-4.233],[51.6375,-4.0778],[52.65,-3.9792],[54.675,-3.9015],[55.6875,-3.6689],[56.7,-3.6944],[57.7125,-3.687],[58.725,-3.471],[59.7375,-3.2791],[60.75,-3.1164],[61.7625,-2.944],[62.775,-2.7449],[63.7875,-2.6365],[64.8,-2.7304],[65.8125,-3.0145],[66.825,-3.5207],[67.8375,-4.201],[68.85,-5.0232],[69.8625,-6.0246],[70.875,-7.4365],[71.8875,-8.5802],[72.9,-9.1281],[73.9125,-9.2317],[74.925,-9.2733],[75.9375,-9.6221],[76.95,-9.7235],[77.9625,-9.4698],[78.975,-9.1267],[79.9875,-8.9379]]
       }
       arm_l_pv_ik_ctrl.Rotation.Z {
        type=MovieSceneScriptingFloatChannel
        default=49.2711
        thinned_from=80
        keys [[0,49.2711],[1.0125,49.8105],[2.025,51.1908],[3.0375,53.0465],[4.05,55.4475],[5.0625,58.4772],[6.075,61.7064],[7.0875,64.6201],[8.1,67.0953],[9.1125,69.1585],[10.125,70.5647],[11.1375,71.2543],[12.15,71.3359],[13.1625,71.0513],[14.175,70.2945],[15.1875,69.0301],[16.2,68.2402],[17.2125,67.8486],[18.225,67.7338],[21.2625,68.3428],[23.2875,69.1439],[24.3,69.7447],[25.3125,70.4915],[30.375,75.3515],[31.3875,76.0453],[32.4,76.506],[34.425,76.9865],[36.45,77.0139],[38.475,76.5992],[39.4875,76.1271],[40.5,75.3689],[41.5125,74.7946],[45.5625,74.2684],[46.575,74.0356],[47.5875,73.6354],[49.6125,72.5598],[53.6625,71.4515],[54.675,70.8389],[55.6875,69.9546],[57.7125,70.2008],[60.75,69.64],[61.7625,69.0099],[65.8125,64.4227],[66.825,62.9514],[67.8375,60.7837],[68.85,57.6587],[69.8625,54.1578],[70.875,50.4627],[71.8875,48.0094],[72.9,47.6293],[73.9125,47.958],[75.9375,48.8107],[77.9625,49.2923],[78.975,49.3349],[79.9875,49.2711]]
       }
       arm_l_pv_ik_ctrl.Scale.X {type=MovieSceneScriptingFloatChannel, default=1, thinned_from=80, keys [[0,1],[79.9875,1]]}
       arm_l_pv_ik_ctrl.Scale.Y {type=MovieSceneScriptingFloatChannel, default=1, const [1,80]}
       arm_l_pv_ik_ctrl.Scale.Z {type=MovieSceneScriptingFloatChannel, default=1, const [1,80]}
       upperarm_r_fk_ctrl.Location.X {type=MovieSceneScriptingFloatChannel, default=0.0001, thinned_from=80, keys [[0,0.0001],[79.9875,0.0001]]}
       upperarm_r_fk_ctrl.Location.Y {type=MovieSceneScriptingFloatChannel, default=0, thinned_from=80, keys [[0,0],[79.9875,0]]}
       upperarm_r_fk_ctrl.Location.Z {type=MovieSceneScriptingFloatChannel, default=-0.0001, thinned_from=80, keys [[0,-0.0001],[79.9875,-0.0001]]}
       upperarm_r_fk_ctrl.Rotation.X {
        type=MovieSceneScriptingFloatChannel
        default=18.5797
        thinned_from=80
        keys [[0,18.5797],[1.0125,16.3462],[2.025,12.6252],[3.0375,11.0907],[4.05,11.9276],[5.0625,13.7499],[6.075,16.1132],[7.0875,18.2874],[8.1,19.8349],[9.1125,20.3271],[10.125,20.4328],[11.1375,20.9658],[12.15,21.039],[13.1625,21.2039],[14.175,23.0309],[15.1875,27.3595],[16.2,29.8283],[17.2125,31.5808],[18.225,34.1378],[19.2375,36.3666],[20.25,37.0127],[21.2625,36.2213],[22.275,34.3789],[23.2875,31.8407],[24.3,28.4318],[25.3125,23.8103],[26.325,18.7778],[27.3375,15.9415],[28.35,15.6836],[29.3625,17.0559],[30.375,18.1592],[31.3875,18.5228],[32.4,19.038],[33.4125,19.3573],[36.45,18.5821],[37.4625,18.105],[38.475,17.4708],[39.4875,16.6533],[40.5,14.5075],[41.5125,12.5814],[42.525,12.4261],[43.5375,12.9534],[44.55,13.7496],[45.5625,14.8068],[46.575,15.9519],[47.5875,16.0373],[48.6,15.2146],[49.6125,13.0909],[50.625,9.8961],[51.6375,8.8151],[52.65,12.4695],[53.6625,14.7531],[54.675,15.0968],[55.6875,14.1415],[56.7,11.8532],[57.7125,11.4587],[58.725,12.9424],[59.7375,14.7494],[60.75,16.2915],[61.7625,17.5729],[62.775,17.4602],[63.7875,16.0442],[64.8,14.0724],[65.8125,11.9941],[67.8375,7.7394],[68.85,7.4184],[69.8625,11.349],[70.875,17.5142],[71.8875,22.1199],[72.9,24.7088],[73.9125,23.0009],[74.925,18.7961],[75.9375,15.7503],[76.95,14.5491],[77.9625,15.4398],[78.975,17.4143],[79.9875,18.5797]]
       }
       upperarm_r_fk_ctrl.Rotation.Y {
        type=MovieSceneScriptingFloatChannel
        default=0.939
        thinned_from=80
        keys [[0,0.939],[1.0125,3.0924],[2.025,6.6831],[3.0375,9.3355],[4.05,11.0688],[5.0625,12.1065],[6.075,12.2637],[7.0875,12.0645],[8.1,11.9716],[9.1125,12.3609],[10.125,12.8329],[11.1375,12.7411],[12.15,12.2846],[13.1625,10.5172],[14.175,6.0927],[15.1875,-0.4424],[16.2,-1.2034],[17.2125,0.6831],[18.225,2.056],[19.2375,3.3728],[20.25,4.3325],[21.2625,4.9311],[22.275,5.6799],[23.2875,6.4148],[24.3,6.8842],[25.3125,7.3834],[26.325,7.7181],[27.3375,8.0735],[28.35,9.0306],[29.3625,9.9454],[30.375,9.8836],[31.3875,8.9914],[32.4,7.943],[33.4125,7.2014],[34.425,6.9884],[35.4375,6.9758],[36.45,7.1571],[37.4625,7.3708],[38.475,7.4348],[39.4875,6.9986],[40.5,6.4088],[41.5125,7.1028],[42.525,8.3224],[43.5375,9.4236],[44.55,10.4391],[45.5625,11.4331],[46.575,13.4031],[47.5875,17.2362],[48.6,22.1066],[49.6125,27.2516],[50.625,32.2679],[51.6375,36.4755],[52.65,38.1153],[53.6625,38.761],[54.675,38.2402],[55.6875,36.1548],[56.7,33.7871],[57.7125,31.1753],[58.725,29.152],[59.7375,28.2177],[61.7625,28.0423],[62.775,27.8247],[63.7875,27.3969],[64.8,26.7149],[65.8125,25.9791],[66.825,25.3313],[67.8375,24.465],[68.85,22.8583],[69.8625,19.9834],[70.875,16.1095],[71.8875,12.1144],[72.9,8.0691],[73.9125,4.974],[74.925,2.0848],[75.9375,-0.7438],[76.95,-2.4152],[77.9625,-1.546],[78.975,0.1138],[79.9875,0.939]]
       }
       upperarm_r_fk_ctrl.Rotation.Z {
        type=MovieSceneScriptingFloatChannel
        default=43.996
        thinned_from=80
        keys [[0,43.996],[1.0125,43.3374],[2.025,42.8334],[3.0375,43.263],[5.0625,44.3496],[6.075,44.6955],[7.0875,44.836],[8.1,44.3411],[10.125,42.8383],[11.1375,42.8774],[12.15,42.765],[13.1625,44.0173],[14.175,47.5095],[15.1875,45.8086],[16.2,36.1605],[17.2125,24.2267],[18.225,13.1381],[19.2375,4.4906],[20.25,0.1734],[21.2625,-1.1147],[22.275,-1.1888],[23.2875,-0.5134],[24.3,1.7804],[25.3125,6.9117],[26.325,15.6623],[27.3375,23.8084],[28.35,28.1997],[29.3625,31.1032],[30.375,32.7227],[32.4,35.1074],[33.4125,35.4631],[34.425,35.1048],[35.4375,34.3237],[36.45,33.2652],[37.4625,32.6011],[39.4875,32.4521],[40.5,30.3961],[41.5125,27.6333],[42.525,27.8451],[43.5375,29.7885],[44.55,32.2974],[45.5625,35.1981],[46.575,36.6121],[47.5875,36.1078],[48.6,35.3762],[49.6125,36.2341],[50.625,39.4221],[51.6375,45.9797],[52.65,52.7058],[53.6625,52.8683],[54.675,51.8431],[55.6875,50.4271],[56.7,46.3192],[57.7125,43.4167],[58.725,42.2213],[59.7375,41.745],[60.75,41.7062],[61.7625,43.0664],[62.775,44.7812],[63.7875,46.6193],[64.8,48.6274],[65.8125,49.9117],[66.825,50.2824],[67.8375,48.9563],[68.85,47.2076],[69.8625,46.8314],[70.875,48.0696],[71.8875,50.1416],[72.9,51.8897],[73.9125,50.7299],[74.925,48.7818],[75.9375,45.3436],[76.95,42.7373],[77.9625,42.5218],[78.975,43.3654],[79.9875,43.996]]
       }
       upperarm_r_fk_ctrl.Scale.X {type=MovieSceneScriptingFloatChannel, default=1, const [1,80]}
       upperarm_r_fk_ctrl.Scale.Y {type=MovieSceneScriptingFloatChannel, default=1, const [1,80]}
       upperarm_r_fk_ctrl.Scale.Z {type=MovieSceneScriptingFloatChannel, default=1, const [1,80]}
       lowerarm_r_fk_ctrl.Location.X {type=MovieSceneScriptingFloatChannel, default=0, thinned_from=80, keys [[0,0],[79.9875,0]]}
       lowerarm_r_fk_ctrl.Location.Y {type=MovieSceneScriptingFloatChannel, default=0, thinned_from=80, keys [[0,0],[79.9875,0]]}
       lowerarm_r_fk_ctrl.Location.Z {type=MovieSceneScriptingFloatChannel, default=0, thinned_from=80, keys [[0,0],[79.9875,0]]}
       lowerarm_r_fk_ctrl.Rotation.X {
        type=MovieSceneScriptingFloatChannel
        default=-3.2045
        thinned_from=80
        keys [[0,-3.2045],[1.0125,-3.2005],[2.025,-3.2084],[5.0625,-3.3049],[6.075,-3.3285],[7.0875,-3.3326],[8.1,-3.3255],[9.1125,-3.3082],[10.125,-3.2839],[12.15,-3.272],[13.1625,-3.2917],[14.175,-3.4393],[15.1875,-3.8316],[16.2,-3.9394],[17.2125,-3.8499],[18.225,-3.7073],[19.2375,-3.575],[20.25,-3.4941],[21.2625,-3.4542],[22.275,-3.4278],[23.2875,-3.4138],[24.3,-3.4142],[25.3125,-3.4338],[26.325,-3.4935],[27.3375,-3.5441],[28.35,-3.5204],[30.375,-3.4636],[31.3875,-3.4426],[32.4,-3.439],[33.4125,-3.4431],[34.425,-3.4345],[35.4375,-3.4116],[37.4625,-3.3483],[38.475,-3.333],[39.4875,-3.3304],[40.5,-3.3021],[41.5125,-3.2605],[42.525,-3.2526],[44.55,-3.2753],[45.5625,-3.2925],[46.575,-3.2898],[48.6,-3.227],[49.6125,-3.2077],[50.625,-3.2005],[52.65,-3.2051],[56.7,-3.2215],[61.7625,-3.2051],[66.825,-3.2009],[70.875,-3.2221],[72.9,-3.2172],[75.9375,-3.201],[79.9875,-3.2045]]
       }
       lowerarm_r_fk_ctrl.Rotation.Y {
        type=MovieSceneScriptingFloatChannel
        default=1.0157
        thinned_from=80
        keys [[0,1.0157],[1.0125,1.085],[2.025,1.2513],[3.0375,1.4084],[4.05,1.5065],[5.0625,1.5882],[6.075,1.636],[7.0875,1.6438],[8.1,1.6302],[9.1125,1.5952],[10.125,1.5402],[11.1375,1.5203],[12.15,1.51],[13.1625,1.5589],[14.175,1.8072],[15.1875,2.1238],[16.2,2.1694],[17.2125,2.1325],[18.225,2.0533],[19.2375,1.9518],[20.25,1.872],[21.2625,1.8258],[22.275,1.7924],[23.2875,1.7735],[24.3,1.774],[25.3125,1.8002],[26.325,1.8713],[27.3375,1.9232],[28.35,1.8997],[30.375,1.8372],[31.3875,1.8115],[32.4,1.8068],[33.4125,1.8121],[34.425,1.8011],[35.4375,1.7705],[36.45,1.7194],[37.4625,1.6721],[38.475,1.6445],[39.4875,1.6396],[40.5,1.5823],[41.5125,1.4781],[42.525,1.4541],[43.5375,1.4839],[44.55,1.5187],[45.5625,1.5608],[46.575,1.5544],[47.5875,1.4703],[49.6125,1.2454],[50.625,1.1442],[51.6375,1.0664],[52.65,1.0099],[53.6625,0.9497],[54.675,0.9309],[55.6875,0.9227],[56.7,0.897],[57.7125,0.919],[58.725,0.9537],[59.7375,0.9774],[60.75,0.9981],[61.7625,1.0097],[62.775,0.9957],[63.7875,1.0045],[64.8,1.0419],[65.8125,1.0959],[67.8375,1.2134],[68.85,1.2795],[69.8625,1.3212],[70.875,1.3356],[71.8875,1.3275],[72.9,1.3101],[73.9125,1.2506],[74.925,1.2034],[75.9375,1.1588],[76.95,1.1088],[77.9625,1.0556],[78.975,1.0251],[79.9875,1.0157]]
       }
       lowerarm_r_fk_ctrl.Rotation.Z {
        type=MovieSceneScriptingFloatChannel
        default=81.4696
        thinned_from=80
        keys [[0,81.4696],[1.0125,77.9223],[2.025,69.4125],[3.0375,61.2282],[4.05,55.9671],[5.0625,51.4391],[6.075,48.7166],[7.0875,48.2669],[8.1,49.0504],[9.1125,51.0449],[10.125,54.1158],[11.1375,55.2088],[12.15,55.7741],[13.1625,53.0834],[14.175,38.2867],[15.1875,12.3167],[16.2,6.3385],[17.2125,11.2846],[18.225,19.6205],[19.2375,28.1445],[20.25,33.9521],[21.2625,37.0689],[22.275,39.2427],[23.2875,40.4457],[24.3,40.4149],[25.3125,38.7408],[26.325,33.9998],[27.3375,30.298],[28.35,32.0008],[29.3625,34.3084],[30.375,36.3137],[31.3875,38.0116],[32.4,38.3123],[33.4125,37.9676],[34.425,38.6823],[35.4375,40.637],[36.45,43.7869],[37.4625,46.6169],[38.475,48.2265],[39.4875,48.5119],[40.5,51.7754],[41.5125,57.5066],[42.525,58.796],[43.5375,57.1905],[44.55,55.2972],[45.5625,52.9773],[46.575,53.3318],[47.5875,57.9261],[49.6125,69.7127],[50.625,74.8988],[51.6375,78.8735],[52.65,81.7661],[53.6625,84.8616],[54.675,85.8325],[55.6875,86.2603],[56.7,87.5915],[57.7125,86.4482],[58.725,84.6538],[60.75,82.3729],[61.7625,81.7735],[62.775,82.4929],[63.7875,82.0408],[64.8,80.1236],[65.8125,77.3657],[67.8375,71.3558],[68.85,67.9571],[69.8625,65.8008],[70.875,65.0483],[71.8875,65.4692],[72.9,66.377],[73.9125,69.4463],[75.9375,74.1517],[76.95,76.7078],[77.9625,79.4226],[78.975,80.9834],[79.9875,81.4696]]
       }
       lowerarm_r_fk_ctrl.Scale.X {type=MovieSceneScriptingFloatChannel, default=1, const [1,80]}
       lowerarm_r_fk_ctrl.Scale.Y {type=MovieSceneScriptingFloatChannel, default=1, const [1,80]}
       lowerarm_r_fk_ctrl.Scale.Z {type=MovieSceneScriptingFloatChannel, default=1, const [1,80]}
       hand_r_fk_ctrl.Location.X {type=MovieSceneScriptingFloatChannel, default=0, thinned_from=80, keys [[0,0],[79.9875,0]]}
       hand_r_fk_ctrl.Location.Y {type=MovieSceneScriptingFloatChannel, default=0, thinned_from=80, keys [[0,0],[79.9875,0]]}
       hand_r_fk_ctrl.Location.Z {type=MovieSceneScriptingFloatChannel, default=0, thinned_from=80, keys [[0,0],[79.9875,0]]}
       hand_r_fk_ctrl.Rotation.X {
        type=MovieSceneScriptingFloatChannel
        default=-11.0662
        thinned_from=80
        keys [[0,-11.0662],[1.0125,-17.9746],[2.025,-28.3981],[3.0375,-32.2727],[4.05,-28.3109],[5.0625,-21.165],[6.075,-15.163],[7.0875,-12.0129],[8.1,-11.7392],[9.1125,-14.3537],[10.125,-16.2333],[11.1375,-16.0904],[12.15,-15.6556],[13.1625,-16.1078],[14.175,-22.5558],[15.1875,-29.1235],[16.2,-21.2178],[17.2125,-9.1631],[18.225,-6.6434],[19.2375,-6.0671],[20.25,-6.1269],[21.2625,-6.896],[22.275,-9.3432],[23.2875,-12.1556],[24.3,-16.4791],[25.3125,-21.0556],[26.325,-23.1049],[27.3375,-23.9761],[28.35,-25.9668],[29.3625,-26.507],[30.375,-22.9414],[31.3875,-18.3311],[32.4,-17.7706],[33.4125,-20.3436],[34.425,-20.4375],[35.4375,-19.9222],[36.45,-19.2723],[37.4625,-18.3698],[38.475,-17.3617],[39.4875,-16.4014],[40.5,-12.9759],[41.5125,-9.5101],[42.525,-9.393],[43.5375,-10.3179],[45.5625,-12.2377],[46.575,-11.8937],[47.5875,-8.9303],[48.6,-5.5682],[49.6125,-0.7787],[50.625,8.1068],[51.6375,19.2177],[52.65,33.453],[53.6625,45.6963],[54.675,50.1343],[55.6875,52.1105],[56.7,52.8355],[57.7125,53.7723],[59.7375,57.9812],[60.75,59.2749],[61.7625,61.5141],[62.775,58.5699],[63.7875,51.7298],[64.8,43.4039],[65.8125,33.3598],[66.825,22.4994],[67.8375,11.7031],[68.85,2.014],[69.8625,-4.7747],[70.875,-9.4524],[71.8875,-13.3001],[72.9,-15.8915],[73.9125,-15.1677],[74.925,-18.0516],[75.9375,-15.4347],[76.95,-12.0365],[77.9625,-10.7992],[78.975,-10.9105],[79.9875,-11.0662]]
       }
       hand_r_fk_ctrl.Rotation.Y {
        type=MovieSceneScriptingFloatChannel
        default=-38.8115
        thinned_from=80
        keys [[0,-38.8115],[1.0125,-36.0712],[2.025,-28.5836],[3.0375,-19.4928],[4.05,-9.2876],[5.0625,1.9974],[6.075,10.3216],[7.0875,13.6362],[8.1,15.6217],[9.1125,17.887],[10.125,17.5166],[11.1375,16.5242],[12.15,16.3783],[13.1625,19.1439],[14.175,27.76],[15.1875,23.5451],[16.2,15.0219],[17.2125,3.0563],[18.225,-2.2948],[19.2375,-4.0048],[20.25,-3.0846],[21.2625,-2.2306],[22.275,-2.3121],[23.2875,-2.5448],[24.3,-1.673],[25.3125,-0.908],[26.325,0.4491],[27.3375,2.9174],[28.35,6.143],[29.3625,11.8388],[30.375,19.3032],[31.3875,26.889],[32.4,33.8087],[33.4125,37.8236],[34.425,37.8531],[35.4375,36.1406],[36.45,33.1424],[37.4625,30.5093],[38.475,29.0447],[39.4875,28.9116],[40.5,26.099],[41.5125,18.8852],[42.525,15.7263],[43.5375,16.9346],[44.55,18.4342],[45.5625,20.153],[46.575,17.9357],[47.5875,10.0094],[48.6,1.6466],[49.6125,-3.8482],[50.625,-7.0089],[51.6375,-12.7194],[52.65,-15.7272],[53.6625,-12.7404],[54.675,-10.997],[55.6875,-9.9653],[56.7,-6.7672],[57.7125,-2.1316],[58.725,0.9545],[59.7375,2.8816],[60.75,4.6346],[61.7625,3.733],[62.775,3.8798],[63.7875,4.6497],[64.8,4.6446],[65.8125,3.7043],[66.825,2.0286],[68.85,-0.7977],[69.8625,-4.858],[70.875,-12.6006],[71.8875,-22.9553],[72.9,-32.31],[73.9125,-38.6258],[74.925,-40.9559],[75.9375,-39.593],[76.95,-37.9994],[77.9625,-38.3102],[78.975,-38.7185],[79.9875,-38.8115]]
       }
       hand_r_fk_ctrl.Rotation.Z {
        type=MovieSceneScriptingFloatChannel
        default=-16.6601
        thinned_from=80
        keys [[0,-16.6601],[1.0125,-11.6789],[2.025,-3.6207],[3.0375,0.3213],[4.05,0.1865],[5.0625,-1.5698],[6.075,-5.0836],[7.0875,-10.7637],[8.1,-16.1429],[9.1125,-18.965],[10.125,-20.0752],[11.1375,-20.516],[12.15,-20.6808],[13.1625,-20.7627],[14.175,-21.7057],[15.1875,-20.1348],[16.2,-19.3636],[17.2125,-22.838],[18.225,-26.4021],[19.2375,-25.7147],[20.25,-17.6257],[21.2625,-7.0248],[22.275,3.9591],[23.2875,10.8827],[24.3,12.2316],[25.3125,12.1615],[26.325,10.4902],[27.3375,8.2086],[28.35,6.9374],[29.3625,3.0495],[30.375,-3.9316],[31.3875,-11.5367],[32.4,-18.6519],[33.4125,-22.7771],[34.425,-23.6562],[35.4375,-23.522],[36.45,-23.0112],[37.4625,-22.0695],[39.4875,-19.6541],[40.5,-16.5073],[41.5125,-13.1721],[42.525,-12.2197],[43.5375,-12.141],[44.55,-12.4882],[45.5625,-13.1118],[46.575,-12.8459],[47.5875,-9.9539],[48.6,-6.9868],[49.6125,-6.6818],[50.625,-7.8021],[51.6375,-5.6792],[52.65,0.8575],[53.6625,5.7988],[54.675,6.2387],[55.6875,5.5831],[56.7,4.7745],[57.7125,4.0661],[58.725,3.1031],[59.7375,1.9998],[60.75,1.0684],[61.7625,0.5654],[62.775,-2.4172],[63.7875,-5.2429],[64.8,-6.923],[65.8125,-7.954],[66.825,-8.1939],[67.8375,-7.8465],[68.85,-6.6897],[69.8625,-4.2828],[70.875,-3.2937],[71.8875,-4.7195],[72.9,-7.721],[73.9125,-11.7893],[74.925,-9.5079],[75.9375,-8.9377],[76.95,-10.0319],[77.9625,-12.2887],[78.975,-15.0844],[79.9875,-16.6601]]
       }
       hand_r_fk_ctrl.Scale.X {type=MovieSceneScriptingFloatChannel, default=1, const [1,80]}
       hand_r_fk_ctrl.Scale.Y {type=MovieSceneScriptingFloatChannel, default=1, const [1,80]}
       hand_r_fk_ctrl.Scale.Z {type=MovieSceneScriptingFloatChannel, default=1, const [1,80]}
       thumb_01_r_ctrl.Location.X {type=MovieSceneScriptingFloatChannel, default=0, thinned_from=80, keys [[0,0],[79.9875,0]]}
       thumb_01_r_ctrl.Location.Y {type=MovieSceneScriptingFloatChannel, default=0, thinned_from=80, keys [[0,0],[79.9875,0]]}
       thumb_01_r_ctrl.Location.Z {type=MovieSceneScriptingFloatChannel, default=0, thinned_from=80, keys [[0,0],[79.9875,0]]}
       thumb_01_r_ctrl.Rotation.X {
        type=MovieSceneScriptingFloatChannel
        default=15.3978
        thinned_from=80
        keys [[0,15.3978],[1.0125,14.6534],[3.0375,11.4453],[4.05,10.8041],[5.0625,12.4415],[6.075,15.4943],[7.0875,19.1684],[8.1,22.2313],[9.1125,23.6752],[10.125,24.4294],[11.1375,24.6443],[13.1625,24.6443],[14.175,22.7561],[15.1875,10.232],[16.2,-7.9568],[17.2125,-16.0739],[18.225,-13.6271],[19.2375,-8.4492],[20.25,-1.5399],[21.2625,6.101],[22.275,13.4736],[23.2875,19.5782],[24.3,23.4148],[25.3125,24.6443],[39.4875,24.6443],[41.5125,24.1606],[46.575,24.1606],[47.5875,20.3149],[48.6,13.7688],[49.6125,11.5412],[50.625,12.4129],[51.6375,13.4772],[52.65,14.4438],[53.6625,15.0959],[54.675,15.263],[66.825,15.3062],[67.8375,15.3459],[68.85,16.0087],[69.8625,17.2887],[70.875,18.0938],[71.8875,17.6392],[72.9,16.4724],[73.9125,15.4713],[74.925,15.2677],[75.9375,15.3908],[79.9875,15.3978]]
       }
       thumb_01_r_ctrl.Rotation.Y {
        type=MovieSceneScriptingFloatChannel
        default=1.9431
        thinned_from=80
        keys [[0,1.9431],[1.0125,1.0825],[2.025,-0.7829],[3.0375,-2.61],[4.05,-3.4517],[5.0625,-4.2104],[6.075,-5.5027],[8.1,-8.0605],[9.1125,-9.3994],[10.125,-10.6005],[11.1375,-11.0978],[13.1625,-11.0978],[14.175,-9.6852],[15.1875,-0.3155],[16.2,13.2922],[17.2125,19.3648],[18.225,17.5344],[19.2375,13.6606],[20.25,8.4915],[21.2625,2.7751],[22.275,-2.7406],[23.2875,-7.3076],[24.3,-10.178],[25.3125,-11.0978],[39.4875,-11.0978],[41.5125,-10.7949],[46.575,-10.7949],[47.5875,-10.3616],[48.6,-8.6848],[49.6125,-5.4077],[50.625,-1.0669],[51.6375,3.2269],[52.65,6.7571],[53.6625,8.8254],[54.675,9.3077],[62.775,9.3077],[63.7875,8.512],[64.8,6.5105],[65.8125,3.9041],[66.825,0.9043],[67.8375,-2.4211],[68.85,-6.8799],[69.8625,-11.7972],[70.875,-14.4392],[71.8875,-11.4862],[72.9,-4.4953],[73.9125,1.503],[74.925,2.7226],[75.9375,1.9852],[76.95,1.9431],[79.9875,1.9431]]
       }
       thumb_01_r_ctrl.Rotation.Z {
        type=MovieSceneScriptingFloatChannel
        default=-2.4793
        thinned_from=80
        keys [[0,-2.4793],[1.0125,0.2843],[2.025,6.2744],[3.0375,12.142],[4.05,14.6076],[5.0625,12.0365],[6.075,7.0916],[7.0875,0.8941],[8.1,-4.8072],[9.1125,-8.7949],[10.125,-11.7911],[11.1375,-12.927],[13.1625,-12.927],[14.175,-11.7744],[15.1875,-4.1296],[16.2,6.973],[17.2125,11.9277],[18.225,10.4342],[19.2375,7.2736],[20.25,3.0561],[21.2625,-1.608],[22.275,-6.1083],[23.2875,-9.8346],[24.3,-12.1765],[25.3125,-12.927],[39.4875,-12.927],[41.5125,-11.8363],[46.575,-11.8363],[47.5875,-11.9123],[48.6,-12.0417],[49.6125,-11.7943],[50.625,-10.9547],[51.6375,-9.9296],[52.65,-8.9985],[53.6625,-8.3705],[54.675,-8.2095],[62.775,-8.2095],[63.7875,-7.1254],[64.8,-4.4542],[65.8125,-0.2364],[66.825,4.649],[67.8375,7.4181],[68.85,7.7609],[69.8625,7.699],[70.875,7.5233],[71.8875,5.468],[72.9,1.3174],[73.9125,-2.2199],[74.925,-2.9468],[75.9375,-2.5045],[76.95,-2.4793],[79.9875,-2.4793]]
       }
       thumb_01_r_ctrl.Scale.X {type=MovieSceneScriptingFloatChannel, default=1, const [1,80]}
       thumb_01_r_ctrl.Scale.Y {type=MovieSceneScriptingFloatChannel, default=1, const [1,80]}
       thumb_01_r_ctrl.Scale.Z {type=MovieSceneScriptingFloatChannel, default=1, const [1,80]}
       thumb_02_r_ctrl.Location.X {type=MovieSceneScriptingFloatChannel, default=0, thinned_from=80, keys [[0,0],[79.9875,0]]}
       thumb_02_r_ctrl.Location.Y {type=MovieSceneScriptingFloatChannel, default=0, thinned_from=80, keys [[0,0],[79.9875,0]]}
       thumb_02_r_ctrl.Location.Z {type=MovieSceneScriptingFloatChannel, default=0, thinned_from=80, keys [[0,0],[79.9875,0]]}
       thumb_02_r_ctrl.Rotation.X {
        type=MovieSceneScriptingFloatChannel
        default=0
        thinned_from=80
        keys [[0,0],[13.1625,0],[14.175,-0.0401],[15.1875,-0.306],[16.2,-0.6921],[17.2125,-0.8644],[18.225,-0.8125],[19.2375,-0.7026],[20.25,-0.5559],[21.2625,-0.3937],[22.275,-0.2372],[23.2875,-0.1075],[24.3,-0.0261],[25.3125,0],[79.9875,0]]
       }
       thumb_02_r_ctrl.Rotation.Y {
        type=MovieSceneScriptingFloatChannel
        default=0
        thinned_from=80
        keys [[0,0],[13.1625,0],[14.175,0.0184],[15.1875,0.1401],[16.2,0.3168],[17.2125,0.3957],[18.225,0.3719],[19.2375,0.3216],[20.25,0.2544],[21.2625,0.1802],[22.275,0.1086],[23.2875,0.0492],[24.3,0.0119],[25.3125,0],[79.9875,0]]
       }
       thumb_02_r_ctrl.Rotation.Z {
        type=MovieSceneScriptingFloatChannel
        default=-11.169
        thinned_from=80
        keys [[0,-11.169],[1.0125,-8.3195],[2.025,-2.1432],[3.0375,3.9066],[4.05,6.5107],[5.0625,5.1916],[6.075,3.0083],[7.0875,1.0068],[8.1,-1.0751],[9.1125,-3.4647],[10.125,-5.6742],[11.1375,-6.6009],[13.1625,-6.6009],[14.175,-4.8999],[15.1875,6.3821],[16.2,22.767],[17.2125,30.0791],[18.225,27.875],[19.2375,23.2106],[20.25,16.9865],[21.2625,10.1035],[22.275,3.462],[23.2875,-2.0371],[24.3,-5.4933],[25.3125,-6.6009],[39.4875,-6.6009],[41.5125,-6.0813],[46.575,-6.0813],[47.5875,-0.0555],[48.6,10.2013],[49.6125,12.4357],[50.625,7.4855],[51.6375,1.3245],[52.65,-4.5234],[53.6625,-8.9519],[54.675,-10.6883],[55.6875,-11.0851],[56.7,-11.1651],[60.75,-11.169],[61.7625,-11.5627],[62.775,-11.2787],[63.7875,-8.1342],[64.8,-3.837],[65.8125,0.5222],[66.825,4.835],[67.8375,8.265],[68.85,10.2141],[69.8625,11.0304],[70.875,11.2156],[71.8875,6.9045],[72.9,-2.5039],[73.9125,-10.5767],[74.925,-12.218],[75.9375,-11.2257],[76.95,-11.169],[79.9875,-11.169]]
       }
       thumb_02_r_ctrl.Scale.X {type=MovieSceneScriptingFloatChannel, default=1, const [1,80]}
       thumb_02_r_ctrl.Scale.Y {type=MovieSceneScriptingFloatChannel, default=1, const [1,80]}
       thumb_02_r_ctrl.Scale.Z {type=MovieSceneScriptingFloatChannel, default=1, const [1,80]}
       thumb_03_r_ctrl.Location.X {type=MovieSceneScriptingFloatChannel, default=0, thinned_from=80, keys [[0,0],[79.9875,0]]}
       thumb_03_r_ctrl.Location.Y {type=MovieSceneScriptingFloatChannel, default=0, thinned_from=80, keys [[0,0],[79.9875,0]]}
       thumb_03_r_ctrl.Location.Z {type=MovieSceneScriptingFloatChannel, default=0, thinned_from=80, keys [[0,0],[79.9875,0]]}
       thumb_03_r_ctrl.Rotation.X {type=MovieSceneScriptingFloatChannel, default=0, thinned_from=80, keys [[0,0],[79.9875,0]]}
       thumb_03_r_ctrl.Rotation.Y {type=MovieSceneScriptingFloatChannel, default=0, thinned_from=80, keys [[0,0],[79.9875,0]]}
       thumb_03_r_ctrl.Rotation.Z {
        type=MovieSceneScriptingFloatChannel
        default=-26.2149
        thinned_from=80
        keys [[0,-26.2149],[1.0125,-23.7145],[2.025,-18.2949],[3.0375,-12.9862],[4.05,-10.9018],[5.0625,-16.0201],[6.075,-20.2874],[7.0875,-14.0597],[8.1,-9.5359],[9.1125,-21.3903],[10.125,-40.3111],[11.1375,-49.6363],[13.1625,-49.6363],[14.175,-47.2304],[15.1875,-31.2725],[16.2,-8.0968],[17.2125,2.2457],[18.225,-0.8718],[19.2375,-7.4694],[20.25,-16.2731],[21.2625,-26.0088],[22.275,-35.4029],[23.2875,-43.1812],[24.3,-48.0698],[25.3125,-49.6363],[39.4875,-49.6363],[41.5125,-48.0986],[46.575,-48.0986],[47.5875,-33.5773],[48.6,-8.8601],[49.6125,-1.3462],[50.625,-7.1706],[51.6375,-14.2824],[52.65,-20.7411],[53.6625,-25.0981],[54.675,-26.2149],[60.75,-26.2149],[61.7625,-26.7992],[62.775,-26.3778],[63.7875,-21.7111],[64.8,-15.3338],[65.8125,-8.8644],[66.825,-2.4638],[67.8375,2.6265],[68.85,5.5192],[69.8625,6.7306],[70.875,7.0055],[71.8875,0.6074],[72.9,-13.3554],[73.9125,-25.336],[74.925,-27.7717],[75.9375,-26.2991],[76.95,-26.2149],[79.9875,-26.2149]]
       }
       thumb_03_r_ctrl.Scale.X {type=MovieSceneScriptingFloatChannel, default=1, const [1,80]}
       thumb_03_r_ctrl.Scale.Y {type=MovieSceneScriptingFloatChannel, default=1, const [1,80]}
       thumb_03_r_ctrl.Scale.Z {type=MovieSceneScriptingFloatChannel, default=1, const [1,80]}
       index_metacarpal_r_ctrl.Location.X {type=MovieSceneScriptingFloatChannel, default=0, thinned_from=80, keys [[0,0],[79.9875,0]]}
       index_metacarpal_r_ctrl.Location.Y {type=MovieSceneScriptingFloatChannel, default=0, thinned_from=80, keys [[0,0],[79.9875,0]]}
       index_metacarpal_r_ctrl.Location.Z {type=MovieSceneScriptingFloatChannel, default=0, thinned_from=80, keys [[0,0],[79.9875,0]]}
       index_metacarpal_r_ctrl.Rotation.X {type=MovieSceneScriptingFloatChannel, default=0, thinned_from=80, keys [[0,0],[79.9875,0]]}
       index_metacarpal_r_ctrl.Rotation.Y {
        type=MovieSceneScriptingFloatChannel
        default=0
        thinned_from=80
        keys [[0,0],[4.05,0],[5.0625,0.2024],[6.075,0.5222],[7.0875,0.8154],[8.1,1.1203],[9.1125,1.4704],[10.125,1.794],[11.1375,1.9297],[13.1625,1.9297],[14.175,1.8411],[15.1875,1.253],[16.2,0.399],[17.2125,0.0178],[18.225,0.1327],[19.2375,0.3759],[20.25,0.7003],[21.2625,1.059],[22.275,1.4052],[23.2875,1.6919],[24.3,1.872],[25.3125,1.9297],[39.4875,1.9297],[41.5125,1.8536],[43.5375,1.8536],[44.55,1.6737],[45.5625,1.1485],[46.575,0.5278],[47.5875,0.1483],[48.6,0.0205],[49.6125,0],[79.9875,0]]
       }
       index_metacarpal_r_ctrl.Rotation.Z {
        type=MovieSceneScriptingFloatChannel
        default=-3.2681
        thinned_from=80
        keys [[0,-3.2681],[1.0125,-2.7432],[2.025,-1.6056],[3.0375,-0.4912],[4.05,0],[43.5375,0],[44.55,-0.0114],[45.5625,-0.0532],[46.575,-0.1287],[47.5875,-0.2105],[48.6,-0.3993],[49.6125,-0.8767],[51.6375,-2.2724],[52.65,-2.8485],[53.6625,-3.1883],[54.675,-3.268],[79.9875,-3.2681]]
       }
       index_metacarpal_r_ctrl.Scale.X {type=MovieSceneScriptingFloatChannel, default=1, const [1,80]}
       index_metacarpal_r_ctrl.Scale.Y {type=MovieSceneScriptingFloatChannel, default=1, const [1,80]}
       index_metacarpal_r_ctrl.Scale.Z {type=MovieSceneScriptingFloatChannel, default=1, const [1,80]}
       index_01_r_ctrl.Location.X {type=MovieSceneScriptingFloatChannel, default=0, thinned_from=80, keys [[0,0],[79.9875,0]]}
       index_01_r_ctrl.Location.Y {type=MovieSceneScriptingFloatChannel, default=0, thinned_from=80, keys [[0,0],[79.9875,0]]}
       index_01_r_ctrl.Location.Z {type=MovieSceneScriptingFloatChannel, default=0, thinned_from=80, keys [[0,0],[79.9875,0]]}
       index_01_r_ctrl.Rotation.X {
        type=MovieSceneScriptingFloatChannel
        default=-7.608
        thinned_from=80
        keys [[0,-7.608],[1.0125,-6.5795],[2.025,-4.3502],[3.0375,-2.1665],[4.05,-1.205],[5.0625,-1.2282],[6.075,-1.2422],[7.0875,-1.1303],[8.1,-0.8902],[9.1125,-0.5302],[10.125,-0.1611],[11.1375,0],[13.1625,0],[14.175,-0.5417],[15.1875,-4.1349],[16.2,-9.3533],[17.2125,-11.6821],[18.225,-10.9802],[19.2375,-9.4946],[20.25,-7.5123],[21.2625,-5.3201],[22.275,-3.2049],[23.2875,-1.4535],[24.3,-0.3527],[25.3125,0],[39.4875,0],[40.5,-0.0469],[41.5125,-0.0926],[43.5375,-0.0926],[44.55,-1.4845],[45.5625,-5.2356],[46.575,-8.6884],[47.5875,-9.7239],[48.6,-9.6577],[49.6125,-8.8704],[50.625,-6.8753],[51.6375,-4.4504],[52.65,-2.1044],[53.6625,-0.4355],[54.675,-0.0579],[55.6875,-0.1943],[56.7,-0.3022],[57.7125,-0.3269],[60.75,-0.3269],[61.7625,-0.3072],[62.775,-0.3736],[63.7875,-0.7071],[64.8,-1.1755],[65.8125,-1.6009],[66.825,-2.0338],[67.8375,-2.7501],[68.85,-4.2757],[69.8625,-6.2217],[70.875,-7.3158],[71.8875,-7.1922],[72.9,-6.7371],[73.9125,-6.4874],[74.925,-6.9873],[75.9375,-7.6424],[76.95,-7.8021],[77.9625,-7.7691],[78.975,-7.6673],[79.9875,-7.608]]
       }
       index_01_r_ctrl.Rotation.Y {
        type=MovieSceneScriptingFloatChannel
        default=-8.246
        thinned_from=80
        keys [[0,-8.246],[1.0125,-7.5178],[2.025,-5.9394],[3.0375,-4.3933],[4.05,-3.6574],[5.0625,-2.5264],[6.075,-0.7504],[7.0875,0.6873],[8.1,2.116],[9.1125,3.749],[10.125,5.2482],[11.1375,5.8751],[13.1625,5.8751],[14.175,5.3648],[15.1875,1.9804],[16.2,-2.9348],[17.2125,-5.1283],[18.225,-4.4672],[19.2375,-3.0679],[20.25,-1.2008],[21.2625,0.864],[22.275,2.8564],[23.2875,4.506],[24.3,5.5428],[25.3125,5.8751],[39.4875,5.8751],[41.5125,5.5242],[43.5375,5.5242],[44.55,4.3003],[45.5625,0.7663],[46.575,-3.2913],[47.5875,-5.5684],[48.6,-6.1918],[49.6125,-5.6952],[50.625,-4.1209],[51.6375,-2.2796],[52.65,-0.907],[53.6625,-0.17],[54.675,-0.0628],[55.6875,-0.2106],[56.7,-0.3276],[57.7125,-0.3543],[60.75,-0.3543],[61.7625,-0.282],[62.775,-0.2748],[63.7875,-0.6203],[64.8,-1.1457],[65.8125,-1.5557],[66.825,-1.9601],[67.8375,-2.9001],[68.85,-5.3042],[69.8625,-8.4944],[70.875,-10.3082],[71.8875,-9.7655],[72.9,-8.2466],[73.9125,-7.096],[74.925,-7.4589],[75.9375,-8.277],[76.95,-8.4563],[77.9625,-8.4206],[78.975,-8.3102],[79.9875,-8.246]]
       }
       index_01_r_ctrl.Rotation.Z {
        type=MovieSceneScriptingFloatChannel
        default=15.5516
        thinned_from=80
        keys [[0,15.5516],[1.0125,14.8855],[2.025,13.4419],[3.0375,12.0278],[4.05,11.8371],[5.0625,20.4911],[6.075,28.3367],[7.0875,25.9797],[8.1,20.9189],[9.1125,13.3328],[10.125,5.5551],[11.1375,2.16],[13.1625,2.16],[14.175,2.9235],[15.1875,7.9875],[16.2,15.342],[17.2125,18.6241],[18.225,17.6348],[19.2375,15.5411],[20.25,12.7474],[21.2625,9.6579],[22.275,6.6768],[23.2875,4.2085],[24.3,2.6571],[25.3125,2.16],[39.4875,2.16],[40.5,3.1477],[41.5125,4.1108],[44.55,4.1084],[45.5625,4.0889],[46.575,4.0271],[47.5875,2.8662],[48.6,0.9854],[49.6125,3.56],[50.625,11.8195],[51.6375,18.8688],[52.65,19.8347],[53.6625,18.4248],[54.675,17.5386],[55.6875,16.5309],[56.7,15.7338],[57.7125,15.5516],[60.75,15.5516],[61.7625,19.6026],[62.775,26.088],[63.7875,27.9808],[64.8,27.6162],[65.8125,26.9147],[66.825,25.657],[67.8375,24.6314],[68.85,24.3236],[69.8625,24.2843],[70.875,24.1136],[71.8875,22.285],[72.9,18.7599],[73.9125,15.7707],[74.925,15.1516],[75.9375,15.5299],[79.9875,15.5516]]
       }
       index_01_r_ctrl.Scale.X {type=MovieSceneScriptingFloatChannel, default=1, const [1,80]}
       index_01_r_ctrl.Scale.Y {type=MovieSceneScriptingFloatChannel, default=1, const [1,80]}
       index_01_r_ctrl.Scale.Z {type=MovieSceneScriptingFloatChannel, default=1, const [1,80]}
       index_02_r_ctrl.Location.X {type=MovieSceneScriptingFloatChannel, default=0, thinned_from=80, keys [[0,0],[79.9875,0]]}
       index_02_r_ctrl.Location.Y {type=MovieSceneScriptingFloatChannel, default=0, thinned_from=80, keys [[0,0],[79.9875,0]]}
       index_02_r_ctrl.Location.Z {type=MovieSceneScriptingFloatChannel, default=0, thinned_from=80, keys [[0,0],[79.9875,0]]}
       index_02_r_ctrl.Rotation.X {type=MovieSceneScriptingFloatChannel, default=0, thinned_from=80, keys [[0,0],[79.9875,0]]}
       index_02_r_ctrl.Rotation.Y {type=MovieSceneScriptingFloatChannel, default=0, thinned_from=80, keys [[0,0],[79.9875,0]]}
       index_02_r_ctrl.Rotation.Z {
        type=MovieSceneScriptingFloatChannel
        default=-32.6501
        thinned_from=80
        keys [[0,-32.6501],[1.0125,-27.7275],[2.025,-17.0578],[3.0375,-6.6063],[4.05,-1.3441],[5.0625,11.7524],[6.075,23.8084],[7.0875,22.6273],[8.1,16.009],[9.1125,-1.9631],[10.125,-23.1025],[11.1375,-32.7578],[13.1625,-32.7578],[14.175,-29.9198],[15.1875,-11.0961],[16.2,16.2417],[17.2125,28.4416],[18.225,24.7642],[19.2375,16.9818],[20.25,6.5971],[21.2625,-4.8871],[22.275,-15.9682],[23.2875,-25.1434],[24.3,-30.9099],[25.3125,-32.7578],[39.4875,-32.7578],[41.5125,-28.5634],[43.5375,-28.5634],[44.55,-23.0758],[45.5625,-7.4741],[46.575,9.6643],[47.5875,18.1945],[48.6,20.4409],[49.6125,22.0769],[50.625,23.87],[51.6375,24.9363],[52.65,22.2269],[53.6625,13.7941],[54.675,-0.7342],[55.6875,-18.5389],[56.7,-30.1482],[57.7125,-32.6501],[59.7375,-32.6501],[60.75,-28.4061],[61.7625,-11.5161],[62.775,13.3069],[63.7875,23.5898],[64.8,25.2436],[65.8125,21.98],[66.825,14.9004],[67.8375,10.1758],[68.85,9.8648],[69.8625,10.5015],[70.875,10.9067],[71.8875,2.5307],[72.9,-15.7833],[73.9125,-31.4973],[74.925,-34.692],[75.9375,-32.7605],[76.95,-32.6501],[79.9875,-32.6501]]
       }
       index_02_r_ctrl.Scale.X {type=MovieSceneScriptingFloatChannel, default=1, const [1,80]}
       index_02_r_ctrl.Scale.Y {type=MovieSceneScriptingFloatChannel, default=1, const [1,80]}
       index_02_r_ctrl.Scale.Z {type=MovieSceneScriptingFloatChannel, default=1, const [1,80]}
       index_03_r_ctrl.Location.X {type=MovieSceneScriptingFloatChannel, default=0, thinned_from=80, keys [[0,0],[79.9875,0]]}
       index_03_r_ctrl.Location.Y {type=MovieSceneScriptingFloatChannel, default=0.0001, thinned_from=80, keys [[0,0.0001],[79.9875,0.0001]]}
       index_03_r_ctrl.Location.Z {type=MovieSceneScriptingFloatChannel, default=-0.0001, thinned_from=80, keys [[0,-0.0001],[79.9875,-0.0001]]}
       index_03_r_ctrl.Rotation.X {type=MovieSceneScriptingFloatChannel, default=0, thinned_from=80, keys [[0,0],[79.9875,0]]}
       index_03_r_ctrl.Rotation.Y {type=MovieSceneScriptingFloatChannel, default=0, thinned_from=80, keys [[0,0],[79.9875,0]]}
       index_03_r_ctrl.Rotation.Z {
        type=MovieSceneScriptingFloatChannel
        default=-50.7194
        thinned_from=80
        keys [[0,-50.7194],[1.0125,-42.8873],[2.025,-25.9112],[3.0375,-9.2823],[4.05,-1.2412],[5.0625,12.97],[6.075,25.8717],[7.0875,22.2667],[8.1,14.5264],[9.1125,2.9236],[10.125,-8.9721],[11.1375,-14.1648],[13.1625,-14.1648],[14.175,-12.6614],[15.1875,-2.6894],[16.2,11.7931],[17.2125,18.2562],[18.225,16.308],[19.2375,12.1852],[20.25,6.6838],[21.2625,0.5999],[22.275,-5.2704],[23.2875,-10.131],[24.3,-13.1859],[25.3125,-14.1648],[39.4875,-14.1648],[40.5,-12.6541],[41.5125,-11.1811],[43.5375,-11.1811],[44.55,-8.0714],[45.5625,0.5258],[46.575,9.1813],[47.5875,12.4686],[48.6,13.1448],[49.6125,16.2586],[50.625,22.0807],[51.6375,26.5509],[52.65,12.4579],[53.6625,-16.2639],[54.675,-35.5323],[55.6875,-45.4299],[56.7,-49.9038],[57.7125,-50.7194],[59.7375,-50.7194],[60.75,-35.4081],[61.7625,-21.7316],[62.775,5.6555],[63.7875,18.5971],[64.8,20.9636],[65.8125,17.2391],[66.825,9.0168],[67.8375,3.5295],[68.85,7.4771],[69.8625,17.0331],[70.875,23.1161],[71.8875,9.3599],[72.9,-21.9154],[73.9125,-48.7507],[74.925,-54.2065],[75.9375,-50.9079],[76.95,-50.7194],[79.9875,-50.7194]]
       }
       index_03_r_ctrl.Scale.X {type=MovieSceneScriptingFloatChannel, default=1, const [1,80]}
       index_03_r_ctrl.Scale.Y {type=MovieSceneScriptingFloatChannel, default=1, const [1,80]}
       index_03_r_ctrl.Scale.Z {type=MovieSceneScriptingFloatChannel, default=1, const [1,80]}
       middle_metacarpal_r_ctrl.Location.X {type=MovieSceneScriptingFloatChannel, default=0, thinned_from=80, keys [[0,0],[79.9875,0]]}
       middle_metacarpal_r_ctrl.Location.Y {type=MovieSceneScriptingFloatChannel, default=0, thinned_from=80, keys [[0,0],[79.9875,0]]}
       middle_metacarpal_r_ctrl.Location.Z {type=MovieSceneScriptingFloatChannel, default=0, thinned_from=80, keys [[0,0],[79.9875,0]]}
       middle_metacarpal_r_ctrl.Rotation.X {type=MovieSceneScriptingFloatChannel, default=0, thinned_from=80, keys [[0,0],[79.9875,0]]}
       middle_metacarpal_r_ctrl.Rotation.Y {
        type=MovieSceneScriptingFloatChannel
        default=0.9293
        thinned_from=80
        keys [[0,0.9293],[1.0125,0.78],[2.025,0.4566],[3.0375,0.1397],[4.05,0],[47.5875,0],[48.6,0.0537],[49.6125,0.19],[50.625,0.4004],[51.6375,0.6169],[52.65,0.7965],[53.6625,0.9038],[54.675,0.9293],[79.9875,0.9293]]
       }
       middle_metacarpal_r_ctrl.Rotation.Z {type=MovieSceneScriptingFloatChannel, default=0, thinned_from=80, keys [[0,0],[79.9875,0]]}
       middle_metacarpal_r_ctrl.Scale.X {type=MovieSceneScriptingFloatChannel, default=1, const [1,80]}
       middle_metacarpal_r_ctrl.Scale.Y {type=MovieSceneScriptingFloatChannel, default=1, const [1,80]}
       middle_metacarpal_r_ctrl.Scale.Z {type=MovieSceneScriptingFloatChannel, default=1, const [1,80]}
       middle_01_r_ctrl.Location.X {type=MovieSceneScriptingFloatChannel, default=0, thinned_from=80, keys [[0,0],[79.9875,0]]}
       middle_01_r_ctrl.Location.Y {type=MovieSceneScriptingFloatChannel, default=0, thinned_from=80, keys [[0,0],[79.9875,0]]}
       middle_01_r_ctrl.Location.Z {type=MovieSceneScriptingFloatChannel, default=0, thinned_from=80, keys [[0,0],[79.9875,0]]}
       middle_01_r_ctrl.Rotation.X {
        type=MovieSceneScriptingFloatChannel
        default=-3.2118
        thinned_from=80
        keys [[0,-3.2118],[1.0125,-2.5696],[2.025,-1.1774],[3.0375,0.1862],[4.05,0.844],[5.0625,1.9756],[6.075,3.0064],[7.0875,2.7677],[8.1,2.2139],[9.1125,1.3252],[10.125,0.4038],[11.1375,0],[13.1625,0],[14.175,-0.4174],[15.1875,-3.1856],[16.2,-7.206],[17.2125,-9.0002],[18.225,-8.4593],[19.2375,-7.3148],[20.25,-5.7876],[21.2625,-4.0987],[22.275,-2.4691],[23.2875,-1.1198],[24.3,-0.2717],[25.3125,0],[39.4875,0],[40.5,0.1134],[41.5125,0.2239],[43.5375,0.2239],[44.55,-0.846],[45.5625,-3.9635],[46.575,-7.6333],[47.5875,-9.8177],[48.6,-10.4894],[49.6125,-9.7692],[50.625,-7.4812],[51.6375,-4.6875],[52.65,-2.1503],[53.6625,-0.4387],[54.675,-0.0245],[55.6875,-0.082],[56.7,-0.1276],[57.7125,-0.138],[60.75,-0.138],[61.7625,-0.7076],[62.775,-0.6807],[63.7875,-1.2234],[64.8,-1.8362],[65.8125,-1.9098],[66.825,-1.8371],[67.8375,-1.7886],[68.85,-1.9241],[69.8625,-2.2407],[70.875,-2.5228],[71.8875,-2.6308],[72.9,-2.6578],[73.9125,-2.7261],[75.9375,-3.2278],[76.95,-3.2938],[77.9625,-3.2799],[78.975,-3.2368],[79.9875,-3.2118]]
       }
       middle_01_r_ctrl.Rotation.Y {
        type=MovieSceneScriptingFloatChannel
        default=0.0026
        thinned_from=80
        keys [[0,0.0026],[1.0125,0.3286],[2.025,1.0351],[3.0375,1.7271],[4.05,1.9716],[5.0625,0.7589],[6.075,-0.3457],[7.0875,-0.0812],[8.1,0.6368],[9.1125,1.926],[10.125,3.2852],[11.1375,3.8844],[13.1625,3.8844],[14.175,3.5129],[15.1875,1.0484],[16.2,-2.5307],[17.2125,-4.1279],[18.225,-3.6465],[19.2375,-2.6276],[20.25,-1.268],[21.2625,0.2355],[22.275,1.6863],[23.2875,2.8875],[24.3,3.6425],[25.3125,3.8844],[39.4875,3.8844],[41.5125,3.5698],[43.5375,3.5698],[44.55,3.557],[45.5625,3.4551],[46.575,3.1315],[47.5875,1.333],[48.6,-1.875],[49.6125,-4.3832],[50.625,-5.984],[51.6375,-7.267],[52.65,-8.2544],[53.6625,-8.7533],[54.675,-8.7856],[55.6875,-8.6268],[56.7,-8.5012],[57.7125,-8.4725],[60.75,-8.4725],[62.775,-8.4499],[63.7875,-8.0228],[64.8,-7.397],[66.825,-6.2175],[67.8375,-5.5745],[68.85,-4.8033],[69.8625,-3.9524],[70.875,-3.1571],[71.8875,-2.5108],[72.9,-1.9492],[73.9125,-1.3656],[74.925,-0.586],[75.9375,0.0498],[76.95,0.2285],[77.9625,0.1901],[78.975,0.0716],[79.9875,0.0026]]
       }
       middle_01_r_ctrl.Rotation.Z {
        type=MovieSceneScriptingFloatChannel
        default=-64.9302
        thinned_from=80
        keys [[0,-64.9302],[1.0125,-54.1354],[2.025,-30.7375],[3.0375,-7.8182],[4.05,2.6577],[5.0625,10.1044],[6.075,16.9826],[7.0875,16.7784],[8.1,15.2012],[9.1125,7.3468],[10.125,-3.8849],[11.1375,-9.2885],[13.1625,-9.2885],[14.175,-8.8795],[15.1875,-6.1662],[16.2,-2.2257],[17.2125,-0.4671],[18.225,-0.9972],[19.2375,-2.119],[20.25,-3.6159],[21.2625,-5.2712],[22.275,-6.8685],[23.2875,-8.191],[24.3,-9.0222],[25.3125,-9.2885],[39.4875,-9.2885],[40.5,-8.3028],[41.5125,-7.3418],[43.5375,-7.3418],[44.55,-6.579],[45.5625,-4.5235],[46.575,-2.6314],[47.5875,-2.9153],[48.6,-4.3722],[49.6125,-4.2781],[50.625,-2.4196],[51.6375,-0.1502],[52.65,1.9107],[53.6625,3.301],[54.675,2.7255],[55.6875,0.5315],[56.7,-1.204],[57.7125,-1.6009],[60.75,-1.6009],[61.7625,2.7461],[62.775,3.0195],[63.7875,5.8584],[64.8,9.0594],[65.8125,8.2234],[66.825,4.9128],[67.8375,1.5703],[68.85,0.0248],[69.8625,-0.9],[70.875,-3.8668],[71.8875,-15.745],[72.9,-35.7762],[73.9125,-53.7107],[74.925,-62.2994],[75.9375,-65.3786],[76.95,-66.6184],[77.9625,-66.3315],[78.975,-65.4454],[79.9875,-64.9302]]
       }
       middle_01_r_ctrl.Scale.X {type=MovieSceneScriptingFloatChannel, default=1, const [1,80]}
       middle_01_r_ctrl.Scale.Y {type=MovieSceneScriptingFloatChannel, default=1, const [1,80]}
       middle_01_r_ctrl.Scale.Z {type=MovieSceneScriptingFloatChannel, default=1, const [1,80]}
       middle_02_r_ctrl.Location.X {type=MovieSceneScriptingFloatChannel, default=0.0009, thinned_from=80, keys [[0,0.0009],[79.9875,0.0009]]}
       middle_02_r_ctrl.Location.Y {type=MovieSceneScriptingFloatChannel, default=0.0008, thinned_from=80, keys [[0,0.0008],[79.9875,0.0008]]}
       middle_02_r_ctrl.Location.Z {type=MovieSceneScriptingFloatChannel, default=0.0476, thinned_from=80, keys [[0,0.0476],[79.9875,0.0476]]}
       middle_02_r_ctrl.Rotation.X {
        type=MovieSceneScriptingFloatChannel
        default=0.0034
        thinned_from=80
        keys [[0,0.0034],[1.0125,0.0022],[3.0375,-0.0032],[4.05,-0.0046],[5.0625,-0.0082],[6.075,-0.0109],[7.0875,-0.0104],[8.1,-0.0088],[9.1125,-0.005],[10.125,-0.0003],[11.1375,0.0018],[13.1625,0.0018],[14.175,0.001],[15.1875,-0.0043],[16.2,-0.0107],[17.2125,-0.0125],[18.225,-0.012],[19.2375,-0.0108],[20.25,-0.0088],[21.2625,-0.006],[22.275,-0.003],[23.2875,-0.0004],[24.3,0.0013],[25.3125,0.0018],[39.4875,0.0018],[41.5125,0.0008],[43.5375,0.0008],[44.55,-0.0008],[45.5625,-0.005],[46.575,-0.0086],[47.5875,-0.0097],[48.6,-0.0099],[51.6375,-0.0112],[52.65,-0.0104],[53.6625,-0.0079],[54.675,-0.0042],[55.6875,0.0001],[56.7,0.0029],[57.7125,0.0034],[59.7375,0.0034],[60.75,0.0025],[61.7625,-0.0026],[62.775,-0.0087],[63.7875,-0.011],[64.8,-0.0115],[65.8125,-0.011],[66.825,-0.0098],[67.8375,-0.0089],[68.85,-0.0091],[69.8625,-0.0096],[70.875,-0.01],[71.8875,-0.0079],[72.9,-0.0022],[73.9125,0.0031],[74.925,0.0041],[75.9375,0.0035],[79.9875,0.0034]]
       }
       middle_02_r_ctrl.Rotation.Y {
        type=MovieSceneScriptingFloatChannel
        default=-0.0131
        thinned_from=80
        keys [[0,-0.0131],[2.025,-0.0136],[4.05,-0.0128],[5.0625,-0.0108],[6.075,-0.0082],[7.0875,-0.0087],[8.1,-0.0103],[9.1125,-0.0126],[10.125,-0.0136],[14.175,-0.0136],[15.1875,-0.0129],[16.2,-0.0084],[17.2125,-0.0053],[18.225,-0.0063],[19.2375,-0.0082],[20.25,-0.0104],[21.2625,-0.0122],[22.275,-0.0133],[23.2875,-0.0136],[44.55,-0.0136],[45.5625,-0.0126],[46.575,-0.0105],[47.5875,-0.0095],[48.6,-0.0093],[49.6125,-0.0089],[51.6375,-0.0076],[52.65,-0.0087],[53.6625,-0.011],[54.675,-0.0129],[55.6875,-0.0136],[57.7125,-0.0131],[59.7375,-0.0131],[60.75,-0.0134],[61.7625,-0.0133],[62.775,-0.0104],[63.7875,-0.008],[64.8,-0.0073],[65.8125,-0.0079],[66.825,-0.0094],[67.8375,-0.0102],[68.85,-0.0101],[70.875,-0.0092],[71.8875,-0.0111],[72.9,-0.0134],[74.925,-0.013],[79.9875,-0.0131]]
       }
       middle_02_r_ctrl.Rotation.Z {
        type=MovieSceneScriptingFloatChannel
        default=-43.9674
        thinned_from=80
        keys [[0,-43.9674],[1.0125,-38.5983],[2.025,-26.9608],[3.0375,-15.5612],[4.05,-9.6647],[5.0625,7.7615],[6.075,23.7008],[7.0875,20.7621],[8.1,11.179],[9.1125,-7.8223],[10.125,-28.1212],[11.1375,-37.111],[13.1625,-37.1109],[14.175,-33.6482],[15.1875,-10.681],[16.2,22.6746],[17.2125,37.5601],[18.225,33.0732],[19.2375,23.5776],[20.25,10.907],[21.2625,-3.1052],[22.275,-16.6255],[23.2875,-27.8204],[24.3,-34.8563],[25.3125,-37.1109],[39.4875,-37.111],[41.5125,-32.593],[43.5375,-32.593],[44.55,-25.984],[45.5625,-7.851],[46.575,9.9416],[47.5875,16.4196],[48.6,17.3888],[49.6125,19.8571],[50.625,23.7259],[51.6375,26.4591],[52.65,20.8083],[53.6625,6.2864],[54.675,-11.4733],[55.6875,-29.943],[56.7,-41.5103],[57.7125,-43.9674],[59.7375,-43.9674],[60.75,-39.8661],[61.7625,-18.1536],[62.775,10.6778],[63.7875,24.4132],[64.8,28.3789],[65.8125,25.0545],[66.825,17.0524],[67.8375,11.7121],[68.85,12.5683],[69.8625,15.759],[70.875,17.79],[71.8875,6.0334],[72.9,-19.9954],[73.9125,-42.329],[74.925,-46.8695],[75.9375,-44.1243],[76.95,-43.9674],[79.9875,-43.9674]]
       }
       middle_02_r_ctrl.Scale.X {type=MovieSceneScriptingFloatChannel, default=1, const [1,80]}
       middle_02_r_ctrl.Scale.Y {type=MovieSceneScriptingFloatChannel, default=1, const [1,80]}
       middle_02_r_ctrl.Scale.Z {type=MovieSceneScriptingFloatChannel, default=1, const [1,80]}
       middle_03_r_ctrl.Location.X {type=MovieSceneScriptingFloatChannel, default=0.0005, thinned_from=80, keys [[0,0.0005],[79.9875,0.0005]]}
       middle_03_r_ctrl.Location.Y {type=MovieSceneScriptingFloatChannel, default=0, thinned_from=80, keys [[0,0],[79.9875,0]]}
       middle_03_r_ctrl.Location.Z {type=MovieSceneScriptingFloatChannel, default=0.0061, thinned_from=80, keys [[0,0.0061],[79.9875,0.0061]]}
       middle_03_r_ctrl.Rotation.X {
        type=MovieSceneScriptingFloatChannel
        default=-0.0109
        thinned_from=80
        keys [[0,-0.0109],[4.05,-0.011],[5.0625,-0.0106],[6.075,-0.0097],[7.0875,-0.01],[9.1125,-0.0109],[15.1875,-0.0109],[16.2,-0.0099],[17.2125,-0.0091],[18.225,-0.0094],[21.2625,-0.0108],[23.2875,-0.0111],[44.55,-0.0111],[45.5625,-0.0109],[46.575,-0.0104],[47.5875,-0.0101],[51.6375,-0.0096],[52.65,-0.01],[53.6625,-0.0105],[59.7375,-0.0109],[60.75,-0.0108],[61.7625,-0.011],[63.7875,-0.0099],[65.8125,-0.0101],[68.85,-0.0106],[79.9875,-0.0109]]
       }
       middle_03_r_ctrl.Rotation.Y {
        type=MovieSceneScriptingFloatChannel
        default=0.0017
        thinned_from=80
        keys [[0,0.0017],[1.0125,0.0015],[3.0375,0.0006],[4.05,0.0006],[5.0625,0.0031],[6.075,0.0053],[7.0875,0.0047],[8.1,0.0035],[10.125,-0.0003],[11.1375,-0.0012],[13.1625,-0.0012],[14.175,-0.0008],[15.1875,0.0016],[16.2,0.0049],[17.2125,0.0063],[18.225,0.0059],[19.2375,0.005],[22.275,0.001],[23.2875,-0.0002],[24.3,-0.001],[25.3125,-0.0012],[39.4875,-0.0012],[41.5125,-0.0007],[43.5375,-0.0007],[44.55,0],[46.575,0.0039],[47.5875,0.0045],[49.6125,0.0048],[51.6375,0.0054],[52.65,0.0048],[53.6625,0.0034],[54.675,0.0024],[55.6875,0.0019],[57.7125,0.0017],[59.7375,0.0017],[60.75,0.0025],[61.7625,0.0007],[62.775,0.0037],[63.7875,0.005],[64.8,0.0049],[65.8125,0.0046],[67.8375,0.0035],[70.875,0.0031],[73.9125,0.0017],[79.9875,0.0017]]
       }
       middle_03_r_ctrl.Rotation.Z {
        type=MovieSceneScriptingFloatChannel
        default=-5.2274
        thinned_from=80
        keys [[0,-5.2274],[1.0125,-6.2649],[2.025,-8.5138],[3.0375,-10.7166],[4.05,-11.0142],[5.0625,2.4559],[6.075,14.7036],[7.0875,11.5643],[8.1,4.8241],[9.1125,-5.2796],[10.125,-15.6384],[11.1375,-20.1602],[13.1625,-20.1602],[14.175,-18.2642],[15.1875,-5.6889],[16.2,12.5743],[17.2125,20.7245],[18.225,18.2678],[19.2375,13.0687],[20.25,6.1311],[21.2625,-1.541],[22.275,-8.9438],[23.2875,-15.0734],[24.3,-18.9257],[25.3125,-20.1602],[39.4875,-20.1602],[41.5125,-17.562],[43.5375,-17.562],[44.55,-13.8276],[45.5625,-3.5769],[46.575,6.4969],[47.5875,10.1962],[48.6,10.7521],[49.6125,12.0495],[50.625,14.0606],[51.6375,15.4755],[52.65,11.6805],[53.6625,3.8828],[54.675,-1.2888],[55.6875,-3.8757],[56.7,-5.0211],[57.7125,-5.2274],[59.7375,-5.2274],[60.75,-0.7032],[61.7625,-10.4219],[62.775,5.3781],[63.7875,12.8694],[64.8,12.265],[65.8125,10.5737],[66.825,7.3406],[67.8375,4.3836],[68.85,3.237],[69.8625,2.9995],[70.875,2.2989],[71.8875,0.1636],[72.9,-2.7199],[73.9125,-5.0565],[74.925,-5.5757],[75.9375,-5.2462],[76.95,-5.2274],[79.9875,-5.2274]]
       }
       middle_03_r_ctrl.Scale.X {type=MovieSceneScriptingFloatChannel, default=1, const [1,80]}
       middle_03_r_ctrl.Scale.Y {type=MovieSceneScriptingFloatChannel, default=1, const [1,80]}
       middle_03_r_ctrl.Scale.Z {type=MovieSceneScriptingFloatChannel, default=1, const [1,80]}
       ring_metacarpal_r_ctrl.Location.X {type=MovieSceneScriptingFloatChannel, default=0, thinned_from=80, keys [[0,0],[79.9875,0]]}
       ring_metacarpal_r_ctrl.Location.Y {type=MovieSceneScriptingFloatChannel, default=0, thinned_from=80, keys [[0,0],[79.9875,0]]}
       ring_metacarpal_r_ctrl.Location.Z {type=MovieSceneScriptingFloatChannel, default=0, thinned_from=80, keys [[0,0],[79.9875,0]]}
       ring_metacarpal_r_ctrl.Rotation.X {
        type=MovieSceneScriptingFloatChannel
        default=0
        thinned_from=80
        keys [[0,0],[1.0125,0.0035],[3.0375,0.0185],[4.05,0.0217],[5.0625,0.0195],[6.075,0.0159],[8.1,0.0091],[9.1125,0.0052],[10.125,0.0015],[11.1375,0],[39.4875,0],[41.5125,0.0009],[44.55,0.0008],[47.5875,0.0001],[79.9875,0]]
       }
       ring_metacarpal_r_ctrl.Rotation.Y {
        type=MovieSceneScriptingFloatChannel
        default=-0.6161
        thinned_from=80
        keys [[0,-0.6161],[1.0125,-0.724],[2.025,-0.9578],[3.0375,-1.1868],[4.05,-1.2884],[5.0625,-1.3003],[8.1,-1.357],[10.125,-1.3985],[11.1375,-1.4069],[13.1625,-1.4069],[14.175,-1.5],[15.1875,-2.1172],[16.2,-3.0136],[17.2125,-3.4137],[18.225,-3.2931],[19.2375,-3.0379],[20.25,-2.6974],[21.2625,-2.3208],[22.275,-1.9575],[23.2875,-1.6566],[24.3,-1.4675],[25.3125,-1.4069],[39.4875,-1.4069],[41.5125,-1.4022],[43.5375,-1.4022],[44.55,-1.8463],[45.5625,-3.0428],[46.575,-4.1442],[47.5875,-4.4187],[48.6,-4.1404],[49.6125,-3.5236],[51.6375,-1.8226],[52.65,-1.1235],[53.6625,-0.7123],[54.675,-0.6162],[79.9875,-0.6161]]
       }
       ring_metacarpal_r_ctrl.Rotation.Z {
        type=MovieSceneScriptingFloatChannel
        default=-1.4089
        thinned_from=80
        keys [[0,-1.4089],[1.0125,-1.3382],[2.025,-1.185],[3.0375,-1.0349],[4.05,-0.9641],[5.0625,-0.8671],[6.075,-0.7065],[7.0875,-0.5594],[8.1,-0.4063],[9.1125,-0.2306],[10.125,-0.0681],[11.1375,0],[13.1625,0],[14.175,-0.0445],[15.1875,-0.3397],[16.2,-0.7684],[17.2125,-0.9597],[18.225,-0.9021],[19.2375,-0.78],[20.25,-0.6172],[21.2625,-0.4371],[22.275,-0.2633],[23.2875,-0.1194],[24.3,-0.029],[25.3125,0],[39.4875,0],[40.5,-0.0193],[41.5125,-0.0382],[43.5375,-0.0382],[44.55,-0.1097],[45.5625,-0.3255],[46.575,-0.6027],[47.5875,-0.8426],[48.6,-1.0329],[49.6125,-1.1661],[50.625,-1.2556],[51.6375,-1.3252],[52.65,-1.378],[53.6625,-1.404],[54.675,-1.4089],[79.9875,-1.4089]]
       }
       ring_metacarpal_r_ctrl.Scale.X {type=MovieSceneScriptingFloatChannel, default=1, const [1,80]}
       ring_metacarpal_r_ctrl.Scale.Y {type=MovieSceneScriptingFloatChannel, default=1, const [1,80]}
       ring_metacarpal_r_ctrl.Scale.Z {type=MovieSceneScriptingFloatChannel, default=1, const [1,80]}
       ring_01_r_ctrl.Location.X {type=MovieSceneScriptingFloatChannel, default=0, thinned_from=80, keys [[0,0],[79.9875,0]]}
       ring_01_r_ctrl.Location.Y {type=MovieSceneScriptingFloatChannel, default=0, thinned_from=80, keys [[0,0],[79.9875,0]]}
       ring_01_r_ctrl.Location.Z {type=MovieSceneScriptingFloatChannel, default=0, thinned_from=80, keys [[0,0],[79.9875,0]]}
       ring_01_r_ctrl.Rotation.X {
        type=MovieSceneScriptingFloatChannel
        default=-5.7649
        thinned_from=80
        keys [[0,-5.7649],[1.0125,-5.4565],[2.025,-4.788],[3.0375,-4.1332],[4.05,-3.8262],[5.0625,-3.4413],[6.075,-2.8041],[7.0875,-2.2201],[8.1,-1.6125],[9.1125,-0.9152],[10.125,-0.2704],[11.1375,0],[13.1625,0],[14.175,-0.4526],[15.1875,-3.4546],[16.2,-7.8143],[17.2125,-9.7599],[18.225,-9.1735],[19.2375,-7.9323],[20.25,-6.2762],[21.2625,-4.4447],[22.275,-2.6776],[23.2875,-1.2143],[24.3,-0.2947],[25.3125,0],[39.4875,0],[40.5,-0.0768],[41.5125,-0.1516],[43.5375,-0.1516],[44.55,-1.2483],[45.5625,-4.2037],[46.575,-6.9241],[47.5875,-7.7413],[48.6,-7.6993],[49.6125,-7.2629],[50.625,-6.3023],[51.6375,-5.5093],[52.65,-5.4527],[53.6625,-5.6883],[54.675,-5.7649],[59.7375,-5.7649],[60.75,-5.7236],[61.7625,-5.7328],[62.775,-5.7594],[63.7875,-5.7319],[64.8,-5.6926],[65.8125,-5.7023],[66.825,-5.7379],[67.8375,-5.7617],[68.85,-5.5223],[69.8625,-5.0259],[70.875,-4.71],[71.8875,-4.8877],[72.9,-5.3443],[73.9125,-5.7361],[74.925,-5.8158],[75.9375,-5.7676],[79.9875,-5.7649]]
       }
       ring_01_r_ctrl.Rotation.Y {
        type=MovieSceneScriptingFloatChannel
        default=0.3633
        thinned_from=80
        keys [[0,0.3633],[1.0125,0.7162],[2.025,1.4812],[3.0375,2.2306],[4.05,2.5486],[5.0625,2.2922],[6.075,1.8678],[7.0875,1.4788],[8.1,1.0741],[9.1125,0.6096],[10.125,0.1801],[11.1375,0],[13.1625,0],[14.175,0.4715],[15.1875,3.5991],[16.2,8.1414],[17.2125,10.1684],[18.225,9.5574],[19.2375,8.2643],[20.25,6.5389],[21.2625,4.6308],[22.275,2.7896],[23.2875,1.2652],[24.3,0.307],[25.3125,0],[39.4875,0],[40.5,0.0511],[41.5125,0.101],[43.5375,0.101],[44.55,0.7838],[45.5625,2.7251],[46.575,4.8573],[47.5875,5.9527],[48.6,6.2202],[49.6125,6.1022],[50.625,5.588],[51.6375,4.6812],[52.65,2.8227],[53.6625,0.9172],[54.675,0.3633],[59.7375,0.3633],[60.75,0.748],[61.7625,0.6669],[62.775,0.4157],[63.7875,0.6337],[64.8,0.9617],[65.8125,0.8818],[66.825,0.5864],[67.8375,0.3892],[68.85,1.0392],[69.8625,2.4224],[70.875,3.3029],[71.8875,2.8076],[72.9,1.5352],[73.9125,0.4434],[74.925,0.2214],[75.9375,0.3556],[76.95,0.3633],[79.9875,0.3633]]
       }
       ring_01_r_ctrl.Rotation.Z {
        type=MovieSceneScriptingFloatChannel
        default=-68.297
        thinned_from=80
        keys [[0,-68.297],[1.0125,-59.0826],[2.025,-39.1105],[3.0375,-19.5469],[4.05,-10.0858],[5.0625,6.6499],[6.075,21.9765],[7.0875,19.6982],[8.1,14.8066],[9.1125,7.4739],[10.125,-0.0437],[11.1375,-3.3254],[13.1625,-3.3254],[14.175,-2.4443],[15.1875,3.3997],[16.2,11.8869],[17.2125,15.6746],[18.225,14.5329],[19.2375,12.1167],[20.25,8.8927],[21.2625,5.3273],[22.275,1.8871],[23.2875,-0.9614],[24.3,-2.7517],[25.3125,-3.3254],[39.4875,-3.3254],[40.5,-2.3707],[41.5125,-1.4398],[43.5375,-1.4398],[44.55,-0.8171],[45.5625,0.8608],[46.575,2.4054],[47.5875,1.5112],[48.6,-2.968],[49.6125,-9.9019],[50.625,-18.3346],[51.6375,-28.7693],[52.65,-46.5017],[53.6625,-63.4794],[54.675,-68.297],[59.7375,-68.297],[60.75,-64.4699],[61.7625,-45.4812],[62.775,-25.5433],[63.7875,-15.9721],[64.8,-9.8498],[65.8125,-5.8322],[66.825,-3.2969],[67.8375,-1.8341],[68.85,-0.9632],[69.8625,-0.9873],[70.875,-3.0076],[71.8875,-19.8959],[72.9,-49.5576],[73.9125,-67.0789],[74.925,-68.297],[79.9875,-68.297]]
       }
       ring_01_r_ctrl.Scale.X {type=MovieSceneScriptingFloatChannel, default=1, const [1,80]}
       ring_01_r_ctrl.Scale.Y {type=MovieSceneScriptingFloatChannel, default=1, const [1,80]}
       ring_01_r_ctrl.Scale.Z {type=MovieSceneScriptingFloatChannel, default=1, const [1,80]}
       ring_02_r_ctrl.Location.X {type=MovieSceneScriptingFloatChannel, default=0, thinned_from=80, keys [[0,0],[79.9875,0]]}
       ring_02_r_ctrl.Location.Y {type=MovieSceneScriptingFloatChannel, default=0, thinned_from=80, keys [[0,0],[79.9875,0]]}
       ring_02_r_ctrl.Location.Z {type=MovieSceneScriptingFloatChannel, default=0, thinned_from=80, keys [[0,0],[79.9875,0]]}
       ring_02_r_ctrl.Rotation.X {type=MovieSceneScriptingFloatChannel, default=0, thinned_from=80, keys [[0,0],[79.9875,0]]}
       ring_02_r_ctrl.Rotation.Y {type=MovieSceneScriptingFloatChannel, default=0, thinned_from=80, keys [[0,0],[79.9875,0]]}
       ring_02_r_ctrl.Rotation.Z {
        type=MovieSceneScriptingFloatChannel
        default=-39.2964
        thinned_from=80
        keys [[0,-39.2964],[1.0125,-36.0746],[2.025,-29.0915],[3.0375,-22.2512],[4.05,-18.4843],[5.0625,-3.4562],[6.075,10.4086],[7.0875,9.6712],[8.1,5.0597],[9.1125,-11.4176],[10.125,-33.006],[11.1375,-43.1693],[13.1625,-43.1693],[14.175,-40.2474],[15.1875,-20.8671],[16.2,7.2792],[17.2125,19.8399],[18.225,16.0537],[19.2375,8.0412],[20.25,-2.6506],[21.2625,-14.4744],[22.275,-25.8832],[23.2875,-35.3298],[24.3,-41.2668],[25.3125,-43.1693],[39.4875,-43.1693],[41.5125,-39.1979],[43.5375,-39.1979],[44.55,-30.5887],[45.5625,-7.3886],[46.575,13.9667],[47.5875,19.9976],[48.6,16.6616],[49.6125,7.5294],[50.625,-6.0395],[51.6375,-19.7714],[52.65,-31.0706],[53.6625,-37.7313],[54.675,-39.2964],[60.75,-39.2964],[61.7625,-36.2389],[62.775,-27.8365],[63.7875,-17.7829],[64.8,-9.7561],[65.8125,-4.1028],[66.825,0.2248],[67.8375,4.0518],[68.85,8.1163],[69.8625,11.4824],[70.875,12.0512],[71.8875,-0.7149],[72.9,-24.2824],[73.9125,-38.3195],[74.925,-39.2964],[79.9875,-39.2964]]
       }
       ring_02_r_ctrl.Scale.X {type=MovieSceneScriptingFloatChannel, default=1, const [1,80]}
       ring_02_r_ctrl.Scale.Y {type=MovieSceneScriptingFloatChannel, default=1, const [1,80]}
       ring_02_r_ctrl.Scale.Z {type=MovieSceneScriptingFloatChannel, default=1, const [1,80]}
       ring_03_r_ctrl.Location.X {type=MovieSceneScriptingFloatChannel, default=0, thinned_from=80, keys [[0,0],[79.9875,0]]}
       ring_03_r_ctrl.Location.Y {type=MovieSceneScriptingFloatChannel, default=0, thinned_from=80, keys [[0,0],[79.9875,0]]}
       ring_03_r_ctrl.Location.Z {type=MovieSceneScriptingFloatChannel, default=0, thinned_from=80, keys [[0,0],[79.9875,0]]}
       ring_03_r_ctrl.Rotation.X {type=MovieSceneScriptingFloatChannel, default=0, thinned_from=80, keys [[0,0],[79.9875,0]]}
       ring_03_r_ctrl.Rotation.Y {type=MovieSceneScriptingFloatChannel, default=0, thinned_from=80, keys [[0,0],[79.9875,0]]}
       ring_03_r_ctrl.Rotation.Z {
        type=MovieSceneScriptingFloatChannel
        default=-10.7468
        thinned_from=80
        keys [[0,-10.7468],[1.0125,-11.2927],[3.0375,-13.6351],[4.05,-13.5219],[5.0625,-1.0385],[6.075,10.2957],[7.0875,7.1438],[8.1,0.3761],[9.1125,-9.7685],[10.125,-20.1693],[11.1375,-24.7095],[13.1625,-24.7095],[14.175,-23.157],[15.1875,-12.8596],[16.2,2.0954],[17.2125,8.7693],[18.225,6.7576],[19.2375,2.5003],[20.25,-3.1806],[21.2625,-9.463],[22.275,-15.5249],[23.2875,-20.5441],[24.3,-23.6986],[25.3125,-24.7095],[39.4875,-24.7095],[41.5125,-22.1008],[43.5375,-22.1008],[44.55,-18.6043],[45.5625,-9.1819],[46.575,-0.5088],[47.5875,1.4118],[48.6,-1.4227],[49.6125,-7.1291],[50.625,-14.7677],[51.6375,-22.3371],[52.65,-28.5632],[53.6625,-32.2147],[54.675,-33.0668],[60.75,-33.0668],[61.7625,-22.5635],[62.775,-7.3891],[63.7875,-2.6118],[64.8,-2.6804],[65.8125,-3.8414],[66.825,-4.8553],[67.8375,-4.8609],[68.85,-2.9856],[69.8625,-0.1109],[70.875,1.5793],[71.8875,-1.1629],[72.9,-6.9742],[73.9125,-10.5008],[74.925,-10.7468],[79.9875,-10.7468]]
       }
       ring_03_r_ctrl.Scale.X {type=MovieSceneScriptingFloatChannel, default=1, const [1,80]}
       ring_03_r_ctrl.Scale.Y {type=MovieSceneScriptingFloatChannel, default=1, const [1,80]}
       ring_03_r_ctrl.Scale.Z {type=MovieSceneScriptingFloatChannel, default=1, const [1,80]}
       pinky_metacarpal_r_ctrl.Location.X {type=MovieSceneScriptingFloatChannel, default=0, thinned_from=80, keys [[0,0],[79.9875,0]]}
       pinky_metacarpal_r_ctrl.Location.Y {type=MovieSceneScriptingFloatChannel, default=0, thinned_from=80, keys [[0,0],[79.9875,0]]}
       pinky_metacarpal_r_ctrl.Location.Z {type=MovieSceneScriptingFloatChannel, default=0, thinned_from=80, keys [[0,0],[79.9875,0]]}
       pinky_metacarpal_r_ctrl.Rotation.X {
        type=MovieSceneScriptingFloatChannel
        default=0
        thinned_from=80
        keys [[0,0],[1.0125,0.0191],[2.025,0.0605],[3.0375,0.101],[4.05,0.1183],[5.0625,0.1064],[6.075,0.0867],[7.0875,0.0687],[8.1,0.0499],[9.1125,0.0283],[10.125,0.0084],[11.1375,0],[39.4875,0],[41.5125,0.0047],[43.5375,0.0047],[44.55,0.0042],[45.5625,0.0029],[46.575,0.0013],[47.5875,0.0004],[48.6,0],[79.9875,0]]
       }
       pinky_metacarpal_r_ctrl.Rotation.Y {
        type=MovieSceneScriptingFloatChannel
        default=-2.7762
        thinned_from=80
        keys [[0,-2.7762],[1.0125,-2.6767],[3.0375,-2.2498],[4.05,-2.1665],[5.0625,-2.3717],[6.075,-2.7113],[7.0875,-3.0227],[8.1,-3.3465],[9.1125,-3.7183],[10.125,-4.062],[11.1375,-4.2061],[13.1625,-4.2061],[14.175,-4.3094],[15.1875,-4.9945],[16.2,-5.9895],[17.2125,-6.4335],[18.225,-6.2996],[19.2375,-6.0164],[20.25,-5.6384],[21.2625,-5.2205],[22.275,-4.8172],[23.2875,-4.4832],[24.3,-4.2733],[25.3125,-4.2061],[39.4875,-4.2061],[41.5125,-4.1253],[43.5375,-4.1253],[44.55,-4.7788],[45.5625,-6.5401],[46.575,-8.1613],[47.5875,-8.5557],[48.6,-8.1203],[49.6125,-7.1781],[51.6375,-4.6015],[52.65,-3.5434],[53.6625,-2.9216],[54.675,-2.7762],[79.9875,-2.7762]]
       }
       pinky_metacarpal_r_ctrl.Rotation.Z {
        type=MovieSceneScriptingFloatChannel
        default=-2.9008
        thinned_from=80
        keys [[0,-2.9008],[1.0125,-2.9419],[3.0375,-3.1179],[4.05,-3.1412],[5.0625,-2.8252],[6.075,-2.3021],[7.0875,-1.8226],[8.1,-1.3238],[9.1125,-0.7513],[10.125,-0.222],[11.1375,0],[13.1625,0],[14.175,-0.145],[15.1875,-1.1069],[16.2,-2.5038],[17.2125,-3.1271],[18.225,-2.9392],[19.2375,-2.5416],[20.25,-2.0109],[21.2625,-1.4241],[22.275,-0.8579],[23.2875,-0.3891],[24.3,-0.0944],[25.3125,0],[39.4875,0],[40.5,-0.063],[41.5125,-0.1245],[43.5375,-0.1245],[44.55,-0.4194],[45.5625,-1.2795],[46.575,-2.2939],[47.5875,-2.9057],[48.6,-3.1028],[49.6125,-3.1142],[50.625,-3.0643],[52.65,-2.9478],[53.6625,-2.9104],[54.675,-2.9008],[79.9875,-2.9008]]
       }
       pinky_metacarpal_r_ctrl.Scale.X {type=MovieSceneScriptingFloatChannel, default=1, const [1,80]}
       pinky_metacarpal_r_ctrl.Scale.Y {type=MovieSceneScriptingFloatChannel, default=1, const [1,80]}
       pinky_metacarpal_r_ctrl.Scale.Z {type=MovieSceneScriptingFloatChannel, default=1, const [1,80]}
       pinky_01_r_ctrl.Location.X {type=MovieSceneScriptingFloatChannel, default=0, thinned_from=80, keys [[0,0],[79.9875,0]]}
       pinky_01_r_ctrl.Location.Y {type=MovieSceneScriptingFloatChannel, default=0, thinned_from=80, keys [[0,0],[79.9875,0]]}
       pinky_01_r_ctrl.Location.Z {type=MovieSceneScriptingFloatChannel, default=0, thinned_from=80, keys [[0,0],[79.9875,0]]}
       pinky_01_r_ctrl.Rotation.X {
        type=MovieSceneScriptingFloatChannel
        default=-6.3509
        thinned_from=80
        keys [[0,-6.3509],[1.0125,-7.4333],[2.025,-9.7793],[3.0375,-12.0774],[4.05,-12.9885],[5.0625,-10.8747],[6.075,-7.6762],[7.0875,-5.2623],[8.1,-3.2783],[9.1125,-1.7001],[10.125,-0.4737],[11.1375,0],[13.1625,0],[14.175,-0.1572],[15.1875,-1.1997],[16.2,-2.7138],[17.2125,-3.3895],[18.225,-3.1858],[19.2375,-2.7548],[20.25,-2.1796],[21.2625,-1.5436],[22.275,-0.9299],[23.2875,-0.4217],[24.3,-0.1024],[25.3125,0],[39.4875,0],[40.5,-0.1925],[41.5125,-0.3802],[43.5375,-0.3802],[44.55,-0.4144],[45.5625,-0.5587],[46.575,-0.8701],[47.5875,-1.5064],[48.6,-2.4142],[49.6125,-3.2743],[50.625,-4.0411],[51.6375,-4.779],[52.65,-5.5847],[53.6625,-6.1928],[54.675,-6.3509],[67.8375,-6.3509],[68.85,-6.124],[69.8625,-5.6596],[70.875,-5.3641],[71.8875,-5.5303],[72.9,-5.9575],[73.9125,-6.324],[74.925,-6.3985],[75.9375,-6.3534],[79.9875,-6.3509]]
       }
       pinky_01_r_ctrl.Rotation.Y {
        type=MovieSceneScriptingFloatChannel
        default=1.9101
        thinned_from=80
        keys [[0,1.9101],[1.0125,2.4586],[2.025,3.6475],[3.0375,4.812],[4.05,5.3659],[5.0625,6.1759],[6.075,6.9085],[7.0875,6.6242],[8.1,5.578],[9.1125,3.3924],[10.125,1.0428],[11.1375,0],[13.1625,0],[14.175,0.9269],[15.1875,7.0745],[16.2,16.0027],[17.2125,19.9871],[18.225,18.7861],[19.2375,16.2444],[20.25,12.8529],[21.2625,9.1023],[22.275,5.4833],[23.2875,2.4868],[24.3,0.6035],[25.3125,0],[39.4875,0],[40.5,0.2597],[41.5125,0.513],[43.5375,0.513],[44.55,1.9672],[45.5625,5.8862],[46.575,9.4936],[47.5875,10.5386],[48.6,10.1809],[49.6125,9.1066],[50.625,7.3586],[51.6375,5.4557],[52.65,3.5831],[53.6625,2.2529],[54.675,1.9101],[67.8375,1.9101],[68.85,2.3641],[69.8625,3.2932],[70.875,3.8845],[71.8875,3.5518],[72.9,2.6972],[73.9125,1.9639],[74.925,1.8148],[75.9375,1.9049],[76.95,1.9101],[79.9875,1.9101]]
       }
       pinky_01_r_ctrl.Rotation.Z {
        type=MovieSceneScriptingFloatChannel
        default=-62.0733
        thinned_from=80
        keys [[0,-62.0733],[1.0125,-57.0313],[2.025,-44.7519],[3.0375,-29.3192],[4.05,-14.5005],[5.0625,1.7523],[6.075,15.6729],[7.0875,21.0968],[8.1,21.6182],[9.1125,16.4504],[10.125,8.2022],[11.1375,4.137],[13.1625,4.137],[14.175,4.8748],[15.1875,9.7682],[16.2,16.8749],[17.2125,20.0464],[18.225,19.0904],[19.2375,17.0673],[20.25,14.3677],[21.2625,11.3823],[22.275,8.5016],[23.2875,6.1165],[24.3,4.6174],[25.3125,4.137],[39.4875,4.137],[40.5,4.5533],[41.5125,4.9591],[43.5375,4.9591],[44.55,7.1916],[45.5625,13.2079],[46.575,18.7457],[47.5875,16.784],[48.6,6.9409],[49.6125,-6.0993],[51.6375,-34.7902],[52.65,-49.1898],[53.6625,-59.4519],[54.675,-62.0733],[60.75,-62.0733],[61.7625,-54.0555],[62.775,-40.3047],[63.7875,-32.8216],[64.8,-29.587],[65.8125,-27.7107],[66.825,-26.5676],[67.8375,-24.694],[68.85,-20.1591],[69.8625,-14.3107],[70.875,-11.0198],[71.8875,-20.6125],[72.9,-42.1957],[73.9125,-60.7147],[74.925,-64.4797],[75.9375,-62.2034],[76.95,-62.0733],[79.9875,-62.0733]]
       }
       pinky_01_r_ctrl.Scale.X {type=MovieSceneScriptingFloatChannel, default=1, const [1,80]}
       pinky_01_r_ctrl.Scale.Y {type=MovieSceneScriptingFloatChannel, default=1, const [1,80]}
       pinky_01_r_ctrl.Scale.Z {type=MovieSceneScriptingFloatChannel, default=1, const [1,80]}
       pinky_02_r_ctrl.Location.X {type=MovieSceneScriptingFloatChannel, default=0, thinned_from=80, keys [[0,0],[79.9875,0]]}
       pinky_02_r_ctrl.Location.Y {type=MovieSceneScriptingFloatChannel, default=0, thinned_from=80, keys [[0,0],[79.9875,0]]}
       pinky_02_r_ctrl.Location.Z {type=MovieSceneScriptingFloatChannel, default=0, thinned_from=80, keys [[0,0],[79.9875,0]]}
       pinky_02_r_ctrl.Rotation.X {type=MovieSceneScriptingFloatChannel, default=0, thinned_from=80, keys [[0,0],[79.9875,0]]}
       pinky_02_r_ctrl.Rotation.Y {type=MovieSceneScriptingFloatChannel, default=0, thinned_from=80, keys [[0,0],[79.9875,0]]}
       pinky_02_r_ctrl.Rotation.Z {
        type=MovieSceneScriptingFloatChannel
        default=-23.4532
        thinned_from=80
        keys [[0,-23.4532],[1.0125,-24.315],[3.0375,-28.0128],[4.05,-27.7456],[5.0625,-6.2686],[6.075,13.5917],[7.0875,13.472],[8.1,11.7956],[9.1125,-1.0536],[10.125,-20.7974],[11.1375,-30.4509],[13.1625,-30.4509],[14.175,-28.1972],[15.1875,-13.2495],[16.2,8.4592],[17.2125,18.1471],[18.225,15.2269],[19.2375,9.047],[20.25,0.8005],[21.2625,-8.319],[22.275,-17.1184],[23.2875,-24.4043],[24.3,-28.9835],[25.3125,-30.4509],[39.4875,-30.4509],[41.5125,-27.188],[43.5375,-27.188],[44.55,-21.0517],[45.5625,-4.5156],[46.575,10.7057],[47.5875,13.1042],[48.6,7.3748],[49.6125,0.1415],[50.625,-7.5407],[51.6375,-14.8232],[52.65,-20.7396],[53.6625,-24.122],[54.675,-24.8932],[60.75,-24.8932],[61.7625,-18.8985],[62.775,-8.9039],[63.7875,-4.1982],[64.8,-3.2347],[65.8125,-3.1024],[66.825,-3.087],[67.8375,-3.0323],[68.85,2.0732],[69.8625,12.3796],[70.875,18.9215],[71.8875,11.2767],[72.9,-6.8026],[73.9125,-22.3152],[74.925,-25.469],[75.9375,-23.5622],[76.95,-23.4532],[79.9875,-23.4532]]
       }
       pinky_02_r_ctrl.Scale.X {type=MovieSceneScriptingFloatChannel, default=1, const [1,80]}
       pinky_02_r_ctrl.Scale.Y {type=MovieSceneScriptingFloatChannel, default=1, const [1,80]}
       pinky_02_r_ctrl.Scale.Z {type=MovieSceneScriptingFloatChannel, default=1, const [1,80]}
       pinky_03_r_ctrl.Location.X {type=MovieSceneScriptingFloatChannel, default=0, thinned_from=80, keys [[0,0],[79.9875,0]]}
       pinky_03_r_ctrl.Location.Y {type=MovieSceneScriptingFloatChannel, default=0, thinned_from=80, keys [[0,0],[79.9875,0]]}
       pinky_03_r_ctrl.Location.Z {type=MovieSceneScriptingFloatChannel, default=0, thinned_from=80, keys [[0,0],[79.9875,0]]}
       pinky_03_r_ctrl.Rotation.X {type=MovieSceneScriptingFloatChannel, default=0, thinned_from=80, keys [[0,0],[79.9875,0]]}
       pinky_03_r_ctrl.Rotation.Y {type=MovieSceneScriptingFloatChannel, default=0, thinned_from=80, keys [[0,0],[79.9875,0]]}
       pinky_03_r_ctrl.Rotation.Z {
        type=MovieSceneScriptingFloatChannel
        default=-31.8495
        thinned_from=80
        keys [[0,-31.8495],[1.0125,-31.9413],[3.0375,-32.3352],[4.05,-31.4488],[5.0625,-12.0031],[6.075,5.7765],[7.0875,2.7041],[8.1,-3.8925],[9.1125,-13.781],[10.125,-23.919],[11.1375,-28.3445],[13.1625,-28.3445],[14.175,-26.8624],[15.1875,-17.0319],[16.2,-2.755],[17.2125,3.6163],[18.225,1.6958],[19.2375,-2.3684],[20.25,-7.7917],[21.2625,-13.7893],[22.275,-19.5763],[23.2875,-24.368],[24.3,-27.3795],[25.3125,-28.3445],[39.4875,-28.3445],[41.5125,-25.8017],[43.5375,-25.8017],[44.55,-21.8517],[45.5625,-11.2074],[46.575,-1.4094],[47.5875,-0.8279],[48.6,-8.2724],[49.6125,-20.3144],[50.625,-35.0641],[51.6375,-49.454],[52.65,-61.2396],[53.6625,-68.0919],[54.675,-69.6786],[60.75,-69.6786],[61.7625,-54.4758],[62.775,-29.4851],[63.7875,-19.4916],[64.8,-18.2122],[65.8125,-16.3629],[66.825,-14.3264],[67.8375,-12.4773],[68.85,-10.559],[69.8625,-8.7749],[70.875,-7.8829],[71.8875,-12.4402],[72.9,-22.544],[73.9125,-31.2135],[74.925,-32.976],[75.9375,-31.9104],[79.9875,-31.8495]]
       }
       pinky_03_r_ctrl.Scale.X {type=MovieSceneScriptingFloatChannel, default=1, const [1,80]}
       pinky_03_r_ctrl.Scale.Y {type=MovieSceneScriptingFloatChannel, default=1, const [1,80]}
       pinky_03_r_ctrl.Scale.Z {type=MovieSceneScriptingFloatChannel, default=1, const [1,80]}
       hand_r_ik_ctrl.Location.X {
        type=MovieSceneScriptingFloatChannel
        default=-29.09
        thinned_from=80
        keys [[0,-29.09],[1.0125,-27.4201],[2.025,-23.9583],[3.0375,-20.9045],[4.05,-18.8803],[5.0625,-17.2489],[6.075,-16.196],[7.0875,-15.9202],[8.1,-15.9014],[9.1125,-16.2525],[10.125,-16.9641],[11.1375,-17.1075],[12.15,-17.0557],[13.1625,-16.0571],[14.175,-11.0332],[15.1875,-2.0432],[16.2,0.3719],[17.2125,-1.0861],[18.225,-4.3349],[19.2375,-8.1078],[20.25,-10.7252],[21.2625,-11.9779],[22.275,-12.4587],[23.2875,-12.3964],[24.3,-11.6676],[25.3125,-10.1813],[26.325,-8.1668],[27.3375,-7.0135],[28.35,-7.3526],[29.3625,-7.9843],[30.375,-8.495],[31.3875,-8.9715],[32.4,-9.1044],[33.4125,-8.934],[34.425,-8.9692],[35.4375,-9.166],[36.45,-9.4961],[37.4625,-9.7666],[39.4875,-10.0925],[40.5,-12.1536],[41.5125,-15.1105],[42.525,-16.1069],[43.5375,-16.0924],[44.55,-16.0176],[45.5625,-15.9776],[46.575,-16.6347],[47.5875,-18.6113],[48.6,-21.3969],[49.6125,-25.1023],[50.625,-29.6931],[51.6375,-34.4187],[52.65,-36.9496],[53.6625,-37.1048],[55.6875,-36.4415],[56.7,-35.8339],[57.7125,-34.8967],[58.725,-34.0858],[59.7375,-33.5939],[60.75,-33.2822],[61.7625,-33.4067],[62.775,-34.1887],[63.7875,-34.7971],[64.8,-35.0185],[65.8125,-34.7265],[66.825,-33.9104],[67.8375,-32.4052],[68.85,-30.178],[69.8625,-27.7906],[70.875,-26.0733],[71.8875,-25.5295],[72.9,-25.5456],[73.9125,-25.7991],[75.9375,-25.9454],[76.95,-26.632],[77.9625,-27.805],[78.975,-28.718],[79.9875,-29.09]]
       }
       hand_r_ik_ctrl.Location.Y {
        type=MovieSceneScriptingFloatChannel
        default=5.0848
        thinned_from=80
        keys [[0,5.0848],[1.0125,3.1219],[2.025,-0.879],[3.0375,-4.2176],[4.05,-6.145],[5.0625,-7.4203],[6.075,-7.7985],[7.0875,-7.6447],[8.1,-7.4041],[9.1125,-7.3658],[10.125,-7.2266],[11.1375,-7.0067],[12.15,-6.9033],[13.1625,-6.7831],[14.175,-6.4333],[15.1875,-3.8256],[16.2,-1.1473],[17.2125,1.8687],[18.225,5.8673],[19.2375,9.1545],[20.25,10.7314],[21.2625,10.9148],[22.275,10.0174],[23.2875,8.4251],[24.3,6.1301],[25.3125,2.5503],[26.325,-2.0574],[27.3375,-5.5324],[28.35,-7.2383],[29.3625,-7.8841],[30.375,-7.707],[31.3875,-7.1923],[32.4,-6.6145],[33.4125,-6.176],[34.425,-6.012],[35.4375,-5.7773],[36.45,-5.4454],[37.4625,-5.2393],[38.475,-5.1949],[39.4875,-5.1787],[40.5,-5.0667],[41.5125,-5.0185],[42.525,-5.2503],[43.5375,-5.6988],[44.55,-6.1983],[45.5625,-6.8183],[46.575,-7.4648],[47.5875,-8.5103],[48.6,-9.8217],[50.625,-12.9421],[51.6375,-13.4712],[52.65,-12.0543],[53.6625,-10.5999],[54.675,-9.7808],[55.6875,-8.9873],[56.7,-8.3891],[57.7125,-8.0951],[59.7375,-7.3662],[61.7625,-7.0373],[62.775,-6.9634],[63.7875,-7.4257],[64.8,-8.2042],[65.8125,-9.1163],[66.825,-10.0565],[67.8375,-10.7758],[68.85,-10.6443],[69.8625,-8.3054],[70.875,-4.4047],[71.8875,-0.6333],[72.9,2.6184],[73.9125,4.2824],[74.925,4.6843],[75.9375,4.9247],[76.95,5.0924],[77.9625,5.0541],[78.975,5.041],[79.9875,5.0848]]
       }
       hand_r_ik_ctrl.Location.Z {
        type=MovieSceneScriptingFloatChannel
        default=26.1516
        thinned_from=80
        keys [[0,26.1516],[2.025,25.8968],[3.0375,25.9529],[4.05,26.3326],[5.0625,26.9132],[6.075,27.5981],[7.0875,28.3472],[8.1,28.6894],[10.125,28.8219],[11.1375,28.8418],[12.15,28.4742],[13.1625,28.2921],[14.175,27.3377],[15.1875,18.705],[16.2,7.4556],[17.2125,-1.8316],[18.225,-8.2794],[19.2375,-12.1048],[20.25,-13.5603],[21.2625,-13.7565],[22.275,-13.2262],[23.2875,-12.1297],[24.3,-10.0417],[25.3125,-6.3729],[26.325,-0.827],[27.3375,5.1815],[28.35,10.3012],[29.3625,14.2995],[30.375,17.1908],[31.3875,19.5796],[32.4,21.3012],[33.4125,21.9586],[34.425,22.0766],[35.4375,22.0201],[36.45,21.8922],[37.4625,21.856],[38.475,21.9648],[39.4875,22.138],[40.5,22.7677],[41.5125,23.2504],[42.525,23.6411],[43.5375,24.1525],[44.55,24.713],[45.5625,25.3514],[46.575,25.4708],[47.5875,25.0268],[48.6,24.6569],[49.6125,25.1974],[50.625,26.6962],[51.6375,28.948],[52.65,30.6508],[53.6625,29.9253],[54.675,28.674],[55.6875,26.8293],[56.7,23.7096],[57.7125,21.3495],[58.725,20.0906],[59.7375,19.4364],[60.75,19.1463],[61.7625,19.6833],[62.775,20.7138],[63.7875,22.0606],[64.8,23.6742],[65.8125,25.164],[66.825,26.3901],[67.8375,26.9332],[68.85,27.0895],[69.8625,27.5811],[71.8875,29.1201],[72.9,29.5828],[73.9125,28.5876],[75.9375,26.2751],[76.95,25.6725],[77.9625,25.6456],[78.975,25.9417],[79.9875,26.1516]]
       }
       hand_r_ik_ctrl.Rotation.X {
        type=MovieSceneScriptingFloatChannel
        default=-34.6093
        thinned_from=80
        keys [[0,-34.6093],[1.0125,-35.2317],[2.025,-36.0566],[3.0375,-35.1097],[4.05,-30.7173],[5.0625,-26.9506],[6.075,-33.4809],[7.0875,-57.1158],[8.1,-78.0414],[9.1125,-97.6689],[10.125,-107.4446],[11.1375,-107.8411],[12.15,-108.0159],[13.1625,-112.0656],[14.175,-107.8337],[15.1875,-40.002],[16.2,-21.2177],[17.2125,-7.8005],[18.225,-5.1052],[19.2375,-4.9966],[20.25,-5.0576],[21.2625,-5.0422],[22.275,-6.0212],[23.2875,-7.6927],[24.3,-11.9073],[25.3125,-16.8457],[26.325,-20.8507],[27.3375,-24.0434],[28.35,-25.7641],[29.3625,-27.0426],[30.375,-32.6565],[31.3875,-54.2782],[32.4,-88.2957],[33.4125,-105.0843],[34.425,-106.3924],[35.4375,-104.4157],[36.45,-101.1823],[37.4625,-98.7293],[38.475,-97.6491],[39.4875,-97.009],[40.5,-90.8118],[41.5125,-74.186],[42.525,-67.216],[43.5375,-70.3065],[44.55,-75.5157],[45.5625,-81.8382],[46.575,-73.746],[47.5875,-43.1021],[48.6,-18.4661],[49.6125,-14.4393],[50.625,-27.6939],[51.6375,-30.3493],[52.65,26.511],[53.6625,165.4582],[54.675,173.4045],[55.6875,176.6704],[56.7,182.2932],[58.725,197.5173],[59.7375,203.4056],[60.75,207.3959],[61.7625,207.1188],[62.775,237.1563],[63.7875,255.9846],[64.8,263.4711],[65.8125,269.0742],[66.825,275.2576],[67.8375,282.5121],[68.85,293.189],[69.8625,318.9658],[70.875,345.6376],[71.8875,350.0297],[72.9,346.3048],[73.9125,337.8167],[74.925,332.7735],[78.975,326.3843],[79.9875,325.3907]]
       }
       hand_r_ik_ctrl.Rotation.Y {
        type=MovieSceneScriptingFloatChannel
        default=60.0706
        thinned_from=80
        keys [[0,60.0706],[1.0125,59.0648],[2.025,57.814],[3.0375,59.03],[4.05,64.5781],[5.0625,72.2377],[6.075,78.2782],[7.0875,79.1989],[8.1,77.2778],[9.1125,75.994],[10.125,75.1187],[11.1375,75.0156],[12.15,74.617],[13.1625,73.9435],[14.175,74.5928],[15.1875,58.2409],[16.2,35.1841],[17.2125,15.7597],[18.225,7.1321],[19.2375,4.2745],[20.25,5.5362],[21.2625,7.3278],[22.275,8.5211],[23.2875,9.7832],[24.3,13.3313],[25.3125,18.2885],[26.325,24.7271],[27.3375,32.5725],[28.35,42.4244],[30.375,65.1033],[31.3875,73.1935],[32.4,73.8631],[33.4125,71.7248],[34.425,70.8272],[35.4375,70.5202],[36.45,70.4105],[37.4625,70.6046],[38.475,71.1001],[39.4875,71.7396],[40.5,73.3197],[41.5125,74.2587],[42.525,74.0602],[45.5625,76.4802],[46.575,77.4441],[47.5875,77.2165],[48.6,74.9583],[49.6125,75.4146],[50.625,78.0258],[51.6375,80.1344],[52.65,88.2438],[53.6625,81.6256],[54.675,79.2179],[55.6875,79.5217],[56.7,80.0625],[57.7125,80.7757],[58.725,81.7104],[59.7375,82.3344],[60.75,82.4068],[61.7625,83.1067],[62.775,82.099],[63.7875,78.9354],[64.8,76.483],[65.8125,75.3223],[67.8375,74.526],[68.85,74.8674],[69.8625,76.4299],[70.875,72.5949],[71.8875,65.6128],[72.9,59.8257],[73.9125,56.139],[74.925,54.9878],[75.9375,55.4926],[76.95,56.8597],[77.9625,58.3473],[78.975,59.5066],[79.9875,60.0706]]
       }
       hand_r_ik_ctrl.Rotation.Z {
        type=MovieSceneScriptingFloatChannel
        default=-22.5378
        thinned_from=80
        keys [[0,-22.5378],[1.0125,-20.951],[2.025,-18.3365],[3.0375,-17.3878],[4.05,-19.1122],[5.0625,-24.2611],[6.075,-38.4147],[7.0875,-67.1018],[8.1,-90.0144],[9.1125,-108.1569],[10.125,-116.1092],[12.15,-116.8153],[13.1625,-120.4291],[14.175,-111.6172],[15.1875,-36.9469],[16.2,-25.1967],[17.2125,-24.5486],[18.225,-22.8117],[19.2375,-17.4436],[20.25,-7.2817],[21.2625,3.1129],[22.275,12.417],[23.2875,16.7606],[24.3,14.5139],[25.3125,9.062],[26.325,0.6231],[27.3375,-6.4979],[28.35,-9.3567],[29.3625,-13.9207],[30.375,-27.1494],[31.3875,-57.2039],[32.4,-95.7113],[33.4125,-112.3475],[34.425,-113.7177],[35.4375,-111.3328],[36.45,-107.235],[37.4625,-104.2041],[38.475,-103],[39.4875,-102.3484],[40.5,-96.6359],[41.5125,-80.4481],[42.525,-73.1552],[43.5375,-76.3773],[44.55,-82.1621],[45.5625,-89.5495],[46.575,-83.5319],[47.5875,-57.447],[48.6,-39.3376],[49.6125,-45.0144],[50.625,-72.6781],[51.6375,-91.4442],[52.65,-52.5955],[53.6625,72.3828],[54.675,76.3299],[55.6875,78.9049],[56.7,85.6024],[57.7125,95.0306],[58.725,103.0547],[59.7375,108.6163],[60.75,112.2675],[61.7625,110.348],[62.775,143.6201],[63.7875,169.137],[65.8125,200.8642],[66.825,218.7612],[67.8375,238.2411],[68.85,261.0604],[69.8625,297.7553],[70.875,333.2969],[71.8875,344.8654],[72.9,347.5062],[73.9125,343.8724],[74.925,345.5275],[75.9375,345.2461],[78.975,338.5929],[79.9875,337.4622]]
       }
       hand_r_ik_ctrl.Scale.X {type=MovieSceneScriptingFloatChannel, default=1, thinned_from=80, keys [[0,1],[79.9875,1]]}
       hand_r_ik_ctrl.Scale.Y {type=MovieSceneScriptingFloatChannel, default=1, const [1,80]}
       hand_r_ik_ctrl.Scale.Z {type=MovieSceneScriptingFloatChannel, default=1, const [1,80]}
       Softness_2 {type=MovieSceneScriptingFloatChannel, default=0, const [0,80]}
       hand_r_ikorient_ctrl.Rotation.Roll {type=MovieSceneScriptingFloatChannel, default=0, const [0,80]}
       hand_r_ikorient_ctrl.Rotation.Pitch {type=MovieSceneScriptingFloatChannel, default=0, const [0,80]}
       hand_r_ikorient_ctrl.Rotation.Yaw {type=MovieSceneScriptingFloatChannel, default=0, const [0,80]}
       arm_r_pv_ik_ctrl.Location.X {
        type=MovieSceneScriptingFloatChannel
        default=0.7958
        thinned_from=80
        keys [[0,0.7958],[1.0125,1.4027],[2.025,2.7883],[3.0375,4.6872],[4.05,6.6229],[5.0625,8.234],[6.075,9.4152],[7.0875,10.2415],[8.1,10.7509],[9.1125,11.0827],[10.125,11.3042],[11.1375,11.5592],[12.15,11.4593],[13.1625,11.1751],[14.175,10.7396],[15.1875,8.573],[16.2,5.7862],[17.2125,2.411],[18.225,-1.0642],[19.2375,-3.8182],[20.25,-5.5211],[21.2625,-6.5092],[22.275,-7.0095],[23.2875,-7.1707],[24.3,-6.9288],[25.3125,-5.7353],[26.325,-3.1415],[27.3375,-0.1929],[28.35,2.1983],[29.3625,4.3747],[30.375,5.5428],[31.3875,5.9595],[32.4,6.2601],[33.4125,6.2996],[34.425,6.1358],[35.4375,5.9065],[36.45,5.6893],[37.4625,5.55],[38.475,5.4665],[39.4875,5.1873],[40.5,4.0177],[41.5125,2.8754],[42.525,2.904],[43.5375,3.5507],[44.55,4.5356],[45.5625,5.844],[46.575,7.2548],[48.6,9.9568],[49.6125,11.3681],[50.625,12.9548],[51.6375,15.268],[52.65,17.4474],[53.6625,17.8661],[54.675,17.4653],[55.6875,16.3935],[56.7,14.1782],[57.7125,11.9093],[58.725,10.3659],[59.7375,9.4663],[60.75,9.0596],[61.7625,9.2383],[62.775,9.455],[63.7875,9.6045],[64.8,9.6192],[65.8125,9.3306],[66.825,8.8414],[68.85,7.1224],[69.8625,7.0282],[70.875,7.4983],[71.8875,7.6877],[72.9,7.1347],[73.9125,5.1531],[74.925,2.1585],[75.9375,-0.9615],[76.95,-2.7298],[77.9625,-1.9344],[78.975,-0.1652],[79.9875,0.7958]]
       }
       arm_r_pv_ik_ctrl.Location.Y {
        type=MovieSceneScriptingFloatChannel
        default=7.2221
        keys [[0,7.2221],[1.0125,6.4187],[2.025,5.7811],[3.0375,6.3012],[4.05,7.2991],[5.0625,8.5738],[6.075,9.7907],[7.0875,10.746],[8.1,11.0306],[9.1125,10.8111],[10.125,10.4223],[11.1375,10.4321],[12.15,10.0945],[13.1625,10.6376],[14.175,12.9427],[15.1875,11.9326],[16.2,3.9215],[17.2125,-5.5188],[18.225,-13.1809],[19.2375,-18.5119],[20.25,-21.3065],[21.2625,-22.544],[22.275,-22.9898],[23.2875,-22.7545],[24.3,-21.5382],[25.3125,-18.662],[26.325,-13.1237],[27.3375,-6.9271],[28.35,-2.8287],[29.3625,0.1832],[30.375,2.4399],[31.3875,4.4493],[32.4,6.2009],[33.4125,7.0263],[34.425,7.0254],[35.4375,6.5917],[36.45,5.8874],[37.4625,5.4092],[38.475,5.3266],[39.4875,5.3959],[40.5,4.691],[41.5125,3.3021],[42.525,3.2267],[43.5375,3.9839],[44.55,4.9201],[45.5625,6.0039],[46.575,5.8717],[47.5875,4.156],[48.6,2.2205],[49.6125,1.4252],[50.625,2.2773],[51.6375,5.0447],[52.65,7.7829],[53.6625,6.9571],[54.675,5.4876],[55.6875,3.469],[56.7,-0.4904],[57.7125,-3.4734],[58.725,-4.9753],[59.7375,-5.8751],[60.75,-6.4284],[61.7625,-5.8557],[62.775,-4.5323],[63.7875,-2.6215],[64.8,-0.2493],[65.8125,1.9332],[66.825,3.7472],[67.8375,4.6529],[68.85,5.0472],[69.8625,5.7649],[70.875,7.3645],[71.8875,9.6333],[72.9,11.7593],[73.9125,11.3756],[74.925,10.198],[75.9375,8.1997],[76.95,6.7714],[77.9625,6.5313],[78.975,6.8939],[79.9875,7.2221]]
       }
       arm_r_pv_ik_ctrl.Location.Z {
        type=MovieSceneScriptingFloatChannel
        default=-1.5724
        thinned_from=80
        keys [[0,-1.5724],[1.0125,-1.1218],[2.025,-0.3458],[3.0375,0.4318],[4.05,1.2184],[5.0625,1.903],[6.075,2.4268],[7.0875,2.8321],[8.1,3.1647],[9.1125,3.4331],[10.125,3.5963],[11.1375,3.524],[12.15,3.1166],[13.1625,2.03],[14.175,-0.0518],[15.1875,-2.7954],[16.2,-3.2804],[17.2125,-0.9408],[18.225,3.0642],[19.2375,7.5064],[20.25,10.3793],[21.2625,11.6348],[22.275,12.1656],[23.2875,12.0815],[24.3,10.8368],[25.3125,8.0523],[26.325,3.8937],[27.3375,0.961],[28.35,0.2055],[29.3625,0.3272],[30.375,0.2925],[31.3875,-0.1343],[32.4,-0.523],[33.4125,-0.7588],[35.4375,-1.0238],[36.45,-1.2018],[37.4625,-1.4864],[38.475,-1.8025],[39.4875,-1.7789],[40.5,-0.3118],[41.5125,1.6123],[42.525,2.2441],[43.5375,2.3965],[44.55,2.4655],[45.5625,2.7111],[46.575,3.7511],[47.5875,5.7147],[48.6,8.1735],[49.6125,10.4212],[50.625,12.1663],[51.6375,13.9594],[52.65,15.9233],[53.6625,16.862],[54.675,16.6469],[55.6875,15.4184],[56.7,13.6578],[57.7125,11.9423],[58.725,10.8488],[59.7375,10.5892],[60.75,10.855],[61.7625,11.1753],[62.775,11.1207],[63.7875,10.6056],[64.8,9.8621],[66.825,8.3038],[67.8375,7.396],[68.85,6.5542],[69.8625,6.0272],[70.875,5.3423],[71.8875,4.0436],[72.9,2.231],[73.9125,-0.4619],[74.925,-3.2639],[75.9375,-4.4894],[76.95,-4.3803],[77.9625,-3.4662],[78.975,-2.2173],[79.9875,-1.5724]]
       }
       arm_r_pv_ik_ctrl.Rotation.X {
        type=MovieSceneScriptingFloatChannel
        default=-16.564
        thinned_from=80
        keys [[0,-16.564],[1.0125,-16.5683],[2.025,-17.3942],[3.0375,-18.4143],[4.05,-18.75],[5.0625,-19.0434],[6.075,-18.9413],[7.0875,-18.6874],[9.1125,-17.2498],[10.125,-16.6187],[11.1375,-16.3944],[12.15,-16.3582],[13.1625,-17.5114],[14.175,-19.7082],[15.1875,-16.5932],[16.2,-7.5388],[17.2125,3.7073],[18.225,14.6198],[19.2375,23.541],[20.25,28.288],[21.2625,29.8083],[22.275,29.8046],[23.2875,28.5361],[24.3,25.0492],[25.3125,18.228],[26.325,7.4655],[27.3375,-2.0848],[28.35,-6.8355],[29.3625,-9.2296],[30.375,-10.7278],[31.3875,-12.3638],[32.4,-13.6956],[33.4125,-14.1343],[34.425,-14.0024],[35.4375,-13.3521],[36.45,-12.2905],[37.4625,-11.5856],[38.475,-11.5146],[39.4875,-11.5976],[40.5,-10.6716],[41.5125,-9.0409],[42.525,-9.1418],[43.5375,-10.3085],[44.55,-11.7839],[45.5625,-13.4454],[46.575,-13.4537],[47.5875,-12.0911],[48.6,-11.2379],[49.6125,-13.5781],[50.625,-20.4198],[51.6375,-29.9581],[52.65,-34.5372],[53.6625,-32.0601],[54.675,-29.8257],[55.6875,-28.0749],[56.7,-24.7775],[57.7125,-21.8687],[58.725,-19.7039],[59.7375,-17.8955],[60.75,-16.4848],[61.7625,-16.388],[62.775,-17.9484],[63.7875,-20.9781],[64.8,-24.5038],[65.8125,-27.2464],[66.825,-28.9242],[67.8375,-28.8771],[68.85,-27.1158],[69.8625,-23.508],[70.875,-20.2029],[71.8875,-19.2455],[72.9,-19.6209],[73.9125,-20.1661],[74.925,-21.1772],[75.9375,-19.9397],[76.95,-18.3314],[77.9625,-17.4443],[78.975,-16.8151],[79.9875,-16.564]]
       }
       arm_r_pv_ik_ctrl.Rotation.Y {
        type=MovieSceneScriptingFloatChannel
        default=10.4413
        thinned_from=80
        keys [[0,10.4413],[1.0125,11.7535],[2.025,14.3439],[3.0375,16.8558],[4.05,18.9144],[5.0625,20.4913],[6.075,21.4774],[7.0875,22.1214],[9.1125,22.8668],[10.125,23.1656],[11.1375,23.3624],[12.15,23.1073],[13.1625,22.2671],[14.175,20.576],[15.1875,15.5463],[16.2,10.1314],[17.2125,5.3587],[18.225,0.4988],[19.2375,-3.1853],[20.25,-4.5737],[21.2625,-4.1854],[22.275,-2.579],[23.2875,-0.4198],[24.3,2.2476],[25.3125,5.8801],[26.325,9.6692],[27.3375,12.1005],[28.35,13.9222],[29.3625,15.6688],[30.375,16.3424],[31.3875,16.1712],[32.4,15.9177],[34.425,15.2244],[35.4375,14.8315],[36.45,14.4703],[37.4625,14.2434],[38.475,14.0994],[39.4875,13.6604],[40.5,12.6406],[41.5125,12.1475],[42.525,12.5216],[43.5375,13.4969],[44.55,14.895],[45.5625,16.7778],[46.575,19.3353],[47.5875,22.9048],[48.6,27.3263],[49.6125,32.1928],[50.625,36.6474],[51.6375,40.2825],[52.65,43.3359],[53.6625,45.1666],[54.675,45.4307],[55.6875,44.5261],[56.7,42.7672],[57.7125,40.3191],[58.725,38.1331],[59.7375,36.9243],[60.75,36.4981],[61.7625,36.2518],[62.775,35.7131],[63.7875,34.9178],[64.8,33.7624],[65.8125,32.2761],[66.825,30.6338],[67.8375,28.7959],[68.85,27.1014],[70.875,24.7186],[71.8875,23.3746],[72.9,21.5134],[73.9125,18.0517],[74.925,13.549],[75.9375,8.8],[76.95,6.0955],[77.9625,7.0091],[78.975,9.2258],[79.9875,10.4413]]
       }
       arm_r_pv_ik_ctrl.Rotation.Z {
        type=MovieSceneScriptingFloatChannel
        default=7.7502
        thinned_from=80
        keys [[0,7.7502],[1.0125,4.9256],[2.025,0.4344],[3.0375,-1.6272],[4.05,-1.4726],[5.0625,-0.5787],[6.075,0.9846],[7.0875,2.4749],[8.1,3.4372],[9.1125,3.4932],[10.125,3.2456],[11.1375,3.5646],[12.15,3.344],[13.1625,3.5795],[14.175,6.5352],[15.1875,12.0064],[16.2,11.2898],[17.2125,8.2512],[18.225,5.9872],[19.2375,3.9531],[20.25,2.3751],[21.2625,1.1435],[22.275,-0.1347],[23.2875,-1.5046],[24.3,-2.9554],[25.3125,-4.5786],[26.325,-6.0587],[27.3375,-6.3214],[28.35,-5.4032],[29.3625,-3.418],[30.375,-1.7628],[31.3875,-0.6849],[32.4,0.5699],[33.4125,1.3768],[34.425,1.2994],[35.4375,1.0101],[36.45,0.6326],[37.4625,0.1738],[38.475,-0.2332],[39.4875,-0.4712],[40.5,-1.4055],[41.5125,-2.741],[42.525,-3.0204],[43.5375,-2.7181],[44.55,-2.1695],[45.5625,-1.3867],[46.575,-0.9576],[47.5875,-2.0887],[48.6,-4.2566],[49.6125,-8.1024],[50.625,-14.2093],[51.6375,-19.806],[52.65,-19.6698],[53.6625,-17.5235],[54.675,-16.2083],[55.6875,-15.6979],[56.7,-16.4953],[57.7125,-16.4807],[58.725,-15.3774],[59.7375,-14.1429],[61.7625,-11.9468],[62.775,-12.0715],[63.7875,-13.3123],[64.8,-14.7221],[65.8125,-15.8642],[66.825,-16.6545],[67.8375,-16.9142],[68.85,-15.2032],[69.8625,-9.2315],[70.875,-1.0385],[71.8875,5.5994],[72.9,10.3762],[73.9125,10.5116],[74.925,8.0279],[75.9375,5.9507],[76.95,4.9516],[77.9625,5.4482],[78.975,6.8549],[79.9875,7.7502]]
       }
       arm_r_pv_ik_ctrl.Scale.X {type=MovieSceneScriptingFloatChannel, default=1, thinned_from=80, keys [[0,1],[79.9875,1]]}
       arm_r_pv_ik_ctrl.Scale.Y {type=MovieSceneScriptingFloatChannel, default=1, const [1,80]}
       arm_r_pv_ik_ctrl.Scale.Z {type=MovieSceneScriptingFloatChannel, default=1, const [1,80]}
       body_ctrl.Location.X {
        type=MovieSceneScriptingFloatChannel
        default=3.0662
        thinned_from=80
        keys [[0,3.0662],[1.0125,3.0798],[3.0375,3.1481],[5.0625,3.202],[8.1,3.3201],[11.1375,3.3949],[12.15,3.4376],[13.1625,3.5263],[14.175,3.6812],[15.1875,3.8939],[16.2,4.0861],[17.2125,4.2311],[18.225,4.2776],[19.2375,4.2222],[21.2625,4.0396],[22.275,3.9644],[25.3125,3.7768],[26.325,3.7212],[27.3375,3.6745],[28.35,3.6415],[29.3625,3.6229],[32.4,3.6107],[33.4125,3.5987],[34.425,3.5729],[35.4375,3.532],[36.45,3.4762],[37.4625,3.4068],[38.475,3.3273],[39.4875,3.1395],[40.5,2.6073],[41.5125,2.1725],[42.525,2.1268],[43.5375,2.1714],[44.55,2.2526],[45.5625,2.3626],[47.5875,2.6159],[49.6125,2.88],[50.625,2.9953],[51.6375,3.0785],[52.65,3.119],[53.6625,3.1046],[54.675,3.0353],[55.6875,2.9307],[56.7,2.8879],[57.7125,2.9972],[58.725,3.1863],[59.7375,3.3483],[60.75,3.4546],[61.7625,3.4954],[63.7875,3.4905],[64.8,3.4679],[65.8125,3.4243],[67.8375,3.2738],[68.85,3.2119],[69.8625,3.1678],[70.875,3.1453],[72.9,3.1177],[74.925,3.1082],[75.9375,3.096],[79.9875,3.0662]]
       }
       body_ctrl.Location.Y {
        type=MovieSceneScriptingFloatChannel
        default=-2.9923
        thinned_from=80
        keys [[0,-2.9923],[1.0125,-2.9826],[2.025,-2.9665],[3.0375,-2.9653],[4.05,-2.9994],[5.0625,-3.0804],[7.0875,-3.2772],[8.1,-3.3595],[10.125,-3.4761],[11.1375,-3.5564],[12.15,-3.6679],[13.1625,-3.8139],[14.175,-3.999],[16.2,-4.403],[17.2125,-4.5884],[18.225,-4.7007],[19.2375,-4.73],[20.25,-4.7245],[22.275,-4.6857],[23.2875,-4.6847],[25.3125,-4.7107],[26.325,-4.7124],[27.3375,-4.6879],[28.35,-4.6003],[29.3625,-4.4246],[30.375,-4.2227],[31.3875,-4.0316],[32.4,-3.8664],[33.4125,-3.7621],[34.425,-3.7001],[36.45,-3.5887],[37.4625,-3.5248],[38.475,-3.4683],[39.4875,-3.3875],[40.5,-3.1606],[41.5125,-2.9707],[42.525,-2.9531],[44.55,-2.9797],[47.5875,-3.0378],[48.6,-3.0665],[51.6375,-3.1858],[52.65,-3.2177],[53.6625,-3.2562],[54.675,-3.373],[55.6875,-3.6863],[56.7,-4.1092],[57.7125,-4.4561],[58.725,-4.6908],[59.7375,-4.833],[60.75,-4.9193],[61.7625,-4.9554],[62.775,-4.9259],[63.7875,-4.7896],[64.8,-4.5169],[65.8125,-4.1572],[66.825,-3.7651],[67.8375,-3.4048],[68.85,-3.1545],[69.8625,-3.0508],[70.875,-3.0491],[71.8875,-3.057],[73.9125,-3.0548],[74.925,-3.0365],[75.9375,-3.0097],[76.95,-2.9952],[79.9875,-2.9923]]
       }
       body_ctrl.Location.Z {
        type=MovieSceneScriptingFloatChannel
        default=-5.1686
        thinned_from=80
        keys [[0,-5.1686],[1.0125,-5.1595],[2.025,-5.135],[6.075,-4.9644],[7.0875,-4.9326],[8.1,-4.9125],[9.1125,-4.9052],[11.1375,-4.9205],[12.15,-4.9453],[13.1625,-5.0267],[15.1875,-5.2962],[16.2,-5.3493],[17.2125,-5.2734],[18.225,-5.1189],[19.2375,-4.9795],[20.25,-4.9278],[21.2625,-4.9448],[24.3,-5.0425],[25.3125,-5.0614],[26.325,-5.0628],[27.3375,-5.0465],[31.3875,-4.9221],[32.4,-4.8989],[33.4125,-4.8861],[34.425,-4.8919],[35.4375,-4.94],[36.45,-5.0426],[37.4625,-5.1649],[38.475,-5.2504],[39.4875,-5.2135],[40.5,-4.8618],[41.5125,-4.4651],[42.525,-4.3934],[43.5375,-4.5098],[44.55,-4.6463],[45.5625,-4.7288],[46.575,-4.7648],[47.5875,-4.7788],[48.6,-4.7742],[51.6375,-4.702],[52.65,-4.7034],[53.6625,-4.7404],[54.675,-4.8272],[55.6875,-5.0662],[56.7,-5.401],[57.7125,-5.6273],[58.725,-5.7377],[59.7375,-5.7886],[60.75,-5.815],[61.7625,-5.8284],[62.775,-5.8148],[63.7875,-5.7613],[64.8,-5.6783],[66.825,-5.4772],[67.8375,-5.386],[68.85,-5.3179],[69.8625,-5.2877],[70.875,-5.3075],[73.9125,-5.4788],[74.925,-5.5167],[75.9375,-5.4909],[76.95,-5.423],[78.975,-5.2196],[79.9875,-5.1686]]
       }
       body_ctrl.Rotation.X {
        type=MovieSceneScriptingFloatChannel
        default=14.7608
        thinned_from=80
        keys [[0,14.7608],[2.025,14.9297],[3.0375,14.8318],[4.05,14.3399],[5.0625,13.3472],[6.075,12.2217],[7.0875,11.1639],[8.1,10.3552],[9.1125,9.9414],[10.125,9.7021],[11.1375,9.5347],[12.15,9.4271],[13.1625,9.4958],[14.175,9.7769],[15.1875,10.218],[16.2,10.7174],[17.2125,11.2902],[18.225,11.951],[19.2375,12.5757],[20.25,13.0875],[21.2625,13.4022],[22.275,13.4516],[23.2875,13.2286],[24.3,12.7992],[25.3125,12.2297],[26.325,11.5884],[29.3625,9.5559],[30.375,8.8925],[31.3875,8.3049],[32.4,7.8886],[33.4125,7.6971],[34.425,7.6926],[35.4375,7.8283],[36.45,8.0975],[38.475,8.8333],[39.4875,9.1472],[40.5,9.3727],[41.5125,9.554],[42.525,9.8173],[43.5375,10.1431],[44.55,10.4253],[45.5625,10.5978],[47.5875,10.8816],[48.6,10.912],[50.625,10.7759],[51.6375,10.8508],[53.6625,11.2254],[55.6875,11.5467],[56.7,11.5489],[57.7125,11.0929],[58.725,10.4573],[59.7375,10.1501],[60.75,10.1843],[61.7625,10.6395],[62.775,11.212],[63.7875,11.6513],[64.8,12.1582],[66.825,13.1231],[67.8375,13.4042],[68.85,13.407],[69.8625,13.3604],[70.875,13.2618],[71.8875,13.196],[72.9,13.2565],[73.9125,13.4685],[74.925,13.9111],[75.9375,14.3851],[76.95,14.6159],[77.9625,14.7194],[78.975,14.756],[79.9875,14.7608]]
       }
       body_ctrl.Rotation.Y {
        type=MovieSceneScriptingFloatChannel
        default=-3.0179
        thinned_from=80
        keys [[0,-3.0179],[1.0125,-3.2124],[2.025,-3.6292],[3.0375,-3.9489],[4.05,-3.8729],[5.0625,-3.4797],[6.075,-3.1131],[7.0875,-2.8325],[8.1,-2.6874],[9.1125,-2.6731],[10.125,-2.6993],[11.1375,-2.7713],[12.15,-2.903],[13.1625,-3.1396],[14.175,-3.4755],[15.1875,-3.9775],[16.2,-4.3607],[17.2125,-4.4933],[18.225,-4.6796],[20.25,-5.2006],[21.2625,-5.4381],[22.275,-5.5973],[23.2875,-5.5514],[24.3,-5.2505],[25.3125,-4.7442],[26.325,-4.0762],[27.3375,-3.3335],[28.35,-2.703],[29.3625,-2.2867],[30.375,-2.0604],[31.3875,-1.979],[32.4,-1.9883],[33.4125,-2.0298],[34.425,-2.0583],[35.4375,-2.0735],[36.45,-2.0963],[37.4625,-2.1435],[38.475,-2.2231],[39.4875,-2.4969],[40.5,-3.0939],[41.5125,-2.9595],[42.525,-2.1691],[43.5375,-1.6295],[44.55,-1.394],[45.5625,-1.4516],[46.575,-1.6516],[47.5875,-1.8677],[48.6,-2.0771],[49.6125,-2.2564],[50.625,-2.3508],[51.6375,-2.4015],[52.65,-2.5],[53.6625,-2.7494],[54.675,-3.3403],[55.6875,-4.6697],[56.7,-6.1206],[57.7125,-6.3037],[58.725,-5.4981],[59.7375,-4.6312],[60.75,-3.9101],[61.7625,-3.2626],[62.775,-2.712],[63.7875,-2.4128],[64.8,-2.2598],[65.8125,-2.0968],[66.825,-1.9275],[67.8375,-1.9015],[68.85,-1.9033],[69.8625,-1.8622],[70.875,-1.8737],[71.8875,-1.955],[72.9,-2.0727],[73.9125,-2.3439],[74.925,-2.6376],[75.9375,-2.8556],[76.95,-3.0516],[77.9625,-3.0907],[78.975,-3.0507],[79.9875,-3.0179]]
       }
       body_ctrl.Rotation.Z {
        type=MovieSceneScriptingFloatChannel
        default=24.0126
        thinned_from=80
        keys [[0,24.0126],[1.0125,23.9117],[3.0375,23.4918],[4.05,23.4511],[5.0625,23.6323],[8.1,24.4954],[10.125,24.6592],[11.1375,24.8598],[12.15,25.3059],[13.1625,26.0486],[14.175,27.1521],[15.1875,28.3821],[16.2,29.3638],[17.2125,29.9059],[18.225,30.1729],[19.2375,30.3379],[20.25,30.4371],[21.2625,30.4667],[22.275,30.4234],[23.2875,30.3056],[24.3,30.1066],[25.3125,29.8012],[26.325,29.3664],[30.375,27.172],[31.3875,26.7099],[32.4,26.3373],[33.4125,26.0845],[34.425,25.9549],[37.4625,25.743],[38.475,25.604],[39.4875,25.2559],[40.5,23.9876],[41.5125,22.7748],[42.525,22.6898],[43.5375,23.0243],[45.5625,23.7943],[46.575,24.0907],[47.5875,24.2518],[48.6,24.3066],[49.6125,24.2838],[52.65,24.0059],[53.6625,24.0113],[54.675,24.2622],[55.6875,25.1771],[57.7125,28.3979],[58.725,29.7244],[59.7375,30.6888],[60.75,31.247],[61.7625,31.4225],[62.775,31.3303],[63.7875,30.9919],[64.8,30.4879],[65.8125,29.9053],[66.825,29.151],[68.85,27.2241],[69.8625,26.4364],[70.875,25.8943],[73.9125,24.699],[74.925,24.3794],[75.9375,24.1495],[76.95,23.97],[77.9625,23.9175],[79.9875,24.0126]]
       }
       body_ctrl.Scale.X {
        type=MovieSceneScriptingFloatChannel
        default=1.0019
        thinned_from=80
        keys [[0,1.0019],[11.1375,1.0019],[14.175,1.0062],[17.2125,0.9977],[19.2375,0.9973],[24.3,1.005],[40.5,1.0017],[44.55,1.0039],[79.9875,1.0019]]
       }
       body_ctrl.Scale.Y {type=MovieSceneScriptingFloatChannel, default=1, const [1,80]}
       body_ctrl.Scale.Z {type=MovieSceneScriptingFloatChannel, default=1, const [1,80]}
       hips_ctrl.Location.X {type=MovieSceneScriptingFloatChannel, default=0, thinned_from=80, keys [[0,0],[79.9875,0]]}
       hips_ctrl.Location.Y {
        type=MovieSceneScriptingFloatChannel
        default=0.0002
        thinned_from=80
        keys [[0,0.0002],[11.1375,0.0002],[14.175,0.0007],[17.2125,-0.0003],[18.225,-0.0004],[24.3,0.0006],[28.35,0.0001],[40.5,0.0002],[43.5375,0.0005],[79.9875,0.0002]]
       }
       hips_ctrl.Location.Z {
        type=MovieSceneScriptingFloatChannel
        default=-0.007
        thinned_from=80
        keys [[0,-0.007],[3.0375,-0.0074],[4.05,-0.007],[5.0625,-0.0063],[6.075,-0.006],[10.125,-0.0064],[11.1375,-0.007],[12.15,-0.0118],[13.1625,-0.0192],[14.175,-0.0223],[15.1875,-0.0159],[16.2,-0.0034],[17.2125,0.0083],[18.225,0.0125],[19.2375,0.0097],[20.25,0.0043],[22.275,-0.0096],[23.2875,-0.0152],[24.3,-0.0183],[25.3125,-0.0174],[26.325,-0.0119],[27.3375,-0.0058],[28.35,-0.0035],[29.3625,-0.0043],[30.375,-0.0055],[32.4,-0.0084],[33.4125,-0.0095],[34.425,-0.0103],[35.4375,-0.0107],[36.45,-0.0107],[37.4625,-0.0097],[38.475,-0.0071],[39.4875,-0.0043],[40.5,-0.0061],[41.5125,-0.0113],[42.525,-0.0147],[43.5375,-0.0157],[44.55,-0.0142],[46.575,-0.009],[47.5875,-0.007],[48.6,-0.0062],[49.6125,-0.0064],[51.6375,-0.0072],[53.6625,-0.0075],[54.675,-0.0079],[55.6875,-0.0091],[56.7,-0.011],[57.7125,-0.0123],[58.725,-0.0126],[59.7375,-0.0123],[60.75,-0.0118],[62.775,-0.01],[65.8125,-0.0086],[68.85,-0.0067],[69.8625,-0.0065],[74.925,-0.0066],[77.9625,-0.007],[79.9875,-0.007]]
       }
       hips_ctrl.Rotation.X {
        type=MovieSceneScriptingFloatChannel
        default=-11.3476
        thinned_from=80
        keys [[0,-11.3476],[2.025,-11.5204],[3.0375,-11.389],[4.05,-10.8176],[5.0625,-9.7728],[6.075,-8.64],[7.0875,-7.601],[8.1,-6.7842],[9.1125,-6.2834],[10.125,-5.9616],[11.1375,-5.7938],[12.15,-5.8446],[13.1625,-6.3542],[14.175,-7.3223],[15.1875,-8.5306],[16.2,-9.6593],[17.2125,-10.5601],[18.225,-11.3331],[19.2375,-11.9816],[20.25,-12.4434],[21.2625,-12.6515],[22.275,-12.554],[23.2875,-12.158],[24.3,-11.5457],[25.3125,-10.8014],[27.3375,-9.2364],[28.35,-8.4822],[29.3625,-7.7501],[30.375,-7.0508],[31.3875,-6.4322],[32.4,-5.9784],[33.4125,-5.7336],[34.425,-5.6566],[35.4375,-5.6969],[36.45,-5.8428],[37.4625,-6.0693],[38.475,-6.3379],[39.4875,-6.4419],[40.5,-5.8961],[41.5125,-5.3639],[42.525,-5.4977],[43.5375,-5.8816],[44.55,-6.3031],[45.5625,-6.6688],[47.5875,-7.3408],[48.6,-7.5989],[49.6125,-7.7547],[50.625,-7.8666],[51.6375,-8.0194],[52.65,-8.1973],[53.6625,-8.3241],[54.675,-8.5776],[55.6875,-9.125],[56.7,-9.7762],[57.7125,-10.142],[59.7375,-10.1668],[60.75,-10.3244],[61.7625,-10.7909],[62.775,-11.286],[63.7875,-11.5083],[64.8,-11.5908],[65.8125,-11.4759],[66.825,-11.1059],[67.8375,-10.4278],[68.85,-9.7242],[69.8625,-9.357],[70.875,-9.2423],[71.8875,-9.2286],[72.9,-9.3692],[73.9125,-9.7424],[74.925,-10.3415],[75.9375,-10.888],[76.95,-11.1719],[77.9625,-11.3003],[78.975,-11.3432],[79.9875,-11.3476]]
       }
       hips_ctrl.Rotation.Y {
        type=MovieSceneScriptingFloatChannel
        default=1.4991
        thinned_from=80
        keys [[0,1.4991],[1.0125,1.7066],[2.025,2.162],[3.0375,2.5422],[4.05,2.5394],[5.0625,2.2245],[6.075,1.9229],[7.0875,1.6856],[8.1,1.565],[9.1125,1.5636],[10.125,1.5971],[11.1375,1.6685],[12.15,1.7577],[13.1625,1.7325],[14.175,1.5811],[15.1875,1.61],[16.2,1.7022],[17.2125,1.7891],[18.225,2.009],[20.25,2.475],[21.2625,2.7273],[22.275,2.9713],[23.2875,3.0754],[24.3,2.9602],[25.3125,2.6556],[26.325,2.1821],[27.3375,1.6037],[28.35,1.0991],[29.3625,0.7857],[30.375,0.6478],[31.3875,0.6372],[32.4,0.6967],[33.4125,0.7692],[34.425,0.821],[36.45,0.8924],[37.4625,0.9452],[38.475,1.0277],[39.4875,1.4101],[40.5,2.697],[41.5125,3.5205],[42.525,3.1104],[43.5375,2.5823],[44.55,2.2774],[45.5625,2.2143],[46.575,2.2433],[48.6,2.2811],[49.6125,2.2708],[50.625,2.1393],[51.6375,1.8822],[52.65,1.5965],[53.6625,1.4027],[54.675,1.4177],[55.6875,1.794],[56.7,2.286],[57.7125,2.2791],[58.725,1.8425],[59.7375,1.3963],[60.75,1.0755],[61.7625,0.7462],[62.775,0.3829],[63.7875,0.1517],[64.8,-0.0218],[65.8125,-0.2272],[66.825,-0.3725],[67.8375,-0.255],[68.85,-0.0383],[69.8625,0.1139],[70.875,0.257],[71.8875,0.4207],[72.9,0.5811],[73.9125,0.8694],[74.925,1.1426],[76.95,1.5058],[77.9625,1.5478],[79.9875,1.4991]]
       }
       hips_ctrl.Rotation.Z {
        type=MovieSceneScriptingFloatChannel
        default=-6.1091
        thinned_from=80
        keys [[0,-6.1091],[1.0125,-6.0217],[2.025,-5.8091],[4.05,-5.266],[5.0625,-5.0129],[6.075,-4.7911],[7.0875,-4.6159],[8.1,-4.5088],[9.1125,-4.472],[10.125,-4.4525],[11.1375,-4.4179],[12.15,-4.3464],[13.1625,-4.1964],[14.175,-3.9613],[15.1875,-3.8101],[16.2,-3.8636],[17.2125,-4.0701],[18.225,-4.3151],[19.2375,-4.5301],[20.25,-4.7032],[21.2625,-4.8264],[22.275,-4.9031],[24.3,-5.0286],[25.3125,-5.0478],[26.325,-4.9898],[27.3375,-4.8652],[28.35,-4.7705],[29.3625,-4.7386],[30.375,-4.737],[33.4125,-4.8081],[36.45,-4.8224],[37.4625,-4.8483],[38.475,-4.8899],[39.4875,-4.9654],[40.5,-5.2283],[41.5125,-5.5383],[42.525,-5.6437],[43.5375,-5.6588],[44.55,-5.7055],[45.5625,-5.8057],[47.5875,-6.0216],[48.6,-6.1059],[49.6125,-6.1383],[50.625,-6.0964],[51.6375,-6.0028],[53.6625,-5.771],[54.675,-5.6362],[55.6875,-5.3529],[56.7,-5.086],[57.7125,-5.0559],[58.725,-5.1211],[59.7375,-5.217],[60.75,-5.3396],[61.7625,-5.5107],[62.775,-5.6939],[63.7875,-5.8469],[64.8,-5.9793],[65.8125,-6.0897],[66.825,-6.152],[67.8375,-6.1904],[68.85,-6.2145],[69.8625,-6.2207],[72.9,-6.3173],[75.9375,-6.3137],[76.95,-6.2945],[77.9625,-6.2357],[78.975,-6.1522],[79.9875,-6.1091]]
       }
       hips_ctrl.Scale.X {
        type=MovieSceneScriptingFloatChannel
        default=0.9981
        thinned_from=80
        keys [[0,0.9981],[11.1375,0.9981],[14.175,0.9939],[17.2125,1.0023],[19.2375,1.0027],[24.3,0.995],[40.5,0.9983],[44.55,0.9961],[79.9875,0.9981]]
       }
       hips_ctrl.Scale.Y {type=MovieSceneScriptingFloatChannel, default=1, const [1,80]}
       hips_ctrl.Scale.Z {type=MovieSceneScriptingFloatChannel, default=1, const [1,80]}
       hips_tan_ctrl.Location.X {
        type=MovieSceneScriptingFloatChannel
        default=0.0093
        thinned_from=80
        keys [[0,0.0093],[1.0125,0.0128],[2.025,0.0206],[3.0375,0.0273],[4.05,0.0261],[5.0625,0.0197],[6.075,0.0148],[7.0875,0.0111],[8.1,0.0086],[9.1125,0.0071],[10.125,0.0063],[11.1375,0.0059],[12.15,0.0063],[13.1625,0.0101],[14.175,0.0173],[15.1875,0.0254],[16.2,0.0317],[17.2125,0.0342],[18.225,0.0338],[19.2375,0.0327],[20.25,0.0308],[21.2625,0.0283],[22.275,0.0253],[26.325,0.012],[27.3375,0.0095],[28.35,0.0078],[29.3625,0.0066],[30.375,0.0058],[31.3875,0.0052],[33.4125,0.0048],[36.45,0.0046],[37.4625,0.0044],[38.475,0.004],[39.4875,0.0004],[40.5,-0.0158],[41.5125,-0.0307],[42.525,-0.0325],[43.5375,-0.0311],[44.55,-0.029],[45.5625,-0.0262],[46.575,-0.0228],[47.5875,-0.0189],[49.6125,-0.0103],[52.65,0.0033],[53.6625,0.0076],[54.675,0.0118],[55.6875,0.0164],[56.7,0.0216],[57.7125,0.026],[58.725,0.0287],[59.7375,0.0296],[60.75,0.029],[62.775,0.026],[63.7875,0.0247],[65.8125,0.0229],[66.825,0.0211],[67.8375,0.0182],[68.85,0.0151],[69.8625,0.0128],[70.875,0.0115],[71.8875,0.0107],[78.975,0.0094],[79.9875,0.0093]]
       }
       hips_tan_ctrl.Location.Y {
        type=MovieSceneScriptingFloatChannel
        default=-0.1536
        thinned_from=80
        keys [[0,-0.1536],[1.0125,-0.1499],[2.025,-0.1413],[3.0375,-0.1321],[4.05,-0.1275],[5.0625,-0.1268],[6.075,-0.1256],[7.0875,-0.124],[11.1375,-0.1158],[12.15,-0.1147],[13.1625,-0.1172],[14.175,-0.1224],[15.1875,-0.1265],[16.2,-0.1281],[17.2125,-0.1265],[18.225,-0.1235],[20.25,-0.1189],[21.2625,-0.1169],[22.275,-0.1159],[23.2875,-0.1166],[24.3,-0.1192],[25.3125,-0.1228],[26.325,-0.127],[27.3375,-0.1322],[29.3625,-0.1447],[30.375,-0.1503],[31.3875,-0.1549],[32.4,-0.1578],[33.4125,-0.1592],[34.425,-0.1593],[35.4375,-0.1585],[36.45,-0.1569],[37.4625,-0.1544],[38.475,-0.1509],[39.4875,-0.1437],[40.5,-0.123],[41.5125,-0.1055],[42.525,-0.1045],[43.5375,-0.1071],[44.55,-0.1104],[45.5625,-0.1142],[46.575,-0.1187],[47.5875,-0.1238],[49.6125,-0.1352],[50.625,-0.1406],[51.6375,-0.1453],[52.65,-0.149],[53.6625,-0.1511],[54.675,-0.1547],[55.6875,-0.162],[56.7,-0.1711],[57.7125,-0.1787],[58.725,-0.1825],[59.7375,-0.1841],[60.75,-0.1846],[61.7625,-0.1843],[62.775,-0.1833],[63.7875,-0.1817],[64.8,-0.1794],[65.8125,-0.1759],[66.825,-0.1702],[67.8375,-0.1626],[68.85,-0.1557],[69.8625,-0.1522],[70.875,-0.1516],[72.9,-0.1515],[78.975,-0.1535],[79.9875,-0.1536]]
       }
       hips_tan_ctrl.Location.Z {
        type=MovieSceneScriptingFloatChannel
        default=0.0323
        thinned_from=80
        keys [[0,0.0323],[1.0125,0.0332],[2.025,0.0349],[3.0375,0.0357],[4.05,0.0334],[5.0625,0.0296],[6.075,0.0278],[7.0875,0.0282],[8.1,0.0297],[9.1125,0.0306],[10.125,0.0312],[11.1375,0.036],[12.15,0.0713],[13.1625,0.1266],[14.175,0.1491],[15.1875,0.1006],[16.2,0.0066],[17.2125,-0.0802],[18.225,-0.112],[19.2375,-0.0918],[20.25,-0.0518],[21.2625,-0.0011],[22.275,0.0507],[23.2875,0.093],[24.3,0.1167],[25.3125,0.1112],[26.325,0.0711],[27.3375,0.0258],[28.35,0.0089],[29.3625,0.0135],[30.375,0.0216],[32.4,0.0415],[33.4125,0.0495],[34.425,0.0551],[35.4375,0.0583],[36.45,0.0588],[37.4625,0.0518],[38.475,0.0333],[39.4875,0.0132],[40.5,0.0284],[41.5125,0.0687],[42.525,0.0947],[43.5375,0.1016],[44.55,0.0899],[45.5625,0.0705],[46.575,0.0507],[47.5875,0.0355],[48.6,0.0295],[49.6125,0.0305],[50.625,0.0335],[51.6375,0.0355],[52.65,0.0365],[53.6625,0.0372],[54.675,0.0401],[55.6875,0.048],[56.7,0.0606],[57.7125,0.0696],[58.725,0.0715],[59.7375,0.0698],[60.75,0.0663],[62.775,0.0542],[63.7875,0.0504],[64.8,0.0473],[65.8125,0.0439],[67.8375,0.0347],[68.85,0.0313],[69.8625,0.0299],[71.8875,0.0303],[73.9125,0.0299],[74.925,0.03],[75.9375,0.0309],[76.95,0.0321],[77.9625,0.0325],[79.9875,0.0323]]
       }
       hips_tan_ctrl.Rotation.X {type=MovieSceneScriptingFloatChannel, default=0, const [0,80]}
       hips_tan_ctrl.Rotation.Y {type=MovieSceneScriptingFloatChannel, default=0, const [0,80]}
       hips_tan_ctrl.Rotation.Z {type=MovieSceneScriptingFloatChannel, default=0, const [0,80]}
       hips_tan_ctrl.Scale.X {type=MovieSceneScriptingFloatChannel, default=1, const [1,80]}
       hips_tan_ctrl.Scale.Y {type=MovieSceneScriptingFloatChannel, default=1, const [1,80]}
       hips_tan_ctrl.Scale.Z {type=MovieSceneScriptingFloatChannel, default=1, const [1,80]}
       spine_01_ctrl.Location.X {
        type=MovieSceneScriptingFloatChannel
        default=-0.0129
        thinned_from=80
        keys [[0,-0.0129],[1.0125,-0.0127],[5.0625,-0.0106],[7.0875,-0.0099],[12.15,-0.0095],[14.175,-0.0088],[15.1875,-0.0086],[16.2,-0.0089],[19.2375,-0.0102],[25.3125,-0.0108],[29.3625,-0.0103],[39.4875,-0.0108],[41.5125,-0.0115],[44.55,-0.0117],[49.6125,-0.0129],[54.675,-0.0122],[56.7,-0.0113],[65.8125,-0.0133],[76.95,-0.0133],[79.9875,-0.0129]]
       }
       spine_01_ctrl.Location.Y {
        type=MovieSceneScriptingFloatChannel
        default=-0.0012
        thinned_from=80
        keys [[0,-0.0012],[10.125,-0.001],[11.1375,-0.0011],[12.15,-0.0018],[13.1625,-0.0029],[14.175,-0.0033],[15.1875,-0.0024],[17.2125,0.0011],[18.225,0.0017],[19.2375,0.0013],[20.25,0.0005],[22.275,-0.0015],[23.2875,-0.0023],[24.3,-0.0028],[25.3125,-0.0027],[27.3375,-0.001],[28.35,-0.0006],[35.4375,-0.0017],[37.4625,-0.0016],[39.4875,-0.0008],[40.5,-0.0011],[41.5125,-0.0018],[42.525,-0.0023],[43.5375,-0.0024],[47.5875,-0.0012],[54.675,-0.0014],[57.7125,-0.002],[79.9875,-0.0012]]
       }
       spine_01_ctrl.Location.Z {
        type=MovieSceneScriptingFloatChannel
        default=0.0173
        thinned_from=80
        keys [[0,0.0173],[1.0125,0.0177],[2.025,0.0185],[3.0375,0.0189],[4.05,0.0176],[5.0625,0.0153],[6.075,0.0139],[7.0875,0.0135],[10.125,0.0142],[11.1375,0.0161],[12.15,0.0305],[13.1625,0.053],[14.175,0.0625],[15.1875,0.0432],[16.2,0.0055],[17.2125,-0.0293],[18.225,-0.0416],[19.2375,-0.0329],[20.25,-0.0163],[22.275,0.0257],[23.2875,0.0426],[24.3,0.0516],[25.3125,0.0487],[26.325,0.0317],[27.3375,0.0128],[28.35,0.0056],[29.3625,0.0073],[30.375,0.0104],[32.4,0.0185],[33.4125,0.0217],[34.425,0.024],[35.4375,0.0253],[36.45,0.0255],[37.4625,0.0226],[38.475,0.0151],[39.4875,0.0069],[40.5,0.013],[41.5125,0.0293],[42.525,0.0401],[43.5375,0.0432],[44.55,0.0386],[45.5625,0.0308],[46.575,0.0228],[47.5875,0.0166],[48.6,0.0141],[49.6125,0.0144],[50.625,0.0155],[51.6375,0.0164],[53.6625,0.0171],[54.675,0.0183],[55.6875,0.0217],[56.7,0.0272],[57.7125,0.031],[58.725,0.0315],[59.7375,0.0306],[60.75,0.0292],[62.775,0.0246],[63.7875,0.0232],[64.8,0.0221],[65.8125,0.0208],[67.8375,0.017],[68.85,0.0154],[69.8625,0.0147],[73.9125,0.0149],[74.925,0.0154],[76.95,0.0169],[77.9625,0.0173],[79.9875,0.0173]]
       }
       spine_01_ctrl.Rotation.X {
        type=MovieSceneScriptingFloatChannel
        default=2.5335
        thinned_from=80
        keys [[0,2.5335],[1.0125,2.4586],[2.025,2.3101],[3.0375,2.2031],[4.05,2.2619],[5.0625,2.5036],[6.075,2.7717],[7.0875,3.0087],[8.1,3.178],[9.1125,3.2533],[10.125,3.284],[11.1375,3.2814],[12.15,3.2262],[13.1625,3.1607],[14.175,3.1302],[16.2,3.1086],[17.2125,3.0094],[18.225,2.7915],[19.2375,2.5363],[20.25,2.295],[21.2625,2.113],[22.275,2.0316],[23.2875,2.0696],[24.3,2.2095],[25.3125,2.4307],[26.325,2.7112],[28.35,3.3369],[29.3625,3.6446],[30.375,3.9435],[31.3875,4.2084],[32.4,4.4016],[33.4125,4.5031],[34.425,4.5302],[35.4375,4.5053],[36.45,4.435],[37.4625,4.3322],[38.475,4.214],[39.4875,4.0328],[41.5125,3.0206],[42.525,2.8288],[43.5375,2.7256],[44.55,2.6839],[45.5625,2.7213],[46.575,2.8048],[47.5875,2.9092],[48.6,3.0441],[49.6125,3.2061],[50.625,3.358],[51.6375,3.4592],[52.65,3.511],[53.6625,3.5363],[54.675,3.5925],[55.6875,3.6947],[56.7,3.8359],[57.7125,4.0149],[58.725,4.1531],[59.7375,4.1984],[60.75,4.1472],[61.7625,3.9472],[62.775,3.7041],[63.7875,3.5465],[64.8,3.4193],[66.825,3.2053],[67.8375,3.1554],[68.85,3.1462],[69.8625,3.1466],[70.875,3.1734],[71.8875,3.1838],[72.9,3.1455],[73.9125,3.0402],[74.925,2.8562],[75.9375,2.6825],[76.95,2.5984],[77.9625,2.5561],[78.975,2.5376],[79.9875,2.5335]]
       }
       spine_01_ctrl.Rotation.Y {
        type=MovieSceneScriptingFloatChannel
        default=1.0266
        thinned_from=80
        keys [[0,1.0266],[1.0125,1.149],[2.025,1.4146],[3.0375,1.6239],[4.05,1.5745],[5.0625,1.3249],[6.075,1.0974],[7.0875,0.919],[8.1,0.8124],[9.1125,0.779],[10.125,0.7737],[11.1375,0.7925],[12.15,0.8467],[13.1625,0.9953],[14.175,1.2464],[15.1875,1.6027],[16.2,1.9137],[17.2125,2.0732],[18.225,2.1709],[19.2375,2.2346],[20.25,2.2631],[21.2625,2.2627],[22.275,2.227],[23.2875,2.1187],[24.3,1.9214],[25.3125,1.6541],[26.325,1.335],[27.3375,1.0009],[28.35,0.7203],[29.3625,0.5273],[30.375,0.4105],[31.3875,0.3532],[32.4,0.3378],[33.4125,0.3448],[35.4375,0.3696],[36.45,0.3858],[37.4625,0.4069],[38.475,0.4324],[39.4875,0.4386],[40.5,0.2665],[41.5125,-0.0295],[42.525,-0.2456],[43.5375,-0.3709],[44.55,-0.3884],[45.5625,-0.2978],[46.575,-0.1553],[47.5875,-0.0002],[48.6,0.1639],[49.6125,0.3198],[50.625,0.4361],[51.6375,0.5169],[52.65,0.5939],[53.6625,0.7012],[54.675,0.8805],[55.6875,1.2199],[56.7,1.6226],[57.7125,1.7918],[58.725,1.7221],[60.75,1.4481],[61.7625,1.3029],[62.775,1.1464],[63.7875,1.0332],[64.8,0.95],[65.8125,0.8482],[66.825,0.7219],[67.8375,0.6274],[68.85,0.554],[69.8625,0.5027],[70.875,0.4964],[71.8875,0.5227],[72.9,0.5779],[73.9125,0.7052],[74.925,0.8446],[75.9375,0.9482],[76.95,1.0304],[77.9625,1.0498],[79.9875,1.0266]]
       }
       spine_01_ctrl.Rotation.Z {
        type=MovieSceneScriptingFloatChannel
        default=6.1266
        thinned_from=80
        keys [[0,6.1266],[1.0125,6.0262],[2.025,5.7866],[3.0375,5.5007],[4.05,5.2592],[5.0625,5.0668],[6.075,4.8951],[7.0875,4.7528],[8.1,4.6666],[9.1125,4.6435],[10.125,4.6342],[11.1375,4.6083],[12.15,4.5459],[13.1625,4.4241],[14.175,4.2486],[15.1875,4.1834],[16.2,4.3055],[17.2125,4.5162],[18.225,4.7084],[19.2375,4.8495],[20.25,4.9418],[21.2625,4.9922],[22.275,5.0187],[23.2875,5.0649],[24.3,5.1342],[25.3125,5.1749],[26.325,5.1364],[27.3375,5.0199],[28.35,4.9232],[29.3625,4.8889],[30.375,4.8919],[31.3875,4.9248],[33.4125,5.0121],[34.425,5.0286],[36.45,5.0337],[37.4625,5.0534],[38.475,5.0875],[39.4875,5.1389],[40.5,5.2996],[41.5125,5.4898],[42.525,5.5275],[43.5375,5.4995],[44.55,5.5308],[45.5625,5.6427],[47.5875,5.9166],[48.6,6.0433],[49.6125,6.1269],[50.625,6.1338],[51.6375,6.0791],[54.675,5.8354],[55.6875,5.6319],[56.7,5.4661],[57.7125,5.5227],[58.725,5.6291],[60.75,5.8076],[61.7625,5.911],[64.8,6.2465],[65.8125,6.3333],[66.825,6.361],[69.8625,6.3325],[70.875,6.3527],[72.9,6.419],[73.9125,6.4179],[76.95,6.3266],[77.9625,6.2586],[78.975,6.1708],[79.9875,6.1266]]
       }
       spine_01_ctrl.Scale.X {
        type=MovieSceneScriptingFloatChannel
        default=0.9981
        thinned_from=80
        keys [[0,0.9981],[11.1375,0.9981],[14.175,0.9939],[17.2125,1.0023],[19.2375,1.0027],[24.3,0.995],[40.5,0.9983],[44.55,0.9961],[79.9875,0.9981]]
       }
       spine_01_ctrl.Scale.Y {type=MovieSceneScriptingFloatChannel, default=1, const [1,80]}
       spine_01_ctrl.Scale.Z {type=MovieSceneScriptingFloatChannel, default=1, const [1,80]}
       spine_02_ctrl.Location.X {
        type=MovieSceneScriptingFloatChannel
        default=-0.0087
        thinned_from=80
        keys [[0,-0.0087],[1.0125,-0.0085],[4.05,-0.0074],[8.1,-0.0067],[11.1375,-0.0066],[12.15,-0.0064],[14.175,-0.0055],[15.1875,-0.0054],[16.2,-0.0061],[17.2125,-0.007],[18.225,-0.0076],[19.2375,-0.0077],[23.2875,-0.0066],[25.3125,-0.0068],[27.3375,-0.0071],[39.4875,-0.0074],[41.5125,-0.0081],[49.6125,-0.0089],[54.675,-0.0083],[56.7,-0.0074],[57.7125,-0.0073],[65.8125,-0.009],[77.9625,-0.0089],[79.9875,-0.0087]]
       }
       spine_02_ctrl.Location.Y {
        type=MovieSceneScriptingFloatChannel
        default=-0.0004
        thinned_from=80
        keys [[0,-0.0004],[15.1875,-0.0003],[17.2125,-0.0001],[20.25,-0.0003],[24.3,0],[36.45,-0.0008],[56.7,-0.0007],[60.75,-0.0008],[79.9875,-0.0004]]
       }
       spine_02_ctrl.Location.Z {
        type=MovieSceneScriptingFloatChannel
        default=-0.0012
        thinned_from=80
        keys [[0,-0.0012],[3.0375,-0.0011],[4.05,-0.0014],[7.0875,-0.0033],[8.1,-0.0038],[11.1375,-0.0043],[16.2,-0.0035],[21.2625,-0.0015],[22.275,-0.0014],[24.3,-0.0018],[32.4,-0.0057],[34.425,-0.006],[39.4875,-0.0052],[42.525,-0.0036],[44.55,-0.003],[48.6,-0.0032],[51.6375,-0.0037],[56.7,-0.004],[58.725,-0.0045],[60.75,-0.0043],[62.775,-0.0032],[65.8125,-0.0023],[66.825,-0.0022],[71.8875,-0.0028],[73.9125,-0.0024],[75.9375,-0.0016],[79.9875,-0.0012]]
       }
       spine_02_ctrl.Rotation.X {
        type=MovieSceneScriptingFloatChannel
        default=0.8477
        keys [[0,0.8477],[1.0125,0.8082],[2.025,0.7571],[3.0375,0.7836],[4.05,0.979],[5.0625,1.3918],[6.075,1.8589],[7.0875,2.2867],[8.1,2.6135],[9.1125,2.7897],[10.125,2.8915],[11.1375,2.9398],[12.15,2.9098],[13.1625,2.8033],[14.175,2.6676],[15.1875,2.5376],[16.2,2.4164],[17.2125,2.2166],[18.225,1.8955],[19.2375,1.5428],[20.25,1.2309],[21.2625,1.0228],[22.275,0.9647],[23.2875,1.0521],[24.3,1.2414],[25.3125,1.5039],[26.325,1.8125],[27.3375,2.1455],[28.35,2.4961],[29.3625,2.863],[30.375,3.2419],[31.3875,3.5955],[32.4,3.865],[33.4125,4.0154],[34.425,4.0654],[35.4375,4.0448],[36.45,3.9628],[37.4625,3.8405],[38.475,3.7053],[39.4875,3.5515],[40.5,3.1953],[41.5125,2.7569],[42.525,2.4504],[43.5375,2.1983],[44.55,2.0337],[45.5625,1.9959],[46.575,2.0267],[47.5875,2.0841],[48.6,2.1931],[49.6125,2.3518],[50.625,2.4954],[51.6375,2.5611],[52.65,2.5653],[53.6625,2.5635],[54.675,2.5974],[55.6875,2.6615],[56.7,2.7612],[57.7125,2.9165],[58.725,3.0463],[59.7375,3.0577],[60.75,2.935],[61.7625,2.5762],[62.775,2.1581],[63.7875,1.9061],[64.8,1.7232],[65.8125,1.5964],[66.825,1.5194],[67.8375,1.5898],[68.85,1.7225],[69.8625,1.7989],[70.875,1.8619],[71.8875,1.8916],[72.9,1.8377],[73.9125,1.6777],[74.925,1.3821],[75.9375,1.0979],[76.95,0.9663],[77.9625,0.8947],[78.975,0.8581],[79.9875,0.8477]]
       }
       spine_02_ctrl.Rotation.Y {
        type=MovieSceneScriptingFloatChannel
        default=1.5974
        thinned_from=80
        keys [[0,1.5974],[1.0125,1.7267],[2.025,1.9937],[3.0375,2.1823],[4.05,2.0971],[6.075,1.452],[7.0875,1.1907],[8.1,1.0351],[9.1125,0.9941],[10.125,0.9925],[11.1375,1.0265],[12.15,1.1077],[13.1625,1.2883],[14.175,1.5786],[15.1875,2.0379],[16.2,2.4645],[17.2125,2.7176],[18.225,2.9221],[19.2375,3.0758],[20.25,3.1767],[21.2625,3.2299],[22.275,3.2212],[23.2875,3.0871],[24.3,2.8032],[25.3125,2.4039],[26.325,1.9171],[27.3375,1.3896],[28.35,0.9296],[29.3625,0.6034],[30.375,0.3984],[31.3875,0.2885],[32.4,0.2489],[33.4125,0.251],[34.425,0.2673],[35.4375,0.2899],[36.45,0.3231],[37.4625,0.3694],[38.475,0.4275],[39.4875,0.519],[40.5,0.5682],[41.5125,0.385],[42.525,0.0762],[43.5375,-0.1398],[44.55,-0.1957],[45.5625,-0.0948],[46.575,0.0779],[47.5875,0.2623],[48.6,0.4506],[49.6125,0.6157],[50.625,0.7129],[51.6375,0.7558],[52.65,0.7971],[53.6625,0.8916],[54.675,1.1142],[55.6875,1.6087],[56.7,2.1959],[57.7125,2.3852],[58.725,2.1984],[59.7375,1.9437],[60.75,1.7353],[61.7625,1.5522],[62.775,1.3591],[63.7875,1.2168],[64.8,1.1092],[65.8125,0.9594],[66.825,0.7763],[67.8375,0.6527],[68.85,0.5653],[69.8625,0.5118],[70.875,0.5212],[71.8875,0.5808],[72.9,0.6902],[73.9125,0.933],[74.925,1.2121],[75.9375,1.4261],[76.95,1.5858],[77.9625,1.6295],[79.9875,1.5974]]
       }
       spine_02_ctrl.Rotation.Z {
        type=MovieSceneScriptingFloatChannel
        default=6.0759
        thinned_from=80
        keys [[0,6.0759],[1.0125,5.9717],[2.025,5.728],[3.0375,5.4456],[4.05,5.2147],[5.0625,5.0389],[6.075,4.8836],[7.0875,4.7542],[8.1,4.6771],[9.1125,4.6594],[10.125,4.6541],[11.1375,4.6314],[12.15,4.5714],[13.1625,4.4478],[14.175,4.2658],[15.1875,4.1935],[16.2,4.3084],[17.2125,4.512],[18.225,4.6967],[19.2375,4.8296],[20.25,4.9146],[21.2625,4.9611],[22.275,4.9886],[23.2875,5.0396],[24.3,5.1151],[25.3125,5.1622],[26.325,5.1294],[27.3375,5.0176],[28.35,4.9255],[29.3625,4.8953],[30.375,4.9019],[31.3875,4.9381],[33.4125,5.0303],[34.425,5.048],[36.45,5.0539],[37.4625,5.0737],[38.475,5.1083],[39.4875,5.1654],[40.5,5.347],[41.5125,5.547],[42.525,5.576],[43.5375,5.5397],[44.55,5.5664],[45.5625,5.6756],[47.5875,5.9434],[48.6,6.0669],[49.6125,6.1478],[50.625,6.1516],[51.6375,6.0922],[54.675,5.8288],[55.6875,5.6107],[56.7,5.4276],[57.7125,5.4779],[58.725,5.5873],[60.75,5.7697],[61.7625,5.8713],[64.8,6.2173],[65.8125,6.3115],[66.825,6.348],[68.85,6.3466],[69.8625,6.3338],[70.875,6.3548],[72.9,6.4174],[73.9125,6.4075],[76.95,6.2806],[77.9625,6.2085],[78.975,6.1198],[79.9875,6.0759]]
       }
       spine_02_ctrl.Scale.X {type=MovieSceneScriptingFloatChannel, default=1, const [1,80]}
       spine_02_ctrl.Scale.Y {type=MovieSceneScriptingFloatChannel, default=1, const [1,80]}
       spine_02_ctrl.Scale.Z {type=MovieSceneScriptingFloatChannel, default=1, const [1,80]}
       spine_03_ctrl.Location.X {
        type=MovieSceneScriptingFloatChannel
        default=0.0049
        thinned_from=80
        keys [[0,0.0049],[8.1,0.0037],[15.1875,0.0036],[21.2625,0.0046],[25.3125,0.0045],[29.3625,0.0038],[39.4875,0.004],[41.5125,0.0043],[44.55,0.0042],[50.625,0.0048],[79.9875,0.0049]]
       }
       spine_03_ctrl.Location.Y {
        type=MovieSceneScriptingFloatChannel
        default=0.0005
        thinned_from=80
        keys [[0,0.0005],[16.2,-0.0001],[60.75,0],[66.825,0.0004],[79.9875,0.0005]]
       }
       spine_03_ctrl.Location.Z {
        type=MovieSceneScriptingFloatChannel
        default=-0.0011
        thinned_from=80
        keys [[0,-0.0011],[3.0375,-0.0008],[4.05,-0.0005],[8.1,0.0017],[11.1375,0.0022],[13.1625,0.0021],[17.2125,0.0011],[20.25,0],[22.275,-0.0002],[28.35,0.0013],[32.4,0.0028],[34.425,0.0031],[40.5,0.0025],[44.55,0.0011],[48.6,0.001],[51.6375,0.0013],[59.7375,0.0015],[60.75,0.0013],[62.775,0.0002],[65.8125,-0.0005],[66.825,-0.0006],[71.8875,0.0002],[73.9125,0],[75.9375,-0.0008],[79.9875,-0.0011]]
       }
       spine_03_ctrl.Rotation.X {
        type=MovieSceneScriptingFloatChannel
        default=-1.4724
        keys [[0,-1.4724],[1.0125,-1.4435],[2.025,-1.3245],[3.0375,-1.0829],[4.05,-0.6659],[5.0625,0.0446],[6.075,0.8651],[7.0875,1.637],[8.1,2.2508],[9.1125,2.6118],[10.125,2.8427],[11.1375,2.9839],[12.15,3.0062],[13.1625,2.8426],[14.175,2.5359],[15.1875,2.2007],[16.2,1.8803],[17.2125,1.4815],[18.225,0.9617],[19.2375,0.4263],[20.25,-0.0241],[21.2625,-0.2926],[22.275,-0.3169],[23.2875,-0.1388],[24.3,0.1511],[25.3125,0.5066],[26.325,0.8839],[27.3375,1.2705],[28.35,1.7075],[29.3625,2.2141],[30.375,2.7752],[31.3875,3.3265],[32.4,3.7637],[33.4125,4.0206],[34.425,4.1193],[35.4375,4.1054],[36.45,3.9941],[37.4625,3.8194],[38.475,3.6276],[39.4875,3.492],[40.5,3.3459],[41.5125,2.9696],[42.525,2.4107],[43.5375,1.8605],[44.55,1.4657],[45.5625,1.3067],[46.575,1.2685],[47.5875,1.2707],[48.6,1.367],[49.6125,1.5549],[50.625,1.7097],[51.6375,1.7236],[52.65,1.6449],[53.6625,1.5947],[54.675,1.6104],[55.6875,1.667],[56.7,1.7702],[57.7125,1.9345],[58.725,2.0568],[59.7375,1.9944],[60.75,1.7216],[61.7625,1.0365],[62.775,0.2539],[63.7875,-0.2037],[64.8,-0.5182],[65.8125,-0.7172],[66.825,-0.7843],[67.8375,-0.5276],[68.85,-0.1575],[69.8625,0.0496],[70.875,0.1856],[71.8875,0.26],[72.9,0.1852],[73.9125,-0.0614],[74.925,-0.5527],[75.9375,-1.036],[76.95,-1.248],[77.9625,-1.3739],[78.975,-1.4479],[79.9875,-1.4724]]
       }
       spine_03_ctrl.Rotation.Y {
        type=MovieSceneScriptingFloatChannel
        default=2.8233
        keys [[0,2.8233],[1.0125,2.9616],[2.025,3.2271],[3.0375,3.3677],[4.05,3.1801],[5.0625,2.6567],[6.075,2.1042],[7.0875,1.6393],[8.1,1.3544],[9.1125,1.2756],[10.125,1.2658],[11.1375,1.3212],[12.15,1.4628],[13.1625,1.7577],[14.175,2.2138],[15.1875,2.9481],[16.2,3.6202],[17.2125,4.0342],[18.225,4.4314],[19.2375,4.7716],[20.25,5.0309],[21.2625,5.204],[22.275,5.2548],[23.2875,5.0642],[24.3,4.589],[25.3125,3.8939],[26.325,3.0346],[27.3375,2.0985],[28.35,1.2726],[29.3625,0.6744],[30.375,0.2836],[31.3875,0.0567],[32.4,-0.0426],[33.4125,-0.0587],[34.425,-0.0388],[35.4375,0.0013],[36.45,0.0676],[37.4625,0.1624],[38.475,0.2818],[39.4875,0.5058],[40.5,0.8319],[41.5125,0.7285],[42.525,0.233],[43.5375,-0.1291],[44.55,-0.2223],[45.5625,-0.0635],[46.575,0.2076],[47.5875,0.4904],[48.6,0.765],[49.6125,0.9849],[50.625,1.0833],[51.6375,1.0951],[52.65,1.1106],[53.6625,1.2183],[54.675,1.5668],[55.6875,2.4254],[56.7,3.4451],[57.7125,3.7221],[58.725,3.323],[59.7375,2.8437],[60.75,2.4966],[61.7625,2.2521],[62.775,2.0007],[63.7875,1.8035],[64.8,1.6463],[65.8125,1.3914],[66.825,1.0669],[67.8375,0.8324],[68.85,0.6607],[69.8625,0.5686],[70.875,0.5935],[71.8875,0.7108],[72.9,0.9293],[73.9125,1.4134],[74.925,1.9931],[75.9375,2.4487],[76.95,2.7725],[77.9625,2.8698],[78.975,2.8522],[79.9875,2.8233]]
       }
       spine_03_ctrl.Rotation.Z {
        type=MovieSceneScriptingFloatChannel
        default=5.9381
        thinned_from=80
        keys [[0,5.9381],[1.0125,5.8387],[2.025,5.6115],[3.0375,5.3554],[4.05,5.1497],[5.0625,5.0003],[6.075,4.8704],[7.0875,4.7599],[8.1,4.6949],[9.1125,4.6842],[10.125,4.6841],[11.1375,4.6664],[12.15,4.6132],[13.1625,4.4946],[14.175,4.3134],[15.1875,4.2425],[16.2,4.3558],[17.2125,4.5529],[18.225,4.7287],[19.2375,4.85],[20.25,4.9232],[21.2625,4.9626],[22.275,4.9907],[23.2875,5.0446],[24.3,5.1202],[25.3125,5.1656],[26.325,5.1299],[27.3375,5.0165],[28.35,4.9263],[29.3625,4.8999],[30.375,4.9101],[31.3875,4.9488],[33.4125,5.0438],[34.425,5.0624],[36.45,5.0698],[37.4625,5.0904],[38.475,5.1259],[39.4875,5.1884],[41.5125,5.6006],[42.525,5.6141],[43.5375,5.5639],[44.55,5.5817],[45.5625,5.6847],[47.5875,5.9403],[48.6,6.0588],[49.6125,6.1373],[50.625,6.1399],[51.6375,6.0778],[54.675,5.8015],[55.6875,5.5707],[56.7,5.3723],[57.7125,5.4189],[58.725,5.5339],[59.7375,5.6314],[61.7625,5.8069],[64.8,6.1394],[65.8125,6.2392],[66.825,6.2867],[68.85,6.3067],[69.8625,6.2991],[72.9,6.3739],[73.9125,6.347],[76.95,6.1488],[78.975,5.9811],[79.9875,5.9381]]
       }
       spine_03_ctrl.Scale.X {type=MovieSceneScriptingFloatChannel, default=1, const [1,80]}
       spine_03_ctrl.Scale.Y {type=MovieSceneScriptingFloatChannel, default=1, const [1,80]}
       spine_03_ctrl.Scale.Z {type=MovieSceneScriptingFloatChannel, default=1, const [1,80]}
       spine_04_ctrl.Location.X {
        type=MovieSceneScriptingFloatChannel
        default=-0.0012
        thinned_from=80
        keys [[0,-0.0012],[3.0375,-0.0014],[7.0875,-0.0005],[11.1375,-0.0004],[12.15,-0.0009],[14.175,-0.003],[15.1875,-0.0028],[16.2,-0.0006],[17.2125,0.0023],[18.225,0.0037],[19.2375,0.0032],[20.25,0.0017],[21.2625,-0.0007],[22.275,-0.0032],[23.2875,-0.0051],[24.3,-0.0056],[25.3125,-0.0046],[26.325,-0.0024],[27.3375,-0.0007],[28.35,-0.0002],[35.4375,0.0003],[39.4875,0],[41.5125,-0.0003],[43.5375,0.0005],[44.55,0.0006],[50.625,-0.0003],[53.6625,-0.0004],[54.675,-0.0006],[55.6875,-0.0013],[56.7,-0.0025],[57.7125,-0.003],[62.775,-0.0011],[68.85,0],[72.9,-0.0002],[76.95,-0.0012],[79.9875,-0.0012]]
       }
       spine_04_ctrl.Location.Y {
        type=MovieSceneScriptingFloatChannel
        default=-0.0052
        thinned_from=80
        keys [[0,-0.0052],[2.025,-0.0052],[3.0375,-0.005],[4.05,-0.0045],[6.075,-0.0033],[8.1,-0.0027],[10.125,-0.0024],[11.1375,-0.0025],[12.15,-0.0041],[13.1625,-0.0067],[14.175,-0.0084],[15.1875,-0.0065],[16.2,-0.0014],[17.2125,0.004],[18.225,0.0064],[19.2375,0.0054],[20.25,0.0026],[22.275,-0.0056],[23.2875,-0.0087],[24.3,-0.0102],[25.3125,-0.0094],[27.3375,-0.003],[28.35,-0.0018],[36.45,-0.0032],[37.4625,-0.003],[39.4875,-0.0015],[40.5,-0.0021],[42.525,-0.0056],[43.5375,-0.0067],[44.55,-0.0065],[45.5625,-0.0056],[46.575,-0.0044],[47.5875,-0.0035],[48.6,-0.0032],[54.675,-0.004],[55.6875,-0.0045],[56.7,-0.0053],[57.7125,-0.0058],[60.75,-0.0059],[63.7875,-0.0063],[65.8125,-0.006],[66.825,-0.0056],[68.85,-0.0044],[71.8875,-0.0041],[73.9125,-0.0042],[76.95,-0.005],[79.9875,-0.0052]]
       }
       spine_04_ctrl.Location.Z {
        type=MovieSceneScriptingFloatChannel
        default=0.0487
        thinned_from=80
        keys [[0,0.0487],[1.0125,0.0493],[2.025,0.0506],[3.0375,0.051],[4.05,0.0485],[5.0625,0.0446],[6.075,0.0425],[7.0875,0.0424],[8.1,0.0434],[9.1125,0.0439],[10.125,0.0442],[11.1375,0.0483],[12.15,0.0799],[13.1625,0.1296],[14.175,0.1504],[15.1875,0.1075],[16.2,0.0237],[17.2125,-0.0539],[18.225,-0.0821],[19.2375,-0.0639],[20.25,-0.028],[21.2625,0.0174],[22.275,0.0636],[23.2875,0.1012],[24.3,0.1222],[25.3125,0.1172],[26.325,0.0813],[27.3375,0.0409],[28.35,0.026],[29.3625,0.0304],[30.375,0.038],[32.4,0.0563],[33.4125,0.0635],[34.425,0.0686],[35.4375,0.0714],[36.45,0.0717],[37.4625,0.0652],[38.475,0.0484],[39.4875,0.0298],[40.5,0.0418],[41.5125,0.0767],[42.525,0.1003],[43.5375,0.1072],[44.55,0.0972],[45.5625,0.0801],[46.575,0.0626],[47.5875,0.0492],[48.6,0.0441],[49.6125,0.0453],[50.625,0.0482],[51.6375,0.0503],[52.65,0.0514],[53.6625,0.0522],[54.675,0.0549],[55.6875,0.0624],[56.7,0.0741],[57.7125,0.0827],[58.725,0.0845],[59.7375,0.0831],[60.75,0.08],[62.775,0.0691],[63.7875,0.0657],[64.8,0.063],[65.8125,0.0598],[67.8375,0.051],[68.85,0.0476],[69.8625,0.0462],[71.8875,0.0465],[73.9125,0.0461],[74.925,0.0463],[75.9375,0.0472],[76.95,0.0483],[77.9625,0.0488],[79.9875,0.0487]]
       }
       spine_04_ctrl.Rotation.X {
        type=MovieSceneScriptingFloatChannel
        default=-2.2977
        thinned_from=80
        keys [[0,-2.2977],[1.0125,-2.1024],[2.025,-1.6502],[3.0375,-1.1154],[4.05,-0.6953],[5.0625,-0.3503],[6.075,-0.003],[7.0875,0.2976],[8.1,0.5218],[9.1125,0.6448],[10.125,0.7363],[11.1375,0.8471],[12.15,1.0849],[13.1625,1.3768],[14.175,1.573],[15.1875,1.6227],[16.2,1.5582],[17.2125,1.47],[18.225,1.4424],[19.2375,1.4571],[20.25,1.4847],[21.2625,1.5338],[22.275,1.5926],[23.2875,1.5788],[24.3,1.4462],[25.3125,1.2097],[26.325,0.8818],[27.3375,0.5205],[28.35,0.2154],[29.3625,-0.0074],[30.375,-0.1547],[31.3875,-0.2231],[32.4,-0.2289],[33.4125,-0.1934],[34.425,-0.1382],[35.4375,-0.0741],[36.45,-0.0049],[37.4625,0.0659],[38.475,0.1426],[39.4875,0.3025],[40.5,0.7194],[41.5125,0.9232],[42.525,0.6811],[43.5375,0.3546],[44.55,0.0527],[45.5625,-0.1664],[46.575,-0.343],[47.5875,-0.5045],[48.6,-0.617],[50.625,-0.7342],[51.6375,-0.8333],[52.65,-0.9445],[53.6625,-1.0056],[54.675,-1.0207],[55.6875,-1.0094],[56.7,-0.9878],[57.7125,-0.9738],[58.725,-0.98],[59.7375,-1.0508],[60.75,-1.1961],[61.7625,-1.5116],[62.775,-1.864],[63.7875,-2.0665],[64.8,-2.2001],[65.8125,-2.2771],[66.825,-2.274],[67.8375,-2.1058],[68.85,-1.8897],[69.8625,-1.769],[70.875,-1.7023],[71.8875,-1.664],[72.9,-1.6873],[73.9125,-1.7712],[75.9375,-2.1414],[76.95,-2.2159],[77.9625,-2.2616],[78.975,-2.2887],[79.9875,-2.2977]]
       }
       spine_04_ctrl.Rotation.Y {
        type=MovieSceneScriptingFloatChannel
        default=0.5872
        thinned_from=80
        keys [[0,0.5872],[1.0125,0.3715],[2.025,-0.0895],[3.0375,-0.4774],[4.05,-0.5231],[5.0625,-0.4116],[6.075,-0.3737],[7.0875,-0.3549],[8.1,-0.318],[9.1125,-0.25],[10.125,-0.204],[11.1375,-0.1768],[12.15,-0.1701],[13.1625,-0.1331],[14.175,-0.0315],[15.1875,0.1476],[16.2,0.2176],[17.2125,0.1541],[18.225,0.1361],[19.2375,0.1726],[20.25,0.2397],[21.2625,0.3317],[22.275,0.4246],[23.2875,0.4557],[24.3,0.3883],[25.3125,0.2452],[26.325,0.0606],[27.3375,-0.1187],[28.35,-0.2527],[29.3625,-0.3223],[30.375,-0.3361],[31.3875,-0.3194],[32.4,-0.2933],[33.4125,-0.2765],[34.425,-0.2758],[35.4375,-0.2872],[36.45,-0.306],[38.475,-0.3479],[39.4875,-0.3273],[40.5,-0.2853],[41.5125,-0.4035],[42.525,-0.6264],[43.5375,-0.7708],[44.55,-0.7962],[45.5625,-0.7257],[46.575,-0.6131],[47.5875,-0.4972],[48.6,-0.391],[49.6125,-0.309],[50.625,-0.2715],[51.6375,-0.2639],[52.65,-0.2569],[53.6625,-0.2212],[54.675,-0.0809],[55.6875,0.3021],[56.7,0.76],[57.7125,0.8695],[58.725,0.6713],[59.7375,0.4503],[60.75,0.3047],[61.7625,0.2242],[62.775,0.1451],[63.7875,0.0744],[64.8,0.0111],[65.8125,-0.1036],[66.825,-0.2519],[67.8375,-0.371],[68.85,-0.4624],[69.8625,-0.5051],[70.875,-0.4945],[71.8875,-0.4449],[72.9,-0.347],[73.9125,-0.1243],[74.925,0.1541],[75.9375,0.3796],[76.95,0.5354],[77.9625,0.5906],[78.975,0.5946],[79.9875,0.5872]]
       }
       spine_04_ctrl.Rotation.Z {
        type=MovieSceneScriptingFloatChannel
        default=6.0957
        thinned_from=80
        keys [[0,6.0957],[1.0125,6.0152],[2.025,5.8152],[3.0375,5.5588],[4.05,5.3188],[5.0625,5.1121],[6.075,4.93],[7.0875,4.7815],[8.1,4.6922],[9.1125,4.668],[10.125,4.6589],[11.1375,4.635],[12.15,4.579],[13.1625,4.4636],[14.175,4.2916],[15.1875,4.2331],[16.2,4.3652],[18.225,4.8094],[19.2375,4.9776],[20.25,5.0928],[21.2625,5.1593],[23.2875,5.2274],[24.3,5.2708],[25.3125,5.2778],[26.325,5.2051],[27.3375,5.0603],[28.35,4.9446],[29.3625,4.8974],[30.375,4.8902],[31.3875,4.9143],[33.4125,4.9905],[34.425,5.0061],[36.45,5.0164],[37.4625,5.0408],[38.475,5.0804],[39.4875,5.1443],[40.5,5.3483],[41.5125,5.5721],[42.525,5.6084],[43.5375,5.5728],[44.55,5.5954],[45.5625,5.6978],[47.5875,5.9494],[48.6,6.0644],[49.6125,6.137],[50.625,6.1346],[51.6375,6.0727],[54.675,5.8106],[55.6875,5.594],[56.7,5.4113],[57.7125,5.4544],[60.75,5.7412],[61.7625,5.8551],[63.7875,6.102],[64.8,6.212],[65.8125,6.3039],[66.825,6.3386],[68.85,6.3341],[69.8625,6.3207],[70.875,6.34],[72.9,6.4015],[73.9125,6.3957],[76.95,6.2928],[77.9625,6.2249],[78.975,6.1387],[79.9875,6.0957]]
       }
       spine_04_ctrl.Scale.X {type=MovieSceneScriptingFloatChannel, default=1, const [1,80]}
       spine_04_ctrl.Scale.Y {type=MovieSceneScriptingFloatChannel, default=1, const [1,80]}
       spine_04_ctrl.Scale.Z {type=MovieSceneScriptingFloatChannel, default=1, const [1,80]}
       head_ctrl.Location.X {type=MovieSceneScriptingFloatChannel, default=0, thinned_from=80, keys [[0,0],[79.9875,0]]}
       head_ctrl.Location.Y {type=MovieSceneScriptingFloatChannel, default=0, thinned_from=80, keys [[0,0],[79.9875,0]]}
       head_ctrl.Location.Z {type=MovieSceneScriptingFloatChannel, default=0, thinned_from=80, keys [[0,0],[79.9875,0]]}
       head_ctrl.Rotation.X {
        type=MovieSceneScriptingFloatChannel
        default=-8.5539
        thinned_from=80
        keys [[0,-8.5539],[1.0125,-8.274],[2.025,-7.4701],[3.0375,-6.1621],[4.05,-4.3851],[5.0625,-1.972],[6.075,0.6027],[7.0875,2.5984],[8.1,3.8187],[9.1125,4.4988],[10.125,4.9147],[11.1375,5.2235],[12.15,5.4878],[13.1625,5.6225],[14.175,5.5974],[15.1875,5.6965],[16.2,6.4782],[17.2125,7.6916],[18.225,8.5421],[19.2375,8.5572],[20.25,8.0493],[21.2625,7.6077],[22.275,7.3372],[23.2875,7.2722],[24.3,7.3756],[25.3125,7.7181],[26.325,8.4355],[29.3625,11.1112],[30.375,11.8615],[31.3875,12.4737],[32.4,12.8833],[33.4125,12.9725],[34.425,12.7873],[35.4375,12.4539],[36.45,12.0638],[37.4625,11.7457],[38.475,11.7011],[39.4875,12.2013],[40.5,13.9193],[41.5125,14.7549],[42.525,13.6002],[43.5375,11.9572],[44.55,10.371],[45.5625,9.3229],[46.575,8.6739],[47.5875,8.2011],[48.6,7.945],[49.6125,7.905],[50.625,7.9503],[51.6375,7.9509],[52.65,7.9356],[53.6625,8.0197],[54.675,8.4747],[56.7,9.9469],[57.7125,10.9858],[58.725,11.9516],[59.7375,11.6308],[60.75,9.8312],[61.7625,7.0135],[62.775,3.1433],[63.7875,-0.5784],[64.8,-3.0531],[65.8125,-4.9735],[66.825,-6.6351],[67.8375,-7.8489],[68.85,-8.476],[69.8625,-8.673],[70.875,-8.6031],[71.8875,-8.4774],[72.9,-8.2925],[73.9125,-7.7705],[74.925,-7.3655],[75.9375,-7.4657],[76.95,-7.63],[77.9625,-7.9386],[78.975,-8.3468],[79.9875,-8.5539]]
       }
       head_ctrl.Rotation.Y {
        type=MovieSceneScriptingFloatChannel
        default=0.5323
        thinned_from=80
        keys [[0,0.5323],[1.0125,0.5176],[2.025,0.6116],[3.0375,0.9314],[4.05,1.5759],[5.0625,2.9565],[6.075,4.7131],[7.0875,5.9557],[8.1,6.6572],[9.1125,7.2719],[10.125,7.7447],[11.1375,8.0691],[12.15,8.1778],[13.1625,8.1484],[15.1875,6.9109],[16.2,6.7948],[17.2125,6.7123],[18.225,6.8327],[19.2375,7.1058],[20.25,7.3585],[21.2625,7.5465],[22.275,7.6156],[23.2875,7.4428],[24.3,6.9847],[25.3125,6.2157],[27.3375,3.9459],[28.35,3.1506],[29.3625,2.6152],[30.375,2.2578],[31.3875,2.0394],[32.4,1.9408],[33.4125,1.9078],[34.425,1.8874],[35.4375,1.887],[36.45,1.9504],[37.4625,2.0721],[38.475,2.2206],[39.4875,2.6229],[40.5,3.7043],[41.5125,4.0665],[42.525,3.2054],[43.5375,2.3534],[44.55,1.7982],[45.5625,1.5942],[46.575,1.6123],[47.5875,1.7154],[48.6,1.832],[49.6125,1.926],[50.625,1.9604],[51.6375,1.9867],[52.65,2.0772],[53.6625,2.2899],[54.675,2.6366],[55.6875,3.2405],[56.7,3.4184],[57.7125,2.104],[58.725,0.0941],[59.7375,-1.5715],[60.75,-2.8451],[61.7625,-3.8287],[62.775,-4.3241],[63.7875,-4.1391],[64.8,-3.4753],[65.8125,-2.6804],[66.825,-1.9348],[67.8375,-1.2085],[68.85,-0.7371],[69.8625,-0.4607],[70.875,-0.2472],[71.8875,-0.0614],[72.9,0.0717],[73.9125,0.2808],[74.925,0.4194],[75.9375,0.4692],[76.95,0.5895],[77.9625,0.6011],[78.975,0.558],[79.9875,0.5323]]
       }
       head_ctrl.Rotation.Z {
        type=MovieSceneScriptingFloatChannel
        default=-17.6838
        thinned_from=80
        keys [[0,-17.6838],[1.0125,-17.2635],[2.025,-16.1022],[3.0375,-14.334],[4.05,-12.0424],[5.0625,-8.4638],[6.075,-3.9784],[7.0875,-0.3947],[8.1,1.8717],[9.1125,3.714],[10.125,5.2516],[11.1375,6.5313],[12.15,7.5273],[13.1625,8.0195],[14.175,7.8757],[15.1875,7.1627],[16.2,6.3773],[17.2125,5.9299],[18.225,5.6072],[19.2375,5.3329],[20.25,5.1582],[21.2625,5.137],[22.275,5.2946],[23.2875,5.6602],[24.3,6.235],[25.3125,7.1353],[27.3375,9.8822],[28.35,10.9084],[29.3625,11.7718],[30.375,12.5589],[31.3875,13.2568],[32.4,13.8251],[33.4125,14.2543],[34.425,14.5633],[35.4375,14.7835],[36.45,14.9389],[37.4625,14.9798],[39.4875,14.7734],[40.5,15.2867],[41.5125,15.8691],[42.525,15.6712],[43.5375,15.0821],[44.55,14.3788],[46.575,12.8695],[47.5875,12.2005],[48.6,11.6472],[49.6125,11.2109],[50.625,10.8789],[52.65,10.304],[53.6625,9.9101],[54.675,8.8492],[55.6875,6.1715],[56.7,2.0101],[57.7125,-2.4126],[58.725,-6.38],[59.7375,-9.897],[60.75,-12.9088],[61.7625,-14.9853],[62.775,-16.2396],[63.7875,-17.0466],[64.8,-17.5608],[65.8125,-17.9582],[66.825,-18.1419],[67.8375,-18.0599],[68.85,-17.8275],[69.8625,-17.6621],[70.875,-17.6622],[72.9,-17.749],[79.9875,-17.6838]]
       }
       head_ctrl.Scale.X {type=MovieSceneScriptingFloatChannel, default=1, thinned_from=80, keys [[0,1],[79.9875,1]]}
       head_ctrl.Scale.Y {type=MovieSceneScriptingFloatChannel, default=1, const [1,80]}
       head_ctrl.Scale.Z {type=MovieSceneScriptingFloatChannel, default=1, const [1,80]}
       clavicle_l_ctrl.Location.X {type=MovieSceneScriptingFloatChannel, default=0, thinned_from=80, keys [[0,0],[79.9875,0]]}
       clavicle_l_ctrl.Location.Y {type=MovieSceneScriptingFloatChannel, default=0, thinned_from=80, keys [[0,0],[79.9875,0]]}
       clavicle_l_ctrl.Location.Z {type=MovieSceneScriptingFloatChannel, default=0, thinned_from=80, keys [[0,0],[79.9875,0]]}
       clavicle_l_ctrl.Rotation.X {type=MovieSceneScriptingFloatChannel, default=-3.062, thinned_from=80, keys [[0,-3.062],[79.9875,-3.062]]}
       clavicle_l_ctrl.Rotation.Y {
        type=MovieSceneScriptingFloatChannel
        default=7.8404
        thinned_from=80
        keys [[0,7.8404],[1.0125,7.9275],[2.025,8.164],[3.0375,8.5161],[4.05,8.9501],[7.0875,10.405],[8.1,10.8282],[9.1125,11.1643],[10.125,11.3794],[11.1375,11.446],[12.15,11.3839],[13.1625,11.2396],[14.175,11.0346],[17.2125,10.2729],[18.225,10.0425],[19.2375,9.8599],[20.25,9.7471],[21.2625,9.7166],[22.275,9.7509],[23.2875,9.8258],[24.3,9.9352],[25.3125,10.073],[27.3375,10.4095],[31.3875,11.1563],[37.4625,12.1166],[39.4875,12.3834],[46.575,13.0764],[49.6125,13.3032],[51.6375,13.402],[55.6875,13.4476],[56.7,13.5156],[57.7125,13.7269],[58.725,14.0634],[60.75,14.9026],[62.775,16.1992],[63.7875,16.501],[64.8,16.5828],[65.8125,16.5876],[66.825,16.4972],[67.8375,16.2095],[68.85,15.7081],[69.8625,14.9969],[70.875,13.9911],[71.8875,12.6192],[72.9,10.9201],[73.9125,7.9811],[74.925,5.7923],[75.9375,6.3252],[76.95,7.2448],[77.9625,7.6103],[78.975,7.7897],[79.9875,7.8404]]
       }
       clavicle_l_ctrl.Rotation.Z {
        type=MovieSceneScriptingFloatChannel
        default=7.8439
        thinned_from=80
        keys [[0,7.8439],[1.0125,7.9049],[2.025,8.0704],[3.0375,8.3169],[4.05,8.6207],[7.0875,9.6392],[8.1,9.9355],[9.1125,10.1707],[10.125,10.3213],[11.1375,10.3648],[12.15,10.2924],[13.1625,10.124],[14.175,9.8849],[15.1875,9.6004],[17.2125,8.9967],[18.225,8.728],[19.2375,8.5151],[20.25,8.3835],[21.2625,8.3495],[22.275,8.3982],[23.2875,8.5019],[24.3,8.649],[25.3125,8.8276],[28.35,9.4338],[29.3625,9.6199],[30.375,9.7782],[31.3875,9.897],[32.4,9.9644],[33.4125,9.9341],[34.425,9.7336],[35.4375,9.3742],[37.4625,8.4903],[38.475,8.1211],[39.4875,7.9039],[40.5,7.8439],[79.9875,7.8439]]
       }
       clavicle_l_ctrl.Scale.X {type=MovieSceneScriptingFloatChannel, default=1, const [1,80]}
       clavicle_l_ctrl.Scale.Y {type=MovieSceneScriptingFloatChannel, default=1, const [1,80]}
       clavicle_l_ctrl.Scale.Z {type=MovieSceneScriptingFloatChannel, default=1, const [1,80]}
       clavicle_r_ctrl.Location.X {type=MovieSceneScriptingFloatChannel, default=0, thinned_from=80, keys [[0,0],[79.9875,0]]}
       clavicle_r_ctrl.Location.Y {type=MovieSceneScriptingFloatChannel, default=0, thinned_from=80, keys [[0,0],[79.9875,0]]}
       clavicle_r_ctrl.Location.Z {type=MovieSceneScriptingFloatChannel, default=0, thinned_from=80, keys [[0,0],[79.9875,0]]}
       clavicle_r_ctrl.Rotation.X {
        type=MovieSceneScriptingFloatChannel
        default=-3.2085
        thinned_from=80
        keys [[0,-3.2085],[1.0125,-2.8381],[2.025,-1.9225],[3.0375,-0.7409],[4.05,0.4269],[5.0625,1.3017],[6.075,1.613],[7.0875,1.2382],[8.1,0.4168],[9.1125,-0.549],[10.125,-1.4025],[11.1375,-2.2097],[12.15,-2.9195],[13.1625,-3.1993],[14.175,-3.1257],[15.1875,-2.9594],[16.2,-2.7375],[17.2125,-2.4907],[18.225,-2.2357],[19.2375,-1.988],[20.25,-1.7513],[21.2625,-1.5078],[22.275,-1.2324],[23.2875,-0.9081],[24.3,-0.5588],[25.3125,-0.2052],[26.325,0.1551],[27.3375,0.5302],[28.35,0.9136],[29.3625,1.2455],[30.375,1.5274],[31.3875,1.8003],[32.4,2.0625],[33.4125,2.3127],[34.425,2.5491],[35.4375,2.7759],[36.45,2.9939],[37.4625,3.1776],[38.475,3.2869],[39.4875,3.1881],[40.5,2.5419],[41.5125,1.875],[42.525,1.5873],[43.5375,1.3567],[45.5625,0.9518],[46.575,0.7437],[47.5875,0.5458],[48.6,0.3512],[49.6125,0.1367],[50.625,-0.0835],[51.6375,-0.2964],[52.65,-0.5563],[53.6625,-0.9472],[54.675,-1.9079],[55.6875,-2.9282],[56.7,-3.2085],[79.9875,-3.2085]]
       }
       clavicle_r_ctrl.Rotation.Y {
        type=MovieSceneScriptingFloatChannel
        default=17.3014
        thinned_from=80
        keys [[0,17.3014],[1.0125,17.3836],[2.025,17.601],[3.0375,17.9131],[6.075,19.0217],[8.1,19.8753],[9.1125,20.2215],[10.125,20.3178],[11.1375,19.9597],[12.15,19.2333],[13.1625,18.2249],[15.1875,14.1926],[16.2,13.0977],[17.2125,12.2601],[18.225,11.6774],[19.2375,11.3375],[20.25,11.1913],[21.2625,11.1861],[22.275,11.351],[24.3,12.1582],[25.3125,12.4158],[26.325,12.6251],[27.3375,12.7811],[31.3875,13.1388],[33.4125,13.2414],[34.425,13.2476],[35.4375,13.0904],[36.45,12.5878],[37.4625,11.8888],[38.475,11.3699],[39.4875,12.0052],[40.5,15.5123],[41.5125,18.6447],[42.525,19.3116],[43.5375,19.4799],[44.55,19.434],[45.5625,19.4579],[46.575,19.7076],[47.5875,20.1467],[48.6,20.758],[49.6125,21.586],[50.625,22.7142],[51.6375,23.6349],[52.65,23.4869],[53.6625,22.2806],[55.6875,18.8294],[56.7,17.567],[57.7125,16.9626],[58.725,17.045],[59.7375,17.6542],[60.75,18.6503],[61.7625,20.4645],[62.775,22.475],[63.7875,23.6446],[64.8,24.2783],[65.8125,24.5302],[66.825,24.394],[67.8375,23.8496],[68.85,22.8932],[69.8625,21.5252],[70.875,19.3689],[72.9,13.5679],[73.9125,10.58],[74.925,8.7706],[75.9375,10.7953],[76.95,13.9002],[77.9625,15.5963],[78.975,16.819],[79.9875,17.3014]]
       }
       clavicle_r_ctrl.Rotation.Z {
        type=MovieSceneScriptingFloatChannel
        default=0.1961
        thinned_from=80
        keys [[0,0.1961],[1.0125,0.8435],[2.025,2.5191],[3.0375,4.8468],[4.05,7.4503],[5.0625,9.9534],[6.075,12.0088],[7.0875,13.6551],[8.1,15.1347],[9.1125,16.4491],[10.125,17.597],[11.1375,18.567],[12.15,19.4007],[13.1625,20.154],[14.175,20.7203],[15.1875,20.884],[16.2,20.5902],[17.2125,19.8354],[18.225,18.5664],[19.2375,16.745],[20.25,14.5189],[21.2625,12.5138],[22.275,11.4409],[23.2875,11.4897],[24.3,12.18],[25.3125,13.433],[27.3375,16.8377],[28.35,18.3298],[29.3625,19.3238],[30.375,19.8036],[31.3875,20.0168],[32.4,20.0592],[34.425,20.0145],[35.4375,20.1864],[36.45,20.6664],[37.4625,21.2744],[38.475,21.662],[39.4875,21.3534],[40.5,19.6433],[41.5125,18.1418],[42.525,17.8093],[43.5375,17.6539],[44.55,17.5538],[45.5625,17.3875],[46.575,17.1131],[47.5875,16.7734],[48.6,16.3556],[49.6125,15.7963],[50.625,15.1211],[51.6375,14.3614],[52.65,13.3998],[53.6625,12.0501],[54.675,9.4523],[55.6875,6.052],[56.7,3.4278],[57.7125,1.7626],[58.725,0.832],[59.7375,0.3756],[60.75,0.2179],[61.7625,0.8957],[62.775,2.5032],[63.7875,3.8988],[64.8,4.5719],[65.8125,4.7488],[66.825,4.4462],[67.8375,3.5986],[68.85,2.482],[69.8625,1.4429],[70.875,0.7292],[71.8875,0.3364],[72.9,0.2053],[73.9125,0.1961],[79.9875,0.1961]]
       }
       clavicle_r_ctrl.Scale.X {type=MovieSceneScriptingFloatChannel, default=1, const [1,80]}
       clavicle_r_ctrl.Scale.Y {type=MovieSceneScriptingFloatChannel, default=1, const [1,80]}
       clavicle_r_ctrl.Scale.Z {type=MovieSceneScriptingFloatChannel, default=1, const [1,80]}
       neck_01_ctrl.Location.X {type=MovieSceneScriptingFloatChannel, default=0, thinned_from=80, keys [[0,0],[79.9875,0]]}
       neck_01_ctrl.Location.Y {type=MovieSceneScriptingFloatChannel, default=0, thinned_from=80, keys [[0,0],[79.9875,0]]}
       neck_01_ctrl.Location.Z {type=MovieSceneScriptingFloatChannel, default=0, thinned_from=80, keys [[0,0],[79.9875,0]]}
       neck_01_ctrl.Rotation.X {
        type=MovieSceneScriptingFloatChannel
        default=-8.2334
        thinned_from=80
        keys [[0,-8.2334],[1.0125,-8.3597],[2.025,-8.6862],[3.0375,-9.1342],[4.05,-9.6241],[5.0625,-10.2949],[6.075,-10.9999],[7.0875,-11.551],[8.1,-11.8753],[9.1125,-11.9385],[10.125,-11.9147],[12.15,-11.7424],[13.1625,-11.6066],[14.175,-11.3477],[15.1875,-10.8239],[16.2,-10.1508],[17.2125,-9.3796],[18.225,-8.4166],[19.2375,-7.3877],[20.25,-6.4198],[21.2625,-5.6729],[22.275,-5.2534],[23.2875,-5.1093],[24.3,-5.1441],[25.3125,-5.3344],[26.325,-5.6418],[27.3375,-6.0127],[28.35,-6.4173],[29.3625,-6.8705],[30.375,-7.3944],[31.3875,-7.9548],[32.4,-8.456],[33.4125,-8.8257],[34.425,-9.0653],[35.4375,-9.2113],[36.45,-9.2816],[37.4625,-9.299],[38.475,-9.2767],[39.4875,-9.1981],[40.5,-8.6886],[41.5125,-7.8598],[42.525,-7.1533],[43.5375,-6.5025],[44.55,-5.9924],[45.5625,-5.7548],[46.575,-5.6921],[47.5875,-5.6925],[48.6,-5.7998],[49.6125,-6.0427],[50.625,-6.3141],[51.6375,-6.4804],[54.675,-6.6748],[55.6875,-6.5721],[56.7,-6.2985],[57.7125,-6.0786],[58.725,-5.9782],[60.75,-5.8905],[61.7625,-5.5448],[62.775,-5.4439],[63.7875,-6.1323],[64.8,-7.1394],[65.8125,-8.3203],[66.825,-9.4041],[67.8375,-10.2056],[68.85,-10.6141],[69.8625,-10.4919],[70.875,-10.1241],[71.8875,-9.7075],[73.9125,-8.8029],[74.925,-8.1818],[75.9375,-7.6866],[76.95,-7.6851],[77.9625,-7.893],[78.975,-8.1239],[79.9875,-8.2334]]
       }
       neck_01_ctrl.Rotation.Y {
        type=MovieSceneScriptingFloatChannel
        default=-4.8098
        thinned_from=80
        keys [[0,-4.8098],[1.0125,-4.6567],[2.025,-4.2596],[3.0375,-3.7091],[4.05,-3.1082],[5.0625,-2.3791],[6.075,-1.5554],[7.0875,-0.7629],[8.1,-0.1058],[9.1125,0.4094],[10.125,0.8899],[11.1375,1.261],[12.15,1.4317],[13.1625,1.2301],[14.175,0.5465],[15.1875,-0.7422],[16.2,-1.9136],[17.2125,-2.4787],[18.225,-2.7469],[19.2375,-2.7864],[20.25,-2.6924],[21.2625,-2.5639],[22.275,-2.4296],[23.2875,-2.0988],[24.3,-1.4361],[25.3125,-0.4953],[26.325,0.6781],[27.3375,1.9636],[28.35,3.0786],[29.3625,3.8522],[30.375,4.3207],[31.3875,4.5621],[32.4,4.651],[33.4125,4.6503],[34.425,4.6048],[38.475,4.3357],[39.4875,4.3297],[40.5,4.911],[41.5125,5.8373],[42.525,6.5167],[43.5375,6.9486],[44.55,6.9361],[45.5625,6.4902],[47.5875,5.2573],[48.6,4.6147],[49.6125,4.0204],[50.625,3.6542],[51.6375,3.5439],[52.65,3.5354],[53.6625,3.4269],[54.675,3.0066],[55.6875,1.9413],[56.7,0.3982],[57.7125,-0.7405],[58.725,-1.1594],[59.7375,-1.2704],[60.75,-1.414],[61.7625,-1.4372],[62.775,-1.3316],[63.7875,-1.4134],[64.8,-1.5986],[65.8125,-1.6928],[66.825,-1.693],[67.8375,-1.8735],[69.8625,-2.3673],[70.875,-2.6813],[71.8875,-2.9798],[72.9,-3.261],[74.925,-4.2755],[75.9375,-4.5967],[76.95,-4.8509],[77.9625,-4.8848],[79.9875,-4.8098]]
       }
       neck_01_ctrl.Rotation.Z {
        type=MovieSceneScriptingFloatChannel
        default=-23.9141
        thinned_from=80
        keys [[0,-23.9141],[1.0125,-23.3995],[2.025,-22.1756],[3.0375,-20.7043],[4.05,-19.425],[5.0625,-18.4464],[6.075,-17.6532],[7.0875,-17.0614],[8.1,-16.7007],[9.1125,-16.4834],[10.125,-16.3294],[11.1375,-16.2894],[12.15,-16.4004],[13.1625,-16.6432],[14.175,-17.2226],[15.1875,-18.6229],[17.2125,-22.6883],[18.225,-24.3683],[19.2375,-25.7691],[20.25,-26.8998],[21.2625,-27.6501],[22.275,-27.9939],[23.2875,-28.1328],[24.3,-28.1677],[25.3125,-27.9133],[26.325,-27.1854],[28.35,-24.9237],[29.3625,-23.8446],[30.375,-22.699],[31.3875,-21.6785],[32.4,-20.9333],[33.4125,-20.515],[34.425,-20.2855],[35.4375,-20.129],[36.45,-20.0418],[38.475,-20.0242],[39.4875,-19.9539],[40.5,-19.603],[41.5125,-19.1837],[42.525,-18.9699],[43.5375,-18.9376],[44.55,-19.2769],[45.5625,-20.0142],[46.575,-20.838],[47.5875,-21.5261],[48.6,-22.0888],[49.6125,-22.4085],[50.625,-22.3478],[51.6375,-22.0116],[52.65,-21.5986],[53.6625,-21.2991],[54.675,-21.3026],[55.6875,-21.6168],[56.7,-22.7168],[57.7125,-24.4043],[58.725,-25.7569],[59.7375,-26.7721],[60.75,-27.5146],[61.7625,-28.0799],[62.775,-28.5408],[63.7875,-28.8465],[64.8,-29.0264],[65.8125,-29.0551],[66.825,-28.6638],[67.8375,-27.9476],[68.85,-27.1486],[69.8625,-26.4528],[70.875,-26.1013],[72.9,-25.7322],[76.95,-24.685],[78.975,-24.0579],[79.9875,-23.9141]]
       }
       neck_01_ctrl.Scale.X {type=MovieSceneScriptingFloatChannel, default=1, const [1,80]}
       neck_01_ctrl.Scale.Y {type=MovieSceneScriptingFloatChannel, default=1, const [1,80]}
       neck_01_ctrl.Scale.Z {type=MovieSceneScriptingFloatChannel, default=1, const [1,80]}
       neck_02_ctrl.Location.X {type=MovieSceneScriptingFloatChannel, default=0, const [0,80]}
       neck_02_ctrl.Location.Y {type=MovieSceneScriptingFloatChannel, default=0, const [0,80]}
       neck_02_ctrl.Location.Z {type=MovieSceneScriptingFloatChannel, default=0, const [0,80]}
       neck_02_ctrl.Rotation.X {type=MovieSceneScriptingFloatChannel, default=0, const [0,80]}
       neck_02_ctrl.Rotation.Y {type=MovieSceneScriptingFloatChannel, default=0, const [0,80]}
       neck_02_ctrl.Rotation.Z {type=MovieSceneScriptingFloatChannel, default=0, const [0,80]}
       neck_02_ctrl.Scale.X {type=MovieSceneScriptingFloatChannel, default=1, const [1,80]}
       neck_02_ctrl.Scale.Y {type=MovieSceneScriptingFloatChannel, default=1, const [1,80]}
       neck_02_ctrl.Scale.Z {type=MovieSceneScriptingFloatChannel, default=1, const [1,80]}
       head_ik_ctrl.Location.X {
        type=MovieSceneScriptingFloatChannel
        default=1.8819
        thinned_from=80
        keys [[0,1.8819],[1.0125,1.8348],[2.025,1.7175],[4.05,1.4149],[5.0625,1.2569],[6.075,1.0889],[7.0875,0.9317],[8.1,0.8053],[9.1125,0.7103],[10.125,0.6231],[11.1375,0.5531],[12.15,0.5134],[13.1625,0.5287],[14.175,0.6216],[15.1875,0.8415],[16.2,1.0812],[17.2125,1.2411],[18.225,1.3532],[19.2375,1.4235],[20.25,1.4648],[21.2625,1.4853],[22.275,1.485],[23.2875,1.4427],[24.3,1.3423],[25.3125,1.1837],[26.325,0.9661],[27.3375,0.7116],[28.35,0.485],[29.3625,0.3189],[30.375,0.2048],[31.3875,0.1362],[32.4,0.1039],[33.4125,0.096],[34.425,0.099],[35.4375,0.1065],[36.45,0.1166],[37.4625,0.1295],[38.475,0.1435],[39.4875,0.1472],[40.5,0.0581],[41.5125,-0.0875],[42.525,-0.2017],[43.5375,-0.2779],[44.55,-0.2669],[45.5625,-0.1668],[46.575,-0.0339],[48.6,0.2245],[49.6125,0.3365],[50.625,0.396],[51.6375,0.4021],[52.65,0.3873],[53.6625,0.3913],[54.675,0.453],[55.6875,0.6177],[56.7,0.8839],[57.7125,1.1213],[58.725,1.2419],[59.7375,1.3029],[60.75,1.3612],[61.7625,1.4011],[62.775,1.4124],[63.7875,1.4362],[64.8,1.4694],[65.8125,1.4829],[66.825,1.471],[67.8375,1.4852],[68.85,1.5131],[69.8625,1.5451],[71.8875,1.6471],[72.9,1.693],[73.9125,1.7705],[74.925,1.8441],[75.9375,1.8879],[76.95,1.92],[77.9625,1.9134],[78.975,1.8935],[79.9875,1.8819]]
       }
       head_ik_ctrl.Location.Y {
        type=MovieSceneScriptingFloatChannel
        default=-1.1632
        thinned_from=80
        keys [[0,-1.1632],[1.0125,-1.1905],[2.025,-1.2593],[4.05,-1.4433],[6.075,-1.6801],[7.0875,-1.7704],[8.1,-1.8204],[9.1125,-1.8279],[10.125,-1.8233],[12.15,-1.7969],[13.1625,-1.7778],[14.175,-1.7422],[15.1875,-1.6661],[16.2,-1.5673],[17.2125,-1.4488],[18.225,-1.2976],[19.2375,-1.1343],[20.25,-0.9792],[21.2625,-0.8578],[22.275,-0.786],[23.2875,-0.7491],[24.3,-0.7304],[25.3125,-0.7329],[26.325,-0.7578],[27.3375,-0.8006],[28.35,-0.8553],[29.3625,-0.924],[30.375,-1.0106],[31.3875,-1.1052],[32.4,-1.1894],[33.4125,-1.2506],[34.425,-1.2908],[35.4375,-1.3168],[36.45,-1.3313],[37.4625,-1.3374],[38.475,-1.3368],[39.4875,-1.3277],[40.5,-1.257],[41.5125,-1.144],[43.5375,-0.9478],[44.55,-0.8593],[45.5625,-0.8029],[46.575,-0.7726],[47.5875,-0.7556],[48.6,-0.7577],[49.6125,-0.786],[50.625,-0.8273],[51.6375,-0.8595],[52.65,-0.8792],[53.6625,-0.8948],[54.675,-0.9053],[55.6875,-0.8878],[56.7,-0.8444],[57.7125,-0.8065],[58.725,-0.7863],[59.7375,-0.7777],[60.75,-0.7718],[61.7625,-0.7173],[62.775,-0.6997],[63.7875,-0.8128],[64.8,-0.9802],[65.8125,-1.1764],[66.825,-1.3573],[67.8375,-1.4891],[68.85,-1.5516],[69.8625,-1.5253],[70.875,-1.4562],[71.8875,-1.3788],[72.9,-1.2986],[73.9125,-1.2229],[74.925,-1.124],[75.9375,-1.0463],[76.95,-1.0504],[77.9625,-1.0926],[78.975,-1.1404],[79.9875,-1.1632]]
       }
       head_ik_ctrl.Location.Z {
        type=MovieSceneScriptingFloatChannel
        default=0.3693
        thinned_from=80
        keys [[0,0.3693],[1.0125,0.3808],[2.025,0.4073],[3.0375,0.4367],[4.05,0.4641],[5.0625,0.5045],[7.0875,0.6133],[8.1,0.6568],[9.1125,0.6765],[10.125,0.684],[11.1375,0.6856],[14.175,0.6845],[16.2,0.7053],[17.2125,0.6887],[18.225,0.6527],[19.2375,0.6107],[20.25,0.5716],[21.2625,0.5441],[22.275,0.5299],[23.2875,0.5132],[24.3,0.4828],[25.3125,0.4442],[26.325,0.4037],[27.3375,0.3722],[28.35,0.3633],[29.3625,0.3791],[30.375,0.4144],[31.3875,0.4611],[32.4,0.5059],[33.4125,0.5387],[34.425,0.5584],[35.4375,0.5681],[36.45,0.5688],[37.4625,0.5632],[38.475,0.5544],[39.4875,0.5422],[40.5,0.4885],[41.5125,0.4055],[42.525,0.3413],[43.5375,0.2887],[44.55,0.2529],[45.5625,0.2404],[46.575,0.2425],[47.5875,0.251],[48.6,0.2701],[49.6125,0.2996],[50.625,0.326],[51.6375,0.3376],[52.65,0.3392],[53.6625,0.3434],[54.675,0.3592],[55.6875,0.3939],[56.7,0.4541],[57.7125,0.5025],[58.725,0.5082],[59.7375,0.4921],[60.75,0.4682],[61.7625,0.4124],[62.775,0.3605],[63.7875,0.3549],[65.8125,0.3705],[66.825,0.3747],[67.8375,0.3961],[68.85,0.4208],[69.8625,0.4262],[70.875,0.4256],[71.8875,0.4226],[72.9,0.4128],[73.9125,0.4043],[74.925,0.3843],[75.9375,0.3665],[76.95,0.3714],[77.9625,0.3727],[79.9875,0.3693]]
       }
       head_ik_ctrl.Rotation.X {
        type=MovieSceneScriptingFloatChannel
        default=-8.5539
        thinned_from=80
        keys [[0,-8.5539],[1.0125,-8.274],[2.025,-7.4701],[3.0375,-6.1621],[4.05,-4.3851],[5.0625,-1.972],[6.075,0.6027],[7.0875,2.5984],[8.1,3.8187],[9.1125,4.4988],[10.125,4.9147],[11.1375,5.2235],[12.15,5.4878],[13.1625,5.6225],[14.175,5.5974],[15.1875,5.6965],[16.2,6.4782],[17.2125,7.6916],[18.225,8.5421],[19.2375,8.5572],[20.25,8.0493],[21.2625,7.6077],[22.275,7.3372],[23.2875,7.2722],[24.3,7.3756],[25.3125,7.7181],[26.325,8.4355],[29.3625,11.1112],[30.375,11.8615],[31.3875,12.4737],[32.4,12.8833],[33.4125,12.9725],[34.425,12.7873],[35.4375,12.4539],[36.45,12.0638],[37.4625,11.7457],[38.475,11.7011],[39.4875,12.2013],[40.5,13.9193],[41.5125,14.7549],[42.525,13.6002],[43.5375,11.9572],[44.55,10.371],[45.5625,9.3229],[46.575,8.6739],[47.5875,8.2011],[48.6,7.945],[49.6125,7.905],[50.625,7.9503],[51.6375,7.9509],[52.65,7.9356],[53.6625,8.0197],[54.675,8.4747],[56.7,9.9469],[57.7125,10.9858],[58.725,11.9516],[59.7375,11.6308],[60.75,9.8312],[61.7625,7.0135],[62.775,3.1433],[63.7875,-0.5784],[64.8,-3.0531],[65.8125,-4.9735],[66.825,-6.6351],[67.8375,-7.8489],[68.85,-8.476],[69.8625,-8.673],[70.875,-8.6031],[71.8875,-8.4774],[72.9,-8.2925],[73.9125,-7.7705],[74.925,-7.3655],[75.9375,-7.4657],[76.95,-7.63],[77.9625,-7.9386],[78.975,-8.3468],[79.9875,-8.5539]]
       }
       head_ik_ctrl.Rotation.Y {
        type=MovieSceneScriptingFloatChannel
        default=0.5323
        thinned_from=80
        keys [[0,0.5323],[1.0125,0.5176],[2.025,0.6116],[3.0375,0.9314],[4.05,1.5759],[5.0625,2.9565],[6.075,4.7131],[7.0875,5.9557],[8.1,6.6572],[9.1125,7.2719],[10.125,7.7447],[11.1375,8.0691],[12.15,8.1778],[13.1625,8.1484],[15.1875,6.9109],[16.2,6.7948],[17.2125,6.7123],[18.225,6.8327],[19.2375,7.1058],[20.25,7.3585],[21.2625,7.5465],[22.275,7.6156],[23.2875,7.4428],[24.3,6.9847],[25.3125,6.2157],[27.3375,3.9459],[28.35,3.1506],[29.3625,2.6152],[30.375,2.2578],[31.3875,2.0394],[32.4,1.9408],[33.4125,1.9078],[34.425,1.8874],[35.4375,1.887],[36.45,1.9504],[37.4625,2.0721],[38.475,2.2206],[39.4875,2.6229],[40.5,3.7043],[41.5125,4.0665],[42.525,3.2054],[43.5375,2.3534],[44.55,1.7982],[45.5625,1.5942],[46.575,1.6123],[47.5875,1.7154],[48.6,1.832],[49.6125,1.926],[50.625,1.9604],[51.6375,1.9867],[52.65,2.0772],[53.6625,2.2899],[54.675,2.6366],[55.6875,3.2405],[56.7,3.4184],[57.7125,2.104],[58.725,0.0941],[59.7375,-1.5715],[60.75,-2.8451],[61.7625,-3.8287],[62.775,-4.3241],[63.7875,-4.1391],[64.8,-3.4753],[65.8125,-2.6804],[66.825,-1.9348],[67.8375,-1.2085],[68.85,-0.7371],[69.8625,-0.4607],[70.875,-0.2472],[71.8875,-0.0614],[72.9,0.0717],[73.9125,0.2808],[74.925,0.4194],[75.9375,0.4692],[76.95,0.5895],[77.9625,0.6011],[78.975,0.558],[79.9875,0.5323]]
       }
       head_ik_ctrl.Rotation.Z {
        type=MovieSceneScriptingFloatChannel
        default=-17.6838
        thinned_from=80
        keys [[0,-17.6838],[1.0125,-17.2635],[2.025,-16.1022],[3.0375,-14.334],[4.05,-12.0424],[5.0625,-8.4638],[6.075,-3.9784],[7.0875,-0.3947],[8.1,1.8717],[9.1125,3.714],[10.125,5.2516],[11.1375,6.5313],[12.15,7.5273],[13.1625,8.0195],[14.175,7.8757],[15.1875,7.1627],[16.2,6.3773],[17.2125,5.9299],[18.225,5.6072],[19.2375,5.3329],[20.25,5.1582],[21.2625,5.137],[22.275,5.2946],[23.2875,5.6602],[24.3,6.235],[25.3125,7.1353],[27.3375,9.8822],[28.35,10.9084],[29.3625,11.7718],[30.375,12.5589],[31.3875,13.2568],[32.4,13.8251],[33.4125,14.2543],[34.425,14.5633],[35.4375,14.7835],[36.45,14.9389],[37.4625,14.9798],[39.4875,14.7734],[40.5,15.2867],[41.5125,15.8691],[42.525,15.6712],[43.5375,15.0821],[44.55,14.3788],[46.575,12.8695],[47.5875,12.2005],[48.6,11.6472],[49.6125,11.2109],[50.625,10.8789],[52.65,10.304],[53.6625,9.9101],[54.675,8.8492],[55.6875,6.1715],[56.7,2.0101],[57.7125,-2.4126],[58.725,-6.38],[59.7375,-9.897],[60.75,-12.9088],[61.7625,-14.9853],[62.775,-16.2396],[63.7875,-17.0466],[64.8,-17.5608],[65.8125,-17.9582],[66.825,-18.1419],[67.8375,-18.0599],[68.85,-17.8275],[69.8625,-17.6621],[70.875,-17.6622],[72.9,-17.749],[79.9875,-17.6838]]
       }
       head_ik_ctrl.Scale.X {type=MovieSceneScriptingFloatChannel, default=1, thinned_from=80, keys [[0,1],[79.9875,1]]}
       head_ik_ctrl.Scale.Y {type=MovieSceneScriptingFloatChannel, default=1, const [1,80]}
       head_ik_ctrl.Scale.Z {type=MovieSceneScriptingFloatChannel, default=1, const [1,80]}
       chest_ctrl.Location.X {
        type=MovieSceneScriptingFloatChannel
        default=-0.3403
        thinned_from=80
        keys [[0,-0.3403],[1.0125,-0.3871],[2.025,-0.489],[3.0375,-0.5712],[4.05,-0.5599],[5.0625,-0.4744],[6.075,-0.3948],[7.0875,-0.3321],[8.1,-0.2964],[9.1125,-0.2882],[10.125,-0.2897],[11.1375,-0.3002],[12.15,-0.323],[13.1625,-0.3809],[14.175,-0.4802],[15.1875,-0.6297],[16.2,-0.7635],[17.2125,-0.8316],[18.225,-0.8703],[19.2375,-0.89],[20.25,-0.8944],[21.2625,-0.8891],[22.275,-0.8732],[23.2875,-0.8286],[24.3,-0.747],[25.3125,-0.6382],[26.325,-0.5094],[27.3375,-0.3729],[28.35,-0.2572],[29.3625,-0.1803],[30.375,-0.1385],[31.3875,-0.1231],[32.4,-0.1244],[33.4125,-0.1323],[34.425,-0.1393],[36.45,-0.1495],[37.4625,-0.1559],[38.475,-0.1644],[39.4875,-0.1708],[40.5,-0.1143],[41.5125,0.0084],[42.525,0.1166],[43.5375,0.1856],[44.55,0.2034],[45.5625,0.1693],[46.575,0.11],[47.5875,0.0444],[48.6,-0.0266],[49.6125,-0.0953],[50.625,-0.1453],[51.6375,-0.1761],[52.65,-0.2028],[53.6625,-0.2434],[54.675,-0.321],[55.6875,-0.478],[56.7,-0.6644],[57.7125,-0.7393],[58.725,-0.7027],[60.75,-0.5681],[61.7625,-0.4917],[62.775,-0.4076],[63.7875,-0.3476],[64.8,-0.3022],[65.8125,-0.2487],[66.825,-0.1871],[67.8375,-0.148],[68.85,-0.1221],[69.8625,-0.1048],[70.875,-0.107],[71.8875,-0.1223],[72.9,-0.1476],[73.9125,-0.204],[74.925,-0.2625],[75.9375,-0.3041],[76.95,-0.3405],[77.9625,-0.3494],[79.9875,-0.3403]]
       }
       chest_ctrl.Location.Y {
        type=MovieSceneScriptingFloatChannel
        default=0.6826
        thinned_from=80
        keys [[0,0.6826],[1.0125,0.6526],[2.025,0.597],[3.0375,0.5675],[4.05,0.6154],[5.0625,0.7477],[6.075,0.8926],[7.0875,1.0211],[8.1,1.1145],[9.1125,1.1592],[10.125,1.1811],[11.1375,1.1861],[12.15,1.1651],[14.175,1.0911],[16.2,1.0219],[17.2125,0.959],[18.225,0.851],[19.2375,0.7312],[20.25,0.6231],[21.2625,0.5464],[22.275,0.518],[23.2875,0.5434],[24.3,0.6135],[25.3125,0.7194],[26.325,0.8515],[28.35,1.1405],[29.3625,1.2789],[30.375,1.4121],[31.3875,1.5298],[32.4,1.6156],[33.4125,1.6608],[34.425,1.6736],[35.4375,1.6637],[36.45,1.6339],[37.4625,1.5903],[38.475,1.5409],[39.4875,1.4731],[40.5,1.3022],[41.5125,1.1275],[42.525,1.0511],[43.5375,0.9994],[44.55,0.9687],[45.5625,0.9687],[46.575,0.9862],[47.5875,1.0112],[48.6,1.0499],[50.625,1.1533],[51.6375,1.1838],[52.65,1.1941],[53.6625,1.1948],[54.675,1.2027],[55.6875,1.2159],[56.7,1.2405],[58.725,1.3524],[59.7375,1.3741],[60.75,1.3529],[61.7625,1.2622],[62.775,1.1538],[63.7875,1.0869],[64.8,1.0355],[65.8125,0.9988],[66.825,0.9712],[67.8375,0.9752],[68.85,0.996],[70.875,1.0248],[71.8875,1.0289],[72.9,1.0071],[73.9125,0.9483],[74.925,0.8502],[75.9375,0.7593],[76.95,0.7137],[77.9625,0.6921],[78.975,0.684],[79.9875,0.6826]]
       }
       chest_ctrl.Location.Z {
        type=MovieSceneScriptingFloatChannel
        default=0.0614
        thinned_from=80
        keys [[0,0.0614],[1.0125,0.0591],[2.025,0.0536],[3.0375,0.0491],[4.05,0.0512],[5.0625,0.0589],[6.075,0.0659],[7.0875,0.0708],[8.1,0.0737],[9.1125,0.0747],[10.125,0.075],[11.1375,0.0767],[12.15,0.0904],[13.1625,0.1109],[14.175,0.1165],[15.1875,0.09],[16.2,0.0443],[17.2125,0.0038],[18.225,-0.0146],[19.2375,-0.0123],[20.25,-0.0015],[21.2625,0.0153],[22.275,0.0357],[23.2875,0.0572],[24.3,0.0755],[25.3125,0.084],[26.325,0.0783],[27.3375,0.0684],[28.35,0.0666],[29.3625,0.0705],[30.375,0.0736],[31.3875,0.0761],[33.4125,0.0802],[34.425,0.082],[35.4375,0.0835],[36.45,0.0845],[37.4625,0.0826],[38.475,0.0761],[39.4875,0.069],[40.5,0.0769],[41.5125,0.0927],[42.525,0.1016],[43.5375,0.1027],[44.55,0.097],[45.5625,0.0897],[46.575,0.0827],[47.5875,0.0776],[48.6,0.076],[49.6125,0.077],[50.625,0.0784],[51.6375,0.0792],[52.65,0.0793],[54.675,0.0786],[55.6875,0.0773],[56.7,0.0745],[57.7125,0.0743],[59.7375,0.0796],[60.75,0.0814],[61.7625,0.0821],[62.775,0.0818],[63.7875,0.0811],[65.8125,0.0789],[66.825,0.0773],[67.8375,0.076],[68.85,0.0753],[69.8625,0.0751],[70.875,0.0755],[71.8875,0.0756],[72.9,0.0748],[73.9125,0.0727],[75.9375,0.065],[76.95,0.0627],[77.9625,0.0617],[78.975,0.0614],[79.9875,0.0614]]
       }
       chest_ctrl.Rotation.X {
        type=MovieSceneScriptingFloatChannel
        default=2.2399
        thinned_from=80
        keys [[0,2.2399],[1.0125,2.1919],[2.025,2.1851],[3.0375,2.3904],[4.05,3.0258],[5.0625,4.2973],[6.075,5.7686],[7.0875,7.1406],[8.1,8.2128],[9.1125,8.8141],[10.125,9.1759],[11.1375,9.3684],[12.15,9.3214],[13.1625,9.0241],[14.175,8.6107],[15.1875,8.2313],[16.2,7.8802],[17.2125,7.2555],[18.225,6.2563],[19.2375,5.1566],[20.25,4.1791],[21.2625,3.5316],[22.275,3.3638],[23.2875,3.6387],[24.3,4.1959],[25.3125,4.9411],[26.325,5.7871],[27.3375,6.6903],[28.35,7.6888],[29.3625,8.8001],[30.375,9.9951],[31.3875,11.1401],[32.4,12.0294],[33.4125,12.5369],[34.425,12.7152],[35.4375,12.6604],[36.45,12.4038],[37.4625,12.0135],[38.475,11.58],[39.4875,11.1219],[40.5,10.0984],[41.5125,8.7088],[42.525,7.5842],[43.5375,6.6388],[44.55,6.0299],[45.5625,5.8932],[46.575,6.0086],[47.5875,6.2176],[48.6,6.6062],[49.6125,7.1606],[50.625,7.6421],[51.6375,7.8424],[53.6625,7.844],[54.675,8.009],[55.6875,8.3495],[56.7,8.8339],[57.7125,9.3909],[58.725,9.7463],[59.7375,9.6834],[60.75,9.1877],[61.7625,7.9006],[62.775,6.4117],[63.7875,5.5115],[64.8,4.8627],[65.8125,4.3947],[66.825,4.0985],[67.8375,4.3415],[68.85,4.8083],[69.8625,5.0744],[70.875,5.2989],[71.8875,5.4255],[72.9,5.2848],[73.9125,4.8333],[74.925,3.9323],[75.9375,3.0439],[76.95,2.6541],[77.9625,2.4218],[78.975,2.2851],[79.9875,2.2399]]
       }
       chest_ctrl.Rotation.Y {
        type=MovieSceneScriptingFloatChannel
        default=4.7937
        keys [[0,4.7937],[1.0125,5.2111],[2.025,6.0622],[3.0375,6.6415],[4.05,6.3142],[5.0625,5.1608],[6.075,3.9961],[7.0875,3.0367],[8.1,2.4481],[9.1125,2.271],[10.125,2.2413],[11.1375,2.3483],[12.15,2.6442],[13.1625,3.303],[14.175,4.3386],[15.1875,5.9014],[16.2,7.2944],[17.2125,8.1163],[18.225,8.849],[19.2375,9.4626],[20.25,9.9112],[21.2625,10.1835],[22.275,10.2076],[23.2875,9.7567],[24.3,8.7545],[25.3125,7.3302],[26.325,5.5997],[27.3375,3.7434],[28.35,2.1157],[29.3625,0.9264],[30.375,0.1342],[31.3875,-0.3396],[32.4,-0.5597],[33.4125,-0.6068],[34.425,-0.5723],[35.4375,-0.4919],[36.45,-0.3592],[37.4625,-0.1743],[38.475,0.0521],[39.4875,0.4047],[40.5,0.6962],[41.5125,0.2265],[42.525,-0.7209],[43.5375,-1.369],[44.55,-1.5133],[45.5625,-1.1827],[46.575,-0.6378],[47.5875,-0.0651],[48.6,0.4995],[49.6125,0.9714],[50.625,1.2308],[51.6375,1.3453],[52.65,1.4803],[53.6625,1.7963],[54.675,2.5404],[55.6875,4.2299],[56.7,6.2181],[57.7125,6.7858],[58.725,6.0722],[59.7375,5.1793],[60.75,4.5004],[61.7625,3.9944],[62.775,3.4746],[63.7875,3.0676],[64.8,2.7526],[65.8125,2.2734],[66.825,1.6711],[67.8375,1.2243],[68.85,0.8822],[69.8625,0.6802],[70.875,0.6924],[71.8875,0.8833],[72.9,1.2746],[73.9125,2.166],[74.925,3.2382],[75.9375,4.0834],[76.95,4.6848],[77.9625,4.8699],[78.975,4.844],[79.9875,4.7937]]
       }
       chest_ctrl.Rotation.Z {
        type=MovieSceneScriptingFloatChannel
        default=18.3838
        thinned_from=80
        keys [[0,18.3838],[1.0125,18.0914],[2.025,17.4035],[3.0375,16.5938],[4.05,15.9124],[5.0625,15.3726],[6.075,14.8823],[7.0875,14.4623],[8.1,14.2071],[9.1125,14.1491],[10.125,14.1352],[11.1375,14.0781],[12.15,13.9217],[13.1625,13.5987],[14.175,13.1252],[15.1875,13.0235],[16.2,13.4716],[17.2125,14.1256],[18.225,14.6892],[19.2375,15.0709],[20.25,15.294],[21.2625,15.4043],[22.275,15.4723],[23.2875,15.6132],[24.3,15.8119],[25.3125,15.9052],[26.325,15.7331],[27.3375,15.3018],[28.35,14.9288],[29.3625,14.7583],[30.375,14.7166],[31.3875,14.7813],[33.4125,15.026],[34.425,15.0816],[36.45,15.1226],[37.4625,15.2021],[38.475,15.3292],[39.4875,15.5379],[41.5125,16.6946],[42.525,16.7129],[43.5375,16.5609],[44.55,16.6292],[45.5625,16.9695],[48.6,18.2289],[49.6125,18.504],[50.625,18.5363],[51.6375,18.3668],[53.6625,17.8803],[54.675,17.6664],[55.6875,17.1584],[56.7,16.8006],[57.7125,17.031],[58.725,17.3079],[60.75,17.6915],[61.7625,17.9188],[64.8,18.7902],[65.8125,19.0314],[66.825,19.0992],[69.8625,19.0102],[70.875,19.0756],[72.9,19.2972],[73.9125,19.313],[76.95,19.0043],[77.9625,18.7893],[78.975,18.5187],[79.9875,18.3838]]
       }
       chest_ctrl.Scale.X {
        type=MovieSceneScriptingFloatChannel
        default=0.9981
        thinned_from=80
        keys [[0,0.9981],[11.1375,0.9981],[14.175,0.9939],[17.2125,1.0023],[19.2375,1.0027],[24.3,0.995],[40.5,0.9983],[44.55,0.9961],[79.9875,0.9981]]
       }
       chest_ctrl.Scale.Y {type=MovieSceneScriptingFloatChannel, default=1, const [1,80]}
       chest_ctrl.Scale.Z {type=MovieSceneScriptingFloatChannel, default=1, const [1,80]}
       chest_tan_ctrl.Location.X {
        type=MovieSceneScriptingFloatChannel
        default=-0.3261
        thinned_from=80
        keys [[0,-0.3261],[1.0125,-0.3464],[2.025,-0.3859],[3.0375,-0.4082],[4.05,-0.3841],[5.0625,-0.312],[6.075,-0.2354],[7.0875,-0.1709],[8.1,-0.1314],[9.1125,-0.1204],[10.125,-0.119],[11.1375,-0.1272],[12.15,-0.1484],[13.1625,-0.1927],[14.175,-0.2608],[15.1875,-0.366],[16.2,-0.458],[18.225,-0.5643],[19.2375,-0.6115],[20.25,-0.6487],[21.2625,-0.6747],[22.275,-0.6836],[23.2875,-0.6574],[24.3,-0.5895],[25.3125,-0.4895],[26.325,-0.3664],[27.3375,-0.2337],[28.35,-0.1171],[29.3625,-0.0322],[30.375,0.0236],[31.3875,0.0565],[32.4,0.0713],[33.4125,0.0742],[34.425,0.0717],[35.4375,0.066],[36.45,0.0566],[37.4625,0.0433],[38.475,0.0267],[39.4875,-0.0046],[40.5,-0.0487],[41.5125,-0.0313],[42.525,0.0399],[43.5375,0.0912],[44.55,0.1046],[45.5625,0.0831],[46.575,0.0458],[47.5875,0.007],[48.6,-0.0307],[49.6125,-0.061],[50.625,-0.075],[51.6375,-0.0774],[52.65,-0.0807],[53.6625,-0.0972],[54.675,-0.1482],[55.6875,-0.274],[56.7,-0.4225],[57.7125,-0.4617],[58.725,-0.4031],[59.7375,-0.3332],[60.75,-0.2824],[61.7625,-0.2462],[62.775,-0.2089],[63.7875,-0.1792],[64.8,-0.1552],[65.8125,-0.1174],[66.825,-0.0703],[67.8375,-0.0365],[68.85,-0.0119],[69.8625,0.0011],[70.875,-0.0022],[71.8875,-0.0186],[72.9,-0.0496],[73.9125,-0.1191],[74.925,-0.2029],[75.9375,-0.269],[76.95,-0.3161],[77.9625,-0.331],[78.975,-0.3297],[79.9875,-0.3261]]
       }
       chest_tan_ctrl.Location.Y {
        type=MovieSceneScriptingFloatChannel
        default=-0.2484
        thinned_from=80
        keys [[0,-0.2484],[1.0125,-0.2432],[2.025,-0.2243],[3.0375,-0.1886],[4.05,-0.1293],[5.0625,-0.0289],[6.075,0.0873],[7.0875,0.1969],[8.1,0.2841],[9.1125,0.3355],[10.125,0.3684],[11.1375,0.3889],[12.15,0.3939],[13.1625,0.3733],[14.175,0.3305],[16.2,0.2307],[17.2125,0.1704],[18.225,0.0958],[19.2375,0.021],[20.25,-0.0413],[21.2625,-0.078],[22.275,-0.0801],[23.2875,-0.0539],[24.3,-0.0121],[25.3125,0.0384],[26.325,0.091],[27.3375,0.1445],[28.35,0.2059],[29.3625,0.2779],[30.375,0.3577],[31.3875,0.4362],[32.4,0.4986],[33.4125,0.5354],[34.425,0.5497],[35.4375,0.548],[36.45,0.5323],[37.4625,0.5072],[38.475,0.4791],[40.5,0.4398],[41.5125,0.3886],[43.5375,0.2317],[44.55,0.1748],[45.5625,0.1513],[46.575,0.145],[47.5875,0.1445],[48.6,0.1577],[49.6125,0.1842],[50.625,0.2062],[51.6375,0.2081],[52.65,0.1968],[53.6625,0.1896],[54.675,0.1919],[55.6875,0.2002],[56.7,0.2153],[57.7125,0.2388],[58.725,0.2562],[59.7375,0.2471],[60.75,0.2081],[61.7625,0.1101],[62.775,-0.002],[63.7875,-0.0676],[64.8,-0.1127],[65.8125,-0.1413],[66.825,-0.151],[67.8375,-0.1142],[68.85,-0.0612],[69.8625,-0.0316],[70.875,-0.0121],[71.8875,-0.0015],[72.9,-0.0122],[73.9125,-0.0474],[74.925,-0.1174],[75.9375,-0.1864],[76.95,-0.2165],[77.9625,-0.2344],[78.975,-0.2449],[79.9875,-0.2484]]
       }
       chest_tan_ctrl.Location.Z {
        type=MovieSceneScriptingFloatChannel
        default=0.2358
        thinned_from=80
        keys [[0,0.2358],[1.0125,0.2369],[2.025,0.2398],[3.0375,0.2439],[4.05,0.2488],[5.0625,0.2567],[6.075,0.2672],[7.0875,0.2786],[8.1,0.2886],[9.1125,0.295],[10.125,0.2993],[11.1375,0.3013],[12.15,0.2956],[13.1625,0.2831],[14.175,0.2747],[15.1875,0.2806],[16.2,0.2959],[17.2125,0.3075],[18.225,0.3075],[19.2375,0.2981],[21.2625,0.2748],[22.275,0.2654],[23.2875,0.2583],[24.3,0.2539],[25.3125,0.255],[26.325,0.2632],[27.3375,0.274],[28.35,0.2824],[29.3625,0.2899],[31.3875,0.3081],[32.4,0.3155],[33.4125,0.3196],[34.425,0.3207],[35.4375,0.3198],[37.4625,0.3147],[38.475,0.3139],[39.4875,0.3147],[40.5,0.3091],[41.5125,0.2944],[42.525,0.279],[43.5375,0.268],[44.55,0.2634],[45.5625,0.264],[48.6,0.2721],[49.6125,0.2751],[50.625,0.2773],[51.6375,0.2772],[52.65,0.2757],[53.6625,0.2749],[54.675,0.2753],[55.6875,0.2775],[56.7,0.282],[57.7125,0.2851],[58.725,0.2847],[59.7375,0.2815],[60.75,0.2757],[61.7625,0.2641],[62.775,0.252],[63.7875,0.2454],[64.8,0.2411],[65.8125,0.2385],[66.825,0.238],[67.8375,0.2421],[68.85,0.2478],[69.8625,0.251],[70.875,0.2529],[71.8875,0.2541],[72.9,0.2531],[73.9125,0.2502],[74.925,0.2447],[75.9375,0.2398],[77.9625,0.2371],[78.975,0.2362],[79.9875,0.2358]]
       }
       chest_tan_ctrl.Rotation.X {
        type=MovieSceneScriptingFloatChannel
        default=1.7583
        thinned_from=80
        keys [[0,1.7583],[1.0125,3.3753],[2.025,3.2],[3.0375,3.1347],[4.05,3.3783],[5.0625,4.0049],[6.075,4.7159],[7.0875,5.3625],[8.1,5.8482],[9.1125,6.0964],[10.125,6.2285],[11.1375,6.2755],[12.15,6.1944],[13.1625,6.0333],[14.175,5.8845],[16.2,5.6669],[17.2125,5.388],[18.225,4.8642],[19.2375,4.2672],[20.25,3.7201],[21.2625,3.3319],[22.275,3.1901],[23.2875,3.3071],[24.3,3.6196],[25.3125,4.0784],[26.325,4.6355],[27.3375,5.2457],[28.35,5.8833],[30.375,7.2062],[31.3875,7.8187],[32.4,8.2793],[33.4125,8.5314],[34.425,8.6094],[35.4375,8.5651],[36.45,8.4146],[37.4625,8.1918],[38.475,7.9411],[39.4875,7.6075],[40.5,6.7345],[41.5125,5.7605],[42.525,5.242],[43.5375,4.8755],[44.55,4.6675],[45.5625,4.6745],[46.575,4.8014],[47.5875,4.9777],[48.6,5.2376],[49.6125,5.5739],[50.625,5.881],[51.6375,6.056],[52.65,6.1196],[53.6625,6.1539],[54.675,6.2619],[55.6875,6.4607],[56.7,6.7393],[57.7125,7.0916],[58.725,7.3539],[59.7375,7.3974],[60.75,7.2108],[61.7625,6.639],[62.775,5.9637],[63.7875,5.5439],[64.8,5.2264],[65.8125,4.9823],[66.825,4.7854],[67.8375,4.7956],[68.85,4.9109],[69.8625,4.9819],[70.875,5.0709],[71.8875,5.114],[72.9,5.0284],[73.9125,4.7783],[74.925,4.3155],[75.9375,3.8702],[76.95,3.6637],[77.9625,3.5514],[78.975,3.4939],[79.9875,3.4774]]
       }
       chest_tan_ctrl.Rotation.Y {
        type=MovieSceneScriptingFloatChannel
        default=0
        thinned_from=80
        keys [[0,0],[1.0125,2.6121],[2.025,3.1689],[3.0375,3.588],[4.05,3.4577],[5.0625,2.8711],[6.075,2.3078],[7.0875,1.8556],[8.1,1.5842],[9.1125,1.5047],[10.125,1.4957],[11.1375,1.5499],[12.15,1.6929],[13.1625,2.0336],[14.175,2.5863],[15.1875,3.405],[16.2,4.1354],[17.2125,4.5436],[18.225,4.8533],[19.2375,5.0856],[20.25,5.232],[21.2625,5.2988],[22.275,5.2608],[23.2875,5.0134],[24.3,4.5176],[25.3125,3.8302],[26.325,3.0022],[27.3375,2.1204],[28.35,1.3593],[29.3625,0.8167],[30.375,0.4698],[31.3875,0.2777],[32.4,0.2024],[33.4125,0.1993],[34.425,0.2238],[35.4375,0.2611],[36.45,0.3161],[37.4625,0.3909],[38.475,0.4821],[39.4875,0.5918],[40.5,0.5037],[41.5125,0.0632],[42.525,0],[47.5875,0],[48.6,0.2914],[49.6125,0.5895],[50.625,0.7857],[52.65,1.019],[53.6625,1.2229],[54.675,1.6233],[55.6875,2.4583],[56.7,3.4435],[57.7125,3.7795],[58.725,3.5022],[59.7375,3.0997],[60.75,2.7547],[61.7625,2.441],[62.775,2.1101],[63.7875,1.8647],[64.8,1.6814],[65.8125,1.4369],[66.825,1.1383],[67.8375,0.9263],[68.85,0.7674],[69.8625,0.6636],[70.875,0.6628],[71.8875,0.7452],[72.9,0.9121],[73.9125,1.2932],[74.925,1.7332],[75.9375,2.0714],[76.95,2.3242],[77.9625,2.3951],[78.975,2.3745],[79.9875,2.3484]]
       }
       chest_tan_ctrl.Rotation.Z {
        type=MovieSceneScriptingFloatChannel
        default=0
        thinned_from=80
        keys [[0,0],[1.0125,12.0705],[2.025,11.5962],[3.0375,11.0333],[4.05,10.5591],[5.0625,10.1826],[6.075,9.8458],[7.0875,9.5647],[8.1,9.3951],[9.1125,9.3529],[10.125,9.3385],[11.1375,9.292],[12.15,9.1737],[13.1625,8.9381],[14.175,8.5972],[15.1875,8.4869],[16.2,8.7492],[17.2125,9.1746],[18.225,9.5536],[19.2375,9.8236],[20.25,9.9935],[21.2625,10.0833],[22.275,10.1324],[23.2875,10.2256],[24.3,10.3641],[25.3125,10.442],[26.325,10.3554],[27.3375,10.106],[28.35,9.8954],[29.3625,9.813],[30.375,9.8097],[31.3875,9.8708],[33.4125,10.0467],[34.425,10.082],[36.45,10.0975],[37.4625,10.1406],[38.475,10.2135],[39.4875,10.3281],[40.5,10.6715],[41.5125,11.0493],[42.525,11.1006],[43.5375,11.0267],[44.55,11.0824],[45.5625,11.3077],[47.5875,11.8656],[48.6,12.1255],[49.6125,12.2996],[50.625,12.3169],[51.6375,12.2063],[54.675,11.7241],[55.6875,11.3381],[56.7,11.0352],[57.7125,11.1625],[58.725,11.3683],[60.75,11.6924],[61.7625,11.879],[64.8,12.5208],[65.8125,12.6911],[66.825,12.7434],[69.8625,12.6852],[70.875,12.7269],[72.9,12.8652],[73.9125,12.8669],[76.95,12.6759],[77.9625,12.537],[78.975,12.3594],[79.9875,12.2703]]
       }
       chest_tan_ctrl.Scale.X {
        type=MovieSceneScriptingFloatChannel
        default=1
        thinned_from=80
        keys [[0,1],[11.1375,1.0019],[14.175,1.0062],[16.2,1.0009],[21.2625,1.0007],[24.3,1.005],[40.5,1.0017],[44.55,1.0039],[79.9875,1.0019]]
       }
       chest_tan_ctrl.Scale.Y {type=MovieSceneScriptingFloatChannel, default=1, const [1,80]}
       chest_tan_ctrl.Scale.Z {type=MovieSceneScriptingFloatChannel, default=1, const [1,80]}
      ]
     }
     {
      type=MovieSceneControlRigParameterSection
      channels [
       global_ctrl.Location.X {type=MovieSceneScriptingFloatChannel, default=0, const [0,4]}
       global_ctrl.Location.Y {type=MovieSceneScriptingFloatChannel, default=0, const [0,4]}
       global_ctrl.Location.Z {type=MovieSceneScriptingFloatChannel, default=0, const [0,4]}
       global_ctrl.Rotation.X {type=MovieSceneScriptingFloatChannel, default=0, const [0,4]}
       global_ctrl.Rotation.Y {type=MovieSceneScriptingFloatChannel, default=0, const [0,4]}
       global_ctrl.Rotation.Z {type=MovieSceneScriptingFloatChannel, default=0, const [0,4]}
       root_ctrl.Location.X {type=MovieSceneScriptingFloatChannel, default=0, const [0,4]}
       root_ctrl.Location.Y {type=MovieSceneScriptingFloatChannel, default=0, const [0,4]}
       root_ctrl.Location.Z {type=MovieSceneScriptingFloatChannel, default=0, const [0,4]}
       root_ctrl.Rotation.X {type=MovieSceneScriptingFloatChannel, default=0, const [0,4]}
       root_ctrl.Rotation.Y {type=MovieSceneScriptingFloatChannel, default=0, const [0,4]}
       root_ctrl.Rotation.Z {type=MovieSceneScriptingFloatChannel, default=0, const [0,4]}
       root_ctrl.Scale.X {type=MovieSceneScriptingFloatChannel, default=0, const [0,4]}
       root_ctrl.Scale.Y {type=MovieSceneScriptingFloatChannel, default=0, const [0,4]}
       root_ctrl.Scale.Z {type=MovieSceneScriptingFloatChannel, default=0, const [0,4]}
       body_offset_ctrl.Location.X {type=MovieSceneScriptingFloatChannel, default=0, const [0,4]}
       body_offset_ctrl.Location.Y {type=MovieSceneScriptingFloatChannel, default=0, const [0,4]}
       body_offset_ctrl.Location.Z {type=MovieSceneScriptingFloatChannel, default=0, const [0,4]}
       body_offset_ctrl.Rotation.X {type=MovieSceneScriptingFloatChannel, default=0, const [0,4]}
       body_offset_ctrl.Rotation.Y {type=MovieSceneScriptingFloatChannel, default=0, const [0,4]}
       body_offset_ctrl.Rotation.Z {type=MovieSceneScriptingFloatChannel, default=0, const [0,4]}
       body_offset_ctrl.Scale.X {type=MovieSceneScriptingFloatChannel, default=0, const [0,4]}
       body_offset_ctrl.Scale.Y {type=MovieSceneScriptingFloatChannel, default=0, const [0,4]}
       body_offset_ctrl.Scale.Z {type=MovieSceneScriptingFloatChannel, default=0, const [0,4]}
       thigh_l_fk_ctrl.Location.X {type=MovieSceneScriptingFloatChannel, default=0, thinned_from=4, keys [[7,0],[52,0]]}
       thigh_l_fk_ctrl.Location.Y {type=MovieSceneScriptingFloatChannel, default=0, thinned_from=4, keys [[7,0],[52,0]]}
       thigh_l_fk_ctrl.Location.Z {type=MovieSceneScriptingFloatChannel, default=0, thinned_from=4, keys [[7,0],[52,0]]}
       thigh_l_fk_ctrl.Rotation.X {type=MovieSceneScriptingFloatChannel, default=0, keys [[7,0.8642],[12,0.9341],[44,0.9341],[52,0.8925]]}
       thigh_l_fk_ctrl.Rotation.Y {type=MovieSceneScriptingFloatChannel, default=0, keys [[7,-0.88],[12,-0.9251],[44,-0.9251],[52,-0.9041]]}
       thigh_l_fk_ctrl.Rotation.Z {type=MovieSceneScriptingFloatChannel, default=0, keys [[7,-0.4709],[12,-0.4564],[44,-0.4564],[52,-0.4713]]}
       thigh_l_fk_ctrl.Scale.X {type=MovieSceneScriptingFloatChannel, default=0, const [0,4]}
       thigh_l_fk_ctrl.Scale.Y {type=MovieSceneScriptingFloatChannel, default=0, const [0,4]}
       thigh_l_fk_ctrl.Scale.Z {type=MovieSceneScriptingFloatChannel, default=0, const [0,4]}
       calf_l_fk_ctrl.Location.X {type=MovieSceneScriptingFloatChannel, default=0, keys [[7,-0.0865],[12,-0.0673],[44,-0.0673],[52,-0.0783]]}
       calf_l_fk_ctrl.Location.Y {type=MovieSceneScriptingFloatChannel, default=0, keys [[7,-0.0015],[12,0.0018],[44,0.0018],[52,0.0001]]}
       calf_l_fk_ctrl.Location.Z {type=MovieSceneScriptingFloatChannel, default=0, keys [[7,-0.003],[12,-0.0024],[44,-0.0024],[52,-0.0027]]}
       calf_l_fk_ctrl.Rotation.X {type=MovieSceneScriptingFloatChannel, default=0, keys [[7,-0.2648],[12,-0.2014],[44,-0.2014],[52,-0.2378]]}
       calf_l_fk_ctrl.Rotation.Y {type=MovieSceneScriptingFloatChannel, default=0, keys [[7,0.9322],[12,0.9941],[44,0.9941],[52,0.9584]]}
       calf_l_fk_ctrl.Rotation.Z {type=MovieSceneScriptingFloatChannel, default=0, keys [[7,0.634],[12,0.6274],[44,0.6274],[52,0.6538]]}
       calf_l_fk_ctrl.Scale.X {type=MovieSceneScriptingFloatChannel, default=0, const [0,4]}
       calf_l_fk_ctrl.Scale.Y {type=MovieSceneScriptingFloatChannel, default=0, const [0,4]}
       calf_l_fk_ctrl.Scale.Z {type=MovieSceneScriptingFloatChannel, default=0, const [0,4]}
       foot_l_fk_ctrl.Location.X {type=MovieSceneScriptingFloatChannel, default=0, thinned_from=4, keys [[7,0],[52,0]]}
       foot_l_fk_ctrl.Location.Y {type=MovieSceneScriptingFloatChannel, default=0, thinned_from=4, keys [[7,0.0398],[52,0.0398]]}
       foot_l_fk_ctrl.Location.Z {type=MovieSceneScriptingFloatChannel, default=0, thinned_from=4, keys [[7,-0.8885],[52,-0.8885]]}
       foot_l_fk_ctrl.Rotation.X {type=MovieSceneScriptingFloatChannel, default=0, keys [[7,-0.14],[12,-0.325],[44,-0.325],[52,-0.2165]]}
       foot_l_fk_ctrl.Rotation.Y {type=MovieSceneScriptingFloatChannel, default=0, keys [[7,0.341],[12,0.3297],[44,0.3297],[52,0.335]]}
       foot_l_fk_ctrl.Rotation.Z {type=MovieSceneScriptingFloatChannel, default=0, keys [[7,-0.3234],[12,-0.3042],[44,-0.3042],[52,-0.3246]]}
       foot_l_fk_ctrl.Scale.X {type=MovieSceneScriptingFloatChannel, default=0, thinned_from=4, keys [[7,0],[52,0]]}
       foot_l_fk_ctrl.Scale.Y {type=MovieSceneScriptingFloatChannel, default=0, const [0,4]}
       foot_l_fk_ctrl.Scale.Z {type=MovieSceneScriptingFloatChannel, default=0, const [0,4]}
       ball_l_fk_ctrl.Location.X {type=MovieSceneScriptingFloatChannel, default=0, thinned_from=4, keys [[7,0],[52,0]]}
       ball_l_fk_ctrl.Location.Y {type=MovieSceneScriptingFloatChannel, default=0, thinned_from=4, keys [[7,0],[52,0]]}
       ball_l_fk_ctrl.Location.Z {type=MovieSceneScriptingFloatChannel, default=0, thinned_from=4, keys [[7,0],[52,0]]}
       ball_l_fk_ctrl.Rotation.X {type=MovieSceneScriptingFloatChannel, default=0, thinned_from=4, keys [[7,0],[52,0]]}
       ball_l_fk_ctrl.Rotation.Y {type=MovieSceneScriptingFloatChannel, default=0, thinned_from=4, keys [[7,0],[52,0]]}
       ball_l_fk_ctrl.Rotation.Z {type=MovieSceneScriptingFloatChannel, default=0, thinned_from=4, keys [[7,0],[52,0]]}
       ball_l_fk_ctrl.Scale.X {type=MovieSceneScriptingFloatChannel, default=0, const [0,4]}
       ball_l_fk_ctrl.Scale.Y {type=MovieSceneScriptingFloatChannel, default=0, const [0,4]}
       ball_l_fk_ctrl.Scale.Z {type=MovieSceneScriptingFloatChannel, default=0, const [0,4]}
       foot_l_ik_ctrl.Location.X {type=MovieSceneScriptingFloatChannel, default=0, const [0,4]}
       foot_l_ik_ctrl.Location.Y {type=MovieSceneScriptingFloatChannel, default=0, const [0,4]}
       foot_l_ik_ctrl.Location.Z {type=MovieSceneScriptingFloatChannel, default=0, const [0,4]}
       foot_l_ik_ctrl.Rotation.X {type=MovieSceneScriptingFloatChannel, default=0, const [0,4]}
       foot_l_ik_ctrl.Rotation.Y {type=MovieSceneScriptingFloatChannel, default=0, const [0,4]}
       foot_l_ik_ctrl.Rotation.Z {type=MovieSceneScriptingFloatChannel, default=0, const [0,4]}
       foot_l_ik_ctrl.Scale.X {type=MovieSceneScriptingFloatChannel, default=0, const [0,4]}
       foot_l_ik_ctrl.Scale.Y {type=MovieSceneScriptingFloatChannel, default=0, const [0,4]}
       foot_l_ik_ctrl.Scale.Z {type=MovieSceneScriptingFloatChannel, default=0, const [0,4]}
       ball_l_ik_ctrl.Location.X {type=MovieSceneScriptingFloatChannel, default=0, const [0,4]}
       ball_l_ik_ctrl.Location.Y {type=MovieSceneScriptingFloatChannel, default=0, const [0,4]}
       ball_l_ik_ctrl.Location.Z {type=MovieSceneScriptingFloatChannel, default=0, const [0,4]}
       ball_l_ik_ctrl.Rotation.X {type=MovieSceneScriptingFloatChannel, default=0, const [0,4]}
       ball_l_ik_ctrl.Rotation.Y {type=MovieSceneScriptingFloatChannel, default=0, const [0,4]}
       ball_l_ik_ctrl.Rotation.Z {type=MovieSceneScriptingFloatChannel, default=0, const [0,4]}
       ball_l_ik_ctrl.Scale.X {type=MovieSceneScriptingFloatChannel, default=0, const [0,4]}
       ball_l_ik_ctrl.Scale.Y {type=MovieSceneScriptingFloatChannel, default=0, const [0,4]}
       ball_l_ik_ctrl.Scale.Z {type=MovieSceneScriptingFloatChannel, default=0, const [0,4]}
       foot_bk1_l_ctrl.Location.X {type=MovieSceneScriptingFloatChannel, default=0, const [0,4]}
       foot_bk1_l_ctrl.Location.Y {type=MovieSceneScriptingFloatChannel, default=0, const [0,4]}
       foot_bk1_l_ctrl.Location.Z {type=MovieSceneScriptingFloatChannel, default=0, const [0,4]}
       foot_bk1_l_ctrl.Rotation.X {type=MovieSceneScriptingFloatChannel, default=0, const [0,4]}
       foot_bk1_l_ctrl.Rotation.Y {type=MovieSceneScriptingFloatChannel, default=0, const [0,4]}
       foot_bk1_l_ctrl.Rotation.Z {type=MovieSceneScriptingFloatChannel, default=0, const [0,4]}
       foot_bk1_l_ctrl.Scale.X {type=MovieSceneScriptingFloatChannel, default=0, const [0,4]}
       foot_bk1_l_ctrl.Scale.Y {type=MovieSceneScriptingFloatChannel, default=0, const [0,4]}
       foot_bk1_l_ctrl.Scale.Z {type=MovieSceneScriptingFloatChannel, default=0, const [0,4]}
       foot_roll_l_ctrl.Location.X {type=MovieSceneScriptingFloatChannel, default=0, const [0,4]}
       foot_roll_l_ctrl.Location.Y {type=MovieSceneScriptingFloatChannel, default=0, const [0,4]}
       foot_roll_l_ctrl.Location.Z {type=MovieSceneScriptingFloatChannel, default=0, const [0,4]}
       foot_roll_l_ctrl.Rotation.X {type=MovieSceneScriptingFloatChannel, default=0, const [0,4]}
       foot_roll_l_ctrl.Rotation.Y {type=MovieSceneScriptingFloatChannel, default=0, const [0,4]}
       foot_roll_l_ctrl.Rotation.Z {type=MovieSceneScriptingFloatChannel, default=0, const [0,4]}
       foot_roll_l_ctrl.Scale.X {type=MovieSceneScriptingFloatChannel, default=0, const [0,4]}
       foot_roll_l_ctrl.Scale.Y {type=MovieSceneScriptingFloatChannel, default=0, const [0,4]}
       foot_roll_l_ctrl.Scale.Z {type=MovieSceneScriptingFloatChannel, default=0, const [0,4]}
       roll_blend_l_ctrl {type=MovieSceneScriptingFloatChannel, default=0, const [0,4]}
       tip_l_ctrl.Location.X {type=MovieSceneScriptingFloatChannel, default=0, const [0,4]}
       tip_l_ctrl.Location.Y {type=MovieSceneScriptingFloatChannel, default=0, const [0,4]}
       tip_l_ctrl.Location.Z {type=MovieSceneScriptingFloatChannel, default=0, const [0,4]}
       tip_l_ctrl.Rotation.X {type=MovieSceneScriptingFloatChannel, default=0, const [0,4]}
       tip_l_ctrl.Rotation.Y {type=MovieSceneScriptingFloatChannel, default=0, const [0,4]}
       tip_l_ctrl.Rotation.Z {type=MovieSceneScriptingFloatChannel, default=0, const [0,4]}
       tip_l_ctrl.Scale.X {type=MovieSceneScriptingFloatChannel, default=0, const [0,4]}
       tip_l_ctrl.Scale.Y {type=MovieSceneScriptingFloatChannel, default=0, const [0,4]}
       tip_l_ctrl.Scale.Z {type=MovieSceneScriptingFloatChannel, default=0, const [0,4]}
       heel_l_ctrl.Location.X {type=MovieSceneScriptingFloatChannel, default=0, const [0,4]}
       heel_l_ctrl.Location.Y {type=MovieSceneScriptingFloatChannel, default=0, const [0,4]}
       heel_l_ctrl.Location.Z {type=MovieSceneScriptingFloatChannel, default=0, const [0,4]}
       heel_l_ctrl.Rotation.X {type=MovieSceneScriptingFloatChannel, default=0, const [0,4]}
       heel_l_ctrl.Rotation.Y {type=MovieSceneScriptingFloatChannel, default=0, const [0,4]}
       heel_l_ctrl.Rotation.Z {type=MovieSceneScriptingFloatChannel, default=0, const [0,4]}
       heel_l_ctrl.Scale.X {type=MovieSceneScriptingFloatChannel, default=0, const [0,4]}
       heel_l_ctrl.Scale.Y {type=MovieSceneScriptingFloatChannel, default=0, const [0,4]}
       heel_l_ctrl.Scale.Z {type=MovieSceneScriptingFloatChannel, default=0, const [0,4]}
       leg_l_softness {type=MovieSceneScriptingFloatChannel, default=0, const [0,4]}
       leg_l_pv_ik_ctrl.Location.X {type=MovieSceneScriptingFloatChannel, default=0, const [0,4]}
       leg_l_pv_ik_ctrl.Location.Y {type=MovieSceneScriptingFloatChannel, default=0, const [0,4]}
       leg_l_pv_ik_ctrl.Location.Z {type=MovieSceneScriptingFloatChannel, default=0, const [0,4]}
       leg_l_pv_ik_ctrl.Rotation.X {type=MovieSceneScriptingFloatChannel, default=0, const [0,4]}
       leg_l_pv_ik_ctrl.Rotation.Y {type=MovieSceneScriptingFloatChannel, default=0, const [0,4]}
       leg_l_pv_ik_ctrl.Rotation.Z {type=MovieSceneScriptingFloatChannel, default=0, const [0,4]}
       leg_l_pv_ik_ctrl.Scale.X {type=MovieSceneScriptingFloatChannel, default=0, const [0,4]}
       leg_l_pv_ik_ctrl.Scale.Y {type=MovieSceneScriptingFloatChannel, default=0, const [0,4]}
       leg_l_pv_ik_ctrl.Scale.Z {type=MovieSceneScriptingFloatChannel, default=0, const [0,4]}
       thigh_r_fk_ctrl.Location.X {type=MovieSceneScriptingFloatChannel, default=0, thinned_from=4, keys [[7,0],[52,0]]}
       thigh_r_fk_ctrl.Location.Y {type=MovieSceneScriptingFloatChannel, default=0, thinned_from=4, keys [[7,0],[52,0]]}
       thigh_r_fk_ctrl.Location.Z {type=MovieSceneScriptingFloatChannel, default=0, thinned_from=4, keys [[7,0],[52,0]]}
       thigh_r_fk_ctrl.Rotation.X {type=MovieSceneScriptingFloatChannel, default=0, keys [[7,0.8283],[12,0.8112],[44,0.8112],[52,0.8333]]}
       thigh_r_fk_ctrl.Rotation.Y {type=MovieSceneScriptingFloatChannel, default=0, thinned_from=4, keys [[7,-0.8127],[12,-0.8068],[52,-0.8071]]}
       thigh_r_fk_ctrl.Rotation.Z {type=MovieSceneScriptingFloatChannel, default=0, keys [[7,-0.5582],[12,-0.5882],[44,-0.5882],[52,-0.5724]]}
       thigh_r_fk_ctrl.Scale.X {type=MovieSceneScriptingFloatChannel, default=0, const [0,4]}
       thigh_r_fk_ctrl.Scale.Y {type=MovieSceneScriptingFloatChannel, default=0, const [0,4]}
       thigh_r_fk_ctrl.Scale.Z {type=MovieSceneScriptingFloatChannel, default=0, const [0,4]}
       calf_r_fk_ctrl.Location.X {type=MovieSceneScriptingFloatChannel, default=0, keys [[7,-0.0936],[12,-0.0907],[44,-0.0907],[52,-0.0883]]}
       calf_r_fk_ctrl.Location.Y {type=MovieSceneScriptingFloatChannel, default=0, keys [[7,-0.0031],[12,-0.0024],[44,-0.0024],[52,-0.0018]]}
       calf_r_fk_ctrl.Location.Z {type=MovieSceneScriptingFloatChannel, default=0, thinned_from=4, keys [[7,-0.0032],[52,-0.003]]}
       calf_r_fk_ctrl.Rotation.X {type=MovieSceneScriptingFloatChannel, default=0, keys [[7,-0.2852],[12,-0.2764],[44,-0.2764],[52,-0.2686]]}
       calf_r_fk_ctrl.Rotation.Y {type=MovieSceneScriptingFloatChannel, default=0, keys [[7,0.8855],[12,0.8946],[44,0.8946],[52,0.9019]]}
       calf_r_fk_ctrl.Rotation.Z {type=MovieSceneScriptingFloatChannel, default=0, keys [[7,0.5382],[12,0.5334],[44,0.5334],[52,0.5534]]}
       calf_r_fk_ctrl.Scale.X {type=MovieSceneScriptingFloatChannel, default=0, const [0,4]}
       calf_r_fk_ctrl.Scale.Y {type=MovieSceneScriptingFloatChannel, default=0, const [0,4]}
       calf_r_fk_ctrl.Scale.Z {type=MovieSceneScriptingFloatChannel, default=0, const [0,4]}
       foot_r_fk_ctrl.Location.X {type=MovieSceneScriptingFloatChannel, default=0, thinned_from=4, keys [[7,0],[52,0]]}
       foot_r_fk_ctrl.Location.Y {type=MovieSceneScriptingFloatChannel, default=0, thinned_from=4, keys [[7,-0.0329],[52,-0.0329]]}
       foot_r_fk_ctrl.Location.Z {type=MovieSceneScriptingFloatChannel, default=0, thinned_from=4, keys [[7,-0.9075],[52,-0.9075]]}
       foot_r_fk_ctrl.Rotation.X {type=MovieSceneScriptingFloatChannel, default=0, keys [[7,-0.1971],[12,-0.2238],[44,-0.2238],[52,-0.2402]]}
       foot_r_fk_ctrl.Rotation.Y {type=MovieSceneScriptingFloatChannel, default=0, keys [[7,0.3259],[12,0.3127],[44,0.3127],[52,0.3047]]}
       foot_r_fk_ctrl.Rotation.Z {type=MovieSceneScriptingFloatChannel, default=0, keys [[7,-0.2049],[12,-0.1958],[44,-0.1958],[52,-0.2075]]}
       foot_r_fk_ctrl.Scale.X {type=MovieSceneScriptingFloatChannel, default=0, thinned_from=4, keys [[7,0],[52,0]]}
       foot_r_fk_ctrl.Scale.Y {type=MovieSceneScriptingFloatChannel, default=0, const [0,4]}
       foot_r_fk_ctrl.Scale.Z {type=MovieSceneScriptingFloatChannel, default=0, const [0,4]}
       ball_r_fk_ctrl.Location.X {type=MovieSceneScriptingFloatChannel, default=0, const [0,4]}
       ball_r_fk_ctrl.Location.Y {type=MovieSceneScriptingFloatChannel, default=0, const [0,4]}
       ball_r_fk_ctrl.Location.Z {type=MovieSceneScriptingFloatChannel, default=0, const [0,4]}
       ball_r_fk_ctrl.Rotation.X {type=MovieSceneScriptingFloatChannel, default=0, const [0,4]}
       ball_r_fk_ctrl.Rotation.Y {type=MovieSceneScriptingFloatChannel, default=0, const [0,4]}
       ball_r_fk_ctrl.Rotation.Z {type=MovieSceneScriptingFloatChannel, default=0, const [0,4]}
       ball_r_fk_ctrl.Scale.X {type=MovieSceneScriptingFloatChannel, default=0, const [0,4]}
       ball_r_fk_ctrl.Scale.Y {type=MovieSceneScriptingFloatChannel, default=0, const [0,4]}
       ball_r_fk_ctrl.Scale.Z {type=MovieSceneScriptingFloatChannel, default=0, const [0,4]}
       foot_r_ik_ctrl.Location.X {type=MovieSceneScriptingFloatChannel, default=0, thinned_from=4, keys [[7,-0.0001],[52,-0.0001]]}
       foot_r_ik_ctrl.Location.Y {type=MovieSceneScriptingFloatChannel, default=0, thinned_from=4, keys [[7,0],[52,-0.0001]]}
       foot_r_ik_ctrl.Location.Z {type=MovieSceneScriptingFloatChannel, default=0, thinned_from=4, keys [[7,0],[52,0]]}
       foot_r_ik_ctrl.Rotation.X {type=MovieSceneScriptingFloatChannel, default=0, const [0,4]}
       foot_r_ik_ctrl.Rotation.Y {type=MovieSceneScriptingFloatChannel, default=0, const [0,4]}
       foot_r_ik_ctrl.Rotation.Z {type=MovieSceneScriptingFloatChannel, default=0, const [0,4]}
       foot_r_ik_ctrl.Scale.X {type=MovieSceneScriptingFloatChannel, default=0, const [0,4]}
       foot_r_ik_ctrl.Scale.Y {type=MovieSceneScriptingFloatChannel, default=0, const [0,4]}
       foot_r_ik_ctrl.Scale.Z {type=MovieSceneScriptingFloatChannel, default=0, const [0,4]}
       ball_r_ik_ctrl.Location.X {type=MovieSceneScriptingFloatChannel, default=0, const [0,4]}
       ball_r_ik_ctrl.Location.Y {type=MovieSceneScriptingFloatChannel, default=0, const [0,4]}
       ball_r_ik_ctrl.Location.Z {type=MovieSceneScriptingFloatChannel, default=0, const [0,4]}
       ball_r_ik_ctrl.Rotation.X {type=MovieSceneScriptingFloatChannel, default=0, const [0,4]}
       ball_r_ik_ctrl.Rotation.Y {type=MovieSceneScriptingFloatChannel, default=0, const [0,4]}
       ball_r_ik_ctrl.Rotation.Z {type=MovieSceneScriptingFloatChannel, default=0, const [0,4]}
       ball_r_ik_ctrl.Scale.X {type=MovieSceneScriptingFloatChannel, default=0, const [0,4]}
       ball_r_ik_ctrl.Scale.Y {type=MovieSceneScriptingFloatChannel, default=0, const [0,4]}
       ball_r_ik_ctrl.Scale.Z {type=MovieSceneScriptingFloatChannel, default=0, const [0,4]}
       foot_bk1_r_ctrl.Location.X {type=MovieSceneScriptingFloatChannel, default=0, const [0,4]}
       foot_bk1_r_ctrl.Location.Y {type=MovieSceneScriptingFloatChannel, default=0, const [0,4]}
       foot_bk1_r_ctrl.Location.Z {type=MovieSceneScriptingFloatChannel, default=0, const [0,4]}
       foot_bk1_r_ctrl.Rotation.X {type=MovieSceneScriptingFloatChannel, default=0, const [0,4]}
       foot_bk1_r_ctrl.Rotation.Y {type=MovieSceneScriptingFloatChannel, default=0, const [0,4]}
       foot_bk1_r_ctrl.Rotation.Z {type=MovieSceneScriptingFloatChannel, default=0, const [0,4]}
       foot_bk1_r_ctrl.Scale.X {type=MovieSceneScriptingFloatChannel, default=0, const [0,4]}
       foot_bk1_r_ctrl.Scale.Y {type=MovieSceneScriptingFloatChannel, default=0, const [0,4]}
       foot_bk1_r_ctrl.Scale.Z {type=MovieSceneScriptingFloatChannel, default=0, const [0,4]}
       foot_roll_r_ctrl.Location.X {type=MovieSceneScriptingFloatChannel, default=0, const [0,4]}
       foot_roll_r_ctrl.Location.Y {type=MovieSceneScriptingFloatChannel, default=0, const [0,4]}
       foot_roll_r_ctrl.Location.Z {type=MovieSceneScriptingFloatChannel, default=0, const [0,4]}
       foot_roll_r_ctrl.Rotation.X {type=MovieSceneScriptingFloatChannel, default=0, const [0,4]}
       foot_roll_r_ctrl.Rotation.Y {type=MovieSceneScriptingFloatChannel, default=0, const [0,4]}
       foot_roll_r_ctrl.Rotation.Z {type=MovieSceneScriptingFloatChannel, default=0, const [0,4]}
       foot_roll_r_ctrl.Scale.X {type=MovieSceneScriptingFloatChannel, default=0, const [0,4]}
       foot_roll_r_ctrl.Scale.Y {type=MovieSceneScriptingFloatChannel, default=0, const [0,4]}
       foot_roll_r_ctrl.Scale.Z {type=MovieSceneScriptingFloatChannel, default=0, const [0,4]}
       roll_blend_r_ctrl {type=MovieSceneScriptingFloatChannel, default=0, const [0,4]}
       tip_r_ctrl.Location.X {type=MovieSceneScriptingFloatChannel, default=0, const [0,4]}
       tip_r_ctrl.Location.Y {type=MovieSceneScriptingFloatChannel, default=0, const [0,4]}
       tip_r_ctrl.Location.Z {type=MovieSceneScriptingFloatChannel, default=0, const [0,4]}
       tip_r_ctrl.Rotation.X {type=MovieSceneScriptingFloatChannel, default=0, const [0,4]}
       tip_r_ctrl.Rotation.Y {type=MovieSceneScriptingFloatChannel, default=0, const [0,4]}
       tip_r_ctrl.Rotation.Z {type=MovieSceneScriptingFloatChannel, default=0, const [0,4]}
       tip_r_ctrl.Scale.X {type=MovieSceneScriptingFloatChannel, default=0, const [0,4]}
       tip_r_ctrl.Scale.Y {type=MovieSceneScriptingFloatChannel, default=0, const [0,4]}
       tip_r_ctrl.Scale.Z {type=MovieSceneScriptingFloatChannel, default=0, const [0,4]}
       heel_r_ctrl.Location.X {type=MovieSceneScriptingFloatChannel, default=0, const [0,4]}
       heel_r_ctrl.Location.Y {type=MovieSceneScriptingFloatChannel, default=0, const [0,4]}
       heel_r_ctrl.Location.Z {type=MovieSceneScriptingFloatChannel, default=0, const [0,4]}
       heel_r_ctrl.Rotation.X {type=MovieSceneScriptingFloatChannel, default=0, const [0,4]}
       heel_r_ctrl.Rotation.Y {type=MovieSceneScriptingFloatChannel, default=0, const [0,4]}
       heel_r_ctrl.Rotation.Z {type=MovieSceneScriptingFloatChannel, default=0, const [0,4]}
       heel_r_ctrl.Scale.X {type=MovieSceneScriptingFloatChannel, default=0, const [0,4]}
       heel_r_ctrl.Scale.Y {type=MovieSceneScriptingFloatChannel, default=0, const [0,4]}
       heel_r_ctrl.Scale.Z {type=MovieSceneScriptingFloatChannel, default=0, const [0,4]}
       leg_r_softness {type=MovieSceneScriptingFloatChannel, default=0, const [0,4]}
       leg_r_pv_ik_ctrl.Location.X {type=MovieSceneScriptingFloatChannel, default=0, const [0,4]}
       leg_r_pv_ik_ctrl.Location.Y {type=MovieSceneScriptingFloatChannel, default=0, const [0,4]}
       leg_r_pv_ik_ctrl.Location.Z {type=MovieSceneScriptingFloatChannel, default=0, const [0,4]}
       leg_r_pv_ik_ctrl.Rotation.X {type=MovieSceneScriptingFloatChannel, default=0, const [0,4]}
       leg_r_pv_ik_ctrl.Rotation.Y {type=MovieSceneScriptingFloatChannel, default=0, const [0,4]}
       leg_r_pv_ik_ctrl.Rotation.Z {type=MovieSceneScriptingFloatChannel, default=0, const [0,4]}
       leg_r_pv_ik_ctrl.Scale.X {type=MovieSceneScriptingFloatChannel, default=0, const [0,4]}
       leg_r_pv_ik_ctrl.Scale.Y {type=MovieSceneScriptingFloatChannel, default=0, const [0,4]}
       leg_r_pv_ik_ctrl.Scale.Z {type=MovieSceneScriptingFloatChannel, default=0, const [0,4]}
       upperarm_l_fk_ctrl.Location.X {type=MovieSceneScriptingFloatChannel, default=0, const [0,4]}
       upperarm_l_fk_ctrl.Location.Y {type=MovieSceneScriptingFloatChannel, default=0, const [0,4]}
       upperarm_l_fk_ctrl.Location.Z {type=MovieSceneScriptingFloatChannel, default=0, const [0,4]}
       upperarm_l_fk_ctrl.Rotation.X {type=MovieSceneScriptingFloatChannel, default=0, const [0,4]}
       upperarm_l_fk_ctrl.Rotation.Y {type=MovieSceneScriptingFloatChannel, default=0, const [0,4]}
       upperarm_l_fk_ctrl.Rotation.Z {type=MovieSceneScriptingFloatChannel, default=0, const [0,4]}
       upperarm_l_fk_ctrl.Scale.X {type=MovieSceneScriptingFloatChannel, default=0, const [0,4]}
       upperarm_l_fk_ctrl.Scale.Y {type=MovieSceneScriptingFloatChannel, default=0, const [0,4]}
       upperarm_l_fk_ctrl.Scale.Z {type=MovieSceneScriptingFloatChannel, default=0, const [0,4]}
       lowerarm_l_fk_ctrl.Location.X {type=MovieSceneScriptingFloatChannel, default=0, const [0,4]}
       lowerarm_l_fk_ctrl.Location.Y {type=MovieSceneScriptingFloatChannel, default=0, const [0,4]}
       lowerarm_l_fk_ctrl.Location.Z {type=MovieSceneScriptingFloatChannel, default=0, const [0,4]}
       lowerarm_l_fk_ctrl.Rotation.X {type=MovieSceneScriptingFloatChannel, default=0, const [0,4]}
       lowerarm_l_fk_ctrl.Rotation.Y {type=MovieSceneScriptingFloatChannel, default=0, const [0,4]}
       lowerarm_l_fk_ctrl.Rotation.Z {type=MovieSceneScriptingFloatChannel, default=0, const [0,4]}
       lowerarm_l_fk_ctrl.Scale.X {type=MovieSceneScriptingFloatChannel, default=0, const [0,4]}
       lowerarm_l_fk_ctrl.Scale.Y {type=MovieSceneScriptingFloatChannel, default=0, const [0,4]}
       lowerarm_l_fk_ctrl.Scale.Z {type=MovieSceneScriptingFloatChannel, default=0, const [0,4]}
       hand_l_fk_ctrl.Location.X {type=MovieSceneScriptingFloatChannel, default=0, const [0,4]}
       hand_l_fk_ctrl.Location.Y {type=MovieSceneScriptingFloatChannel, default=0, const [0,4]}
       hand_l_fk_ctrl.Location.Z {type=MovieSceneScriptingFloatChannel, default=0, const [0,4]}
       hand_l_fk_ctrl.Rotation.X {type=MovieSceneScriptingFloatChannel, default=0, const [0,4]}
       hand_l_fk_ctrl.Rotation.Y {type=MovieSceneScriptingFloatChannel, default=0, const [0,4]}
       hand_l_fk_ctrl.Rotation.Z {type=MovieSceneScriptingFloatChannel, default=0, const [0,4]}
       hand_l_fk_ctrl.Scale.X {type=MovieSceneScriptingFloatChannel, default=0, const [0,4]}
       hand_l_fk_ctrl.Scale.Y {type=MovieSceneScriptingFloatChannel, default=0, const [0,4]}
       hand_l_fk_ctrl.Scale.Z {type=MovieSceneScriptingFloatChannel, default=0, const [0,4]}
       thumb_01_l_ctrl.Location.X {type=MovieSceneScriptingFloatChannel, default=0, thinned_from=4, keys [[7,0],[52,0]]}
       thumb_01_l_ctrl.Location.Y {type=MovieSceneScriptingFloatChannel, default=0, thinned_from=4, keys [[7,0],[52,0]]}
       thumb_01_l_ctrl.Location.Z {type=MovieSceneScriptingFloatChannel, default=0, thinned_from=4, keys [[7,0],[52,0]]}
       thumb_01_l_ctrl.Rotation.X {type=MovieSceneScriptingFloatChannel, default=0, const [0,4]}
       thumb_01_l_ctrl.Rotation.Y {type=MovieSceneScriptingFloatChannel, default=0, const [0,4]}
       thumb_01_l_ctrl.Rotation.Z {type=MovieSceneScriptingFloatChannel, default=0, const [0,4]}
       thumb_01_l_ctrl.Scale.X {type=MovieSceneScriptingFloatChannel, default=0, const [0,4]}
       thumb_01_l_ctrl.Scale.Y {type=MovieSceneScriptingFloatChannel, default=0, const [0,4]}
       thumb_01_l_ctrl.Scale.Z {type=MovieSceneScriptingFloatChannel, default=0, const [0,4]}
       thumb_02_l_ctrl.Location.X {type=MovieSceneScriptingFloatChannel, default=0, const [0,4]}
       thumb_02_l_ctrl.Location.Y {type=MovieSceneScriptingFloatChannel, default=0, thinned_from=4, keys [[7,0],[52,0]]}
       thumb_02_l_ctrl.Location.Z {type=MovieSceneScriptingFloatChannel, default=0, const [0,4]}
       thumb_02_l_ctrl.Rotation.X {type=MovieSceneScriptingFloatChannel, default=0, const [0,4]}
       thumb_02_l_ctrl.Rotation.Y {type=MovieSceneScriptingFloatChannel, default=0, const [0,4]}
       thumb_02_l_ctrl.Rotation.Z {type=MovieSceneScriptingFloatChannel, default=0, const [0,4]}
       thumb_02_l_ctrl.Scale.X {type=MovieSceneScriptingFloatChannel, default=0, const [0,4]}
       thumb_02_l_ctrl.Scale.Y {type=MovieSceneScriptingFloatChannel, default=0, const [0,4]}
       thumb_02_l_ctrl.Scale.Z {type=MovieSceneScriptingFloatChannel, default=0, const [0,4]}
       thumb_03_l_ctrl.Location.X {type=MovieSceneScriptingFloatChannel, default=0, const [0,4]}
       thumb_03_l_ctrl.Location.Y {type=MovieSceneScriptingFloatChannel, default=0, const [0,4]}
       thumb_03_l_ctrl.Location.Z {type=MovieSceneScriptingFloatChannel, default=0, thinned_from=4, keys [[7,0],[52,0]]}
       thumb_03_l_ctrl.Rotation.X {type=MovieSceneScriptingFloatChannel, default=0, const [0,4]}
       thumb_03_l_ctrl.Rotation.Y {type=MovieSceneScriptingFloatChannel, default=0, const [0,4]}
       thumb_03_l_ctrl.Rotation.Z {type=MovieSceneScriptingFloatChannel, default=0, const [0,4]}
       thumb_03_l_ctrl.Scale.X {type=MovieSceneScriptingFloatChannel, default=0, const [0,4]}
       thumb_03_l_ctrl.Scale.Y {type=MovieSceneScriptingFloatChannel, default=0, const [0,4]}
       thumb_03_l_ctrl.Scale.Z {type=MovieSceneScriptingFloatChannel, default=0, const [0,4]}
       index_metacarpal_l_ctrl.Location.X {type=MovieSceneScriptingFloatChannel, default=0, const [0,4]}
       index_metacarpal_l_ctrl.Location.Y {type=MovieSceneScriptingFloatChannel, default=0, const [0,4]}
       index_metacarpal_l_ctrl.Location.Z {type=MovieSceneScriptingFloatChannel, default=0, const [0,4]}
       index_metacarpal_l_ctrl.Rotation.X {type=MovieSceneScriptingFloatChannel, default=0, const [0,4]}
       index_metacarpal_l_ctrl.Rotation.Y {type=MovieSceneScriptingFloatChannel, default=0, const [0,4]}
       index_metacarpal_l_ctrl.Rotation.Z {type=MovieSceneScriptingFloatChannel, default=0, const [0,4]}
       index_metacarpal_l_ctrl.Scale.X {type=MovieSceneScriptingFloatChannel, default=0, const [0,4]}
       index_metacarpal_l_ctrl.Scale.Y {type=MovieSceneScriptingFloatChannel, default=0, const [0,4]}
       index_metacarpal_l_ctrl.Scale.Z {type=MovieSceneScriptingFloatChannel, default=0, const [0,4]}
       index_01_l_ctrl.Location.X {type=MovieSceneScriptingFloatChannel, default=0, thinned_from=4, keys [[7,0],[52,0]]}
       index_01_l_ctrl.Location.Y {type=MovieSceneScriptingFloatChannel, default=0, thinned_from=4, keys [[7,0],[52,0]]}
       index_01_l_ctrl.Location.Z {type=MovieSceneScriptingFloatChannel, default=0, thinned_from=4, keys [[7,0],[52,0]]}
       index_01_l_ctrl.Rotation.X {type=MovieSceneScriptingFloatChannel, default=0, const [0,4]}
       index_01_l_ctrl.Rotation.Y {type=MovieSceneScriptingFloatChannel, default=0, const [0,4]}
       index_01_l_ctrl.Rotation.Z {type=MovieSceneScriptingFloatChannel, default=0, const [0,4]}
       index_01_l_ctrl.Scale.X {type=MovieSceneScriptingFloatChannel, default=0, const [0,4]}
       index_01_l_ctrl.Scale.Y {type=MovieSceneScriptingFloatChannel, default=0, const [0,4]}
       index_01_l_ctrl.Scale.Z {type=MovieSceneScriptingFloatChannel, default=0, const [0,4]}
       index_02_l_ctrl.Location.X {type=MovieSceneScriptingFloatChannel, default=0, thinned_from=4, keys [[7,0],[52,0]]}
       index_02_l_ctrl.Location.Y {type=MovieSceneScriptingFloatChannel, default=0, thinned_from=4, keys [[7,0],[52,0]]}
       index_02_l_ctrl.Location.Z {type=MovieSceneScriptingFloatChannel, default=0, thinned_from=4, keys [[7,0],[52,0]]}
       index_02_l_ctrl.Rotation.X {type=MovieSceneScriptingFloatChannel, default=0, const [0,4]}
       index_02_l_ctrl.Rotation.Y {type=MovieSceneScriptingFloatChannel, default=0, const [0,4]}
       index_02_l_ctrl.Rotation.Z {type=MovieSceneScriptingFloatChannel, default=0, const [0,4]}
       index_02_l_ctrl.Scale.X {type=MovieSceneScriptingFloatChannel, default=0, const [0,4]}
       index_02_l_ctrl.Scale.Y {type=MovieSceneScriptingFloatChannel, default=0, const [0,4]}
       index_02_l_ctrl.Scale.Z {type=MovieSceneScriptingFloatChannel, default=0, const [0,4]}
       index_03_l_ctrl.Location.X {type=MovieSceneScriptingFloatChannel, default=0, const [0,4]}
       index_03_l_ctrl.Location.Y {type=MovieSceneScriptingFloatChannel, default=0, const [0,4]}
       index_03_l_ctrl.Location.Z {type=MovieSceneScriptingFloatChannel, default=0, const [0,4]}
       index_03_l_ctrl.Rotation.X {type=MovieSceneScriptingFloatChannel, default=0, const [0,4]}
       index_03_l_ctrl.Rotation.Y {type=MovieSceneScriptingFloatChannel, default=0, const [0,4]}
       index_03_l_ctrl.Rotation.Z {type=MovieSceneScriptingFloatChannel, default=0, const [0,4]}
       index_03_l_ctrl.Scale.X {type=MovieSceneScriptingFloatChannel, default=0, const [0,4]}
       index_03_l_ctrl.Scale.Y {type=MovieSceneScriptingFloatChannel, default=0, const [0,4]}
       index_03_l_ctrl.Scale.Z {type=MovieSceneScriptingFloatChannel, default=0, const [0,4]}
       middle_metacarpal_l_ctrl.Location.X {type=MovieSceneScriptingFloatChannel, default=0, thinned_from=4, keys [[7,0],[52,0]]}
       middle_metacarpal_l_ctrl.Location.Y {type=MovieSceneScriptingFloatChannel, default=0, thinned_from=4, keys [[7,0],[52,0]]}
       middle_metacarpal_l_ctrl.Location.Z {type=MovieSceneScriptingFloatChannel, default=0, thinned_from=4, keys [[7,0],[52,0]]}
       middle_metacarpal_l_ctrl.Rotation.X {type=MovieSceneScriptingFloatChannel, default=0, const [0,4]}
       middle_metacarpal_l_ctrl.Rotation.Y {type=MovieSceneScriptingFloatChannel, default=0, const [0,4]}
       middle_metacarpal_l_ctrl.Rotation.Z {type=MovieSceneScriptingFloatChannel, default=0, const [0,4]}
       middle_metacarpal_l_ctrl.Scale.X {type=MovieSceneScriptingFloatChannel, default=0, const [0,4]}
       middle_metacarpal_l_ctrl.Scale.Y {type=MovieSceneScriptingFloatChannel, default=0, const [0,4]}
       middle_metacarpal_l_ctrl.Scale.Z {type=MovieSceneScriptingFloatChannel, default=0, const [0,4]}
       middle_01_l_ctrl.Location.X {type=MovieSceneScriptingFloatChannel, default=0, thinned_from=4, keys [[7,0],[52,0]]}
       middle_01_l_ctrl.Location.Y {type=MovieSceneScriptingFloatChannel, default=0, const [0,4]}
       middle_01_l_ctrl.Location.Z {type=MovieSceneScriptingFloatChannel, default=0, const [0,4]}
       middle_01_l_ctrl.Rotation.X {type=MovieSceneScriptingFloatChannel, default=0, const [0,4]}
       middle_01_l_ctrl.Rotation.Y {type=MovieSceneScriptingFloatChannel, default=0, const [0,4]}
       middle_01_l_ctrl.Rotation.Z {type=MovieSceneScriptingFloatChannel, default=0, const [0,4]}
       middle_01_l_ctrl.Scale.X {type=MovieSceneScriptingFloatChannel, default=0, const [0,4]}
       middle_01_l_ctrl.Scale.Y {type=MovieSceneScriptingFloatChannel, default=0, const [0,4]}
       middle_01_l_ctrl.Scale.Z {type=MovieSceneScriptingFloatChannel, default=0, const [0,4]}
       middle_02_l_ctrl.Location.X {type=MovieSceneScriptingFloatChannel, default=0, const [0,4]}
       middle_02_l_ctrl.Location.Y {type=MovieSceneScriptingFloatChannel, default=0, thinned_from=4, keys [[7,0],[52,0]]}
       middle_02_l_ctrl.Location.Z {type=MovieSceneScriptingFloatChannel, default=0, thinned_from=4, keys [[7,0],[52,0]]}
       middle_02_l_ctrl.Rotation.X {type=MovieSceneScriptingFloatChannel, default=0, const [0,4]}
       middle_02_l_ctrl.Rotation.Y {type=MovieSceneScriptingFloatChannel, default=0, const [0,4]}
       middle_02_l_ctrl.Rotation.Z {type=MovieSceneScriptingFloatChannel, default=0, const [0,4]}
       middle_02_l_ctrl.Scale.X {type=MovieSceneScriptingFloatChannel, default=0, const [0,4]}
       middle_02_l_ctrl.Scale.Y {type=MovieSceneScriptingFloatChannel, default=0, const [0,4]}
       middle_02_l_ctrl.Scale.Z {type=MovieSceneScriptingFloatChannel, default=0, const [0,4]}
       middle_03_l_ctrl.Location.X {type=MovieSceneScriptingFloatChannel, default=0, const [0,4]}
       middle_03_l_ctrl.Location.Y {type=MovieSceneScriptingFloatChannel, default=0, thinned_from=4, keys [[7,0],[52,0]]}
       middle_03_l_ctrl.Location.Z {type=MovieSceneScriptingFloatChannel, default=0, thinned_from=4, keys [[7,0],[52,0]]}
       middle_03_l_ctrl.Rotation.X {type=MovieSceneScriptingFloatChannel, default=0, const [0,4]}
       middle_03_l_ctrl.Rotation.Y {type=MovieSceneScriptingFloatChannel, default=0, const [0,4]}
       middle_03_l_ctrl.Rotation.Z {type=MovieSceneScriptingFloatChannel, default=0, const [0,4]}
       middle_03_l_ctrl.Scale.X {type=MovieSceneScriptingFloatChannel, default=0, const [0,4]}
       middle_03_l_ctrl.Scale.Y {type=MovieSceneScriptingFloatChannel, default=0, const [0,4]}
       middle_03_l_ctrl.Scale.Z {type=MovieSceneScriptingFloatChannel, default=0, const [0,4]}
       ring_metacarpal_l_ctrl.Location.X {type=MovieSceneScriptingFloatChannel, default=0, thinned_from=4, keys [[7,0],[52,0]]}
       ring_metacarpal_l_ctrl.Location.Y {type=MovieSceneScriptingFloatChannel, default=0, thinned_from=4, keys [[7,0],[52,0]]}
       ring_metacarpal_l_ctrl.Location.Z {type=MovieSceneScriptingFloatChannel, default=0, thinned_from=4, keys [[7,0],[52,0]]}
       ring_metacarpal_l_ctrl.Rotation.X {type=MovieSceneScriptingFloatChannel, default=0, const [0,4]}
       ring_metacarpal_l_ctrl.Rotation.Y {type=MovieSceneScriptingFloatChannel, default=0, const [0,4]}
       ring_metacarpal_l_ctrl.Rotation.Z {type=MovieSceneScriptingFloatChannel, default=0, const [0,4]}
       ring_metacarpal_l_ctrl.Scale.X {type=MovieSceneScriptingFloatChannel, default=0, const [0,4]}
       ring_metacarpal_l_ctrl.Scale.Y {type=MovieSceneScriptingFloatChannel, default=0, const [0,4]}
       ring_metacarpal_l_ctrl.Scale.Z {type=MovieSceneScriptingFloatChannel, default=0, const [0,4]}
       ring_01_l_ctrl.Location.X {type=MovieSceneScriptingFloatChannel, default=0, thinned_from=4, keys [[7,0],[52,0]]}
       ring_01_l_ctrl.Location.Y {type=MovieSceneScriptingFloatChannel, default=0, const [0,4]}
       ring_01_l_ctrl.Location.Z {type=MovieSceneScriptingFloatChannel, default=0, const [0,4]}
       ring_01_l_ctrl.Rotation.X {type=MovieSceneScriptingFloatChannel, default=0, const [0,4]}
       ring_01_l_ctrl.Rotation.Y {type=MovieSceneScriptingFloatChannel, default=0, const [0,4]}
       ring_01_l_ctrl.Rotation.Z {type=MovieSceneScriptingFloatChannel, default=0, const [0,4]}
       ring_01_l_ctrl.Scale.X {type=MovieSceneScriptingFloatChannel, default=0, const [0,4]}
       ring_01_l_ctrl.Scale.Y {type=MovieSceneScriptingFloatChannel, default=0, const [0,4]}
       ring_01_l_ctrl.Scale.Z {type=MovieSceneScriptingFloatChannel, default=0, const [0,4]}
       ring_02_l_ctrl.Location.X {type=MovieSceneScriptingFloatChannel, default=0, const [0,4]}
       ring_02_l_ctrl.Location.Y {type=MovieSceneScriptingFloatChannel, default=0, const [0,4]}
       ring_02_l_ctrl.Location.Z {type=MovieSceneScriptingFloatChannel, default=0, const [0,4]}
       ring_02_l_ctrl.Rotation.X {type=MovieSceneScriptingFloatChannel, default=0, const [0,4]}
       ring_02_l_ctrl.Rotation.Y {type=MovieSceneScriptingFloatChannel, default=0, const [0,4]}
       ring_02_l_ctrl.Rotation.Z {type=MovieSceneScriptingFloatChannel, default=0, const [0,4]}
       ring_02_l_ctrl.Scale.X {type=MovieSceneScriptingFloatChannel, default=0, const [0,4]}
       ring_02_l_ctrl.Scale.Y {type=MovieSceneScriptingFloatChannel, default=0, const [0,4]}
       ring_02_l_ctrl.Scale.Z {type=MovieSceneScriptingFloatChannel, default=0, const [0,4]}
       ring_03_l_ctrl.Location.X {type=MovieSceneScriptingFloatChannel, default=0, const [0,4]}
       ring_03_l_ctrl.Location.Y {type=MovieSceneScriptingFloatChannel, default=0, thinned_from=4, keys [[7,0],[52,0]]}
       ring_03_l_ctrl.Location.Z {type=MovieSceneScriptingFloatChannel, default=0, thinned_from=4, keys [[7,0],[52,0]]}
       ring_03_l_ctrl.Rotation.X {type=MovieSceneScriptingFloatChannel, default=0, const [0,4]}
       ring_03_l_ctrl.Rotation.Y {type=MovieSceneScriptingFloatChannel, default=0, const [0,4]}
       ring_03_l_ctrl.Rotation.Z {type=MovieSceneScriptingFloatChannel, default=0, const [0,4]}
       ring_03_l_ctrl.Scale.X {type=MovieSceneScriptingFloatChannel, default=0, const [0,4]}
       ring_03_l_ctrl.Scale.Y {type=MovieSceneScriptingFloatChannel, default=0, const [0,4]}
       ring_03_l_ctrl.Scale.Z {type=MovieSceneScriptingFloatChannel, default=0, const [0,4]}
       pinky_metacarpal_l_ctrl.Location.X {type=MovieSceneScriptingFloatChannel, default=0, thinned_from=4, keys [[7,0],[52,0]]}
       pinky_metacarpal_l_ctrl.Location.Y {type=MovieSceneScriptingFloatChannel, default=0, thinned_from=4, keys [[7,0],[52,0]]}
       pinky_metacarpal_l_ctrl.Location.Z {type=MovieSceneScriptingFloatChannel, default=0, thinned_from=4, keys [[7,0],[52,0]]}
       pinky_metacarpal_l_ctrl.Rotation.X {type=MovieSceneScriptingFloatChannel, default=0, const [0,4]}
       pinky_metacarpal_l_ctrl.Rotation.Y {type=MovieSceneScriptingFloatChannel, default=0, const [0,4]}
       pinky_metacarpal_l_ctrl.Rotation.Z {type=MovieSceneScriptingFloatChannel, default=0, const [0,4]}
       pinky_metacarpal_l_ctrl.Scale.X {type=MovieSceneScriptingFloatChannel, default=0, const [0,4]}
       pinky_metacarpal_l_ctrl.Scale.Y {type=MovieSceneScriptingFloatChannel, default=0, const [0,4]}
       pinky_metacarpal_l_ctrl.Scale.Z {type=MovieSceneScriptingFloatChannel, default=0, const [0,4]}
       pinky_01_l_ctrl.Location.X {type=MovieSceneScriptingFloatChannel, default=0, const [0,4]}
       pinky_01_l_ctrl.Location.Y {type=MovieSceneScriptingFloatChannel, default=0, const [0,4]}
       pinky_01_l_ctrl.Location.Z {type=MovieSceneScriptingFloatChannel, default=0, const [0,4]}
       pinky_01_l_ctrl.Rotation.X {type=MovieSceneScriptingFloatChannel, default=0, const [0,4]}
       pinky_01_l_ctrl.Rotation.Y {type=MovieSceneScriptingFloatChannel, default=0, const [0,4]}
       pinky_01_l_ctrl.Rotation.Z {type=MovieSceneScriptingFloatChannel, default=0, const [0,4]}
       pinky_01_l_ctrl.Scale.X {type=MovieSceneScriptingFloatChannel, default=0, const [0,4]}
       pinky_01_l_ctrl.Scale.Y {type=MovieSceneScriptingFloatChannel, default=0, const [0,4]}
       pinky_01_l_ctrl.Scale.Z {type=MovieSceneScriptingFloatChannel, default=0, const [0,4]}
       pinky_02_l_ctrl.Location.X {type=MovieSceneScriptingFloatChannel, default=0, const [0,4]}
       pinky_02_l_ctrl.Location.Y {type=MovieSceneScriptingFloatChannel, default=0, thinned_from=4, keys [[7,0],[52,0]]}
       pinky_02_l_ctrl.Location.Z {type=MovieSceneScriptingFloatChannel, default=0, const [0,4]}
       pinky_02_l_ctrl.Rotation.X {type=MovieSceneScriptingFloatChannel, default=0, const [0,4]}
       pinky_02_l_ctrl.Rotation.Y {type=MovieSceneScriptingFloatChannel, default=0, const [0,4]}
       pinky_02_l_ctrl.Rotation.Z {type=MovieSceneScriptingFloatChannel, default=0, const [0,4]}
       pinky_02_l_ctrl.Scale.X {type=MovieSceneScriptingFloatChannel, default=0, const [0,4]}
       pinky_02_l_ctrl.Scale.Y {type=MovieSceneScriptingFloatChannel, default=0, const [0,4]}
       pinky_02_l_ctrl.Scale.Z {type=MovieSceneScriptingFloatChannel, default=0, const [0,4]}
       pinky_03_l_ctrl.Location.X {type=MovieSceneScriptingFloatChannel, default=0, const [0,4]}
       pinky_03_l_ctrl.Location.Y {type=MovieSceneScriptingFloatChannel, default=0, thinned_from=4, keys [[7,0],[52,0]]}
       pinky_03_l_ctrl.Location.Z {type=MovieSceneScriptingFloatChannel, default=0, const [0,4]}
       pinky_03_l_ctrl.Rotation.X {type=MovieSceneScriptingFloatChannel, default=0, const [0,4]}
       pinky_03_l_ctrl.Rotation.Y {type=MovieSceneScriptingFloatChannel, default=0, const [0,4]}
       pinky_03_l_ctrl.Rotation.Z {type=MovieSceneScriptingFloatChannel, default=0, const [0,4]}
       pinky_03_l_ctrl.Scale.X {type=MovieSceneScriptingFloatChannel, default=0, const [0,4]}
       pinky_03_l_ctrl.Scale.Y {type=MovieSceneScriptingFloatChannel, default=0, const [0,4]}
       pinky_03_l_ctrl.Scale.Z {type=MovieSceneScriptingFloatChannel, default=0, const [0,4]}
       hand_l_ik_ctrl.Location.X {type=MovieSceneScriptingFloatChannel, default=0, keys [[7,0.0073],[12,-0.0352],[44,-0.0352],[52,0.0064]]}
       hand_l_ik_ctrl.Location.Y {type=MovieSceneScriptingFloatChannel, default=0, keys [[7,-0.0106],[12,0.0001],[44,0.0001],[52,-0.0024]]}
       hand_l_ik_ctrl.Location.Z {type=MovieSceneScriptingFloatChannel, default=0, keys [[7,-0.0062],[12,0.0042],[44,0.0042],[52,0.0007]]}
       hand_l_ik_ctrl.Rotation.X {type=MovieSceneScriptingFloatChannel, default=0, keys [[7,0.0513],[12,0.0417],[44,0.0417],[52,-0.0124]]}
       hand_l_ik_ctrl.Rotation.Y {type=MovieSceneScriptingFloatChannel, default=0, keys [[7,0.0225],[12,0.0048],[44,0.0048],[52,-0.0117]]}
       hand_l_ik_ctrl.Rotation.Z {type=MovieSceneScriptingFloatChannel, default=0, keys [[7,0.0465],[12,0.0657],[44,0.0657],[52,-0.0136]]}
       hand_l_ik_ctrl.Scale.X {type=MovieSceneScriptingFloatChannel, default=0, thinned_from=4, keys [[7,0],[52,0]]}
       hand_l_ik_ctrl.Scale.Y {type=MovieSceneScriptingFloatChannel, default=0, const [0,4]}
       hand_l_ik_ctrl.Scale.Z {type=MovieSceneScriptingFloatChannel, default=0, const [0,4]}
       Softness {type=MovieSceneScriptingFloatChannel, default=0, const [0,4]}
       hand_l_ikorient_ctrl.Rotation.Roll {type=MovieSceneScriptingFloatChannel, default=0, const [0,4]}
       hand_l_ikorient_ctrl.Rotation.Pitch {type=MovieSceneScriptingFloatChannel, default=0, const [0,4]}
       hand_l_ikorient_ctrl.Rotation.Yaw {type=MovieSceneScriptingFloatChannel, default=0, const [0,4]}
       arm_l_pv_ik_ctrl.Location.X {type=MovieSceneScriptingFloatChannel, default=0, keys [[7,0.0127],[12,-0.0144],[44,-0.0144],[52,0.0033]]}
       arm_l_pv_ik_ctrl.Location.Y {type=MovieSceneScriptingFloatChannel, default=0, keys [[7,-0.0156],[12,-0.0039],[44,-0.0039],[52,0]]}
       arm_l_pv_ik_ctrl.Location.Z {type=MovieSceneScriptingFloatChannel, default=0, keys [[7,-0.0117],[12,0.0154],[44,0.0154],[52,-0.0011]]}
       arm_l_pv_ik_ctrl.Rotation.X {type=MovieSceneScriptingFloatChannel, default=0, keys [[7,0.0502],[12,0.0048],[44,0.0048],[52,-0.0042]]}
       arm_l_pv_ik_ctrl.Rotation.Y {type=MovieSceneScriptingFloatChannel, default=0, keys [[7,-0.0206],[12,0.0035],[44,0.0035],[52,0]]}
       arm_l_pv_ik_ctrl.Rotation.Z {type=MovieSceneScriptingFloatChannel, default=0, keys [[7,-0.0147],[12,-0.0001],[44,-0.0001],[52,-0.0042]]}
       arm_l_pv_ik_ctrl.Scale.X {type=MovieSceneScriptingFloatChannel, default=0, const [0,4]}
       arm_l_pv_ik_ctrl.Scale.Y {type=MovieSceneScriptingFloatChannel, default=0, const [0,4]}
       arm_l_pv_ik_ctrl.Scale.Z {type=MovieSceneScriptingFloatChannel, default=0, const [0,4]}
       upperarm_r_fk_ctrl.Location.X {type=MovieSceneScriptingFloatChannel, default=0, const [0,4]}
       upperarm_r_fk_ctrl.Location.Y {type=MovieSceneScriptingFloatChannel, default=0, const [0,4]}
       upperarm_r_fk_ctrl.Location.Z {type=MovieSceneScriptingFloatChannel, default=0, const [0,4]}
       upperarm_r_fk_ctrl.Rotation.X {type=MovieSceneScriptingFloatChannel, default=0, const [0,4]}
       upperarm_r_fk_ctrl.Rotation.Y {type=MovieSceneScriptingFloatChannel, default=0, const [0,4]}
       upperarm_r_fk_ctrl.Rotation.Z {type=MovieSceneScriptingFloatChannel, default=0, const [0,4]}
       upperarm_r_fk_ctrl.Scale.X {type=MovieSceneScriptingFloatChannel, default=0, const [0,4]}
       upperarm_r_fk_ctrl.Scale.Y {type=MovieSceneScriptingFloatChannel, default=0, const [0,4]}
       upperarm_r_fk_ctrl.Scale.Z {type=MovieSceneScriptingFloatChannel, default=0, const [0,4]}
       lowerarm_r_fk_ctrl.Location.X {type=MovieSceneScriptingFloatChannel, default=0, const [0,4]}
       lowerarm_r_fk_ctrl.Location.Y {type=MovieSceneScriptingFloatChannel, default=0, const [0,4]}
       lowerarm_r_fk_ctrl.Location.Z {type=MovieSceneScriptingFloatChannel, default=0, const [0,4]}
       lowerarm_r_fk_ctrl.Rotation.X {type=MovieSceneScriptingFloatChannel, default=0, const [0,4]}
       lowerarm_r_fk_ctrl.Rotation.Y {type=MovieSceneScriptingFloatChannel, default=0, const [0,4]}
       lowerarm_r_fk_ctrl.Rotation.Z {type=MovieSceneScriptingFloatChannel, default=0, const [0,4]}
       lowerarm_r_fk_ctrl.Scale.X {type=MovieSceneScriptingFloatChannel, default=0, const [0,4]}
       lowerarm_r_fk_ctrl.Scale.Y {type=MovieSceneScriptingFloatChannel, default=0, const [0,4]}
       lowerarm_r_fk_ctrl.Scale.Z {type=MovieSceneScriptingFloatChannel, default=0, const [0,4]}
       hand_r_fk_ctrl.Location.X {type=MovieSceneScriptingFloatChannel, default=0, const [0,4]}
       hand_r_fk_ctrl.Location.Y {type=MovieSceneScriptingFloatChannel, default=0, const [0,4]}
       hand_r_fk_ctrl.Location.Z {type=MovieSceneScriptingFloatChannel, default=0, const [0,4]}
       hand_r_fk_ctrl.Rotation.X {type=MovieSceneScriptingFloatChannel, default=0, const [0,4]}
       hand_r_fk_ctrl.Rotation.Y {type=MovieSceneScriptingFloatChannel, default=0, const [0,4]}
       hand_r_fk_ctrl.Rotation.Z {type=MovieSceneScriptingFloatChannel, default=0, const [0,4]}
       hand_r_fk_ctrl.Scale.X {type=MovieSceneScriptingFloatChannel, default=0, const [0,4]}
       hand_r_fk_ctrl.Scale.Y {type=MovieSceneScriptingFloatChannel, default=0, const [0,4]}
       hand_r_fk_ctrl.Scale.Z {type=MovieSceneScriptingFloatChannel, default=0, const [0,4]}
       thumb_01_r_ctrl.Location.X {type=MovieSceneScriptingFloatChannel, default=0, thinned_from=4, keys [[7,0],[52,0]]}
       thumb_01_r_ctrl.Location.Y {type=MovieSceneScriptingFloatChannel, default=0, thinned_from=4, keys [[7,0],[52,0]]}
       thumb_01_r_ctrl.Location.Z {type=MovieSceneScriptingFloatChannel, default=0, thinned_from=4, keys [[7,0],[52,0]]}
       thumb_01_r_ctrl.Rotation.X {type=MovieSceneScriptingFloatChannel, default=0, const [0,4]}
       thumb_01_r_ctrl.Rotation.Y {type=MovieSceneScriptingFloatChannel, default=0, const [0,4]}
       thumb_01_r_ctrl.Rotation.Z {type=MovieSceneScriptingFloatChannel, default=0, const [0,4]}
       thumb_01_r_ctrl.Scale.X {type=MovieSceneScriptingFloatChannel, default=0, const [0,4]}
       thumb_01_r_ctrl.Scale.Y {type=MovieSceneScriptingFloatChannel, default=0, const [0,4]}
       thumb_01_r_ctrl.Scale.Z {type=MovieSceneScriptingFloatChannel, default=0, const [0,4]}
       thumb_02_r_ctrl.Location.X {type=MovieSceneScriptingFloatChannel, default=0, thinned_from=4, keys [[7,0],[52,0]]}
       thumb_02_r_ctrl.Location.Y {type=MovieSceneScriptingFloatChannel, default=0, thinned_from=4, keys [[7,0],[52,0]]}
       thumb_02_r_ctrl.Location.Z {type=MovieSceneScriptingFloatChannel, default=0, thinned_from=4, keys [[7,0],[52,0]]}
       thumb_02_r_ctrl.Rotation.X {type=MovieSceneScriptingFloatChannel, default=0, const [0,4]}
       thumb_02_r_ctrl.Rotation.Y {type=MovieSceneScriptingFloatChannel, default=0, const [0,4]}
       thumb_02_r_ctrl.Rotation.Z {type=MovieSceneScriptingFloatChannel, default=0, const [0,4]}
       thumb_02_r_ctrl.Scale.X {type=MovieSceneScriptingFloatChannel, default=0, const [0,4]}
       thumb_02_r_ctrl.Scale.Y {type=MovieSceneScriptingFloatChannel, default=0, const [0,4]}
       thumb_02_r_ctrl.Scale.Z {type=MovieSceneScriptingFloatChannel, default=0, const [0,4]}
       thumb_03_r_ctrl.Location.X {type=MovieSceneScriptingFloatChannel, default=0, thinned_from=4, keys [[7,0],[52,0]]}
       thumb_03_r_ctrl.Location.Y {type=MovieSceneScriptingFloatChannel, default=0, thinned_from=4, keys [[7,0],[52,0]]}
       thumb_03_r_ctrl.Location.Z {type=MovieSceneScriptingFloatChannel, default=0, thinned_from=4, keys [[7,0],[52,0]]}
       thumb_03_r_ctrl.Rotation.X {type=MovieSceneScriptingFloatChannel, default=0, const [0,4]}
       thumb_03_r_ctrl.Rotation.Y {type=MovieSceneScriptingFloatChannel, default=0, const [0,4]}
       thumb_03_r_ctrl.Rotation.Z {type=MovieSceneScriptingFloatChannel, default=0, const [0,4]}
       thumb_03_r_ctrl.Scale.X {type=MovieSceneScriptingFloatChannel, default=0, const [0,4]}
       thumb_03_r_ctrl.Scale.Y {type=MovieSceneScriptingFloatChannel, default=0, const [0,4]}
       thumb_03_r_ctrl.Scale.Z {type=MovieSceneScriptingFloatChannel, default=0, const [0,4]}
       index_metacarpal_r_ctrl.Location.X {type=MovieSceneScriptingFloatChannel, default=0, thinned_from=4, keys [[7,0],[52,0]]}
       index_metacarpal_r_ctrl.Location.Y {type=MovieSceneScriptingFloatChannel, default=0, thinned_from=4, keys [[7,0],[52,0]]}
       index_metacarpal_r_ctrl.Location.Z {type=MovieSceneScriptingFloatChannel, default=0, thinned_from=4, keys [[7,0],[52,0]]}
       index_metacarpal_r_ctrl.Rotation.X {type=MovieSceneScriptingFloatChannel, default=0, const [0,4]}
       index_metacarpal_r_ctrl.Rotation.Y {type=MovieSceneScriptingFloatChannel, default=0, const [0,4]}
       index_metacarpal_r_ctrl.Rotation.Z {type=MovieSceneScriptingFloatChannel, default=0, const [0,4]}
       index_metacarpal_r_ctrl.Scale.X {type=MovieSceneScriptingFloatChannel, default=0, const [0,4]}
       index_metacarpal_r_ctrl.Scale.Y {type=MovieSceneScriptingFloatChannel, default=0, const [0,4]}
       index_metacarpal_r_ctrl.Scale.Z {type=MovieSceneScriptingFloatChannel, default=0, const [0,4]}
       index_01_r_ctrl.Location.X {type=MovieSceneScriptingFloatChannel, default=0, const [0,4]}
       index_01_r_ctrl.Location.Y {type=MovieSceneScriptingFloatChannel, default=0, const [0,4]}
       index_01_r_ctrl.Location.Z {type=MovieSceneScriptingFloatChannel, default=0, const [0,4]}
       index_01_r_ctrl.Rotation.X {type=MovieSceneScriptingFloatChannel, default=0, keys [[7,0],[12,-3.3795],[44,-3.9225],[52,0]]}
       index_01_r_ctrl.Rotation.Y {type=MovieSceneScriptingFloatChannel, default=0, keys [[7,0],[12,-2.6312],[44,-3.3146],[52,0]]}
       index_01_r_ctrl.Rotation.Z {type=MovieSceneScriptingFloatChannel, default=0, keys [[7,0],[12,13.1595],[44,2.4664],[52,0]]}
       index_01_r_ctrl.Scale.X {type=MovieSceneScriptingFloatChannel, default=0, const [0,4]}
       index_01_r_ctrl.Scale.Y {type=MovieSceneScriptingFloatChannel, default=0, const [0,4]}
       index_01_r_ctrl.Scale.Z {type=MovieSceneScriptingFloatChannel, default=0, const [0,4]}
       index_02_r_ctrl.Location.X {type=MovieSceneScriptingFloatChannel, default=0, const [0,4]}
       index_02_r_ctrl.Location.Y {type=MovieSceneScriptingFloatChannel, default=0, const [0,4]}
       index_02_r_ctrl.Location.Z {type=MovieSceneScriptingFloatChannel, default=0, const [0,4]}
       index_02_r_ctrl.Rotation.X {type=MovieSceneScriptingFloatChannel, default=0, const [0,4]}
       index_02_r_ctrl.Rotation.Y {type=MovieSceneScriptingFloatChannel, default=0, const [0,4]}
       index_02_r_ctrl.Rotation.Z {type=MovieSceneScriptingFloatChannel, default=0, const [0,4]}
       index_02_r_ctrl.Scale.X {type=MovieSceneScriptingFloatChannel, default=0, const [0,4]}
       index_02_r_ctrl.Scale.Y {type=MovieSceneScriptingFloatChannel, default=0, const [0,4]}
       index_02_r_ctrl.Scale.Z {type=MovieSceneScriptingFloatChannel, default=0, const [0,4]}
       index_03_r_ctrl.Location.X {type=MovieSceneScriptingFloatChannel, default=0, const [0,4]}
       index_03_r_ctrl.Location.Y {type=MovieSceneScriptingFloatChannel, default=0, const [0,4]}
       index_03_r_ctrl.Location.Z {type=MovieSceneScriptingFloatChannel, default=0, const [0,4]}
       index_03_r_ctrl.Rotation.X {type=MovieSceneScriptingFloatChannel, default=0, const [0,4]}
       index_03_r_ctrl.Rotation.Y {type=MovieSceneScriptingFloatChannel, default=0, const [0,4]}
       index_03_r_ctrl.Rotation.Z {type=MovieSceneScriptingFloatChannel, default=0, const [0,4]}
       index_03_r_ctrl.Scale.X {type=MovieSceneScriptingFloatChannel, default=0, const [0,4]}
       index_03_r_ctrl.Scale.Y {type=MovieSceneScriptingFloatChannel, default=0, const [0,4]}
       index_03_r_ctrl.Scale.Z {type=MovieSceneScriptingFloatChannel, default=0, const [0,4]}
       middle_metacarpal_r_ctrl.Location.X {type=MovieSceneScriptingFloatChannel, default=0, const [0,4]}
       middle_metacarpal_r_ctrl.Location.Y {type=MovieSceneScriptingFloatChannel, default=0, const [0,4]}
       middle_metacarpal_r_ctrl.Location.Z {type=MovieSceneScriptingFloatChannel, default=0, const [0,4]}
       middle_metacarpal_r_ctrl.Rotation.X {type=MovieSceneScriptingFloatChannel, default=0, const [0,4]}
       middle_metacarpal_r_ctrl.Rotation.Y {type=MovieSceneScriptingFloatChannel, default=0, const [0,4]}
       middle_metacarpal_r_ctrl.Rotation.Z {type=MovieSceneScriptingFloatChannel, default=0, const [0,4]}
       middle_metacarpal_r_ctrl.Scale.X {type=MovieSceneScriptingFloatChannel, default=0, const [0,4]}
       middle_metacarpal_r_ctrl.Scale.Y {type=MovieSceneScriptingFloatChannel, default=0, const [0,4]}
       middle_metacarpal_r_ctrl.Scale.Z {type=MovieSceneScriptingFloatChannel, default=0, const [0,4]}
       middle_01_r_ctrl.Location.X {type=MovieSceneScriptingFloatChannel, default=0, const [0,4]}
       middle_01_r_ctrl.Location.Y {type=MovieSceneScriptingFloatChannel, default=0, const [0,4]}
       middle_01_r_ctrl.Location.Z {type=MovieSceneScriptingFloatChannel, default=0, const [0,4]}
       middle_01_r_ctrl.Rotation.X {type=MovieSceneScriptingFloatChannel, default=0, keys [[7,0],[12,2.3139],[44,1.9102],[52,0]]}
       middle_01_r_ctrl.Rotation.Y {type=MovieSceneScriptingFloatChannel, default=0, keys [[7,0],[12,-0.8129],[44,-0.5471],[52,0]]}
       middle_01_r_ctrl.Rotation.Z {type=MovieSceneScriptingFloatChannel, default=0, keys [[7,0],[12,32.4593],[44,25.2476],[52,0]]}
       middle_01_r_ctrl.Scale.X {type=MovieSceneScriptingFloatChannel, default=0, const [0,4]}
       middle_01_r_ctrl.Scale.Y {type=MovieSceneScriptingFloatChannel, default=0, const [0,4]}
       middle_01_r_ctrl.Scale.Z {type=MovieSceneScriptingFloatChannel, default=0, const [0,4]}
       middle_02_r_ctrl.Location.X {type=MovieSceneScriptingFloatChannel, default=0, const [0,4]}
       middle_02_r_ctrl.Location.Y {type=MovieSceneScriptingFloatChannel, default=0, const [0,4]}
       middle_02_r_ctrl.Location.Z {type=MovieSceneScriptingFloatChannel, default=0, const [0,4]}
       middle_02_r_ctrl.Rotation.X {type=MovieSceneScriptingFloatChannel, default=0, const [0,4]}
       middle_02_r_ctrl.Rotation.Y {type=MovieSceneScriptingFloatChannel, default=0, const [0,4]}
       middle_02_r_ctrl.Rotation.Z {type=MovieSceneScriptingFloatChannel, default=0, const [0,4]}
       middle_02_r_ctrl.Scale.X {type=MovieSceneScriptingFloatChannel, default=0, const [0,4]}
       middle_02_r_ctrl.Scale.Y {type=MovieSceneScriptingFloatChannel, default=0, const [0,4]}
       middle_02_r_ctrl.Scale.Z {type=MovieSceneScriptingFloatChannel, default=0, const [0,4]}
       middle_03_r_ctrl.Location.X {type=MovieSceneScriptingFloatChannel, default=0, const [0,4]}
       middle_03_r_ctrl.Location.Y {type=MovieSceneScriptingFloatChannel, default=0, const [0,4]}
       middle_03_r_ctrl.Location.Z {type=MovieSceneScriptingFloatChannel, default=0, const [0,4]}
       middle_03_r_ctrl.Rotation.X {type=MovieSceneScriptingFloatChannel, default=0, const [0,4]}
       middle_03_r_ctrl.Rotation.Y {type=MovieSceneScriptingFloatChannel, default=0, const [0,4]}
       middle_03_r_ctrl.Rotation.Z {type=MovieSceneScriptingFloatChannel, default=0, const [0,4]}
       middle_03_r_ctrl.Scale.X {type=MovieSceneScriptingFloatChannel, default=0, const [0,4]}
       middle_03_r_ctrl.Scale.Y {type=MovieSceneScriptingFloatChannel, default=0, const [0,4]}
       middle_03_r_ctrl.Scale.Z {type=MovieSceneScriptingFloatChannel, default=0, const [0,4]}
       ring_metacarpal_r_ctrl.Location.X {type=MovieSceneScriptingFloatChannel, default=0, thinned_from=4, keys [[7,0],[52,0]]}
       ring_metacarpal_r_ctrl.Location.Y {type=MovieSceneScriptingFloatChannel, default=0, thinned_from=4, keys [[7,0],[52,0]]}
       ring_metacarpal_r_ctrl.Location.Z {type=MovieSceneScriptingFloatChannel, default=0, thinned_from=4, keys [[7,0],[52,0]]}
       ring_metacarpal_r_ctrl.Rotation.X {type=MovieSceneScriptingFloatChannel, default=0, const [0,4]}
       ring_metacarpal_r_ctrl.Rotation.Y {type=MovieSceneScriptingFloatChannel, default=0, const [0,4]}
       ring_metacarpal_r_ctrl.Rotation.Z {type=MovieSceneScriptingFloatChannel, default=0, const [0,4]}
       ring_metacarpal_r_ctrl.Scale.X {type=MovieSceneScriptingFloatChannel, default=0, const [0,4]}
       ring_metacarpal_r_ctrl.Scale.Y {type=MovieSceneScriptingFloatChannel, default=0, const [0,4]}
       ring_metacarpal_r_ctrl.Scale.Z {type=MovieSceneScriptingFloatChannel, default=0, const [0,4]}
       ring_01_r_ctrl.Location.X {type=MovieSceneScriptingFloatChannel, default=0, const [0,4]}
       ring_01_r_ctrl.Location.Y {type=MovieSceneScriptingFloatChannel, default=0, thinned_from=4, keys [[7,0],[52,0]]}
       ring_01_r_ctrl.Location.Z {type=MovieSceneScriptingFloatChannel, default=0, const [0,4]}
       ring_01_r_ctrl.Rotation.X {type=MovieSceneScriptingFloatChannel, default=0, keys [[7,0],[12,2.5393],[44,2.8223],[52,0]]}
       ring_01_r_ctrl.Rotation.Y {type=MovieSceneScriptingFloatChannel, default=0, keys [[7,0],[12,-2.1151],[44,-1.7191],[52,0]]}
       ring_01_r_ctrl.Rotation.Z {type=MovieSceneScriptingFloatChannel, default=0, keys [[7,0],[12,38.1277],[44,29.6824],[52,0]]}
       ring_01_r_ctrl.Scale.X {type=MovieSceneScriptingFloatChannel, default=0, const [0,4]}
       ring_01_r_ctrl.Scale.Y {type=MovieSceneScriptingFloatChannel, default=0, const [0,4]}
       ring_01_r_ctrl.Scale.Z {type=MovieSceneScriptingFloatChannel, default=0, const [0,4]}
       ring_02_r_ctrl.Location.X {type=MovieSceneScriptingFloatChannel, default=0, thinned_from=4, keys [[7,0],[52,0]]}
       ring_02_r_ctrl.Location.Y {type=MovieSceneScriptingFloatChannel, default=0, const [0,4]}
       ring_02_r_ctrl.Location.Z {type=MovieSceneScriptingFloatChannel, default=0, const [0,4]}
       ring_02_r_ctrl.Rotation.X {type=MovieSceneScriptingFloatChannel, default=0, keys [[7,0],[12,-0.1742],[44,0],[52,0]]}
       ring_02_r_ctrl.Rotation.Y {type=MovieSceneScriptingFloatChannel, default=0, keys [[7,0],[12,-0.8496],[44,0],[52,0]]}
       ring_02_r_ctrl.Rotation.Z {type=MovieSceneScriptingFloatChannel, default=0, keys [[7,0],[12,20.8298],[44,1.2144],[52,0]]}
       ring_02_r_ctrl.Scale.X {type=MovieSceneScriptingFloatChannel, default=0, const [0,4]}
       ring_02_r_ctrl.Scale.Y {type=MovieSceneScriptingFloatChannel, default=0, const [0,4]}
       ring_02_r_ctrl.Scale.Z {type=MovieSceneScriptingFloatChannel, default=0, const [0,4]}
       ring_03_r_ctrl.Location.X {type=MovieSceneScriptingFloatChannel, default=0, const [0,4]}
       ring_03_r_ctrl.Location.Y {type=MovieSceneScriptingFloatChannel, default=0, thinned_from=4, keys [[7,0],[52,0]]}
       ring_03_r_ctrl.Location.Z {type=MovieSceneScriptingFloatChannel, default=0, const [0,4]}
       ring_03_r_ctrl.Rotation.X {type=MovieSceneScriptingFloatChannel, default=0, thinned_from=4, keys [[7,0],[52,0]]}
       ring_03_r_ctrl.Rotation.Y {type=MovieSceneScriptingFloatChannel, default=0, thinned_from=4, keys [[7,0],[52,0]]}
       ring_03_r_ctrl.Rotation.Z {type=MovieSceneScriptingFloatChannel, default=0, keys [[7,0],[12,10.4475],[44,35.0455],[52,0]]}
       ring_03_r_ctrl.Scale.X {type=MovieSceneScriptingFloatChannel, default=0, const [0,4]}
       ring_03_r_ctrl.Scale.Y {type=MovieSceneScriptingFloatChannel, default=0, const [0,4]}
       ring_03_r_ctrl.Scale.Z {type=MovieSceneScriptingFloatChannel, default=0, const [0,4]}
       pinky_metacarpal_r_ctrl.Location.X {type=MovieSceneScriptingFloatChannel, default=0, thinned_from=4, keys [[7,0],[52,0]]}
       pinky_metacarpal_r_ctrl.Location.Y {type=MovieSceneScriptingFloatChannel, default=0, thinned_from=4, keys [[7,0],[52,0]]}
       pinky_metacarpal_r_ctrl.Location.Z {type=MovieSceneScriptingFloatChannel, default=0, thinned_from=4, keys [[7,0],[52,0]]}
       pinky_metacarpal_r_ctrl.Rotation.X {type=MovieSceneScriptingFloatChannel, default=0, const [0,4]}
       pinky_metacarpal_r_ctrl.Rotation.Y {type=MovieSceneScriptingFloatChannel, default=0, const [0,4]}
       pinky_metacarpal_r_ctrl.Rotation.Z {type=MovieSceneScriptingFloatChannel, default=0, const [0,4]}
       pinky_metacarpal_r_ctrl.Scale.X {type=MovieSceneScriptingFloatChannel, default=0, const [0,4]}
       pinky_metacarpal_r_ctrl.Scale.Y {type=MovieSceneScriptingFloatChannel, default=0, const [0,4]}
       pinky_metacarpal_r_ctrl.Scale.Z {type=MovieSceneScriptingFloatChannel, default=0, const [0,4]}
       pinky_01_r_ctrl.Location.X {type=MovieSceneScriptingFloatChannel, default=0, const [0,4]}
       pinky_01_r_ctrl.Location.Y {type=MovieSceneScriptingFloatChannel, default=0, const [0,4]}
       pinky_01_r_ctrl.Location.Z {type=MovieSceneScriptingFloatChannel, default=0, const [0,4]}
       pinky_01_r_ctrl.Rotation.X {type=MovieSceneScriptingFloatChannel, default=0, keys [[7,0],[12,0.1044],[44,0.1044],[52,0]]}
       pinky_01_r_ctrl.Rotation.Y {type=MovieSceneScriptingFloatChannel, default=0, keys [[7,0],[12,-1.0913],[44,-1.0913],[52,0]]}
       pinky_01_r_ctrl.Rotation.Z {type=MovieSceneScriptingFloatChannel, default=0, keys [[7,0],[12,51.1358],[44,51.1358],[52,0]]}
       pinky_01_r_ctrl.Scale.X {type=MovieSceneScriptingFloatChannel, default=0, const [0,4]}
       pinky_01_r_ctrl.Scale.Y {type=MovieSceneScriptingFloatChannel, default=0, const [0,4]}
       pinky_01_r_ctrl.Scale.Z {type=MovieSceneScriptingFloatChannel, default=0, const [0,4]}
       pinky_02_r_ctrl.Location.X {type=MovieSceneScriptingFloatChannel, default=0, const [0,4]}
       pinky_02_r_ctrl.Location.Y {type=MovieSceneScriptingFloatChannel, default=0, const [0,4]}
       pinky_02_r_ctrl.Location.Z {type=MovieSceneScriptingFloatChannel, default=0, const [0,4]}
       pinky_02_r_ctrl.Rotation.X {type=MovieSceneScriptingFloatChannel, default=0, keys [[7,0],[12,0.4525],[44,0.3929],[52,0]]}
       pinky_02_r_ctrl.Rotation.Y {type=MovieSceneScriptingFloatChannel, default=0, keys [[7,0],[12,0.4377],[44,0.4919],[52,0]]}
       pinky_02_r_ctrl.Rotation.Z {type=MovieSceneScriptingFloatChannel, default=0, keys [[7,0],[12,21.4946],[44,14.158],[52,0]]}
       pinky_02_r_ctrl.Scale.X {type=MovieSceneScriptingFloatChannel, default=0, const [0,4]}
       pinky_02_r_ctrl.Scale.Y {type=MovieSceneScriptingFloatChannel, default=0, const [0,4]}
       pinky_02_r_ctrl.Scale.Z {type=MovieSceneScriptingFloatChannel, default=0, const [0,4]}
       pinky_03_r_ctrl.Location.X {type=MovieSceneScriptingFloatChannel, default=0, const [0,4]}
       pinky_03_r_ctrl.Location.Y {type=MovieSceneScriptingFloatChannel, default=0, const [0,4]}
       pinky_03_r_ctrl.Location.Z {type=MovieSceneScriptingFloatChannel, default=0, const [0,4]}
       pinky_03_r_ctrl.Rotation.X {type=MovieSceneScriptingFloatChannel, default=0, keys [[7,0],[12,-0.6024],[44,0],[52,0]]}
       pinky_03_r_ctrl.Rotation.Y {type=MovieSceneScriptingFloatChannel, default=0, keys [[7,0],[12,0.1737],[44,0],[52,0]]}
       pinky_03_r_ctrl.Rotation.Z {type=MovieSceneScriptingFloatChannel, default=0, keys [[7,0],[12,28.1914],[44,0],[52,0]]}
       pinky_03_r_ctrl.Scale.X {type=MovieSceneScriptingFloatChannel, default=0, const [0,4]}
       pinky_03_r_ctrl.Scale.Y {type=MovieSceneScriptingFloatChannel, default=0, const [0,4]}
       pinky_03_r_ctrl.Scale.Z {type=MovieSceneScriptingFloatChannel, default=0, const [0,4]}
       hand_r_ik_ctrl.Location.X {type=MovieSceneScriptingFloatChannel, default=0, keys [[7,-0.0084],[12,0.0085],[44,0.0085],[52,-0.0315]]}
       hand_r_ik_ctrl.Location.Y {type=MovieSceneScriptingFloatChannel, default=0, keys [[7,0.0031],[12,-0.0046],[44,-0.0046],[52,0.055]]}
       hand_r_ik_ctrl.Location.Z {type=MovieSceneScriptingFloatChannel, default=0, keys [[7,-0.0073],[12,-0.0255],[44,-0.0255],[52,0.024]]}
       hand_r_ik_ctrl.Rotation.X {type=MovieSceneScriptingFloatChannel, default=0, keys [[7,0.1933],[12,0.0205],[44,0.0205],[52,-7.2623]]}
       hand_r_ik_ctrl.Rotation.Y {type=MovieSceneScriptingFloatChannel, default=0, keys [[7,0.0703],[12,0.0145],[44,0.0145],[52,-0.5273]]}
       hand_r_ik_ctrl.Rotation.Z {type=MovieSceneScriptingFloatChannel, default=0, keys [[7,0.1425],[12,0.0316],[44,0.0316],[52,-10.5551]]}
       hand_r_ik_ctrl.Scale.X {type=MovieSceneScriptingFloatChannel, default=0, thinned_from=4, keys [[7,0],[52,0]]}
       hand_r_ik_ctrl.Scale.Y {type=MovieSceneScriptingFloatChannel, default=0, const [0,4]}
       hand_r_ik_ctrl.Scale.Z {type=MovieSceneScriptingFloatChannel, default=0, const [0,4]}
       Softness_2 {type=MovieSceneScriptingFloatChannel, default=0, const [0,4]}
       hand_r_ikorient_ctrl.Rotation.Roll {type=MovieSceneScriptingFloatChannel, default=0, const [0,4]}
       hand_r_ikorient_ctrl.Rotation.Pitch {type=MovieSceneScriptingFloatChannel, default=0, const [0,4]}
       hand_r_ikorient_ctrl.Rotation.Yaw {type=MovieSceneScriptingFloatChannel, default=0, const [0,4]}
       arm_r_pv_ik_ctrl.Location.X {type=MovieSceneScriptingFloatChannel, default=0, keys [[7,-0.003],[12,0.0087],[44,0.0087],[52,-0.024]]}
       arm_r_pv_ik_ctrl.Location.Y {type=MovieSceneScriptingFloatChannel, default=0, keys [[7,-0.0227],[12,-0.0199],[44,-0.0199],[52,0.1092]]}
       arm_r_pv_ik_ctrl.Location.Z {type=MovieSceneScriptingFloatChannel, default=0, keys [[7,-0.0006],[12,-0.0205],[44,-0.0205],[52,-0.1384]]}
       arm_r_pv_ik_ctrl.Rotation.X {type=MovieSceneScriptingFloatChannel, default=0, keys [[7,0.0042],[12,-0.0018],[44,-0.0018],[52,-0.4761]]}
       arm_r_pv_ik_ctrl.Rotation.Y {type=MovieSceneScriptingFloatChannel, default=0, keys [[7,-0.0055],[12,-0.0053],[44,-0.0053],[52,-0.2004]]}
       arm_r_pv_ik_ctrl.Rotation.Z {type=MovieSceneScriptingFloatChannel, default=0, thinned_from=4, keys [[7,-0.0004],[44,-0.0003],[52,-0.7506]]}
       arm_r_pv_ik_ctrl.Scale.X {type=MovieSceneScriptingFloatChannel, default=0, const [0,4]}
       arm_r_pv_ik_ctrl.Scale.Y {type=MovieSceneScriptingFloatChannel, default=0, const [0,4]}
       arm_r_pv_ik_ctrl.Scale.Z {type=MovieSceneScriptingFloatChannel, default=0, const [0,4]}
       body_ctrl.Location.X {type=MovieSceneScriptingFloatChannel, default=0, const [0,4]}
       body_ctrl.Location.Y {type=MovieSceneScriptingFloatChannel, default=0, const [0,4]}
       body_ctrl.Location.Z {type=MovieSceneScriptingFloatChannel, default=0, const [0,4]}
       body_ctrl.Rotation.X {type=MovieSceneScriptingFloatChannel, default=0, const [0,4]}
       body_ctrl.Rotation.Y {type=MovieSceneScriptingFloatChannel, default=0, const [0,4]}
       body_ctrl.Rotation.Z {type=MovieSceneScriptingFloatChannel, default=0, const [0,4]}
       body_ctrl.Scale.X {type=MovieSceneScriptingFloatChannel, default=0, const [0,4]}
       body_ctrl.Scale.Y {type=MovieSceneScriptingFloatChannel, default=0, const [0,4]}
       body_ctrl.Scale.Z {type=MovieSceneScriptingFloatChannel, default=0, const [0,4]}
       hips_ctrl.Location.X {type=MovieSceneScriptingFloatChannel, default=0, const [0,4]}
       hips_ctrl.Location.Y {type=MovieSceneScriptingFloatChannel, default=0, const [0,4]}
       hips_ctrl.Location.Z {type=MovieSceneScriptingFloatChannel, default=0, const [0,4]}
       hips_ctrl.Rotation.X {type=MovieSceneScriptingFloatChannel, default=0, const [0,4]}
       hips_ctrl.Rotation.Y {type=MovieSceneScriptingFloatChannel, default=0, const [0,4]}
       hips_ctrl.Rotation.Z {type=MovieSceneScriptingFloatChannel, default=0, const [0,4]}
       hips_ctrl.Scale.X {type=MovieSceneScriptingFloatChannel, default=0, const [0,4]}
       hips_ctrl.Scale.Y {type=MovieSceneScriptingFloatChannel, default=0, const [0,4]}
       hips_ctrl.Scale.Z {type=MovieSceneScriptingFloatChannel, default=0, const [0,4]}
       hips_tan_ctrl.Location.X {type=MovieSceneScriptingFloatChannel, default=0, keys [[7,0.2959],[12,0.4146],[44,0.4146],[52,0.3188]]}
       hips_tan_ctrl.Location.Y {type=MovieSceneScriptingFloatChannel, default=0, keys [[7,1.2087],[12,0.9601],[44,0.9601],[52,1.2725]]}
       hips_tan_ctrl.Location.Z {type=MovieSceneScriptingFloatChannel, default=0, keys [[7,-0.0391],[12,-0.0233],[44,-0.0233],[52,-0.0426]]}
       hips_tan_ctrl.Rotation.X {type=MovieSceneScriptingFloatChannel, default=0, const [0,4]}
       hips_tan_ctrl.Rotation.Y {type=MovieSceneScriptingFloatChannel, default=0, const [0,4]}
       hips_tan_ctrl.Rotation.Z {type=MovieSceneScriptingFloatChannel, default=0, const [0,4]}
       hips_tan_ctrl.Scale.X {type=MovieSceneScriptingFloatChannel, default=0, const [0,4]}
       hips_tan_ctrl.Scale.Y {type=MovieSceneScriptingFloatChannel, default=0, const [0,4]}
       hips_tan_ctrl.Scale.Z {type=MovieSceneScriptingFloatChannel, default=0, const [0,4]}
       spine_01_ctrl.Location.X {type=MovieSceneScriptingFloatChannel, default=0, const [0,4]}
       spine_01_ctrl.Location.Y {type=MovieSceneScriptingFloatChannel, default=0, const [0,4]}
       spine_01_ctrl.Location.Z {type=MovieSceneScriptingFloatChannel, default=0, const [0,4]}
       spine_01_ctrl.Rotation.X {type=MovieSceneScriptingFloatChannel, default=0, const [0,4]}
       spine_01_ctrl.Rotation.Y {type=MovieSceneScriptingFloatChannel, default=0, const [0,4]}
       spine_01_ctrl.Rotation.Z {type=MovieSceneScriptingFloatChannel, default=0, const [0,4]}
       spine_01_ctrl.Scale.X {type=MovieSceneScriptingFloatChannel, default=0, const [0,4]}
       spine_01_ctrl.Scale.Y {type=MovieSceneScriptingFloatChannel, default=0, const [0,4]}
       spine_01_ctrl.Scale.Z {type=MovieSceneScriptingFloatChannel, default=0, const [0,4]}
       spine_02_ctrl.Location.X {type=MovieSceneScriptingFloatChannel, default=0, const [0,4]}
       spine_02_ctrl.Location.Y {type=MovieSceneScriptingFloatChannel, default=0, const [0,4]}
       spine_02_ctrl.Location.Z {type=MovieSceneScriptingFloatChannel, default=0, const [0,4]}
       spine_02_ctrl.Rotation.X {type=MovieSceneScriptingFloatChannel, default=0, const [0,4]}
       spine_02_ctrl.Rotation.Y {type=MovieSceneScriptingFloatChannel, default=0, const [0,4]}
       spine_02_ctrl.Rotation.Z {type=MovieSceneScriptingFloatChannel, default=0, const [0,4]}
       spine_02_ctrl.Scale.X {type=MovieSceneScriptingFloatChannel, default=0, const [0,4]}
       spine_02_ctrl.Scale.Y {type=MovieSceneScriptingFloatChannel, default=0, const [0,4]}
       spine_02_ctrl.Scale.Z {type=MovieSceneScriptingFloatChannel, default=0, const [0,4]}
       spine_03_ctrl.Location.X {type=MovieSceneScriptingFloatChannel, default=0, const [0,4]}
       spine_03_ctrl.Location.Y {type=MovieSceneScriptingFloatChannel, default=0, const [0,4]}
       spine_03_ctrl.Location.Z {type=MovieSceneScriptingFloatChannel, default=0, const [0,4]}
       spine_03_ctrl.Rotation.X {type=MovieSceneScriptingFloatChannel, default=0, const [0,4]}
       spine_03_ctrl.Rotation.Y {type=MovieSceneScriptingFloatChannel, default=0, const [0,4]}
       spine_03_ctrl.Rotation.Z {type=MovieSceneScriptingFloatChannel, default=0, const [0,4]}
       spine_03_ctrl.Scale.X {type=MovieSceneScriptingFloatChannel, default=0, const [0,4]}
       spine_03_ctrl.Scale.Y {type=MovieSceneScriptingFloatChannel, default=0, const [0,4]}
       spine_03_ctrl.Scale.Z {type=MovieSceneScriptingFloatChannel, default=0, const [0,4]}
       spine_04_ctrl.Location.X {type=MovieSceneScriptingFloatChannel, default=0, const [0,4]}
       spine_04_ctrl.Location.Y {type=MovieSceneScriptingFloatChannel, default=0, const [0,4]}
       spine_04_ctrl.Location.Z {type=MovieSceneScriptingFloatChannel, default=0, const [0,4]}
       spine_04_ctrl.Rotation.X {type=MovieSceneScriptingFloatChannel, default=0, const [0,4]}
       spine_04_ctrl.Rotation.Y {type=MovieSceneScriptingFloatChannel, default=0, const [0,4]}
       spine_04_ctrl.Rotation.Z {type=MovieSceneScriptingFloatChannel, default=0, const [0,4]}
       spine_04_ctrl.Scale.X {type=MovieSceneScriptingFloatChannel, default=0, const [0,4]}
       spine_04_ctrl.Scale.Y {type=MovieSceneScriptingFloatChannel, default=0, const [0,4]}
       spine_04_ctrl.Scale.Z {type=MovieSceneScriptingFloatChannel, default=0, const [0,4]}
       head_ctrl.Location.X {type=MovieSceneScriptingFloatChannel, default=0, const [0,4]}
       head_ctrl.Location.Y {type=MovieSceneScriptingFloatChannel, default=0, const [0,4]}
       head_ctrl.Location.Z {type=MovieSceneScriptingFloatChannel, default=0, const [0,4]}
       head_ctrl.Rotation.X {type=MovieSceneScriptingFloatChannel, default=0, const [0,4]}
       head_ctrl.Rotation.Y {type=MovieSceneScriptingFloatChannel, default=0, const [0,4]}
       head_ctrl.Rotation.Z {type=MovieSceneScriptingFloatChannel, default=0, const [0,4]}
       head_ctrl.Scale.X {type=MovieSceneScriptingFloatChannel, default=0, const [0,4]}
       head_ctrl.Scale.Y {type=MovieSceneScriptingFloatChannel, default=0, const [0,4]}
       head_ctrl.Scale.Z {type=MovieSceneScriptingFloatChannel, default=0, const [0,4]}
       clavicle_l_ctrl.Location.X {type=MovieSceneScriptingFloatChannel, default=0, const [0,4]}
       clavicle_l_ctrl.Location.Y {type=MovieSceneScriptingFloatChannel, default=0, const [0,4]}
       clavicle_l_ctrl.Location.Z {type=MovieSceneScriptingFloatChannel, default=0, const [0,4]}
       clavicle_l_ctrl.Rotation.X {type=MovieSceneScriptingFloatChannel, default=0, const [0,4]}
       clavicle_l_ctrl.Rotation.Y {type=MovieSceneScriptingFloatChannel, default=0, const [0,4]}
       clavicle_l_ctrl.Rotation.Z {type=MovieSceneScriptingFloatChannel, default=0, const [0,4]}
       clavicle_l_ctrl.Scale.X {type=MovieSceneScriptingFloatChannel, default=0, const [0,4]}
       clavicle_l_ctrl.Scale.Y {type=MovieSceneScriptingFloatChannel, default=0, const [0,4]}
       clavicle_l_ctrl.Scale.Z {type=MovieSceneScriptingFloatChannel, default=0, const [0,4]}
       clavicle_r_ctrl.Location.X {type=MovieSceneScriptingFloatChannel, default=0, const [0,4]}
       clavicle_r_ctrl.Location.Y {type=MovieSceneScriptingFloatChannel, default=0, const [0,4]}
       clavicle_r_ctrl.Location.Z {type=MovieSceneScriptingFloatChannel, default=0, const [0,4]}
       clavicle_r_ctrl.Rotation.X {type=MovieSceneScriptingFloatChannel, default=0, const [0,4]}
       clavicle_r_ctrl.Rotation.Y {type=MovieSceneScriptingFloatChannel, default=0, const [0,4]}
       clavicle_r_ctrl.Rotation.Z {type=MovieSceneScriptingFloatChannel, default=0, const [0,4]}
       clavicle_r_ctrl.Scale.X {type=MovieSceneScriptingFloatChannel, default=0, const [0,4]}
       clavicle_r_ctrl.Scale.Y {type=MovieSceneScriptingFloatChannel, default=0, const [0,4]}
       clavicle_r_ctrl.Scale.Z {type=MovieSceneScriptingFloatChannel, default=0, const [0,4]}
       neck_01_ctrl.Location.X {type=MovieSceneScriptingFloatChannel, default=0, const [0,4]}
       neck_01_ctrl.Location.Y {type=MovieSceneScriptingFloatChannel, default=0, const [0,4]}
       neck_01_ctrl.Location.Z {type=MovieSceneScriptingFloatChannel, default=0, const [0,4]}
       neck_01_ctrl.Rotation.X {type=MovieSceneScriptingFloatChannel, default=0, const [0,4]}
       neck_01_ctrl.Rotation.Y {type=MovieSceneScriptingFloatChannel, default=0, const [0,4]}
       neck_01_ctrl.Rotation.Z {type=MovieSceneScriptingFloatChannel, default=0, const [0,4]}
       neck_01_ctrl.Scale.X {type=MovieSceneScriptingFloatChannel, default=0, const [0,4]}
       neck_01_ctrl.Scale.Y {type=MovieSceneScriptingFloatChannel, default=0, const [0,4]}
       neck_01_ctrl.Scale.Z {type=MovieSceneScriptingFloatChannel, default=0, const [0,4]}
       neck_02_ctrl.Location.X {type=MovieSceneScriptingFloatChannel, default=0, const [0,4]}
       neck_02_ctrl.Location.Y {type=MovieSceneScriptingFloatChannel, default=0, const [0,4]}
       neck_02_ctrl.Location.Z {type=MovieSceneScriptingFloatChannel, default=0, const [0,4]}
       neck_02_ctrl.Rotation.X {type=MovieSceneScriptingFloatChannel, default=0, const [0,4]}
       neck_02_ctrl.Rotation.Y {type=MovieSceneScriptingFloatChannel, default=0, const [0,4]}
       neck_02_ctrl.Rotation.Z {type=MovieSceneScriptingFloatChannel, default=0, const [0,4]}
       neck_02_ctrl.Scale.X {type=MovieSceneScriptingFloatChannel, default=0, const [0,4]}
       neck_02_ctrl.Scale.Y {type=MovieSceneScriptingFloatChannel, default=0, const [0,4]}
       neck_02_ctrl.Scale.Z {type=MovieSceneScriptingFloatChannel, default=0, const [0,4]}
       head_ik_ctrl.Location.X {type=MovieSceneScriptingFloatChannel, default=0, keys [[7,0],[12,-0.0004],[44,-0.0004],[52,-0.0008]]}
       head_ik_ctrl.Location.Y {type=MovieSceneScriptingFloatChannel, default=0, thinned_from=4, keys [[7,0],[12,0.0004],[52,0.0003]]}
       head_ik_ctrl.Location.Z {type=MovieSceneScriptingFloatChannel, default=0, keys [[7,0],[12,0.0011],[44,0.0011],[52,0.0005]]}
       head_ik_ctrl.Rotation.X {type=MovieSceneScriptingFloatChannel, default=0, const [0,4]}
       head_ik_ctrl.Rotation.Y {type=MovieSceneScriptingFloatChannel, default=0, const [0,4]}
       head_ik_ctrl.Rotation.Z {type=MovieSceneScriptingFloatChannel, default=0, const [0,4]}
       head_ik_ctrl.Scale.X {type=MovieSceneScriptingFloatChannel, default=0, const [0,4]}
       head_ik_ctrl.Scale.Y {type=MovieSceneScriptingFloatChannel, default=0, const [0,4]}
       head_ik_ctrl.Scale.Z {type=MovieSceneScriptingFloatChannel, default=0, const [0,4]}
       chest_ctrl.Location.X {type=MovieSceneScriptingFloatChannel, default=0, keys [[7,0],[12,-0.0005],[44,-0.0005],[52,0.0001]]}
       chest_ctrl.Location.Y {type=MovieSceneScriptingFloatChannel, default=0, keys [[7,0],[12,0.0008],[44,0.0008],[52,-0.0002]]}
       chest_ctrl.Location.Z {type=MovieSceneScriptingFloatChannel, default=0, keys [[7,0],[12,-0.0002],[44,-0.0002],[52,0]]}
       chest_ctrl.Rotation.X {type=MovieSceneScriptingFloatChannel, default=0, keys [[7,0],[12,0.0134],[44,0.0134],[52,0.0152]]}
       chest_ctrl.Rotation.Y {type=MovieSceneScriptingFloatChannel, default=0, keys [[7,0],[12,-0.0002],[44,-0.0002],[52,0.0022]]}
       chest_ctrl.Rotation.Z {type=MovieSceneScriptingFloatChannel, default=0, keys [[7,0],[12,-0.003],[44,-0.003],[52,0.0003]]}
       chest_ctrl.Scale.X {type=MovieSceneScriptingFloatChannel, default=0, const [0,4]}
       chest_ctrl.Scale.Y {type=MovieSceneScriptingFloatChannel, default=0, const [0,4]}
       chest_ctrl.Scale.Z {type=MovieSceneScriptingFloatChannel, default=0, const [0,4]}
       chest_tan_ctrl.Location.X {type=MovieSceneScriptingFloatChannel, default=0, thinned_from=4, keys [[7,0],[12,-0.0001],[52,-0.0001]]}
       chest_tan_ctrl.Location.Y {type=MovieSceneScriptingFloatChannel, default=0, keys [[7,0],[12,0.0003],[44,0.0003],[52,0.0005]]}
       chest_tan_ctrl.Location.Z {type=MovieSceneScriptingFloatChannel, default=0, thinned_from=4, keys [[7,0],[52,0.0001]]}
       chest_tan_ctrl.Rotation.X {type=MovieSceneScriptingFloatChannel, default=0, keys [[7,-5.3127],[12,-4.7462],[44,-4.7462],[52,-6.0847]]}
       chest_tan_ctrl.Rotation.Y {type=MovieSceneScriptingFloatChannel, default=0, keys [[7,-1.8878],[12,0],[44,0],[52,-0.9399]]}
       chest_tan_ctrl.Rotation.Z {type=MovieSceneScriptingFloatChannel, default=0, keys [[7,-9.5847],[12,-11.0415],[44,-11.0415],[52,-12.1526]]}
       chest_tan_ctrl.Scale.X {type=MovieSceneScriptingFloatChannel, default=0, keys [[7,-0.0017],[12,-0.0042],[44,-0.0042],[52,-0.002]]}
       chest_tan_ctrl.Scale.Y {type=MovieSceneScriptingFloatChannel, default=0, const [0,4]}
       chest_tan_ctrl.Scale.Z {type=MovieSceneScriptingFloatChannel, default=0, const [0,4]}
      ]
     }
    ]
   }
  ]
 }
]
display_rate {denominator=1, numerator=30}
tick_resolution {denominator=1, numerator=24000}
tracks []
