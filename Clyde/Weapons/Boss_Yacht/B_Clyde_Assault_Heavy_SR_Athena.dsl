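; Boss Yacht heavy assault rifle weapon Blueprint that prototypes a first-bullet bonus against shielded players, with debug prints only and no damage applied.
; On BeginPlay it binds a check to start-firing, which acts only when the magazine holds exactly 50 rounds.
; It traces the aim point of the first pawn in the player list, not explicitly the weapon's owner.
; Any shield or overshield on the hit pawn prints a buff message; otherwise the values are printed.
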(event EventBeginPlay
  (Default|BindEventtoOnStartFiring (delegate Custom|FirtBulletBuff)))

(event Custom|FirtBulletBuff
  (if (== (FortWeapon|GetMagazineAmmoCount) 50)
    (bind _outfortplayerpawns (Fortnite|GetFortPlayerPawns))
    (bind _returnvalue (Fortnite|GetPlayerAimPointHit (Fortnite|GetFortPlayerControllerfromActor (Utilities|Array|Get(acopy) _outfortplayerpawns)) 2.5e+04 (Utilities|Array|MakeArray 0)))
    (bind (_returnvalue_bblockinghit _returnvalue_binitialoverlap _returnvalue_time _returnvalue_distance _returnvalue_location _returnvalue_impactpoint _returnvalue_normal _returnvalue_impactnormal _returnvalue_physmat _returnvalue_hitactor) (Collision|BreakHitResult _returnvalue))
    (bind _asfort_player_pawn (Utilities|Casting|CastToFortPlayerPawn _returnvalue_hitactor))
    (if (Math|Boolean|ORBoolean (> (FortPawn|GetShield _asfort_player_pawn) 0) (> (FortPawn|GetOvershield _asfort_player_pawn) 0) (> (FortPawn|GetCurrentOvershield _asfort_player_pawn) 0))
      (Development|PrintString "Apply first bullet buff dmg" true true "(R=0.000000,G=0.660000,B=1.000000,A=1.000000)" 5.0)
      (else
        (Development|PrintString (Utilities|String|ToString(Float) (FortPawn|GetShield _asfort_player_pawn)) true true "(R=0.000000,G=0.660000,B=1.000000,A=1.000000)" 5.0)
        (Development|PrintString (Utilities|String|ToString(Float) (FortPawn|GetOvershield _asfort_player_pawn)) true true "(R=0.000000,G=0.660000,B=1.000000,A=1.000000)" 5.0)
        (Development|PrintString (Utilities|String|ToString(Float) (FortPawn|GetCurrentOvershield _asfort_player_pawn)) true true "(R=0.000000,G=0.660000,B=1.000000,A=1.000000)" 5.0)))))

(fn ConstructionScript
  (|CallParentFunction))
